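# Season Pass purchase screen, a CommonUI activatable widget for buying the pass or its bundle, gifting it, or getting V-Bucks.
# View-model handlers drive affordability, reward text and theming, with switchers pairing buy and Get V-Bucks buttons.
# Purchases optionally pass a right-of-withdrawal dialog, then run an async action behind a progress modal; success pops the screen and fires BackToHub.
# Trace IDs tag analytics, and mobile gets aspect-ratio-aware compact layout.
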WidgetBlueprint /SeasonPass/Widgets/SeasonPass_PurchaseScreen
graphs [
 AddPassNameAndId {nodes [
   n1 = Inputs!() -> (Info:"Flow Analytics Additional Info Structure (by ref)", Value:String=Free, UpsellType:ENexusUpsellEntryType Enum=Free)
   n2 = GetOfferId!(self:Fort Simple Offer VM Object Reference=GetBenefitsForUpsellType!(
     self:Season Pass VM Object Reference=Get SeasonPassVM() -> (SeasonPassVM:Season Pass VM Object Reference)
     UpsellType:ENexusUpsellEntryType Enum=@n1.UpsellType
    ) -> (BenefitData:Nexus Upsell Benefits Structure, ReturnValue:Fort Simple Offer VM Object Reference).ReturnValue) -> (ReturnValue:String)
   n3 = String To Name(InString=@n1.Value)
   Outputs!(ReturnValue:Flow Analytics Additional Info Structure="Add Name (Non-Localized)"!(
     Info:"Flow Analytics Additional Info Structure (by ref)"="Add Name (Non-Localized)"(Info=@n1, Key=PassName, Value=@n3)
     Key:String=PassID
     Value:"Name (by ref)"=String To Name(InString=@n2)
    ) -> (ReturnValue:Flow Analytics Additional Info Structure))
  ]}
 BP_GetDesiredFocusTarget {nodes [
   Get Desired Focus Target()
   > Return Node!(
    execute
    ReturnValue:Widget Object Reference=GetActiveWidget!(self:Widget Switcher Object Reference=Get PurchaseSwitcher() -> (PurchaseSwitcher:Common Activatable Widget Switcher Object Reference)) -> (ReturnValue:Widget Object Reference)
   )
  ]}
 BackToHub {nodes [BackToHub() {is=title K2Node_FunctionEntry}]}
 CF_CanPurchaseOffer {nodes [
   n1 = CF_CanPurchaseOffer!() -> (then, Offer:Object Reference, ServicesStatus:bool=~) {is=title K2Node_FunctionEntry}
   > n2 = Branch(Condition=@n1.ServicesStatus)
   n3 = Get Offer() -> (Offer:Object Reference)
   > n4 = Cast To FortSimpleOfferVM!(execute, Object=@n3) -> (then, CastFailed, AsFort Simple Offer VM:Fort Simple Offer VM Object Reference) {is=title K2Node_DynamicCast}
   > n5 = Set CanPurchase(CanPurchase:bool=NOT Boolean(A=Get bIsOwned!(self:Fort Simple Offer VM Object Reference=@n4) -> (bIsOwned:bool))) -> (Output_Get:bool)
   Return Node!(
    execute=[@n5.then, @n6.then]
    CanPurchaseVisibility:ESlateVisibility Enum=Select(Option 0=Collapsed, Option 1=SelfHitTestInvisible, Index=Get CanPurchase() -> (CanPurchase:bool))
   )
   n6 = Set CanPurchase(execute=@n4.CastFailed, CanPurchase=IsValid(Object=@n3))
   Return Node(execute=@n2.else, CanPurchaseVisibility=Collapsed)
  ]}
 CF_GetVBucksText {nodes [
   n1 = CF_GetVBucksText!() -> (then, CanAfford:bool=~, OfferName:Text, IsBundle:bool=~) {is=title K2Node_FunctionEntry}
   > n2 = Branch(Condition=@n1)
   > Return Node!(execute, Result:Text)
   Return Node(
    execute=@n2.else
    Result=Format Text!(
     Format="LOCTEXT(\"{OfferName}{IsBundle} - \")"
     OfferName:Text=@n1.OfferName
     IsBundle:Text=Select(Option 0:Text, Option 1:Text="LOCTEXT(\" Bundle\")", Index=@n1.IsBundle) -> (ReturnValue:Text)
    ) -> (Result)
   )
  ]}
 EventGraph {nodes [
   "On Clicked (GetVBucks_Button)"!() -> (OutputDelegate, then, Button:Common Button Base Object Reference) {is=title K2Node_ComponentBoundEvent}
   > ClearPurchaseUXTraceId!(
    execute
    self:Fort Mcp Context Object Reference=Get FortMcpContext!() -> (ReturnValue:Fort Mcp Context Object Reference) {is=title K2Node_GetBlueprintContextEx}
    DebugSourceContext:String
   ) -> (then)
   > PopContentWidget!(execute) -> (then)
   n1 = Get SeasonPassVM()
   > TryShowReloadMtxStore!(
    execute
    OfferVM:Fort Simple Offer VM Object Reference=Get PassOffer!(self:Season Pass VM Object Reference=@n1) -> (PassOffer:Fort Simple Offer VM Object Reference)
    TitleOverride:"Text (by ref)"="To Text (String)"(InString=GetDisplayName(Object=@n1))
    SourceContext:String=BattlePassPurchaseScreen
   ) -> (then)
   n2 = "On Clicked (BackButton)"!() -> (OutputDelegate, then, Button:Common Button Base Object Reference) {is=title K2Node_ComponentBoundEvent}
   > n3 = GetPurchaseTraceId!(execute) -> (then, ReturnValue:Flow Analytics Additional Info Structure)
   > Fire Event Element Interacted(Element=@n2.Button, AdditionalInfo=@n3)
   > ClearPurchaseUXTraceId(self=Get FortMcpContext())
   > PopContentWidget()
   n4 = "On Clicked (Gift_Button)"!() -> (OutputDelegate, then, Button:Common Button Base Object Reference) {is=title K2Node_ComponentBoundEvent}
   > n5 = GetPurchaseTraceId()
   > Fire Event Element Interacted(Element=@n4.Button, AdditionalInfo=@n5)
   > ShowGiftingScreen(self:Season Pass VM Object Reference=Get SeasonPassVM())
   "On Clicked (GetVBucksBundle_Button)"!() -> (OutputDelegate, then, Button:Common Button Base Object Reference) {is=title K2Node_ComponentBoundEvent}
   > ClearPurchaseUXTraceId(self=Get FortMcpContext())
   > PopContentWidget()
   n6 = Get SeasonPassVM()
   > TryShowReloadMtxStore(
    OfferVM=Get PassBundleOffer!(self:Season Pass VM Object Reference=@n6) -> (PassBundleOffer:Season Pass Bundle Offer VM Object Reference)
    TitleOverride="To Text (String)"(InString=GetDisplayName(Object=@n6))
    SourceContext=BattlePassPurchaseScreen
   )
   n7 = "On Clicked (ConfirmPurchase_Button)"!() -> (OutputDelegate, then, Button:Common Button Base Object Reference) {is=title K2Node_ComponentBoundEvent}
   > n8 = GetPurchaseTraceId()
   > Fire Event Element Interacted(
    Element=@n7.Button
    AdditionalInfo=Add Pass Name and Id!(Info:"Flow Analytics Additional Info Structure (by ref)"=@n8, Value:String=Paid, UpsellType:ENexusUpsellEntryType Enum=Paid) -> (ReturnValue:Flow Analytics Additional Info Structure) {is=title K2Node_MacroInstance}
   )
   > n9 = ShouldShowRightOfWithdrawalModal!(execute, Context:Widget Object Reference) -> (then, ReturnValue:bool)
   > n10 = Branch(Condition=@n9)
   n11 = OnCpcDismissedPass!() -> (OutputDelegate, then, Accepted:bool=~) {is=title K2Node_CustomEvent}
   n12 = Self-Reference()
   OnCpcDialogRequested!(
    execute=@n10.then
    Outer:Object Reference=@n12
    BodyText:Text="LOCTABLE(\"/SeasonPass/Data/ST_Passes\", \"CPC_BattlePass\")"
    OnCpcDialogDismissed:"Delegate (by ref)"=@n11
    Target:Widget Object Reference=@n12
    Context:User Widget Object Reference=@n12
   ) -> (then)
   n13 = Get SeasonPassVM()
   Process Pass Purchase!(
    execute:Exec=[@n27.then, @n10.else]
    SeasonPass:Season Pass VM Object Reference=@n13
    Offer:Fort Simple Offer VM Object Reference=Get PassOffer(self=@n13)
   ) -> (then:Exec) {is=title K2Node_MacroInstance}
   > ClearPurchaseUXTraceId(self=Get FortMcpContext())
   n14 = "On Clicked (BuyBundle_Button)"!() -> (OutputDelegate, then, Button:Common Button Base Object Reference) {is=title K2Node_ComponentBoundEvent}
   > n15 = GetPurchaseTraceId()
   > Fire Event Element Interacted(Element=@n14.Button, AdditionalInfo=Add Pass Name and Id(Info=@n15, Value=Bundle, UpsellType=Bundle))
   > n16 = ShouldShowRightOfWithdrawalModal()
   > n17 = Branch(Condition=@n16)
   n18 = OnCpcDismissedPassBundle!() -> (OutputDelegate, then, Accepted:bool=~) {is=title K2Node_CustomEvent}
   n19 = Self-Reference()
   OnCpcDialogRequested(
    execute=@n17.then
    Outer=@n19
    BodyText="LOCTABLE(\"&ST_Passes\", \"CPC_BattlePass\")"
    OnCpcDialogDismissed=@n18
    Target=@n19
    Context=@n19
   )
   n20 = Get SeasonPassVM()
   Process Pass Purchase(execute=[@n26.then, @n17.else], SeasonPass=@n20, Offer=Get PassBundleOffer(self=@n20))
   > ClearPurchaseUXTraceId(self=Get FortMcpContext())
   Event OnInitialized()
   > SetCompactUIStyle() {is=title K2Node_CallFunction}
   > SetScrollWidget!(execute, ToScroll:Widget Object Reference=Get ScrollBox_0() -> (ScrollBox_0:Scroll Box Object Reference)) -> (then)
   > n21 = ShouldShowRightOfWithdrawalModal()
   > Branch(Condition=@n21)
   n22 = Get ConfirmPurchase_Button() -> (ConfirmPurchase_Button:WBP UIKit Button Loud Object Reference)
   n23 = Get BuyBundle_Button() -> (BuyBundle_Button:WBP UIKit Button Regular Object Reference)
   > SetRequiresHold!(execute, self:Common Button Base Object Reference=[@n22, @n23], bInRequiresHold:bool=false) -> (then)
   > SetHideInputAction!(execute, self:Common Button Base Object Reference=[@n22, @n23], bInHideInputAction:bool=true) -> (then)
   Event On Activated()
   n24 = Get SeasonPassVM()
   > n25 = GeneratePurchaseTraceId!(execute, OfferId:String=GetEpicOfferId!(self:Fort Simple Offer VM Object Reference=Get PassOffer(self=@n24)) -> (ReturnValue:String)) -> (then, AnalyticsAdditionalInfo:Flow Analytics Additional Info Structure)
   > Fire Event Sub View Opened!(execute, View:Widget Object Reference=Self-Reference(), AdditionalInfo:Flow Analytics Additional Info Structure=@n25, bIsOverlay:bool=false) -> (then)
   n26 = Branch(execute=@n18.then, Condition=@n18.Accepted)
   n27 = Branch(execute=@n11.then, Condition=@n11.Accepted)
  ]}
 GetFrontendFlowInterruptConfig {nodes [
   GetFrontendFlowInterruptConfig() {is=title K2Node_FunctionEntry}
   > Return Node!(execute, ReturnValue:Frontend Flow Interrup Config Structure)
  ]}
 GetPrimaryContentSetup {nodes [
   GetPrimaryContentSetup() {is=title K2Node_FunctionEntry}
   > Return Node(ReturnValue:Primary Content Setup Structure=Make Primary Content Setup!(
     bShowBottomBar:bool=false
     bShowChatWidget:bool=false
     bEnableNewTextChat:bool=False
     bShowTopBar:bool=false
     bTabSupportsInterface:bool=False
    ) -> (PrimaryContentSetup:Primary Content Setup Structure) {is=title K2Node_MakeStruct})
  ]}
 GetPurchaseActiveWidget {nodes [
   n1 = GetPurchaseActiveWidget!() -> (then, CanAffordOffer:bool=~) {is=title K2Node_FunctionEntry}
   > Return Node!(
    execute
    Index:Widget Object Reference=Select(
     Option 0:WBP UIKit Button Loud Object Reference=Get GetVBucks_Button() -> (GetVBucks_Button:WBP UIKit Button Loud Object Reference)
     Option 1:WBP UIKit Button Loud Object Reference=Get ConfirmPurchase_Button()
     Index=@n1
    ) -> (ReturnValue:WBP UIKit Button Loud Object Reference)
   )
  ]}
 GetPurchaseBundleActiveWidget {nodes [
   n1 = GetPurchaseBundleActiveWidget!() -> (then, CanAffordBundle:bool=~) {is=title K2Node_FunctionEntry}
   > Return Node(Index=Select(
     Option 0:WBP UIKit Button Regular Object Reference=Get GetVBucksBundle_Button() -> (GetVBucksBundle_Button:WBP UIKit Button Regular Object Reference)
     Option 1:WBP UIKit Button Regular Object Reference=Get BuyBundle_Button()
     Index=@n1
    ) -> (ReturnValue:WBP UIKit Button Regular Object Reference))
  ]}
 GetRewardsBundleText {nodes [
   n1 = GetRewardsBundleText!() -> (then, RewardsAmount:int) {is=title K2Node_FunctionEntry}
   > Return Node!(
    execute
    Result:Text=Format Text!(Format="LOCTEXT(\"The Battle Bundle Includes the first {Amount} BATTLE PASS REWARDS at a discounted rate!\")", Amount:int=@n1) -> (Result)
   )
  ]}
 HandleContentWarningInfoBox {nodes [
   n1 = HandleContentWarningInfoBox!() -> (then, ContentWarningText:Text) {is=title K2Node_FunctionEntry}
   > n2 = Branch(Condition=TextIsEmpty(InText=@n1))
   n3 = Get WBP_UIKit_InfoBox_ContentWarning() -> (WBP_UIKit_InfoBox_ContentWarning:WBP UIKit Info Box Object Reference)
   > SetVisibility(self=@n3, InVisibility=Collapsed)
   SetText(execute=@n2.else, self:WBP UIKit Info Box Object Reference=@n3, InText:Text=Get ContentWarningText() -> (ContentWarningText:Text))
   > SetVisibility(self=@n3, InVisibility=SelfHitTestInvisible)
  ]}
 OnCompletePassPurchaseFlow {nodes [
   n1 = OnCompletePassPurchaseFlow!() -> (then, Condition:bool=~) {is=title K2Node_FunctionEntry}
   > HideProgressModal!(execute) -> (then)
   > Branch(Condition=@n1)
   > PopContentWidget()
   > Call Back to Hub!(execute:Exec, self:Season Pass Purchase Screen Object Reference) -> (then:Exec) {is=title K2Node_CallDelegate}
   > Unbind all Events from Back to Hub!(execute:Exec, self:Season Pass Purchase Screen Object Reference) -> (then:Exec) {is=title K2Node_ClearDelegate}
  ]}
 ProcessPassPurchase {nodes [
   n1 = Inputs!() -> (execute:Exec, SeasonPass:Season Pass VM Object Reference, Offer:Fort Simple Offer VM Object Reference)
   > n2 = PurchaseSeasonPassOffer!(execute:Exec, SeasonPass:Season Pass VM Object Reference=@n1, Offer:Fort Simple Offer VM Object Reference=@n1.Offer, Quantity:int=1) -> (then:Exec, OnOperationCompleted:Exec, Success:bool=~) {is=title K2Node_AsyncAction}
   > ShowProgressModal!(execute, Title:Text="LOCTEXT(\"Completing Purchase\")", Description:Text) -> (then)
   OnCompletePassPurchaseFlow!(execute=@n2.OnOperationCompleted, self, Condition:bool=@n2) -> (then) {is=title K2Node_CallFunction}
   > Outputs!(then:Exec)
  ]}
 SetCompactUIStyle {nodes [
   SetCompactUIStyle() {is=title K2Node_FunctionEntry}
   > Branch(Condition=IsMobileGame())
   > Set!(
    execute
    Padding_Left:float=48
    Padding_Top:float=24
    Padding_Right:float=48
    Padding_Bottom:float=24
    self:Safe Zone Slot Object Reference=SlotAsSafeBoxSlot!(Widget:Widget Object Reference=Get VB_ContentWrapper() -> (VB_ContentWrapper:Vertical Box Object Reference)) -> (ReturnValue:Safe Zone Slot Object Reference)
   ) -> (then, Output_Get:Margin Structure)
   > SetPadding!(
    execute
    self:Overlay Slot Object Reference=SlotAsOverlaySlot(Widget=Get Txt_Legal() -> (Txt_Legal:Common Text Object Reference))
    InPadding_Left:float=0
    InPadding_Top:float=0
    InPadding_Right:float=0
    InPadding_Bottom:float=0
   ) -> (then)
   n1 = GetViewportSize!() -> (ReturnValue_X:double, ReturnValue_Y:double)
   n2 = float / float(A=@n1, B=@n1.ReturnValue_Y)
   > SetPadding(
    self=SlotAsOverlaySlot(Widget=Get CrewDetails() -> (CrewDetails:Vertical Box Object Reference))
    InPadding_Left=128
    InPadding_Top=SelectFloat(A=20, B=256, bPickA="float >= float"(A=@n2, B=1.34))
    InPadding_Right=0
    InPadding_Bottom=0
   )
   > K2Node_Composite_0 = Crew Details Adjustments!(then:Exec) -> (exec:Exec) {is=title K2Node_Composite}
   > SetPadding(
    self=SlotAsOverlaySlot(Widget=Get WBP_ActionBarContainer() -> (WBP_ActionBarContainer:WBP Action Bar Container Object Reference))
    InPadding_Left=0
    InPadding_Top=0
    InPadding_Right=0
    InPadding_Bottom=0
   )
   > SetPadding(
    self=SlotAsOverlaySlot(Widget=Get WBP_PlayerInfo() -> (WBP_PlayerInfo:WBP Player Info Object Reference))
    InPadding_Left=0
    InPadding_Top=0
    InPadding_Right=0
    InPadding_Bottom=0
   )
   > SetHorizontalAlignment(self:Vertical Box Slot Object Reference=SlotAsVerticalBoxSlot(Widget=Get ConfirmPurchase_Button()), InHorizontalAlignment=HAlign_Fill)
   > SetHorizontalAlignment(self=SlotAsVerticalBoxSlot(Widget=Get BuyBundle_Button()), InHorizontalAlignment=HAlign_Fill)
  ]}
 Crew Details Adjustments {
  path=SetCompactUIStyle.K2Node_Composite_0.Crew Details Adjustments
  nodes [
   Inputs!() -> (then:Exec)
   > SetPadding(
    self:Vertical Box Slot Object Reference=SlotAsVerticalBoxSlot(Widget=Get SeasonPassBadgeIcon() -> (SeasonPassBadgeIcon:Season Pass Badge Icon Object Reference))
    InPadding_Left=0
    InPadding_Top=0
    InPadding_Right=0
    InPadding_Bottom=0
   )
   > SetPadding(
    self=SlotAsVerticalBoxSlot(Widget=Get Text_Header() -> (Text_Header:Common Text Object Reference))
    InPadding_Left=0
    InPadding_Top=0
    InPadding_Right=0
    InPadding_Bottom=0
   )
   > SetPadding(
    self=SlotAsVerticalBoxSlot(Widget=Get PassBenefits() -> (PassBenefits:Common Text Object Reference))
    InPadding_Left=0
    InPadding_Top=0
    InPadding_Right=0
    InPadding_Bottom=16
   )
   > SetPadding(self=SlotAsVerticalBoxSlot(Widget=Get BuyBundle_Button()), InPadding_Left=0, InPadding_Top=0, InPadding_Right=0, InPadding_Bottom=24)
   > SetPadding(
    self=SlotAsVerticalBoxSlot(Widget=Get LegalTextBundle() -> (LegalTextBundle:Common Rich Text Block Object Reference))
    InPadding_Left=0
    InPadding_Top=0
    InPadding_Right=0
    InPadding_Bottom=16
   )
   > SetPadding(
    self=SlotAsVerticalBoxSlot(Widget=Get Image_Separator() -> (Image_Separator:Image Object Reference))
    InPadding_Left=0
    InPadding_Top=0
    InPadding_Right=0
    InPadding_Bottom=24
   )
   > SetPadding(
    self=SlotAsVerticalBoxSlot(Widget=Get Gift_Button() -> (Gift_Button:WBP UIKit Button Regular Object Reference))
    InPadding_Left=0
    InPadding_Top=0
    InPadding_Right=0
    InPadding_Bottom=0
   )
   n1 = GetViewportSize()
   n2 = SafeDivide(A=@n1, B=@n1.ReturnValue_Y)
   > Set Size!(
    execute
    Size:Vector 2D Structure=Select(Option 0:Vector 2D Structure=(X=140,Y=140), Option 1:Vector 2D Structure=(X=80,Y=80), Index="float >= float"(A=@n2, B=1.34)) -> (ReturnValue:Vector 2D Structure)
    self:Season Pass Badge Icon Object Reference=Get SeasonPassBadgeIcon()
   ) -> (then, Output_Get:Vector 2D Structure)
   > Outputs!(exec:Exec)
  ]
 }
 SetLevelsToGrantText {nodes [
   n1 = SetLevelsToGrantText!() -> (then, GetLevelsToGrant:int) {is=title K2Node_FunctionEntry}
   > SetText(
    self:Rich Text Block Object Reference=Get LegalTextBundle()
    InText:"Text (by ref)"=Format Text!(
     Format="LOCTEXT(\"The Battle Bundle includes <text color=\\\"FFFFFFFF\\\" fontface=\\\"75 Medium\\\">{GetLevelsToGrant} BATTLE PASS REWARDS</> at a discount rate!\")"
     GetLevelsToGrant:int=@n1
    ) -> (Result)
   )
  ]}
 UpdateBundleOfferPriceChanged {nodes [
   n1 = UpdateBundleOfferPriceChanged!() -> (then, BundlePrice:int) {is=title K2Node_FunctionEntry}
   n2 = Get FortLocalPlayerVM() -> (FortLocalPlayerVM:Fort Local Player VM Object Reference)
   > Set with Broadcast bCanAffordBundleOffer!(
    execute
    bCanAffordBundleOffer:bool="integer <= integer"(A=@n1, B=GetMtxBalance!(self:Fort Local Player VM Object Reference=@n2) -> (ReturnValue:int=0))
   ) -> (then, Output_Get:bool)
   > Set Text(
    Text=Select(Option 0:Text="LOCTEXT(\"Get V-Bucks\")", Option 1:Text="LOCTEXT(\"Reload V-Bucks\")", Index=IsReloadMtxEnabled!() -> (ReturnValue:bool)) -> (ReturnValue:Text)
    self:Fort CTAButton Object Reference=Get GetVBucksBundle_Button()
   )
  ]}
 UpdateOfferPriceChanged {nodes [
   n1 = UpdateOfferPriceChanged!() -> (then, OfferPrice:int) {is=title K2Node_FunctionEntry}
   n2 = Get FortLocalPlayerVM()
   > Set with Broadcast bCanAffordOffer!(execute, bCanAffordOffer:bool="integer <= integer"(A=@n1, B=GetMtxBalance(self=@n2) -> (ReturnValue=0))) -> (then, Output_Get:bool)
   > Set Text(
    Text=Select(Option 0="LOCTEXT(\"Get V-Bucks\")", Option 1="LOCTEXT(\"Reload V-Bucks\")", Index=IsReloadMtxEnabled())
    self=Get GetVBucks_Button()
   )
  ]}
 VM_SetPotentialRewards {nodes [
   n1 = VM_SetPotentialRewards!() -> (then, PotentialRewards:int) {is=title K2Node_FunctionEntry}
   n2 = Get SeasonPassVM()
   n3 = Get RewardsAfterPurchaseInfo() -> (RewardsAfterPurchaseInfo:WBP UIKit Info Box Object Reference)
   > SetText(
    self:WBP UIKit Info Box Object Reference=@n3
    InText:Text=Format Text!(
     Format="LOCTEXT(\"After purchasing the {passName}, you will be able to claim {numRewards} rewards.\")"
     passName:Text=GetDisplayName!(self:Season Pass VM Object Reference=@n2) -> (ReturnValue:Text)
     numRewards:int=@n1
    ) -> (Result)
   )
   n4 = Get PotentialRewards() -> (PotentialRewards:int=0)
   > SetVisibility(
    self=@n3
    InVisibility=Select(Option 0:ESlateVisibility Enum=Collapsed, Option 1:ESlateVisibility Enum=SelfHitTestInvisible, Index="integer > integer"(A=@n4)) -> (ReturnValue:ESlateVisibility Enum)
   )
  ]}
 VM_SetRewardCountAndTheme {nodes [
   n1 = VM_SetRewardCountAndTheme!() -> (then, RewardCount:int, Theme:Season Pass Theme Data Structure) {is=title K2Node_FunctionEntry}
   n2 = Break Season Pass Theme Data!(SeasonPassThemeData:"Season Pass Theme Data Structure (by ref)"=@n1.Theme) -> (
    TitleColor:Linear Color Structure
    TitleMaterial:Soft Object Reference
    TitleOutline:Font Outline Settings Structure
    SeasonColor:Linear Color Structure
    ProgressionColor:Linear Color Structure
    OverlayColor:Linear Color Structure
    VaultBackgroundMaterial:Soft Object Reference
    GiftboxVaultBackgroundMaterial:Soft Object Reference
    UIBackgroundMaterial:Soft Object Reference
    UpsellBackgroundMaterial:Soft Object Reference
    CrewCategoryBackgroundMaterial:Soft Object Reference
    CustomLogo:Soft Object Reference
    CustomSeasonTitle:Text
    CustomSeasonTitleColor:Linear Color Structure
    Upsell2BackgrounMaterial:Soft Object Reference
    ImageReferenceTable:Fort Image Reference Table Structure
    ColorReferenceTable:Fort Color Reference Table Structure
    TextReferenceTable:Fort Text Reference Table Structure
   ) {is=title K2Node_BreakStruct}
   n3 = Get SeasonPassVM()
   > Return Node!(
    execute
    Text:Text=Format Text!(
     Format="LOCTEXT(\"You have claimed: <text color=\\\"{color}\\\" fontface=\\\"75 Medium\\\">{count}</><text color=\\\"FFFFFFFF\\\" fontface=\\\"75 Medium\\\">/{total}</>\")"
     color:String=ToHex!(InColor:Linear Color Structure=@n2.ProgressionColor) -> (ReturnValue:String)
     count:int=@n1
     total:int=GetTotalRewards!(self:Season Pass VM Object Reference=@n3) -> (ReturnValue:int=0)
    ) -> (Result)
   )
  ]}
 VM_SetThemeData {nodes [
   n1 = VM_SetThemeData!() -> (then, ThemeData:Season Pass Theme Data Structure) {is=title K2Node_FunctionEntry}
   > ApplyThemeToTitleText!(execute, TitleText:Common Text Object Reference=Get Text_Header(), ThemeData:Season Pass Theme Data Structure=@n1) -> (then)
   > ApplyThemeToSeasonText!(
    execute
    SeasonText:Common Text Object Reference=Get Text_Season() -> (Text_Season:Common Text Object Reference)
    ThemeData:Season Pass Theme Data Structure=@n1
   ) -> (then)
   > ApplyCustomThemeData!(
    execute
    ThemeData:Season Pass Theme Data Structure=@n1
    Switcher:Common Visibility Switcher Object Reference=Get CustomPassSwitcher() -> (CustomPassSwitcher:Common Visibility Switcher Object Reference)
    LazyImage:Fort Lazy Image Object Reference=Get CustomPassImage() -> (CustomPassImage:Fort Lazy Image Object Reference)
    Text:Common Text Object Reference=Get CustomPassName() -> (CustomPassName:Common Text Object Reference)
   ) -> (then)
   n2 = Break Season Pass Theme Data!(SeasonPassThemeData=@n1) -> (
    GiftboxVaultBackgroundMaterial
    UIBackgroundMaterial
    CrewCategoryBackgroundMaterial
    CustomLogo
    CustomSeasonTitle
    CustomSeasonTitleColor
    Upsell2BackgrounMaterial
    ImageReferenceTable
    ColorReferenceTable
    TextReferenceTable
   )
   > n3 = Branch(Condition=IsValidSoftObjectReference(SoftObjectReference=@n2.UIBackgroundMaterial))
   n4 = Get PassThemeBG() -> (PassThemeBG:Fort Lazy Image Object Reference)
   > SetBrushFromLazyDisplayAsset!(
    execute
    self:Common Lazy Image Object Reference=@n4
    LazyObject:"Soft Object Reference (by ref)"=@n2.UIBackgroundMaterial
    bMatchTextureSize:bool=false
   ) -> (then)
   > n5 = Set ColorAndOpacity!(execute, ColorAndOpacity:Linear Color Structure=(R=1,G=1,B=1,A=1), self:Image Object Reference=@n4) -> (then, Output_Get:Linear Color Structure)
   SetScalarParameterValue!(
    execute=[@n5.then, @n6.then]
    Collection:Material Parameter Collection Object Reference=/Game/Environments/FrontEnd/Blueprints/ItemPreview/VaultWorldTransitions/MaterialCollectionParameters/MPC_UI_Master_VaultWorld_Settings
    ParameterName:Name=Is Horizontal Orientation
    ParameterValue:float=0
   ) -> (then) {is=title K2Node_CallMaterialParameterCollectionFunction}
   > SetVectorParameterValue!(
    execute
    Collection:Material Parameter Collection Object Reference=&MPC_UI_Master_VaultWorld_Settings
    ParameterName:Name=FocalPointPosition
    ParameterValue_R:float=0
    ParameterValue_G:float=0
    ParameterValue_B:float=0
    ParameterValue_A:float=0
   ) -> (then) {is=title K2Node_CallMaterialParameterCollectionFunction}
   n6 = Set ColorAndOpacity(execute=@n3.else, ColorAndOpacity=(R=0,G=0,B=0,A=1), self=@n4)
  ]}
 VM_UpdateMtxBalanceChanged {nodes [
   VM_UpdateMtxBalanceChanged!() -> (then, MtxBalance:int) {is=title K2Node_FunctionEntry}
   n1 = Get SeasonPassVM()
   n2 = Get PassOffer(self=@n1)
   > UpdateOfferPriceChanged!(
    execute
    self
    OfferPrice:int="To Integer (Integer64)"!(InInt:Integer64=GetPrice!(self:Fort Base Offer VM Object Reference=@n2) -> (ReturnValue:Integer64=0)) -> (ReturnValue:int=0)
   ) -> (then) {is=title K2Node_CallFunction}
   n3 = Get PassBundleOffer(self=@n1)
   > UpdateBundleOfferPriceChanged!(execute, self, BundlePrice:int="To Integer (Integer64)"(InInt=GetPrice(self=@n3) -> (ReturnValue=0)) -> (ReturnValue=0)) -> (then) {is=title K2Node_CallFunction}
  ]}
]
variables [
 BackToHub {}
 bCanAffordBundleOffer {type=boolean}
 bCanAffordOffer {type=boolean}
]
widgets {
 info {
  inheritedWidgetCount=0
  namedSlotCount=5
  parentClass=&CommonActivatableWidget
  rootWidgetClass=&Overlay
  widgetCount=54
 }
 widgets [
  Overlay_1 {widgetClassPath=&Overlay, widget=Overlay_1}
  PassThemeBG {
   widgetClassPath=&FortLazyImage
   bIsVariable=true
   parent=Overlay_1
   widget=PassThemeBG
   slot {class=OverlaySlot, properties {horizontalAlignment=HAlign_Fill, verticalAlignment=VAlign_Fill}}
  }
  CMS_BG {
   widgetClassPath=/Script/EpicCMSUIFramework.EpicCMSImage
   bIsVariable=true
   parent=Overlay_1
   widget=CMS_BG
   properties {
    bMatchImageSize=false
    materialTextureParamName=Image
    visibility=HitTestInvisible
    brush {
     bIsDynamicallyLoaded=false
     drawAs=Image
     imageType=FullColor
     mirroring=NoMirror
     resourceName=None
     resourceObject=/SeasonPass/Materials/Background/MI_UI_SeasonPass_PurchaseScreen_Background
     tiling=NoTile
     imageSize=<32 32>
     margin=<ltrb 0 0 0 0>
     outlineSettings {
      bUseBrushTransparency=false
      roundingType=HalfHeightRadius
      width=0
      color {colorUseRule=UseColor_Specified, specifiedColor=<rgba 0 0 0 0>}
      cornerRadii=<0 0 0 0>
     }
     tintColor {colorUseRule=UseColor_Specified, specifiedColor=<rgba 1 1 1 1>}
     uVRegion {bIsValid=false, max=<0 0>, min=<0 0>}
    }
   }
   slot {class=OverlaySlot, properties {horizontalAlignment=HAlign_Fill, verticalAlignment=VAlign_Fill}}
  }
  WBP_CaptureForPostBufferUpdate {
   widgetClassPath=&WBP_CaptureForPostBufferUpdate
   bIsVariable=true
   parent=Overlay_1
   slot {class=OverlaySlot, properties {horizontalAlignment=HAlign_Fill, verticalAlignment=VAlign_Fill}}
  }
  SafeZone_0 {
   widgetClassPath=&SafeZone
   parent=Overlay_1
   widget=SafeZone_0
   slot {class=OverlaySlot, properties {horizontalAlignment=HAlign_Fill, verticalAlignment=VAlign_Fill}}
  }
  VB_ContentWrapper {
   widgetClassPath=&VerticalBox
   bIsVariable=true
   parent=SafeZone_0
   widget=VB_ContentWrapper
   slot {class=SafeZoneSlot, properties {padding=<ltrb 80 40 80 32>}}
  }
  HB_Content {
   widgetClassPath=&HorizontalBox
   parent=VB_ContentWrapper
   widget=HB_Content
   slot {class=VerticalBoxSlot, properties {size {sizeRule=Fill, value=1}}}
  }
  ScrollBox_0 {
   widgetClassPath=&ScrollBox
   bIsVariable=true
   parent=HB_Content
   widget=ScrollBox_0
   properties {scrollbarPadding=<ltrb 16 2 2 2>}
   slot {class=HorizontalBoxSlot, properties {horizontalAlignment=HAlign_Center, verticalAlignment=VAlign_Center, padding=<ltrb 8 24 0 0>}}
  }
  Overlay_0 {
   widgetClassPath=&Overlay
   parent=ScrollBox_0
   widget=Overlay_0
   slot {class=ScrollBoxSlot, properties {verticalAlignment=VAlign_Center}}
  }
  Min_Width {
   widgetClassPath=&Spacer
   parent=Overlay_0
   widget=Min_Width
   properties {size=<800 1>}
   slot {class=OverlaySlot, properties {horizontalAlignment=HAlign_Fill, verticalAlignment=VAlign_Fill}}
  }
  CrewDetails {
   widgetClassPath=&VerticalBox
   bIsVariable=true
   parent=Overlay_0
   widget=CrewDetails
   slot {class=OverlaySlot, properties {horizontalAlignment=HAlign_Fill, verticalAlignment=VAlign_Fill}}
  }
  SeasonPassBadgeIcon {
   widgetClassPath=&/SeasonPassBadgeIcon
   bIsVariable=true
   parent=CrewDetails
   widget=SeasonPassBadgeIcon
   properties {overrideToOwnPass=true, size=<140 140>}
   slot {class=VerticalBoxSlot, properties {horizontalAlignment=HAlign_Center, verticalAlignment=VAlign_Top, padding=<ltrb 0 0 0 24>}}
  }
  CustomPassSwitcher {
   widgetClassPath=&CommonVisibilitySwitcher
   bIsVariable=true
   parent=CrewDetails
   widget=CustomPassSwitcher
   slot {class=VerticalBoxSlot, properties {verticalAlignment=VAlign_Top, padding=<ltrb 0 0 0 16>}}
  }
  VerticalBox_0 {widgetClassPath=&VerticalBox, parent=CustomPassSwitcher, widget=VerticalBox_0, slot {class=CommonVisibilitySwitcherSlot}}
  Text_Season {
   widgetClassPath=&CommonTextBlock
   bIsVariable=true
   parent=VerticalBox_0
   widget=Text_Season
   properties {
    autoWrapText=true
    justification=Center
    text=Chapter - Season
    wrapTextAt=696
    colorAndOpacity {colorUseRule=UseColor_Specified, specifiedColor=<rgba 0.5776 0.5776 0.5776 1>}
    font {
     bForceMonospaced=false
     bMaterialIsStencil=false
     fontMaterial=null
     fontObject=&NotoSans
     letterSpacing=0
     monospacedWidth=1
     size=24
     skewAmount=0
     typefaceFontName=NotoSans-Regular
     outlineSettings {
      bApplyOutlineToDropShadows=false
      bMiteredCorners=false
      bSeparateFillAlpha=false
      outlineMaterial=null
      outlineSize=0
      outlineColor=<rgba 0 0 0 1>
     }
    }
    shadowOffset=<0 0>
   }
   slot {class=VerticalBoxSlot}
  }
  Text_Header {
   widgetClassPath=&CommonTextBlock
   bIsVariable=true
   parent=VerticalBox_0
   widget=Text_Header
   properties {
    autoWrapText=true
    justification=Center
    text=Battle Pass
    wrapTextAt=696
    colorAndOpacity {colorUseRule=UseColor_Specified, specifiedColor=<rgba 0.5776 0.5776 0.5776 1>}
    font {
     bForceMonospaced=false
     bMaterialIsStencil=false
     fontMaterial=null
     fontObject=&NotoSans
     letterSpacing=0
     monospacedWidth=1
     size=24
     skewAmount=0
     typefaceFontName=NotoSans-Regular
     outlineSettings {
      bApplyOutlineToDropShadows=false
      bMiteredCorners=false
      bSeparateFillAlpha=false
      outlineMaterial=null
      outlineSize=0
      outlineColor=<rgba 0 0 0 1>
     }
    }
    shadowOffset=<0 0>
   }
   slot {class=VerticalBoxSlot}
  }
  VerticalBox {widgetClassPath=&VerticalBox, parent=CustomPassSwitcher, widget=VerticalBox, slot {class=CommonVisibilitySwitcherSlot}}
  CustomPassImage {
   widgetClassPath=&FortLazyImage
   bIsVariable=true
   parent=VerticalBox
   widget=CustomPassImage
   properties {brush {
     bIsDynamicallyLoaded=false
     drawAs=Image
     imageType=NoImage
     mirroring=NoMirror
     resourceName=None
     resourceObject=null
     tiling=NoTile
     imageSize=<200 100>
     margin=<ltrb 0 0 0 0>
     outlineSettings {
      bUseBrushTransparency=false
      roundingType=HalfHeightRadius
      width=0
      color {colorUseRule=UseColor_Specified, specifiedColor=<rgba 0 0 0 0>}
      cornerRadii=<0 0 0 0>
     }
     tintColor {colorUseRule=UseColor_Specified, specifiedColor=<rgba 1 1 1 1>}
     uVRegion {bIsValid=false, max=<0 0>, min=<0 0>}
    }}
   slot {class=VerticalBoxSlot, properties {horizontalAlignment=HAlign_Center, verticalAlignment=VAlign_Top}}
  }
  CustomPassName {
   widgetClassPath=&CommonTextBlock
   bIsVariable=true
   parent=VerticalBox
   widget=CustomPassName
   properties {
    autoWrapText=true
    justification=Center
    text=Battle Pass
    wrapTextAt=696
    colorAndOpacity {colorUseRule=UseColor_Specified, specifiedColor=<rgba 0.5776 0.5776 0.5776 1>}
    font {
     bForceMonospaced=false
     bMaterialIsStencil=false
     fontMaterial=null
     fontObject=&NotoSans
     letterSpacing=0
     monospacedWidth=1
     size=24
     skewAmount=0
     typefaceFontName=NotoSans-Regular
     outlineSettings {
      bApplyOutlineToDropShadows=false
      bMiteredCorners=false
      bSeparateFillAlpha=false
      outlineMaterial=null
      outlineSize=0
      outlineColor=<rgba 0 0 0 1>
     }
    }
    shadowOffset=<0 0>
   }
   slot {class=VerticalBoxSlot, properties {padding=<ltrb 0 8 0 0>}}
  }
  PassBenefits {
   widgetClassPath=&CommonTextBlock
   bIsVariable=true
   parent=CrewDetails
   widget=PassBenefits
   properties {
    autoWrapText=true
    justification=Center
    text=Pass benefits
    wrapTextAt=800
    colorAndOpacity {colorUseRule=UseColor_Specified, specifiedColor=<rgba 1 1 1 0.75>}
    font {
     bForceMonospaced=false
     bMaterialIsStencil=false
     fontMaterial=null
     fontObject=&NotoSans
     letterSpacing=0
     monospacedWidth=1
     size=24
     skewAmount=0
     typefaceFontName=NotoSans-Regular
     outlineSettings {
      bApplyOutlineToDropShadows=false
      bMiteredCorners=false
      bSeparateFillAlpha=false
      outlineMaterial=null
      outlineSize=0
      outlineColor=<rgba 0 0 0 1>
     }
    }
    shadowOffset=<0 0>
   }
   slot {class=VerticalBoxSlot, properties {verticalAlignment=VAlign_Top, padding=<ltrb 0 0 0 16>}}
  }
  SizeBox_1 {
   widgetClassPath=&SizeBox
   parent=CrewDetails
   widget=SizeBox_1
   properties {bOverride_MaxDesiredWidth=true, maxDesiredWidth=600}
   slot {class=VerticalBoxSlot, properties {padding=<ltrb 32 0 32 24>}}
  }
  RewardsAfterPurchaseInfo {
   widgetClassPath=/FortUILibrary/UIKit/InfoBox/WBP_UIKit_InfoBox
   bIsVariable=true
   parent=SizeBox_1
   widget=RewardsAfterPurchaseInfo
   properties {previewText="After purchasing the Battle Pass, you will be able to claim 76 rewards", showIcon=false, visibility=Collapsed}
   slot {class=SizeBoxSlot}
  }
  ServicesDownError {
   widgetClassPath=&WBP_UIKit_InfoBox
   bIsVariable=true
   parent=CrewDetails
   widget=ServicesDownError
   properties {previewText="Services down, please try again later.", showIcon=false, showStatusIndicator=true, use Preview Text=true, visibility=Collapsed}
   slot {class=VerticalBoxSlot, properties {padding=<ltrb 0 0 0 24>}}
  }
  SizeBox_3 {
   widgetClassPath=&SizeBox
   parent=CrewDetails
   widget=SizeBox_3
   properties {bOverride_WidthOverride=true, widthOverride=800}
   slot {class=VerticalBoxSlot, properties {padding=<ltrb 0 0 0 24>}}
  }
  PurchaseSwitcher {
   widgetClassPath=&CommonActivatableWidgetSwitcher
   bIsVariable=true
   parent=SizeBox_3
   widget=PurchaseSwitcher
   slot {class=SizeBoxSlot}
  }
  ConfirmPurchase_Button {
   widgetClassPath=/FortUILibrary/UIKit/Buttons/WBP_UIKit_Button_Loud
   bIsVariable=true
   parent=PurchaseSwitcher
   widget=ConfirmPurchase_Button
   properties {bRequiresHold=true, buttonSize=Large, minHeight=56, overrideSecondaryTextSizeMultiplier=0.5, text=Confirm Purchase}
   slot {class=WidgetSwitcherSlot, properties {padding=<ltrb 16 0 16 0>}}
  }
  ConfirmPurchasePrice {
   widgetClassPath=/FortUILibrary/UIKit/Price/WBP_PriceSwitcher_SimpleOffer
   bIsVariable=true
   namedSlotHost=ConfirmPurchase_Button
   widget=ConfirmPurchasePrice
   properties {
    hasRealPrice=true
    realPriceTransformPolicy=ToUpper
    realPriceVerticalAlignment=VAlign_Center
    showCurrencyIcon=true
    showRealPriceOnTheSide=true
    widgetHeightWithRealPrice=28
    defaultRealPricePadding=<ltrb 4 0 0 0>
    realPriceTextColor=<rgba 0.0048 0.004 0.0116 1>
    realPriceTextFont {
     bForceMonospaced=false
     bMaterialIsStencil=false
     fontMaterial=/FortUILibrary/Systems/ItemShopTile/Materials/PriceSwitcher/MI_UIKit_PriceSwitcher_Text
     fontObject=&HeadingNow
     letterSpacing=0
     monospacedWidth=1
     size=15.75
     skewAmount=0
     typefaceFontName=76 Bold
     outlineSettings {
      bApplyOutlineToDropShadows=false
      bMiteredCorners=false
      bSeparateFillAlpha=false
      outlineMaterial=null
      outlineSize=0
      outlineColor=<rgba 0 0 0 1>
     }
    }
   }
  }
  GetVBucks_Button {
   widgetClassPath=&WBP_UIKit_Button_Loud
   bIsVariable=true
   parent=PurchaseSwitcher
   widget=GetVBucks_Button
   properties {bHideInputAction=true, buttonSize=Large, minHeight=56, overrideSecondaryTextSizeMultiplier=0.5, text=Get V-Bucks}
   slot {class=WidgetSwitcherSlot, properties {padding=<ltrb 16 0 16 0>}}
  }
  HorizontalBox_113 {widgetClassPath=&HorizontalBox, namedSlotHost=GetVBucks_Button, widget=HorizontalBox_113}
  GetVBuckSecondary {
   widgetClassPath=&CommonTextBlock
   parent=HorizontalBox_113
   widget=GetVBuckSecondary
   properties {
    text=Battle Pass
    textTransformPolicy=ToUpper
    colorAndOpacity {colorUseRule=UseColor_Specified, specifiedColor=<rgba 0 0 0 1>}
    font {
     bForceMonospaced=false
     bMaterialIsStencil=false
     fontMaterial=null
     fontObject=&HeadingNow
     letterSpacing=0
     monospacedWidth=1
     size=15.75
     skewAmount=0
     typefaceFontName=76 Bold
     outlineSettings {
      bApplyOutlineToDropShadows=false
      bMiteredCorners=false
      bSeparateFillAlpha=false
      outlineMaterial=null
      outlineSize=0
      outlineColor=<rgba 0 0 0 1>
     }
    }
    shadowOffset=<0 0>
   }
   slot {class=HorizontalBoxSlot}
  }
  GetVBucksPrice {
   widgetClassPath=&WBP_PriceSwitcher_SimpleOffer
   bIsVariable=true
   parent=HorizontalBox_113
   widget=GetVBucksPrice
   properties {
    hasRealPrice=true
    realPriceTransformPolicy=ToUpper
    realPriceVerticalAlignment=VAlign_Center
    showCurrencyIcon=true
    showRealPriceOnTheSide=true
    widgetHeightWithRealPrice=28
    defaultRealPricePadding=<ltrb 4 0 0 0>
    realPriceTextColor=<rgba 0.0048 0.004 0.0116 1>
    realPriceTextFont {
     bForceMonospaced=false
     bMaterialIsStencil=false
     fontMaterial=&MI_UIKit_PriceSwitcher_Text
     fontObject=&HeadingNow
     letterSpacing=0
     monospacedWidth=1
     size=15.75
     skewAmount=0
     typefaceFontName=76 Bold
     outlineSettings {
      bApplyOutlineToDropShadows=false
      bMiteredCorners=false
      bSeparateFillAlpha=false
      outlineMaterial=null
      outlineSize=0
      outlineColor=<rgba 0 0 0 1>
     }
    }
   }
   slot {class=HorizontalBoxSlot}
  }
  SizeBox_2 {
   widgetClassPath=&SizeBox
   parent=CrewDetails
   widget=SizeBox_2
   properties {bOverride_WidthOverride=true, widthOverride=800}
   slot {class=VerticalBoxSlot}
  }
  BundlePurchaseSwitcher {
   widgetClassPath=&CommonActivatableWidgetSwitcher
   bIsVariable=true
   parent=SizeBox_2
   widget=BundlePurchaseSwitcher
   slot {class=SizeBoxSlot}
  }
  BuyBundle_Button {
   widgetClassPath=&WBP_UIKit_Button_Regular
   bIsVariable=true
   parent=BundlePurchaseSwitcher
   widget=BuyBundle_Button
   properties {bRequiresHold=true, buttonSize=Large, minHeight=56, text=Buy Battle Bundle}
   slot {class=WidgetSwitcherSlot, properties {verticalAlignment=VAlign_Top, padding=<ltrb 16 0 16 34>}}
  }
  BuyBundlePrice {
   widgetClassPath=&WBP_PriceSwitcher_SimpleOffer
   bIsVariable=true
   namedSlotHost=BuyBundle_Button
   widget=BuyBundlePrice
   properties {
    hasRealPrice=true
    realPriceTransformPolicy=ToUpper
    realPriceVerticalAlignment=VAlign_Center
    showCurrencyIcon=true
    showRealPriceOnTheSide=true
    widgetHeightWithRealPrice=28
    defaultRealPricePadding=<ltrb 4 0 0 0>
    realPriceTextColor=<rgba 0.0048 0.004 0.0116 1>
    realPriceTextFont {
     bForceMonospaced=false
     bMaterialIsStencil=false
     fontMaterial=&MI_UIKit_PriceSwitcher_Text
     fontObject=&HeadingNow
     letterSpacing=0
     monospacedWidth=1
     size=15.75
     skewAmount=0
     typefaceFontName=76 Bold
     outlineSettings {
      bApplyOutlineToDropShadows=false
      bMiteredCorners=false
      bSeparateFillAlpha=false
      outlineMaterial=null
      outlineSize=0
      outlineColor=<rgba 0 0 0 1>
     }
    }
   }
  }
  GetVBucksBundle_Button {
   widgetClassPath=&WBP_UIKit_Button_Regular
   bIsVariable=true
   parent=BundlePurchaseSwitcher
   widget=GetVBucksBundle_Button
   properties {bHideInputAction=true, buttonSize=Large, minHeight=56, text=Get V-Bucks}
   slot {class=WidgetSwitcherSlot, properties {verticalAlignment=VAlign_Top, padding=<ltrb 16 0 16 34>}}
  }
  HorizontalBox_114 {widgetClassPath=&HorizontalBox, namedSlotHost=GetVBucksBundle_Button, widget=HorizontalBox_114}
  GetVBucksBundleSecondary {
   widgetClassPath=&CommonTextBlock
   parent=HorizontalBox_114
   widget=GetVBucksBundleSecondary
   properties {
    text=Battle Pass
    textTransformPolicy=ToUpper
    colorAndOpacity {colorUseRule=UseColor_Specified, specifiedColor=<rgba 0 0 0 1>}
    font {
     bForceMonospaced=false
     bMaterialIsStencil=false
     fontMaterial=null
     fontObject=&HeadingNow
     letterSpacing=0
     monospacedWidth=1
     size=15.75
     skewAmount=0
     typefaceFontName=76 Bold
     outlineSettings {
      bApplyOutlineToDropShadows=false
      bMiteredCorners=false
      bSeparateFillAlpha=false
      outlineMaterial=null
      outlineSize=0
      outlineColor=<rgba 0 0 0 1>
     }
    }
    shadowOffset=<0 0>
   }
   slot {class=HorizontalBoxSlot}
  }
  GetVBucksBundlePrice {
   widgetClassPath=&WBP_PriceSwitcher_SimpleOffer
   bIsVariable=true
   parent=HorizontalBox_114
   widget=GetVBucksBundlePrice
   properties {
    hasRealPrice=true
    realPriceTransformPolicy=ToUpper
    showCurrencyIcon=true
    showRealPriceOnTheSide=true
    widgetHeightWithRealPrice=28
    defaultRealPricePadding=<ltrb 4 0 0 0>
    realPriceTextColor=<rgba 0.0048 0.004 0.0116 1>
    realPriceTextFont {
     bForceMonospaced=false
     bMaterialIsStencil=false
     fontMaterial=&MI_UIKit_PriceSwitcher_Text
     fontObject=&HeadingNow
     letterSpacing=0
     monospacedWidth=1
     size=15.75
     skewAmount=0
     typefaceFontName=76 Bold
     outlineSettings {
      bApplyOutlineToDropShadows=false
      bMiteredCorners=false
      bSeparateFillAlpha=false
      outlineMaterial=null
      outlineSize=0
      outlineColor=<rgba 0 0 0 1>
     }
    }
   }
   slot {class=HorizontalBoxSlot}
  }
  LegalTextBundle {
   widgetClassPath=&CommonRichTextBlock
   bIsVariable=true
   parent=CrewDetails
   widget=LegalTextBundle
   properties {
    autoWrapText=true
    bOverrideDefaultStyle=true
    justification=Center
    text="The Battle Bundle includes <text color=\"FFFFFFFF\" fontface=\"75 Medium\">25 BATTLE PASS REWARDS</> at a discount rate!"
    wrapTextAt=632
    defaultTextStyleOverride {
     overflowPolicy=Clip
     transformPolicy=None
     colorAndOpacity {colorUseRule=UseColor_Specified, specifiedColor=<rgba 1 1 1 0.75>}
     font {
      bForceMonospaced=false
      bMaterialIsStencil=false
      fontMaterial=null
      fontObject=&HeadingNow
      letterSpacing=0
      monospacedWidth=1
      size=15.75
      skewAmount=0
      typefaceFontName=74 Regular
      outlineSettings {
       bApplyOutlineToDropShadows=false
       bMiteredCorners=false
       bSeparateFillAlpha=false
       outlineMaterial=null
       outlineSize=0
       outlineColor=<rgba 0 0 0 1>
      }
     }
     highlightColor {colorUseRule=UseColor_Specified, specifiedColor=<rgba 0 0 0 1>}
     highlightShape {
      bIsDynamicallyLoaded=false
      drawAs=Image
      imageType=NoImage
      mirroring=NoMirror
      resourceName=None
      resourceObject=null
      tiling=NoTile
      imageSize=<32 32>
      margin=<ltrb 0 0 0 0>
      outlineSettings {
       bUseBrushTransparency=false
       roundingType=HalfHeightRadius
       width=0
       color {colorUseRule=UseColor_Specified, specifiedColor=<rgba 0 0 0 0>}
       cornerRadii=<0 0 0 0>
      }
      tintColor {colorUseRule=UseColor_Specified, specifiedColor=<rgba 1 1 1 1>}
      uVRegion {bIsValid=false, max=<0 0>, min=<0 0>}
     }
     selectedBackgroundColor {colorUseRule=UseColor_Foreground, specifiedColor=<rgba 1 0 1 1>}
     shadowColorAndOpacity=<rgba 0 0 0 1>
     shadowOffset=<0 0>
     strikeBrush {
      bIsDynamicallyLoaded=false
      drawAs=Image
      imageType=NoImage
      mirroring=NoMirror
      resourceName=None
      resourceObject=null
      tiling=NoTile
      imageSize=<32 32>
      margin=<ltrb 0 0 0 0>
      outlineSettings {
       bUseBrushTransparency=false
       roundingType=HalfHeightRadius
       width=0
       color {colorUseRule=UseColor_Specified, specifiedColor=<rgba 0 0 0 0>}
       cornerRadii=<0 0 0 0>
      }
      tintColor {colorUseRule=UseColor_Specified, specifiedColor=<rgba 1 1 1 1>}
      uVRegion {bIsValid=false, max=<0 0>, min=<0 0>}
     }
     underlineBrush {
      bIsDynamicallyLoaded=false
      drawAs=Image
      imageType=NoImage
      mirroring=NoMirror
      resourceName=None
      resourceObject=null
      tiling=NoTile
      imageSize=<32 32>
      margin=<ltrb 0 0 0 0>
      outlineSettings {
       bUseBrushTransparency=false
       roundingType=HalfHeightRadius
       width=0
       color {colorUseRule=UseColor_Specified, specifiedColor=<rgba 0 0 0 0>}
       cornerRadii=<0 0 0 0>
      }
      tintColor {colorUseRule=UseColor_Specified, specifiedColor=<rgba 1 1 1 1>}
      uVRegion {bIsValid=false, max=<0 0>, min=<0 0>}
     }
    }
   }
   slot {class=VerticalBoxSlot, properties {verticalAlignment=VAlign_Center, padding=<ltrb 32 0 32 24>}}
  }
  Image_Separator {
   widgetClassPath=&Image
   bIsVariable=true
   parent=CrewDetails
   widget=Image_Separator
   properties {
    visibility=HitTestInvisible
    brush {
     bIsDynamicallyLoaded=false
     drawAs=Image
     imageType=NoImage
     mirroring=NoMirror
     resourceName=None
     resourceObject=null
     tiling=NoTile
     imageSize=<32 2>
     margin=<ltrb 0 0 0 0>
     outlineSettings {
      bUseBrushTransparency=false
      roundingType=HalfHeightRadius
      width=0
      color {colorUseRule=UseColor_Specified, specifiedColor=<rgba 0 0 0 0>}
      cornerRadii=<0 0 0 0>
     }
     tintColor {colorUseRule=UseColor_Specified, specifiedColor=<rgba 1 1 1 0.2>}
     uVRegion {bIsValid=false, max=<0 0>, min=<0 0>}
    }
   }
   slot {class=VerticalBoxSlot, properties {verticalAlignment=VAlign_Top, padding=<ltrb 0 0 0 24>}}
  }
  Gift_Button {
   widgetClassPath=&WBP_UIKit_Button_Regular
   bIsVariable=true
   parent=CrewDetails
   widget=Gift_Button
   properties {bHideInputAction=true, text=Gift The Pass}
   slot {class=VerticalBoxSlot, properties {verticalAlignment=VAlign_Top, padding=<ltrb 16 0 16 0>}}
  }
  Spacer_281 {
   widgetClassPath=&Spacer
   parent=CrewDetails
   widget=Spacer_281
   slot {class=VerticalBoxSlot, properties {size {sizeRule=Fill, value=1}}}
  }
  SizeBox_0 {
   widgetClassPath=&SizeBox
   parent=CrewDetails
   widget=SizeBox_0
   properties {bOverride_MaxDesiredWidth=true, maxDesiredWidth=800}
   slot {class=VerticalBoxSlot, properties {verticalAlignment=VAlign_Bottom, padding=<ltrb 16 16 16 16>}}
  }
  WBP_UIKit_InfoBox_ContentWarning {
   widgetClassPath=&WBP_UIKit_InfoBox
   bIsVariable=true
   parent=SizeBox_0
   widget=WBP_UIKit_InfoBox_ContentWarning
   properties {previewText=Disclaimer text here, showIcon=false, showStatusIndicator=true}
   slot {class=SizeBoxSlot}
  }
  WBP_PlayerInfo {
   widgetClassPath=/Game/UI/Frontend/ItemShop/MultiProduct/WBP_PlayerInfo
   bIsVariable=true
   parent=HB_Content
   widget=WBP_PlayerInfo
   slot {class=HorizontalBoxSlot, properties {horizontalAlignment=HAlign_Right, verticalAlignment=VAlign_Top, size {sizeRule=Fill, value=1}}}
  }
  WBP_UIKit_InfoBox {
   widgetClassPath=&WBP_UIKit_InfoBox
   bIsVariable=true
   parent=VB_ContentWrapper
   widget=WBP_UIKit_InfoBox
   properties {autoWrapText=false, previewText=You have claimed: XXX/XXX, showIcon=false}
   slot {class=VerticalBoxSlot, properties {horizontalAlignment=HAlign_Right, padding=<ltrb 0 0 0 16>}}
  }
  HB_Footer {widgetClassPath=&HorizontalBox, parent=VB_ContentWrapper, widget=HB_Footer, slot {class=VerticalBoxSlot}}
  VerticalBox_LegalTexts {
   widgetClassPath=&VerticalBox
   parent=HB_Footer
   widget=VerticalBox_LegalTexts
   slot {class=HorizontalBoxSlot, properties {verticalAlignment=VAlign_Bottom, padding=<ltrb 0 0 48 0>, size {sizeRule=Fill, value=1}}}
  }
  Txt_Legal {
   widgetClassPath=&CommonTextBlock
   bIsVariable=true
   parent=VerticalBox_LegalTexts
   widget=Txt_Legal
   properties {
    autoWrapText=true
    mobileFontSizeMultiplier=1.4
    text="[PH] Lots of legal text goes here. Read all this small text or else. jpdoisfj dsfoj dsfpoj dsfj fopdsfjojpdsf gfdopjkg opfdop opfd o pdfpo fdjpo"
    wrapTextAt=1208
    colorAndOpacity {colorUseRule=UseColor_Specified, specifiedColor=<rgba 0.5776 0.5776 0.5776 1>}
    font {
     bForceMonospaced=false
     bMaterialIsStencil=false
     fontMaterial=null
     fontObject=&NotoSans
     letterSpacing=0
     monospacedWidth=1
     size=24
     skewAmount=0
     typefaceFontName=NotoSans-Regular
     outlineSettings {
      bApplyOutlineToDropShadows=false
      bMiteredCorners=false
      bSeparateFillAlpha=false
      outlineMaterial=null
      outlineSize=0
      outlineColor=<rgba 0 0 0 1>
     }
    }
    shadowOffset=<0 0>
   }
   slot {class=VerticalBoxSlot}
  }
  WBP_RealPrice_LegalText {
   widgetClassPath=/FortUILibrary/UIKit/Price/WBP_RealPrice_LegalText
   bIsVariable=true
   parent=VerticalBox_LegalTexts
   widget=WBP_RealPrice_LegalText
   properties {autoWrap=true, wrapTextAt=0, textColor {colorUseRule=UseColor_Specified, specifiedColor=<rgba 0.5776 0.5776 0.5776 1>}}
   slot {class=VerticalBoxSlot, properties {padding=<ltrb 0 8 0 0>}}
  }
  WBP_ActionBarContainer {
   widgetClassPath=/Game/UI/Frontend/BottomBar/WBP_ActionBarContainer
   bIsVariable=true
   parent=HB_Footer
   widget=WBP_ActionBarContainer
   slot {class=HorizontalBoxSlot, properties {verticalAlignment=VAlign_Bottom}}
  }
  HorizontalBox {widgetClassPath=&HorizontalBox, namedSlotHost=WBP_ActionBarContainer, widget=HorizontalBox}
  BackButton {
   widgetClassPath=/FortUILibrary/UIKit/Buttons/WBP_UIKit_Button_Quiet
   bIsVariable=true
   parent=HorizontalBox
   widget=BackButton
   properties {bIsFocusable=false, text=Back, triggeringInputAction {dataTable=&InputActionDataTable, rowName=GenericBack}}
   slot {class=HorizontalBoxSlot}
  }
 ]
}
