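# Translucent UI material for a Season Pass upsell background, drawing an image parameter with an angled cut and a division line.
# The image is sampled with 16:9 scaled UVs and fades into a near-black color along an eased vertical gradient with adjustable range.
# A box-angle mask, horizontally or otherwise flipped through a parameter collection, provides the cut, toggled by DisplayCut.
# Subtracting a second, offset box-angle mask yields the division line blended over the color.
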Material /SeasonPassHub/Widgets/Materials/M_UI_Background_Upsell_8Up
graph {
 properties {
  bAutomaticallySetUsageInEditor=false
  bUsedWithStaticMesh=false
  blendMode=BLEND_Translucent
  materialDomain=MD_UI
  usedShadingModels=Default Lit
 }
 nodes [
  n1 = TextureCoordinate()
  n2 = MaterialFunctionCall(
   UV=LinearInterpolate(
    A=OneMinus(None=@n1)
    B=@n1
    Alpha=CollectionParameter() {collection=/SeasonPass/MaterialFunctions/MPC_UI_Upsell_Variables, parameterName=Flip Mask}
   )
   ScalePivotPoint=Constant2Vector()
  ) {materialFunction=&MF_UI_UVScaling}
  n3 = MaterialFunctionCall!(UV=@n2.ScaleY, Corner=Constant() {r=3}, Angle=Constant() {r=4}, Flip, Falloff) {materialFunction=/FortUILibrary/MaterialFunctions/Shapes/MF_UI_BoxAngle}
  LinearInterpolate(B=@n3.Result, Alpha=ScalarParameter() {defaultValue=1, parameterName=DisplayCut, sliderMax=1}) {constA=1}
  n4 = MaterialFunctionCall(TextureSize=Constant2Vector() {g=1080, r=1920}) {materialFunction=&MF_UI_UVScaling}
  n5 = TextureObjectParameter() {group=Upsell Settings, parameterName=Image, texture=/Game/Global/Textures/Default/Blanks/T_Blank_White_ZeroAlpha_sRGBColor}
  n6 = MaterialFunctionCall!(Use Local UV, Texture Size, Scale Pivot Point) {materialFunction=/Game/Environments/FrontEnd/Blueprints/ItemPreview/VaultWorldTransitions/MaterialFunctions/MF_UI_VaultWorld_UV}
  n7 = Saturate(None=MaterialFunctionCall(
    Input=ComponentMask(None=@n6.UV) {g=true}
    Input Low=ScalarParameter() {defaultValue=0.4, group=Upsell Settings, parameterName=BtmGradInputLow, sliderMax=1}
    Input High=ScalarParameter() {defaultValue=0.8, group=Upsell Settings, parameterName=BtmGradInputHigh, sliderMax=1}
   ) {materialFunction=&RemapValue_0-1}.Result)
  LinearInterpolate(
   A=LinearInterpolate(
    A=TextureSample(UVs=@n4.ScaleY, Tex=@n5).RGB
    B=Constant3Vector() {constant=<rgba 0.0048 0.004 0.0116 1>}
    Alpha=MaterialFunctionCall("Value (0-1)"=@n7) {desc=Color Gradient Range, materialFunction=&MF_UI_EaseCurves}.EaseOutCubic
   )
   Alpha=NamedRerouteUsage()
  )
  n8 = MaterialFunctionCall(UV=Add(A=@n2.ScaleY, B=Constant2Vector() {desc=Division Line Offset, r=-0.015}) {constB=0}, Corner=Constant() {r=3}, Angle=Constant() {r=6}) {materialFunction=&MF_UI_BoxAngle}
  NamedRerouteDeclaration(None=Saturate(None=Subtract(A=@n3.Result, B=@n8.Result))) {name=Division Line, nodeColor=<rgba 0 0 0 1>}
 ]
}
