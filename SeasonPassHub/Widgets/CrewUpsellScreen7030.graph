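# Full-screen activatable upsell offering a Season Pass beside Crew, using a two-column grid whose sides and mask can flip.
# Buttons log analytics and push the pass purchase or Crew info screen; focus returns to the last choice.
# Theme data async-loads the pass background material, ignoring stale loads; CMS image URLs provide an alternative.
# It also toggles background animation, requests hidden top/bottom bars and chat, and adjusts layout for mobile.
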WidgetBlueprint /SeasonPassHub/Widgets/CrewUpsellScreen7030
animations [Anim_Intro {binding_count=0, length_seconds=5, object=&self:Anim_Intro, sequence {bindings []}}]
graphs [
 BP_GetDesiredFocusTarget {nodes [
   Get Desired Focus Target()
   n1 = NOT Boolean(A=IsVisible(self=Get PassSide() -> (PassSide:Overlay Object Reference)))
   n2 = Get WasCrewButtonSelected() -> (WasCrewButtonSelected:bool)
   > Return Node!(
    execute
    ReturnValue:Widget Object Reference=Select(
     Option 0:Widget Object Reference=Get GetPassButton() -> (GetPassButton:WBP UIKit Button Loud Object Reference)
     Option 1:Widget Object Reference=Get JoinCrewButton() -> (JoinCrewButton:WBP UIKit Button Loud Object Reference)
     Index=OR Boolean(A=@n2, B=@n1)
    ) -> (ReturnValue:Widget Object Reference)
   )
  ]}
 EventGraph {nodes [
   n1 = "On Clicked (JoinCrewButton)"!() -> (OutputDelegate, then, Button:Common Button Base Object Reference) {is=title K2Node_ComponentBoundEvent}
   n2 = Get SeasonPassVM() -> (SeasonPassVM:Season Pass VM Object Reference)
   > Fire Event Element Interacted(
    Element=@n1.Button
    AdditionalInfo="Add Text (Localized)"!(
     Info:"Flow Analytics Additional Info Structure (by ref)"=Make Flow Analytics Additional Info!() -> (FlowAnalyticsAdditionalInfo:Flow Analytics Additional Info Structure) {is=title K2Node_MakeStruct}
     Key:String=PassName
     Value:"Text (by ref)"=GetDisplayName!(self:Season Pass VM Object Reference=@n2) -> (ReturnValue:Text)
    ) -> (ReturnValue:Flow Analytics Additional Info Structure)
   )
   > n3 = PushPrimaryContentClassAsync!(execute:Exec, WidgetClass:Common Activatable Widget Class Reference=/Crew/UI/CrewInfoContainer.CrewInfoContainer_C) -> (then:Exec, OnPrimaryContentPushed:Exec, ActivatableWidget:Common Activatable Widget Object Reference) {is=title K2Node_AsyncAction}
   Set WasCrewButtonSelected(execute=@n3.OnPrimaryContentPushed, WasCrewButtonSelected:bool=true) -> (Output_Get:bool)
   > n4 = Cast To CrewInfoContainer!(execute, Object=@n3) -> (then, CastFailed, AsCrew Info Container:Crew Info Container Object Reference) {is=title K2Node_DynamicCast}
   > Bind Event to Crew Purchased(self:Crew Info Container Object Reference=@n4, Delegate=Create Event()) {is=title K2Node_AddDelegate}
   n5 = "On Clicked (GetPassButton)"!() -> (OutputDelegate, then, Button:Common Button Base Object Reference) {is=title K2Node_ComponentBoundEvent}
   n6 = Get SeasonPassVM()
   > Fire Event Element Interacted(
    Element=@n5.Button
    AdditionalInfo="Add Text (Localized)"(Info=Make Flow Analytics Additional Info(), Key=PassName, Value=GetDisplayName(self=@n6))
   )
   > n7 = PushPrimaryContentClassAsync(WidgetClass=/SeasonPass/Widgets/SeasonPass_PurchaseScreen.SeasonPass_PurchaseScreen_C)
   n8 = Cast To SeasonPass_PurchaseScreen!(execute=@n7.OnPrimaryContentPushed, Object=@n7) -> (then, CastFailed, AsSeason Pass Purchase Screen:Season Pass Purchase Screen Object Reference) {is=title K2Node_DynamicCast}
   > SetSeasonPassVM!(execute, self:Season Pass Purchase Screen Object Reference=@n8, Viewmodel:Season Pass VM Object Reference=Get SeasonPassVM()) -> (then)
   > n9 = Set WasCrewButtonSelected(WasCrewButtonSelected=false)
   n10 = PopSelf() {is=title K2Node_CustomEvent}
   Bind Event to Back to Hub(execute=@n9.then, self:Season Pass Purchase Screen Object Reference=@n8, Delegate=@n10) {is=title K2Node_AddDelegate}
   "On Clicked (BackButton)"!() -> (OutputDelegate, then, Button:Common Button Base Object Reference) {is=title K2Node_ComponentBoundEvent}
   > DeactivateWidget()
   > Call On Closed!(execute:Exec, self:Crew Upsell Screen 7030 Object Reference) -> (then:Exec) {is=title K2Node_CallDelegate}
   n11 = RequestLoadSeasonPassBackground!() -> (OutputDelegate, then, Background Asset:Soft Object Reference) {is=title K2Node_CustomEvent}
   > Set RequestedSeasonPassBgSoftPath(RequestedSeasonPassBgSoftPath:Soft Object Path Structure=To Soft Object Path!(SoftObjectReference:Soft Object Reference=@n11.Background Asset) -> (ReturnValue:Soft Object Path Structure)) -> (Output_Get:Soft Object Path Structure)
   > n12 = Async Load Asset(Asset=@n11.Background Asset)
   n13 = Get RequestedSeasonPassBgSoftPath() -> (RequestedSeasonPassBgSoftPath:Soft Object Path Structure)
   n14 = GetSoftObjectPath(Object=@n12)
   Branch(
    execute=@n12.Completed
    Condition="Equal Exactly (String)"(A=BreakSoftObjectPath(InSoftObjectPath=@n14), B=BreakSoftObjectPath(InSoftObjectPath=@n13))
   )
   n15 = Get BPUpsellInterstitialVM() -> (BPUpsellInterstitialVM:BPUpsell Interstitial VM Object Reference)
   n16 = Break Crew Upsell Config!(CrewUpsellConfig:"Crew Upsell Config Structure (by ref)"=Get Config!(self:BPUpsell Interstitial VM Object Reference=@n15) -> (Config:Crew Upsell Config Structure)) -> (
    OverrideUpsellPass:String
    PassBackgroundUrl:Array of Motd Image Structures
    CrewBackgroundUrl:Array of Motd Image Structures
    bUsePassBackgroundMaterial:bool=~
    bUseCrewBackgroundMaterial:bool=~
    bCrewOnRightInUpsell:bool=~
    bCrew50PercentInUpsell:bool=~
    bCrew30PercentInUpsell:bool=~
    bCrewAsSliver:bool=~
    bCrewOnlyInUpsell:bool=~
    bPassOnlyInUpsell:bool=~
    bAnimateCrewBG:bool=~
    bUseNewUpsellFormat:bool=~
    bAllowToggleBundle:bool=~
    bDefaultShowBundle:bool=~
    EntryOverrides:Array of Nexus Upsell Entry Structures
   ) {is=title K2Node_BreakStruct}
   > n17 = Branch(Condition=AND Boolean(A=HasValidData!(self:Fort Interstitial VM Object Reference=@n15) -> (ReturnValue:bool), B=NOT Boolean(A=@n16.bUsePassBackgroundMaterial)))
   > SetupPassTexture!(execute, self, Brush Resource Obj:Object Reference) -> (then)
   SetupPassTexture(execute=@n17.else, Brush Resource Obj=@n12)
   n18 = Event PreConstruct()
   > Branch(Condition=@n18.IsDesignTime)
   > Set ColorAndOpacity!(
    execute
    ColorAndOpacity:Linear Color Structure=(R=1,G=1,B=1,A=1)
    self:Image Object Reference=Get Crew_Background() -> (Crew_Background:Epic CMSImage Object Reference)
   ) -> (then, Output_Get:Linear Color Structure)
   > SetColorAndOpacity(
    self:Image Object Reference=Get CMSImage_BG_SeasonPass() -> (CMSImage_BG_SeasonPass:Epic CMSImage Object Reference)
    InColorAndOpacity:Linear Color Structure=(R=0.0704,G=0,B=0.2874,A=1)
   )
   Event On Activated()
   > n19 = Sequence!(execute) -> (then_0, then_1, then_2:Exec)
   > n20 = Branch(Condition=IsValid(Object=Get SeasonPassVM()))
   Set SeasonPassVM(
    execute=@n20.else
    SeasonPassVM:Season Pass VM Object Reference=Get SelectedPass!(self:Season Pass Selection VM Object Reference=Get SeasonPassSelectionVM() -> (SeasonPassSelectionVM:Season Pass Selection VM Object Reference)) -> (SelectedPass:Season Pass VM Object Reference)
   ) -> (Output_Get:Season Pass VM Object Reference)
   DelayUntilNextTick!(execute=@n19.then_1) -> (then)
   > PlayUpsellMusic!(execute, self, bPlay:bool=true) -> (then)
   Fire Event Sub View Opened!(
    execute=@n19.then_2
    View:Widget Object Reference=Self-Reference()
    AdditionalInfo:Flow Analytics Additional Info Structure
    bIsOverlay:bool=false
   ) -> (then)
   Event On Deactivated()
   > Parent: On Deactivated() {is=title K2Node_CallParentFunction}
   > PlayUpsellMusic(bPlay=false)
   Event OnInitialized()
   > SetMobileStyle() {is=title K2Node_CallFunction}
   Call On Closed(execute=@n21.then)
   n21 = PopContentWidget!(execute=@n10.then) -> (then)
  ]}
 GetPrimaryContentSetup {nodes [
   GetPrimaryContentSetup() {is=title K2Node_FunctionEntry}
   > Return Node(ReturnValue:Primary Content Setup Structure=Make Primary Content Setup!(
     bShowBottomBar:bool=false
     bShowChatWidget:bool=false
     bEnableNewTextChat:bool=False
     bShowTopBar:bool=false
     bTabSupportsInterface:bool=False
    ) -> (PrimaryContentSetup:Primary Content Setup Structure) {is=title K2Node_MakeStruct})
  ]}
 LoadSeasonPassCMSImage {nodes [
   LoadSeasonPassCMSImage() {is=title K2Node_FunctionEntry}
   n1 = Get SeasonPassVM()
   > Branch(Condition=IsValid(Object=@n1))
   n2 = Get CMSImage_BG_SeasonPass()
   > SetMaterialTextureParamName!(execute, self:Common Lazy Image Object Reference=@n2, TextureParamName:Name=Image) -> (then)
   > SetMediaURL!(
    execute
    self:Epic CMSImage Object Reference=@n2
    MediaUrl:String=Break Season Pass CMSData!(SeasonPassCMSData:"Season Pass CMSData Structure (by ref)"=Get SeasonPassCMSData!(self:Season Pass VM Object Reference=@n1) -> (SeasonPassCMSData:Season Pass CMSData Structure)) -> (PurchaseBackgroundURL:String) {is=title K2Node_BreakStruct}
   ) -> (then)
  ]}
 OnClosed {nodes [OnClosed() {is=title K2Node_FunctionEntry}]}
 PlayUpsellMusic {nodes [PlayUpsellMusic!() -> (then, bPlay:bool=~) {is=title K2Node_FunctionEntry}]}
 SetAnimatingCrew {nodes [
   n1 = SetAnimatingCrew!() -> (then, Enable:bool=~) {is=title K2Node_FunctionEntry}
   > Set IsAnimating(IsAnimating:bool=@n1) -> (Output_Get:bool)
   > n2 = Sequence!(execute) -> (then_0, then_1)
   > n3 = GetDynamicMaterial(self=Get CMSImage_BG_SeasonPass())
   > SetScalarParameterValue(self=@n3, ParameterName=AnimateBackground, Value="To Float (Boolean)"(InBool=@n1))
   > SetScalarParameterValue(self=@n3, ParameterName=DisplayCut, Value=1)
   > SetScalarParameterValue(self=@n3, ParameterName=UseLocalUV, Value=1)
   n4 = GetDynamicMaterial(execute=@n2.then_1, self=Get Crew_Background())
   > SetScalarParameterValue(self=@n4, ParameterName=AnimateBackground, Value="To Float (Boolean)"(InBool=@n1))
   > SetScalarParameterValue(self=@n4, ParameterName=DisplayCut, Value=1)
  ]}
 SetConfig {nodes [
   n1 = SetConfig!() -> (then, Config:Crew Upsell Config Structure) {is=title K2Node_FunctionEntry}
   n2 = Break Crew Upsell Config(CrewUpsellConfig=@n1) -> (
    bUsePassBackgroundMaterial=~
    bUseCrewBackgroundMaterial=~
    bCrewOnRightInUpsell=~
    bCrew50PercentInUpsell=~
    bCrew30PercentInUpsell=~
    bCrewAsSliver=~
    bCrewOnlyInUpsell=~
    bPassOnlyInUpsell=~
    bAnimateCrewBG=~
    bUseNewUpsellFormat=~
    bAllowToggleBundle=~
    bDefaultShowBundle=~
   )
   > VM_SetSides!(execute, self, CrewOnRight:bool=@n2.bCrewOnRightInUpsell) -> (then)
   > SetAnimatingCrew!(execute, self, Enable:bool=@n2.bAnimateCrewBG) -> (then) {is=title K2Node_CallFunction}
  ]}
 SetMobileStyle {nodes [
   SetMobileStyle() {is=title K2Node_FunctionEntry}
   > Branch(Condition=IsMobileGame())
   n1 = SlotAsOverlaySlot(Widget=Get SZ_CrewDetails() -> (SZ_CrewDetails:Safe Zone Object Reference))
   n2 = Get!(self:Overlay Slot Object Reference=@n1) -> (Padding_Left:float=~, Padding_Top:float=~, Padding_Right:float=~, Padding_Bottom:float=~)
   > SetPadding!(
    execute
    self:Overlay Slot Object Reference=@n1
    InPadding_Left:float=@n2
    InPadding_Top:float=-96
    InPadding_Right:float=@n2.Padding_Right
    InPadding_Bottom:float=@n2.Padding_Bottom
   ) -> (then)
   n3 = SlotAsOverlaySlot(Widget=Get SZ_PassDetails() -> (SZ_PassDetails:Safe Zone Object Reference))
   n4 = Get(self=@n3) -> (Padding_Left=~, Padding_Top=~, Padding_Right=~, Padding_Bottom=~)
   > SetPadding(self=@n3, InPadding_Left=@n4, InPadding_Top=-96, InPadding_Right=@n4.Padding_Right, InPadding_Bottom=@n4.Padding_Bottom)
   > Set Size!(
    execute
    Size:Vector 2D Structure=(X=100,Y=100)
    self:Season Pass Badge Icon Object Reference=Get SeasonPassBadgeIcon() -> (SeasonPassBadgeIcon:Season Pass Badge Icon Object Reference)
   ) -> (then, Output_Get:Vector 2D Structure)
   n5 = Get CustomPassImage() -> (CustomPassImage:Fort Lazy Image Object Reference)
   n6 = Get!(self:Image Object Reference=@n5) -> (
    Brush_TintColor:Slate Color Structure
    Brush_DrawAs:ESlateBrushDrawType Enum
    Brush_Tiling:ESlateBrushTileType Enum
    Brush_Mirroring:ESlateBrushMirrorType Enum
    Brush_ImageSize:Vector 2D Structure
    Brush_Margin:Margin Structure
    Brush_ResourceObject:Object Reference
    Brush_OutlineSettings:Slate Brush Outline Settings Structure
   )
   > Set!(
    execute
    Brush_TintColor:Slate Color Structure=@n6
    Brush_DrawAs:ESlateBrushDrawType Enum=@n6.Brush_DrawAs
    Brush_Tiling:ESlateBrushTileType Enum=@n6.Brush_Tiling
    Brush_Mirroring:ESlateBrushMirrorType Enum=@n6.Brush_Mirroring
    Brush_ImageSize_X:double=230
    Brush_ImageSize_Y:double=115
    Brush_Margin:Margin Structure=@n6.Brush_Margin
    Brush_ResourceObject:Object Reference=@n6.Brush_ResourceObject
    Brush_OutlineSettings:Slate Brush Outline Settings Structure=@n6.Brush_OutlineSettings
    self:Image Object Reference=@n5
   ) -> (then, Output_Get:Slate Brush Structure)
  ]}
 SetThemeData {nodes [
   SetThemeData!() -> (then, ThemeData:Season Pass Theme Data Structure) {is=title K2Node_FunctionEntry}
   > n1 = Sequence!(execute) -> (then_0, then_1, then_2:Exec, then_3:Exec)
   > ApplyThemeToTitleText!(
    execute
    TitleText:Common Text Object Reference=Get PassName() -> (PassName:Common Text Object Reference)
    ThemeData:Season Pass Theme Data Structure=Get ThemeData() -> (ThemeData:Season Pass Theme Data Structure)
   ) -> (then)
   n2 = Get CMSImage_BG_SeasonPass()
   SetBrush(execute=@n1.then_1, self=@n2, InBrush=Make Slate Brush!() -> (SlateBrush:Slate Brush Structure) {is=title K2Node_MakeStruct})
   > SetColorAndOpacity(self=@n2, InColorAndOpacity=(R=0,G=0,B=0,A=1))
   n3 = Break Season Pass Theme Data!(SeasonPassThemeData:"Season Pass Theme Data Structure (by ref)"=Get ThemeData()) -> (
    GiftboxVaultBackgroundMaterial:Soft Object Reference
    UpsellBackgroundMaterial:Soft Object Reference
    CustomLogo:Soft Object Reference
    CustomSeasonTitle:Text
    CustomSeasonTitleColor:Linear Color Structure
    Upsell2BackgrounMaterial:Soft Object Reference
    ImageReferenceTable:Fort Image Reference Table Structure
    ColorReferenceTable:Fort Color Reference Table Structure
    TextReferenceTable:Fort Text Reference Table Structure
   ) {is=title K2Node_BreakStruct}
   Branch(execute=@n1.then_2, Condition=IsValidSoftObjectReference(SoftObjectReference=@n3.UpsellBackgroundMaterial))
   > RequestLoadSeasonPassBackground!(execute, self, Background Asset:Soft Object Reference=@n3.UpsellBackgroundMaterial) -> (then) {is=title K2Node_CallFunction}
   ApplyCustomThemeData!(
    execute=@n1.then_3
    ThemeData:Season Pass Theme Data Structure=Get ThemeData()
    Switcher:Common Visibility Switcher Object Reference=Get CustomPassImageSwitcher() -> (CustomPassImageSwitcher:Common Visibility Switcher Object Reference)
    LazyImage:Fort Lazy Image Object Reference=Get CustomPassImage()
    Text:Common Text Object Reference=Get CustomPassName() -> (CustomPassName:Common Text Object Reference)
   ) -> (then)
   > n4 = Branch(Condition=IsValidSoftObjectReference(SoftObjectReference=@n3.CustomLogo))
   > SetPadding(
    self:Vertical Box Slot Object Reference=SlotAsVerticalBoxSlot(Widget=Get GetPassButton())
    InPadding_Left=0
    InPadding_Top=32
    InPadding_Right=0
    InPadding_Bottom=0
   )
   SetPadding(
    execute=@n4.else
    self=SlotAsVerticalBoxSlot(Widget=Get GetPassButton())
    InPadding_Left=0
    InPadding_Top=56
    InPadding_Right=0
    InPadding_Bottom=0
   )
  ]}
 SetupPassTexture {nodes [
   n1 = SetupPassTexture!() -> (then, Brush Resource Obj:Object Reference) {is=title K2Node_FunctionEntry}
   > n2 = Is Valid(InputObject=@n1)
   n3 = Get CMSImage_BG_SeasonPass()
   n4 = SetBrushFromMaterial(execute=@n2.Is Not Valid, self=@n3, Material=/SeasonPassHub/Widgets/Materials/M_UI_Background_Upsell_8Up)
   SetColorAndOpacity(execute=[@n11.then, @n4.then], self=Get CMSImage_BG_SeasonPass(), InColorAndOpacity=(R=1,G=1,B=1,A=1))
   n5 = Get Config(self=Get BPUpsellInterstitialVM())
   n6 = Break Motd Image!(MotdImage:"Motd Image Structure (by ref)"="Get (a ref)"(
     Array:Array of Motd Image Structures=Break Crew Upsell Config(CrewUpsellConfig=@n5) -> (
      bUsePassBackgroundMaterial=~
      bUseCrewBackgroundMaterial=~
      bCrewOnRightInUpsell=~
      bCrew50PercentInUpsell=~
      bCrew30PercentInUpsell=~
      bCrewAsSliver=~
      bCrewOnlyInUpsell=~
      bPassOnlyInUpsell=~
      bAnimateCrewBG=~
      bUseNewUpsellFormat=~
      bAllowToggleBundle=~
      bDefaultShowBundle=~
     ).PassBackgroundUrl
     Dimension 1=0
    ) -> (Output:"Motd Image Structure (by ref)")) -> (Url:String, Width:int, Height:int) {is=title K2Node_BreakStruct}
   > n7 = Branch(Condition="Equal (Boolean)"(A:bool=IsEmpty(InString=@n6), B:bool) -> (ReturnValue:bool=~) {is=title K2Node_PromotableOperator})
   n8 = Get CMSImage_BG_SeasonPass()
   > SetMaterialTextureParamName(self=@n8, TextureParamName=Image)
   > n9 = SetMediaURL(self=@n8, MediaUrl=@n6)
   SetAnimatingCrew(execute=[@n9.then, @n10.then], Enable=Get IsAnimating() -> (IsAnimating:bool))
   n10 = LoadSeasonPassCMSImage(execute=@n7.else) {is=title K2Node_CallFunction}
   n11 = SetBrushResourceObject!(execute=@n2.Is Valid, self:Image Object Reference=@n3, ResourceObject:Object Reference=@n1) -> (then)
  ]}
 VM_SetSides {nodes [
   VM_SetSides!() -> (then, CrewOnRight:bool=~) {is=title K2Node_FunctionEntry}
   > SetGridSlotColumnAndRow!(
    execute
    Widget:Widget Object Reference=Get PassSide()
    Row:int=0
    Row Span:int=0
    Column:int=Select(Option 0:int=1, Option 1:int=0, Index=Get CrewOnRight() -> (CrewOnRight:bool)) -> (ReturnValue:int)
    Column Span:int=0
   ) -> (then)
   > SetGridSlotColumnAndRow(
    Widget=Get B_CrewDetails() -> (B_CrewDetails:Overlay Object Reference)
    Row=0
    Row Span=0
    Column=Select(Option 0=0, Option 1=1, Index=Get CrewOnRight())
    Column Span=0
   )
   > SetScalarParameterValue!(
    execute
    Collection:Material Parameter Collection Object Reference=/SeasonPass/MaterialFunctions/MPC_UI_Upsell_Variables
    ParameterName:Name=Flip Mask
    ParameterValue:float=Select(Option 0:float=1, Option 1:float=0, Index=Get CrewOnRight()) -> (ReturnValue:float)
   ) -> (then) {is=title K2Node_CallMaterialParameterCollectionFunction}
  ]}
]
variables [
 IsAnimating {type=boolean}
 OnClosed {}
 RequestedSeasonPassBgSoftPath {type=SoftObjectPath, default=""}
 UpsellMusicComponent {type=AudioComponent, default=null}
 WasCrewButtonSelected {type=boolean}
]
widgets {
 info {
  inheritedWidgetCount=0
  namedSlotCount=1
  parentClass=&CommonActivatableWidget
  rootWidgetClass=&Overlay
  widgetCount=48
 }
 widgets [
  Overlay_81 {widgetClassPath=&Overlay, widget=Overlay_81}
  Crew_Background {
   widgetClassPath=/Script/EpicCMSUIFramework.EpicCMSImage
   bIsVariable=true
   parent=Overlay_81
   widget=Crew_Background
   properties {
    bMatchImageSize=false
    materialTextureParamName=CMS_Crew_Bento
    visibility=HitTestInvisible
    brush {
     bIsDynamicallyLoaded=false
     drawAs=Image
     imageType=FullColor
     mirroring=NoMirror
     resourceName=None
     resourceObject=/Crew/UI/Materials/M_UI_CrewUpsell_50x50_US
     tiling=NoTile
     imageSize=<32 32>
     margin=<ltrb 0 0 0 0>
     outlineSettings {
      bUseBrushTransparency=false
      roundingType=HalfHeightRadius
      width=0
      color {colorUseRule=UseColor_Specified, specifiedColor=<rgba 0 0 0 0>}
      cornerRadii=<0 0 0 0>
     }
     tintColor {colorUseRule=UseColor_Specified, specifiedColor=<rgba 1 1 1 1>}
     uVRegion {bIsValid=false, max=<0 0>, min=<0 0>}
    }
   }
   slot {class=OverlaySlot, properties {horizontalAlignment=HAlign_Fill, verticalAlignment=VAlign_Fill}}
  }
  O_Content {
   widgetClassPath=&Overlay
   parent=Overlay_81
   widget=O_Content
   slot {class=OverlaySlot, properties {horizontalAlignment=HAlign_Fill, verticalAlignment=VAlign_Fill}}
  }
  GP_Sections {
   widgetClassPath=&GridPanel
   bIsVariable=true
   parent=O_Content
   widget=GP_Sections
   properties {bOverride_Cursor=true, columnFill [0.7,0.3], rowFill [1]}
   slot {class=OverlaySlot, properties {horizontalAlignment=HAlign_Fill, verticalAlignment=VAlign_Fill}}
  }
  PassSide {
   widgetClassPath=&Overlay
   bIsVariable=true
   parent=GP_Sections
   widget=PassSide
   slot {class=GridSlot, properties {padding=<ltrb -32 0 -32 0>}}
  }
  CMSImage_BG_SeasonPass {
   widgetClassPath=&EpicCMSImage
   bIsVariable=true
   parent=PassSide
   widget=CMSImage_BG_SeasonPass
   properties {bMatchImageSize=false, visibility=SelfHitTestInvisible}
   slot {class=OverlaySlot, properties {horizontalAlignment=HAlign_Fill, verticalAlignment=VAlign_Fill}}
  }
  SZ_PassDetails {
   widgetClassPath=&SafeZone
   bIsVariable=true
   parent=PassSide
   widget=SZ_PassDetails
   slot {class=OverlaySlot, properties {horizontalAlignment=HAlign_Fill, verticalAlignment=VAlign_Fill, padding=<ltrb 80 96 80 128>}}
  }
  VB_PassDetails {widgetClassPath=&VerticalBox, parent=SZ_PassDetails, widget=VB_PassDetails, slot {class=SafeZoneSlot}}
  Spacer_1 {
   widgetClassPath=&Spacer
   parent=VB_PassDetails
   widget=Spacer_1
   slot {class=VerticalBoxSlot, properties {size {sizeRule=Fill, value=4.5}}}
  }
  FortVisualAttachment_0 {
   widgetClassPath=&FortVisualAttachment
   parent=VB_PassDetails
   widget=FortVisualAttachment_0
   properties {contentAnchor=<0.5 1>}
   slot {class=VerticalBoxSlot, properties {horizontalAlignment=HAlign_Center, verticalAlignment=VAlign_Top, padding=<ltrb 0 0 0 40>}}
  }
  VerticalBox_2 {widgetClassPath=&VerticalBox, parent=FortVisualAttachment_0, widget=VerticalBox_2, slot {class=SizeBoxSlot}}
  Overlay_0 {
   widgetClassPath=&Overlay
   parent=VerticalBox_2
   widget=Overlay_0
   slot {class=VerticalBoxSlot, properties {horizontalAlignment=HAlign_Center, verticalAlignment=VAlign_Top}}
  }
  BadgeShadow {
   widgetClassPath=&Image
   parent=Overlay_0
   widget=BadgeShadow
   properties {
    visibility=HitTestInvisible
    brush {
     bIsDynamicallyLoaded=false
     drawAs=Image
     imageType=FullColor
     mirroring=NoMirror
     resourceName=None
     resourceObject=/SeasonPassHub/Widgets/Materials/MI_UI_CrewUpsell_BadgeShadow_Porridge
     tiling=NoTile
     imageSize=<32 32>
     margin=<ltrb 0 0 0 0>
     outlineSettings {
      bUseBrushTransparency=false
      roundingType=HalfHeightRadius
      width=0
      color {colorUseRule=UseColor_Specified, specifiedColor=<rgba 0 0 0 0>}
      cornerRadii=<0 0 0 0>
     }
     tintColor {colorUseRule=UseColor_Specified, specifiedColor=<rgba 1 1 1 1>}
     uVRegion {bIsValid=false, max=<0 0>, min=<0 0>}
    }
    colorAndOpacity=<rgba 1 1 1 0>
   }
   slot {class=OverlaySlot, properties {horizontalAlignment=HAlign_Fill, verticalAlignment=VAlign_Fill, padding=<ltrb -40 -40 -40 -40>}}
  }
  SeasonPassBadgeIcon {
   widgetClassPath=/SeasonPass/Widgets/SeasonPassBadgeIcon
   bIsVariable=true
   parent=Overlay_0
   widget=SeasonPassBadgeIcon
   properties {overrideToOwnPass=true, size=<140 140>}
   slot {class=OverlaySlot}
  }
  CustomPassImageSwitcher {
   widgetClassPath=&CommonVisibilitySwitcher
   bIsVariable=true
   parent=VerticalBox_2
   widget=CustomPassImageSwitcher
   slot {class=VerticalBoxSlot, properties {padding=<ltrb 0 16 0 0>}}
  }
  PassName {
   widgetClassPath=&CommonTextBlock
   bIsVariable=true
   parent=CustomPassImageSwitcher
   widget=PassName
   properties {
    justification=Center
    text=Pass Name
    wrapTextAt=656
    colorAndOpacity {colorUseRule=UseColor_Specified, specifiedColor=<rgba 0.5776 0.5776 0.5776 1>}
    font {
     bForceMonospaced=false
     bMaterialIsStencil=false
     fontMaterial=null
     fontObject=&NotoSans
     letterSpacing=0
     monospacedWidth=1
     size=24
     skewAmount=0
     typefaceFontName=NotoSans-Regular
     outlineSettings {
      bApplyOutlineToDropShadows=false
      bMiteredCorners=false
      bSeparateFillAlpha=false
      outlineMaterial=null
      outlineSize=0
      outlineColor=<rgba 0 0 0 1>
     }
    }
    shadowOffset=<0 0>
   }
   slot {class=CommonVisibilitySwitcherSlot}
  }
  VerticalBox_64 {
   widgetClassPath=&VerticalBox
   parent=CustomPassImageSwitcher
   widget=VerticalBox_64
   slot {class=CommonVisibilitySwitcherSlot, properties {padding=<ltrb 0 0 0 -24>}}
  }
  CustomPassImage {
   widgetClassPath=&FortLazyImage
   bIsVariable=true
   parent=VerticalBox_64
   widget=CustomPassImage
   properties {brush {
     bIsDynamicallyLoaded=false
     drawAs=Image
     imageType=NoImage
     mirroring=NoMirror
     resourceName=None
     resourceObject=null
     tiling=NoTile
     imageSize=<300 150>
     margin=<ltrb 0 0 0 0>
     outlineSettings {
      bUseBrushTransparency=false
      roundingType=HalfHeightRadius
      width=0
      color {colorUseRule=UseColor_Specified, specifiedColor=<rgba 0 0 0 0>}
      cornerRadii=<0 0 0 0>
     }
     tintColor {colorUseRule=UseColor_Specified, specifiedColor=<rgba 1 1 1 1>}
     uVRegion {bIsValid=false, max=<0 0>, min=<0 0>}
    }}
   slot {class=VerticalBoxSlot, properties {horizontalAlignment=HAlign_Center, verticalAlignment=VAlign_Top}}
  }
  CustomPassName {
   widgetClassPath=&CommonTextBlock
   bIsVariable=true
   parent=VerticalBox_64
   widget=CustomPassName
   properties {
    justification=Center
    text=Pass Name
    wrapTextAt=656
    colorAndOpacity {colorUseRule=UseColor_Specified, specifiedColor=<rgba 0.5776 0.5776 0.5776 1>}
    font {
     bForceMonospaced=false
     bMaterialIsStencil=false
     fontMaterial=null
     fontObject=&NotoSans
     letterSpacing=0
     monospacedWidth=1
     size=24
     skewAmount=0
     typefaceFontName=NotoSans-Regular
     outlineSettings {
      bApplyOutlineToDropShadows=false
      bMiteredCorners=false
      bSeparateFillAlpha=false
      outlineMaterial=null
      outlineSize=0
      outlineColor=<rgba 0 0 0 1>
     }
    }
    shadowOffset=<0 0>
   }
   slot {class=VerticalBoxSlot, properties {padding=<ltrb 0 8 0 0>}}
  }
  VerticalBox_0 {
   widgetClassPath=&VerticalBox
   parent=VB_PassDetails
   widget=VerticalBox_0
   slot {class=VerticalBoxSlot, properties {verticalAlignment=VAlign_Top}}
  }
  WBP_PriceSwitcher_Base {
   widgetClassPath=/FortUILibrary/Systems/ItemShopTile/WBP_PriceSwitcher_Base
   bIsVariable=true
   parent=VerticalBox_0
   widget=WBP_PriceSwitcher_Base
   properties {
    hasRealPrice=true
    priceDisplayMode=PriceOnly
    realPriceVerticalAlignment=VAlign_Center
    showRealPriceOnTheSide=true
    widgetHeight=44
    widgetHeightWithRealPrice=44
    currencyImageSize=<28 28>
    defaultRealPricePadding=<ltrb 4 0 0 0>
    mobileRealPricePadding=<ltrb 4 0 0 0>
    priceTextFont {
     bForceMonospaced=false
     bMaterialIsStencil=false
     fontMaterial=/FortUILibrary/Systems/ItemShopTile/Materials/PriceSwitcher/MI_UIKit_PriceSwitcher_Text
     fontObject=&HeadingNow
     letterSpacing=-10
     monospacedWidth=1
     size=25.5
     skewAmount=0
     typefaceFontName=86 Bold Italic
     outlineSettings {
      bApplyOutlineToDropShadows=false
      bMiteredCorners=false
      bSeparateFillAlpha=false
      outlineMaterial=null
      outlineSize=0
      outlineColor=<rgba 0 0 0 1>
     }
    }
    realPriceTextFont {
     bForceMonospaced=false
     bMaterialIsStencil=false
     fontMaterial=&MI_UIKit_PriceSwitcher_Text
     fontObject=&HeadingNow
     letterSpacing=0
     monospacedWidth=1
     size=15.75
     skewAmount=0
     typefaceFontName=74 Regular
     outlineSettings {
      bApplyOutlineToDropShadows=false
      bMiteredCorners=false
      bSeparateFillAlpha=false
      outlineMaterial=null
      outlineSize=0
      outlineColor=<rgba 0 0 0 1>
     }
    }
   }
   slot {class=VerticalBoxSlot, properties {horizontalAlignment=HAlign_Center, verticalAlignment=VAlign_Top}}
  }
  GetPassButton {
   widgetClassPath=/FortUILibrary/UIKit/Buttons/WBP_UIKit_Button_Loud
   bIsVariable=true
   parent=VerticalBox_0
   widget=GetPassButton
   properties {
    bHideInputAction=true
    minHeight=56
    minWidth=326
    navigation=&self:WidgetTree.GetPassButton.WidgetNavigation_0
    text=Get Pass
   }
   slot {class=VerticalBoxSlot, properties {horizontalAlignment=HAlign_Center, verticalAlignment=VAlign_Top, padding=<ltrb 0 56 0 0>}}
  }
  FortVisualAttachment_3 {
   widgetClassPath=&FortVisualAttachment
   parent=VB_PassDetails
   widget=FortVisualAttachment_3
   properties {contentAnchor=<0.5 0>}
   slot {class=VerticalBoxSlot, properties {horizontalAlignment=HAlign_Center, verticalAlignment=VAlign_Top, padding=<ltrb 0 40 0 0>}}
  }
  Txt_PassBenefits {
   widgetClassPath=&CommonTextBlock
   bIsVariable=true
   parent=FortVisualAttachment_3
   widget=Txt_PassBenefits
   properties {
    justification=Center
    text=Pass Benefits list
    wrapTextAt=656
    font {
     bForceMonospaced=false
     bMaterialIsStencil=false
     fontMaterial=null
     fontObject=&NotoSans
     letterSpacing=0
     monospacedWidth=1
     size=24
     skewAmount=0
     typefaceFontName=NotoSans-Regular
     outlineSettings {
      bApplyOutlineToDropShadows=false
      bMiteredCorners=false
      bSeparateFillAlpha=false
      outlineMaterial=null
      outlineSize=0
      outlineColor=<rgba 0 0 0 1>
     }
    }
    shadowOffset=<0 0>
   }
   slot {class=SizeBoxSlot}
  }
  Spacer_2 {
   widgetClassPath=&Spacer
   parent=VB_PassDetails
   widget=Spacer_2
   slot {class=VerticalBoxSlot, properties {size {sizeRule=Fill, value=1}}}
  }
  B_CrewDetails {
   widgetClassPath=&Overlay
   bIsVariable=true
   parent=GP_Sections
   widget=B_CrewDetails
   slot {class=GridSlot, properties {column=1}}
  }
  SZ_CrewDetails {
   widgetClassPath=&SafeZone
   bIsVariable=true
   parent=B_CrewDetails
   widget=SZ_CrewDetails
   slot {class=OverlaySlot, properties {horizontalAlignment=HAlign_Fill, verticalAlignment=VAlign_Fill, padding=<ltrb 80 96 80 128>}}
  }
  VB_CrewDetails {widgetClassPath=&VerticalBox, parent=SZ_CrewDetails, widget=VB_CrewDetails, slot {class=SafeZoneSlot}}
  Spacer_268 {
   widgetClassPath=&Spacer
   parent=VB_CrewDetails
   widget=Spacer_268
   slot {class=VerticalBoxSlot, properties {size {sizeRule=Fill, value=4.5}}}
  }
  FortVisualAttachment_1 {
   widgetClassPath=&FortVisualAttachment
   parent=VB_CrewDetails
   widget=FortVisualAttachment_1
   properties {contentAnchor=<0.5 1>}
   slot {class=VerticalBoxSlot, properties {horizontalAlignment=HAlign_Center, verticalAlignment=VAlign_Bottom, padding=<ltrb 0 0 0 40>}}
  }
  VerticalBox_3 {widgetClassPath=&VerticalBox, parent=FortVisualAttachment_1, widget=VerticalBox_3, slot {class=SizeBoxSlot}}
  Logo_Crew {
   widgetClassPath=&Image
   parent=VerticalBox_3
   widget=Logo_Crew
   properties {
    visibility=HitTestInvisible
    brush {
     bIsDynamicallyLoaded=false
     drawAs=Image
     imageType=FullColor
     mirroring=NoMirror
     resourceName=None
     resourceObject=/Game/Athena/UI/Logos/Textures/M_UI_CrewLogo_Large
     tiling=NoTile
     imageSize=<432 128>
     margin=<ltrb 0 0 0 0>
     outlineSettings {
      bUseBrushTransparency=false
      roundingType=HalfHeightRadius
      width=0
      color {colorUseRule=UseColor_Specified, specifiedColor=<rgba 0 0 0 0>}
      cornerRadii=<0 0 0 0>
     }
     tintColor {colorUseRule=UseColor_Specified, specifiedColor=<rgba 1 1 1 1>}
     uVRegion {bIsValid=false, max=<0 0>, min=<0 0>}
    }
   }
   slot {class=VerticalBoxSlot, properties {horizontalAlignment=HAlign_Center, verticalAlignment=VAlign_Top, padding=<ltrb 0 0 0 32>}}
  }
  VerticalBox_1 {
   widgetClassPath=&VerticalBox
   parent=VB_CrewDetails
   widget=VerticalBox_1
   slot {class=VerticalBoxSlot, properties {verticalAlignment=VAlign_Top}}
  }
  Txt_CrewPrice {
   widgetClassPath=&CommonTextBlock
   bIsVariable=true
   parent=VerticalBox_1
   widget=Txt_CrewPrice
   properties {
    justification=Center
    text=$11.99
    wrapTextAt=1056
    font {
     bForceMonospaced=false
     bMaterialIsStencil=false
     fontMaterial=null
     fontObject=&NotoSans
     letterSpacing=0
     monospacedWidth=1
     size=24
     skewAmount=0
     typefaceFontName=NotoSans-Regular
     outlineSettings {
      bApplyOutlineToDropShadows=false
      bMiteredCorners=false
      bSeparateFillAlpha=false
      outlineMaterial=null
      outlineSize=0
      outlineColor=<rgba 0 0 0 1>
     }
    }
    shadowOffset=<0 0>
   }
   slot {class=VerticalBoxSlot}
  }
  Txt_CrewFrequency {
   widgetClassPath=&CommonTextBlock
   parent=VerticalBox_1
   widget=Txt_CrewFrequency
   properties {
    justification=Center
    text=Per Month
    wrapTextAt=1056
    colorAndOpacity {colorUseRule=UseColor_Specified, specifiedColor=<rgba 0.5776 0.5776 0.5776 1>}
    font {
     bForceMonospaced=false
     bMaterialIsStencil=false
     fontMaterial=null
     fontObject=&NotoSans
     letterSpacing=0
     monospacedWidth=1
     size=24
     skewAmount=0
     typefaceFontName=NotoSans-Regular
     outlineSettings {
      bApplyOutlineToDropShadows=false
      bMiteredCorners=false
      bSeparateFillAlpha=false
      outlineMaterial=null
      outlineSize=0
      outlineColor=<rgba 0 0 0 1>
     }
    }
    shadowOffset=<0 0>
   }
   slot {class=VerticalBoxSlot}
  }
  JoinCrewButton {
   widgetClassPath=&WBP_UIKit_Button_Loud
   bIsVariable=true
   parent=VerticalBox_1
   widget=JoinCrewButton
   properties {
    bHideInputAction=true
    minHeight=56
    minWidth=352
    navigation=&self:WidgetTree.JoinCrewButton.WidgetNavigation_0
    text=View Offer
   }
   slot {class=VerticalBoxSlot, properties {horizontalAlignment=HAlign_Center, padding=<ltrb 0 32 0 0>}}
  }
  FortVisualAttachment {
   widgetClassPath=&FortVisualAttachment
   parent=VB_CrewDetails
   widget=FortVisualAttachment
   properties {contentAnchor=<0.5 0>}
   slot {class=VerticalBoxSlot, properties {horizontalAlignment=HAlign_Center, verticalAlignment=VAlign_Top, padding=<ltrb 0 40 0 0>}}
  }
  Txt_CrewBenefits {
   widgetClassPath=&CommonTextBlock
   parent=FortVisualAttachment
   widget=Txt_CrewBenefits
   properties {
    justification=Center
    text=Crew benefits list
    wrapTextAt=480
    font {
     bForceMonospaced=false
     bMaterialIsStencil=false
     fontMaterial=null
     fontObject=&NotoSans
     letterSpacing=0
     monospacedWidth=1
     size=24
     skewAmount=0
     typefaceFontName=NotoSans-Regular
     outlineSettings {
      bApplyOutlineToDropShadows=false
      bMiteredCorners=false
      bSeparateFillAlpha=false
      outlineMaterial=null
      outlineSize=0
      outlineColor=<rgba 0 0 0 1>
     }
    }
    shadowOffset=<0 0>
   }
   slot {class=SizeBoxSlot}
  }
  Spacer {
   widgetClassPath=&Spacer
   parent=VB_CrewDetails
   widget=Spacer
   slot {class=VerticalBoxSlot, properties {size {sizeRule=Fill, value=1}}}
  }
  WBP_CaptureForPostBufferUpdate {
   widgetClassPath=&WBP_CaptureForPostBufferUpdate
   parent=O_Content
   slot {class=OverlaySlot, properties {horizontalAlignment=HAlign_Fill, verticalAlignment=VAlign_Fill}}
  }
  SZ_LegalTxt {
   widgetClassPath=&SafeZone
   parent=O_Content
   widget=SZ_LegalTxt
   properties {padRight=false, padTop=false}
   slot {class=OverlaySlot, properties {verticalAlignment=VAlign_Bottom, padding=<ltrb 80 0 0 32>}}
  }
  VerticalBox_LegalTexts {widgetClassPath=&VerticalBox, parent=SZ_LegalTxt, widget=VerticalBox_LegalTexts, slot {class=SafeZoneSlot}}
  Txt_Legal {
   widgetClassPath=&CommonTextBlock
   bIsVariable=true
   parent=VerticalBox_LegalTexts
   widget=Txt_Legal
   properties {
    mobileFontSizeMultiplier=1.25
    text="[PH] Lots of legal text goes here. Read all this small text or else. jpdoisfj dsfoj dsfpoj dsfj fopdsfjojpdsf gfdopjkg opfdop opfd o pdfpo fdjpo"
    wrapTextAt=672
    colorAndOpacity {colorUseRule=UseColor_Specified, specifiedColor=<rgba 0.5765 0.5765 0.5765 1>}
    font {
     bForceMonospaced=false
     bMaterialIsStencil=false
     fontMaterial=null
     fontObject=&NotoSans
     letterSpacing=0
     monospacedWidth=1
     size=24
     skewAmount=0
     typefaceFontName=NotoSans-Regular
     outlineSettings {
      bApplyOutlineToDropShadows=false
      bMiteredCorners=false
      bSeparateFillAlpha=false
      outlineMaterial=null
      outlineSize=0
      outlineColor=<rgba 0 0 0 1>
     }
    }
    shadowOffset=<0 0>
   }
   slot {class=VerticalBoxSlot}
  }
  WBP_RealPrice_LegalText {
   widgetClassPath=/FortUILibrary/UIKit/Price/WBP_RealPrice_LegalText
   bIsVariable=true
   parent=VerticalBox_LegalTexts
   widget=WBP_RealPrice_LegalText
   properties {autoWrap=true, wrapTextAt=0, textColor {colorUseRule=UseColor_Specified, specifiedColor=<rgba 0.5765 0.5765 0.5765 1>}}
   slot {class=VerticalBoxSlot, properties {padding=<ltrb 0 8 0 0>}}
  }
  SZ_ActionBar {
   widgetClassPath=&SafeZone
   parent=O_Content
   widget=SZ_ActionBar
   properties {padLeft=false, padTop=false}
   slot {class=OverlaySlot, properties {horizontalAlignment=HAlign_Right, verticalAlignment=VAlign_Bottom, padding=<ltrb 0 0 80 32>}}
  }
  WBP_ActionBarContainer {
   widgetClassPath=/Game/UI/Frontend/BottomBar/WBP_ActionBarContainer
   bIsVariable=true
   parent=SZ_ActionBar
   widget=WBP_ActionBarContainer
   slot {class=SafeZoneSlot}
  }
  HorizontalBox {widgetClassPath=&HorizontalBox, namedSlotHost=WBP_ActionBarContainer, widget=HorizontalBox}
  BackButton {
   widgetClassPath=/FortUILibrary/UIKit/Buttons/WBP_UIKit_Button_Quiet
   bIsVariable=true
   parent=HorizontalBox
   widget=BackButton
   properties {bIsFocusable=false, text=Back, triggeringInputAction {dataTable=&InputActionDataTable, rowName=GenericBack}}
   slot {class=HorizontalBoxSlot}
  }
 ]
}
