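# UI material for season pass upsell selling points, showing an item render over an anti-aliased circle mask.
# Scaled-to-fit UVs and a Content Offset parameter shift the render relative to the circle.
# The content mask joins the circle with a stepped upper-half band, apparently letting the item overflow the top, and multiplies by an edge clip.
# The render's RGB blends from a theme color using itself as alpha; its masked alpha is maxed with another mask.
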Material /SeasonPassHub/Widgets/Materials/M_UI_Upsell_SellingPoints
graph {
 properties {
  bAutomaticallySetUsageInEditor=false
  bUsedWithStaticMesh=false
  blendMode=BLEND_Translucent
  materialDomain=MD_UI
  usedShadingModels=Default Lit
 }
 nodes [
  n1 = Constant()
  n2 = ScalarParameter() {defaultValue=0.075, parameterName=Content Offset}
  n3 = TextureSampleParameter2D(UVs=Add(A=NamedRerouteUsage(), B=AppendVector(A=@n1, B=@n2))) {parameterName=Item Render, texture=/Game/UI/Foundation/Textures/Icons/Heroes/Athena/Soldier/S29/T-Soldier-Hero-PiteousKicks-L}
  LinearInterpolate(A=VectorParameter() {group=Theming, parameterName=Color Theme 2, defaultValue=<rgba 0.0331 0.3515 0.7084 1>}.RGB, B=@n3.RGB, Alpha=@n3.RGB)
  Max(A=Multiply(A=@n3.A, B=NamedRerouteUsage()), B=NamedRerouteUsage())
  n4 = Constant()
  n5 = Multiply(A=ScalarParameter() {defaultValue=0.075, parameterName=Content Offset}) {constB=-1}
  NamedRerouteDeclaration(None=Add(
    A=MaterialFunctionCall() {materialFunction=&MF_UI_UVScaling}.Scaled to Fit
    B=AppendVector(A=@n4, B=@n5)
   )) {name=UV, nodeColor=<rgba 0 0 0 1>}
  n6 = ScalarParameter() {defaultValue=0.8, parameterName=Circle Mask Size}
  n7 = Multiply(A=Distance(A=NamedRerouteUsage(), B=Constant() {r=0.5})) {constB=2}
  NamedRerouteDeclaration(None=OneMinus(None=MaterialFunctionCall!(Threshold=@n6, Gradient=@n7, Filter Width Override, Filter Width Fudge Factor) {materialFunction=/Game/UI/Materials/AAThreshold}.Result)) {name=Circle, nodeColor=<rgba 0 0 0 1>}
  n8 = NamedRerouteUsage()
  n9 = Distance(A=ComponentMask(None=@n8) {r=true}, B=Constant() {r=0.5})
  n10 = ComponentMask(None=@n8) {g=true}
  n11 = Min(A=Step(Y=Multiply(A=@n9) {constB=2}, X=ScalarParameter() {defaultValue=0.8, parameterName=Circle Mask Size}), B=Step(Y=@n10) {constX=0.5})
  n12 = NamedRerouteUsage()
  n13 = OneMinus(None=@n10)
  n14 = Add(A=Constant() {r=1}, B=ScalarParameter() {defaultValue=0.075, parameterName=Content Offset})
  NamedRerouteDeclaration(None=Multiply(A=Max(A=@n11, B=@n12), B=Step(Y=@n13, X=@n14) {constX=1.075})) {name=Content Mask, nodeColor=<rgba 0 0 0 1>}
 ]
}
