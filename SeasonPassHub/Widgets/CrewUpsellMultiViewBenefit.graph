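# Crew upsell widget in the Season Pass Hub that shows one benefit as an image beside a label and subtext.
# SetBenefit loads the CMS image URL, tints its dynamic material with the given color, and fills the texts, collapsing empty ones or the whole text box when ShowText is false.
# On Construct and a tick after each update, compact UI at 16:9 or wider disables subtext wrapping and wraps the label only when subtext is hidden.
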WidgetBlueprint /SeasonPassHub/Widgets/CrewUpsellMultiViewBenefit
graphs [
 EventGraph {nodes [
   OnBenefitUpdate() {is=title K2Node_CustomEvent}
   > DelayUntilNextTick!(execute) -> (then)
   > SetCompactUIStyle() {is=title K2Node_CallFunction}
   Event Construct()
   > SetCompactUIStyle()
  ]}
 SetBenefit {nodes [
   n1 = SetBenefit!() -> (then, Benefit:"Nexus Simple Benefit Structure (by ref)", ShowText:"Boolean (by ref)", Color:Linear Color Structure) {is=title K2Node_FunctionEntry}
   > OnBenefitUpdate() {is=title K2Node_CallFunction}
   n2 = Break Nexus Simple Benefit!(NexusSimpleBenefit:"Nexus Simple Benefit Structure (by ref)"=@n1) -> (Label:Text, SubText:Text, LabelB:Text, SubTextB:Text, ImageUrl:String, PosX:float=~, PosY:float=~) {is=title K2Node_BreakStruct}
   n3 = Get BenefitImage() -> (BenefitImage:Epic CMSImage Object Reference)
   > SetMediaURL!(execute, self:Epic CMSImage Object Reference=@n3, MediaUrl:String=@n2.ImageUrl) -> (then)
   > n4 = GetDynamicMaterial(self=@n3)
   > SetVectorParameterValue(self=@n4, ParameterName=Color Theme 2, Value=@n1.Color)
   > n5 = Branch(Condition=Get ShowText() -> (ShowText:"Boolean (by ref)"=false))
   n6 = Get Text!() -> (Text:Common Rich Text Block Object Reference)
   > SetVisibility(self=@n6, InVisibility=Select(Option 0=SelfHitTestInvisible, Option 1=Collapsed, Index=TextIsEmpty(InText=@n2)))
   > SetText(self=[@n6, Get Text()], InText=@n2)
   n7 = Get SubText() -> (SubText:Common Rich Text Block Object Reference)
   > SetVisibility(self=@n7, InVisibility=Select(Option 0=SelfHitTestInvisible, Option 1=Collapsed, Index=TextIsEmpty(InText=@n2.SubText)))
   > SetText(self=@n7, InText=@n2.SubText)
   SetVisibility(execute=@n5.else, self=Get VerticalBox_Text() -> (VerticalBox_Text:Vertical Box Object Reference), InVisibility=Collapsed)
  ]}
 SetCompactUIStyle {nodes [
   SetCompactUIStyle() {is=title K2Node_FunctionEntry}
   > Branch(Condition=AND Boolean(
     A=IsUsingCompactUI()
     B=CompareToScreenAspectRatio!(Comparison:EAspectRatioComparison Enum=GreaterThanOrEqualTo, AspectRatio:EAspectRatio Enum=Ratio_16_9) -> (ReturnValue:bool)
    ))
   n1 = Get SubText()
   > Set AutoWrapText!(execute, AutoWrapText:bool=false, self:Text Layout Widget Object Reference=@n1) -> (then, Output_Get:bool)
   > Set AutoWrapText(AutoWrapText=NOT Boolean(A=IsVisible(self=@n1)), self=Get Text())
  ]}
]
variables []
widgets {
 info {inheritedWidgetCount=0, parentClass=&CommonUserWidget, rootWidgetClass=&SizeBox, widgetCount=7}
 widgets [
  RootSizeBox {widgetClassPath=&SizeBox, bIsVariable=true, widget=RootSizeBox, properties {bOverride_MaxDesiredWidth=true, maxDesiredWidth=425}}
  HorizontalBox_28 {widgetClassPath=&HorizontalBox, parent=RootSizeBox, widget=HorizontalBox_28, slot {class=SizeBoxSlot}}
  BenefitImage {
   widgetClassPath=/Script/EpicCMSUIFramework.EpicCMSImage
   bIsVariable=true
   parent=HorizontalBox_28
   widget=BenefitImage
   properties {
    bMatchImageSize=false
    bOverride_Cursor=true
    materialTextureParamName=Item Render
    visibility=SelfHitTestInvisible
    brush {
     bIsDynamicallyLoaded=false
     drawAs=Image
     imageType=FullColor
     mirroring=NoMirror
     resourceName=None
     resourceObject=/SeasonPassHub/Widgets/Materials/M_UI_Upsell_SellingPoints
     tiling=NoTile
     imageSize=<88 80>
     margin=<ltrb 0 0 0 0>
     outlineSettings {
      bUseBrushTransparency=false
      roundingType=HalfHeightRadius
      width=0
      color {colorUseRule=UseColor_Specified, specifiedColor=<rgba 0 0 0 0>}
      cornerRadii=<0 0 0 0>
     }
     tintColor {colorUseRule=UseColor_Specified, specifiedColor=<rgba 1 1 1 1>}
     uVRegion {bIsValid=false, max=<0 0>, min=<0 0>}
    }
   }
   slot {class=HorizontalBoxSlot, properties {padding=<ltrb -8 -8 0 8>}}
  }
  SizeBox_0 {
   widgetClassPath=&SizeBox
   parent=HorizontalBox_28
   widget=SizeBox_0
   properties {bOverride_MaxDesiredWidth=true, bOverride_MinDesiredWidth=true, maxDesiredWidth=285, minDesiredWidth=200}
   slot {class=HorizontalBoxSlot, properties {verticalAlignment=VAlign_Center, size {sizeRule=Fill, value=1}}}
  }
  VerticalBox_Text {widgetClassPath=&VerticalBox, bIsVariable=true, parent=SizeBox_0, widget=VerticalBox_Text, slot {class=SizeBoxSlot}}
  Text {
   widgetClassPath=&CommonRichTextBlock
   bIsVariable=true
   parent=VerticalBox_Text
   widget=Text
   properties {
    autoWrapText=true
    bOverrideDefaultStyle=true
    mobileTextBlockScale=1.15
    scrollStyle=&TextScrollStyle-Base_C
    text="[PH] Benefit"
    defaultTextStyleOverride {
     overflowPolicy=Clip
     transformPolicy=None
     colorAndOpacity {colorUseRule=UseColor_Specified, specifiedColor=<rgba 1 1 1 1>}
     font {
      bForceMonospaced=false
      bMaterialIsStencil=false
      fontMaterial=null
      fontObject=&HeadingNow
      letterSpacing=10
      monospacedWidth=1
      size=15.75
      skewAmount=0
      typefaceFontName=75 Medium
      outlineSettings {
       bApplyOutlineToDropShadows=false
       bMiteredCorners=false
       bSeparateFillAlpha=false
       outlineMaterial=null
       outlineSize=0
       outlineColor=<rgba 0 0 0 1>
      }
     }
     highlightColor {colorUseRule=UseColor_Specified, specifiedColor=<rgba 0 0 0 1>}
     highlightShape {
      bIsDynamicallyLoaded=false
      drawAs=Image
      imageType=NoImage
      mirroring=NoMirror
      resourceName=None
      resourceObject=null
      tiling=NoTile
      imageSize=<32 32>
      margin=<ltrb 0 0 0 0>
      outlineSettings {
       bUseBrushTransparency=false
       roundingType=HalfHeightRadius
       width=0
       color {colorUseRule=UseColor_Specified, specifiedColor=<rgba 0 0 0 0>}
       cornerRadii=<0 0 0 0>
      }
      tintColor {colorUseRule=UseColor_Specified, specifiedColor=<rgba 1 1 1 1>}
      uVRegion {bIsValid=false, max=<0 0>, min=<0 0>}
     }
     selectedBackgroundColor {colorUseRule=UseColor_Foreground, specifiedColor=<rgba 1 0 1 1>}
     shadowColorAndOpacity=<rgba 0 0 0 1>
     shadowOffset=<0 0>
     strikeBrush {
      bIsDynamicallyLoaded=false
      drawAs=Image
      imageType=NoImage
      mirroring=NoMirror
      resourceName=None
      resourceObject=null
      tiling=NoTile
      imageSize=<32 32>
      margin=<ltrb 0 0 0 0>
      outlineSettings {
       bUseBrushTransparency=false
       roundingType=HalfHeightRadius
       width=0
       color {colorUseRule=UseColor_Specified, specifiedColor=<rgba 0 0 0 0>}
       cornerRadii=<0 0 0 0>
      }
      tintColor {colorUseRule=UseColor_Specified, specifiedColor=<rgba 1 1 1 1>}
      uVRegion {bIsValid=false, max=<0 0>, min=<0 0>}
     }
     underlineBrush {
      bIsDynamicallyLoaded=false
      drawAs=Image
      imageType=NoImage
      mirroring=NoMirror
      resourceName=None
      resourceObject=null
      tiling=NoTile
      imageSize=<32 32>
      margin=<ltrb 0 0 0 0>
      outlineSettings {
       bUseBrushTransparency=false
       roundingType=HalfHeightRadius
       width=0
       color {colorUseRule=UseColor_Specified, specifiedColor=<rgba 0 0 0 0>}
       cornerRadii=<0 0 0 0>
      }
      tintColor {colorUseRule=UseColor_Specified, specifiedColor=<rgba 1 1 1 1>}
      uVRegion {bIsValid=false, max=<0 0>, min=<0 0>}
     }
    }
   }
   slot {class=VerticalBoxSlot}
  }
  SubText {
   widgetClassPath=&CommonRichTextBlock
   bIsVariable=true
   parent=VerticalBox_Text
   widget=SubText
   properties {
    bAutoCollapseWithEmptyText=true
    bOverrideDefaultStyle=true
    mobileTextBlockScale=1.15
    scrollStyle=&TextScrollStyle-Base_C
    text="[PH] Subtext"
    visibility=Collapsed
    defaultTextStyleOverride {
     overflowPolicy=Clip
     transformPolicy=None
     colorAndOpacity {colorUseRule=UseColor_Specified, specifiedColor=<rgba 1 1 1 0.5>}
     font {
      bForceMonospaced=false
      bMaterialIsStencil=false
      fontMaterial=null
      fontObject=&HeadingNow
      letterSpacing=20
      monospacedWidth=1
      size=15.75
      skewAmount=0
      typefaceFontName=74 Regular
      outlineSettings {
       bApplyOutlineToDropShadows=false
       bMiteredCorners=false
       bSeparateFillAlpha=false
       outlineMaterial=null
       outlineSize=0
       outlineColor=<rgba 0 0 0 1>
      }
     }
     highlightColor {colorUseRule=UseColor_Specified, specifiedColor=<rgba 0 0 0 1>}
     highlightShape {
      bIsDynamicallyLoaded=false
      drawAs=Image
      imageType=NoImage
      mirroring=NoMirror
      resourceName=None
      resourceObject=null
      tiling=NoTile
      imageSize=<32 32>
      margin=<ltrb 0 0 0 0>
      outlineSettings {
       bUseBrushTransparency=false
       roundingType=HalfHeightRadius
       width=0
       color {colorUseRule=UseColor_Specified, specifiedColor=<rgba 0 0 0 0>}
       cornerRadii=<0 0 0 0>
      }
      tintColor {colorUseRule=UseColor_Specified, specifiedColor=<rgba 1 1 1 1>}
      uVRegion {bIsValid=false, max=<0 0>, min=<0 0>}
     }
     selectedBackgroundColor {colorUseRule=UseColor_Foreground, specifiedColor=<rgba 1 0 1 1>}
     shadowColorAndOpacity=<rgba 0 0 0 1>
     shadowOffset=<0 0>
     strikeBrush {
      bIsDynamicallyLoaded=false
      drawAs=Image
      imageType=NoImage
      mirroring=NoMirror
      resourceName=None
      resourceObject=null
      tiling=NoTile
      imageSize=<32 32>
      margin=<ltrb 0 0 0 0>
      outlineSettings {
       bUseBrushTransparency=false
       roundingType=HalfHeightRadius
       width=0
       color {colorUseRule=UseColor_Specified, specifiedColor=<rgba 0 0 0 0>}
       cornerRadii=<0 0 0 0>
      }
      tintColor {colorUseRule=UseColor_Specified, specifiedColor=<rgba 1 1 1 1>}
      uVRegion {bIsValid=false, max=<0 0>, min=<0 0>}
     }
     underlineBrush {
      bIsDynamicallyLoaded=false
      drawAs=Image
      imageType=NoImage
      mirroring=NoMirror
      resourceName=None
      resourceObject=null
      tiling=NoTile
      imageSize=<32 32>
      margin=<ltrb 0 0 0 0>
      outlineSettings {
       bUseBrushTransparency=false
       roundingType=HalfHeightRadius
       width=0
       color {colorUseRule=UseColor_Specified, specifiedColor=<rgba 0 0 0 0>}
       cornerRadii=<0 0 0 0>
      }
      tintColor {colorUseRule=UseColor_Specified, specifiedColor=<rgba 1 1 1 1>}
      uVRegion {bIsValid=false, max=<0 0>, min=<0 0>}
     }
    }
   }
   slot {class=VerticalBoxSlot, properties {padding=<ltrb 0 4 0 0>}}
  }
 ]
}
