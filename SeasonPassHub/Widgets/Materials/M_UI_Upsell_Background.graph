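# UI material for the Season Pass hub upsell background, drawing an animated, color-themed spotlight over a dark gradient with a vignette.
# A State Controller scalar blends colors, focal position and spotlight scale across three slots, with Display 2 Tiles switching colors and positions to a two-slot layout.
# Panning, rotating noise supplies spotlight distortion, and eased radial falloff shapes its color-dodge glow.
# Time-seeded screen-space randomness adds subtle hue-shift and dodge grain.
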Material /SeasonPassHub/Widgets/Materials/M_UI_Upsell_Background
graph {
 properties {bAutomaticallySetUsageInEditor=false, bUsedWithStaticMesh=false, materialDomain=MD_UI, usedShadingModels=Default Lit}
 nodes [
  MaterialFunctionCall(Speed=Constant() {r=0.2}) {materialFunction=&DebugTimeSine}
  n1 = MaterialFunctionCall!(UV=ScreenPosition().ViewportUV, Seed=Time()) {materialFunction=&MF_UI_Random}
  n2 = NamedRerouteUsage()
  n3 = VectorParameter() {parameterName=Vignette Color, defaultValue=<rgba 0.0021 0.0017 0.005 1>}
  n4 = Multiply(A=NamedRerouteUsage(), B=ScalarParameter() {defaultValue=0.8, parameterName=Vignetter Opacity})
  MaterialFunctionCall!(
   Blend=Min(A=@n1.Result) {constB=0.15}
   Base=MaterialFunctionCall(Hue Shift Percentage=Multiply(A=@n1.Result) {constB=0.01}, Texture=LinearInterpolate(A=@n2, B=@n3.RGB, Alpha=@n4) {constA=1}) {materialFunction=&HueShift}.Result
  ) {materialFunction=/Engine/Functions/Engine_MaterialFunctions03/Blends/Blend_ColorDodge}
  n5 = VectorParameter() {group=Multiple Values, parameterName=Slot 1 - Color Theme 2, defaultValue=<rgba 0.0003 0.3467 0.9734 1>}
  NamedRerouteDeclaration(None=@n5.RGB) {name=Slot 1 - Color 1, nodeColor=<rgba 1 1 1 1>}
  n6 = VectorParameter() {group=Multiple Values, parameterName=Slot 1 - Color Theme 4, defaultValue=<rgba 0.0024 0.0052 0.0103 1>}
  NamedRerouteDeclaration(None=@n6.RGB) {name=Slot 1 - Color 2, nodeColor=<rgba 1 1 1 1>}
  n7 = ComponentMask(None=TextureCoordinate()) {g=true}
  n8 = LinearInterpolate(
   A=NamedRerouteUsage()
   B=NamedRerouteUsage()
   Alpha=OneMinus(None=MaterialFunctionCall("Value (0-1)"=@n7) {materialFunction=&MF_UI_EaseCurves}.EaseInOutSine)
  )
  NamedRerouteDeclaration(None=Multiply(A=Power(Base=@n8) {constExponent=1.2}, B=ScalarParameter() {defaultValue=0.1, parameterName=Base Gradient Darkness}) {constB=0.1}) {name=Base Gradient, nodeColor=<rgba 0 0 0 1>}
  NamedRerouteDeclaration(None=AppendVector(
    A=ScalarParameter() {defaultValue=0.3, group=Multiple Values, parameterName=Slot 1 - Focal Position X}
    B=ScalarParameter() {defaultValue=0.1, group=Multiple Values, parameterName=Slot 1 - Focal Position Y}
   )) {name=Slot 1 - Focal Position, nodeColor=<rgba 1 1 1 1>}
  NamedRerouteDeclaration(None=AppendVector(
    A=ScalarParameter() {defaultValue=0.15, group=2 Slot Positions, parameterName=Left Slot - Focal Position X}
    B=ScalarParameter() {defaultValue=0.1, group=2 Slot Positions, parameterName=Left Slot - Focal Position Y}
   )) {name=Left - Focal Position, nodeColor=<rgba 1 1 1 1>}
  NamedRerouteDeclaration(None=AppendVector(
    A=ScalarParameter() {group=Multiple Values, parameterName=Slot 2 - Focal Position X}
    B=ScalarParameter() {group=Multiple Values, parameterName=Slot 2 - Focal Position Y}
   )) {name=Slot 2 - Focal Position, nodeColor=<rgba 0.5 0.5 0.5 1>}
  NamedRerouteDeclaration(None=AppendVector(
    A=ScalarParameter() {defaultValue=-0.15, group=2 Slot Positions, parameterName=Right Slot - Focal Position X}
    B=ScalarParameter() {defaultValue=0.1, group=2 Slot Positions, parameterName=Right Slot - Focal Position Y}
   )) {name=Right - Focal Position, nodeColor=<rgba 0 0 0 1>}
  NamedRerouteDeclaration(None=AppendVector(
    A=ScalarParameter() {defaultValue=-0.3, group=Multiple Values, parameterName=Slot 3 - Focal Position X}
    B=ScalarParameter() {defaultValue=0.1, group=Multiple Values, parameterName=Slot 3 - Focal Position Y}
   )) {name=Slot 3 - Focal Position, nodeColor=<rgba 0 0 0 1>}
  n9 = TextureCoordinate()
  n10 = Multiply(A=NamedRerouteUsage()) {constB=0.15}
  n11 = NamedRerouteUsage()
  n12 = LinearInterpolate(
   A=LinearInterpolate(A=NamedRerouteUsage(), B=NamedRerouteUsage(), Alpha=SmoothStep(Value=@n11) {constMax=0.5})
   B=NamedRerouteUsage()
   Alpha=SmoothStep(Value=NamedRerouteUsage()) {constMin=0.5}
  )
  n13 = ScalarParameter() {parameterName=Display 2 Tiles}
  n14 = LinearInterpolate(A=NamedRerouteUsage(), B=NamedRerouteUsage(), Alpha=NamedRerouteUsage())
  NamedRerouteDeclaration(None=Add(A=Add(A=@n9, B=@n10), B=LinearInterpolate(A=@n12, B=@n14, Alpha=@n13))) {name=Focal Position, nodeColor=<rgba 0 0 0 1>}
  n15 = MaterialFunctionCall(
   UVs=MaterialFunctionCall(UV=NamedRerouteUsage()) {materialFunction=&MF_UI_UVScaling}.Scaled To Fill
   Texture Scale=NamedRerouteUsage()
  ) {materialFunction=&ScaleUVsByCenter}
  n16 = Constant() {r=0.5}
  n17 = MaterialFunctionCall("Value (0-1)"=Multiply(A=Distance(A=@n15.UVs, B=@n16)) {constB=2}) {materialFunction=&MF_UI_EaseCurves}
  n18 = Multiply(A=OneMinus(None=@n17.EaseInOutSine), B=ScalarParameter() {defaultValue=0.5, parameterName=Spotlight Color Dodge Opacity}) {constB=0.8}
  n19 = NamedRerouteUsage()
  n20 = Multiply(
   A=LinearInterpolate(A=NamedRerouteUsage(), B=NamedRerouteUsage(), Alpha=@n17.EaseOutCubic)
   B=ScalarParameter() {defaultValue=0.75, parameterName=Spotlight Opacity}
  ) {constB=0.5}
  n21 = NamedRerouteUsage()
  NamedRerouteDeclaration(None=MaterialFunctionCall(Blend=Multiply(A=@n18, B=@n19), Base=LinearInterpolate(A=@n20, B=@n21, Alpha=@n17.EaseOutCubic)) {materialFunction=&Blend_ColorDodge}.Result) {name=Spotlight, nodeColor=<rgba 0 0 0 1>}
  NamedRerouteDeclaration(None=VectorParameter() {group=Multiple Values, parameterName=Slot 2 - Color Theme 2, defaultValue=<rgba 0.5333 1 0.0742 1>}.RGB) {name=Slot 2 - Color 1, nodeColor=<rgba 0.5 0.5 0.5 1>}
  n22 = NamedRerouteUsage()
  n23 = NamedRerouteUsage()
  n24 = SmoothStep(Value=NamedRerouteUsage()) {constMax=0.5}
  n25 = NamedRerouteUsage()
  NamedRerouteDeclaration(None=LinearInterpolate(A=LinearInterpolate(A=@n22, B=@n23, Alpha=@n24), B=NamedRerouteUsage(), Alpha=SmoothStep(Value=@n25) {constMin=0.5})) {name=SpotlightScale, nodeColor=<rgba 0 0 0 1>}
  NamedRerouteDeclaration(None=ScalarParameter() {defaultValue=2, group=Slot 1, parameterName=Slot 1 - Spotlight Scale}) {name=Slot 1 - Spotlight Scale, nodeColor=<rgba 0 0 0 1>}
  n26 = TextureCoordinate()
  n27 = Multiply(A=ScalarParameter() {defaultValue=2, parameterName=Distortion Speed}, B=Time())
  n28 = MaterialFunctionCall(
   UVs=Panner(Coordinate=Rotator(Coordinate=@n26, Time=@n27) {speed=0.2}, Time=@n27) {speedX=0.4}
   Texture Scale=ScalarParameter() {defaultValue=4, parameterName=Distortion Scale}
  ) {materialFunction=&ScaleUVsByCenter}
  NamedRerouteDeclaration(None=LinearInterpolate(Alpha=TextureSample(UVs=@n28.UVs) {samplerType=SAMPLERTYPE_LinearGrayscale, texture=/Game/Global/Textures/Default/Noise/LowResBlurredNoise_NoMips}.R) {constA=-0.5, constB=0.5}) {name=Spotlight Distortion, nodeColor=<rgba 0 0 0 1>}
  NamedRerouteDeclaration(None=ScalarParameter() {defaultValue=1.5, group=Slot 2, parameterName=Slot 2 - Spotlight Scale}) {name=Slot 2 - Spotlight Scale, nodeColor=<rgba 0 0 0 1>}
  NamedRerouteDeclaration(None=ScalarParameter() {defaultValue=2, group=Slot 3, parameterName=Slot 3 - Spotlight Scale}) {name=Slot 3 - Spotlight Scale, nodeColor=<rgba 0 0 0 1>}
  n29 = ComponentMask(None=TextureCoordinate()) {g=true}
  n30 = Power(Base=SmoothStep(Value=OneMinus(None=@n29)) {constMin=0.5}) {constExponent=3}
  n31 = SmoothStep(Value=ComponentMask(None=TextureCoordinate()) {g=true}) {constMin=0.1}
  NamedRerouteDeclaration(None=Max(A=Multiply(A=@n30) {constB=0.7}, B=Power(Base=@n31) {constExponent=3})) {name=Vignette, nodeColor=<rgba 0 0 0 1>}
  n32 = VectorParameter() {group=Multiple Values, parameterName=Slot 2 - Color Theme 4, defaultValue=<rgba 0.0027 0.0086 0.0018 1>}
  NamedRerouteDeclaration(None=@n32.RGB) {name=Slot 2 - Color 2, nodeColor=<rgba 0.5 0.5 0.5 1>}
  n33 = NamedRerouteUsage()
  n34 = LinearInterpolate(A=Constant() {r=2}, B=Constant() {r=1}, Alpha=ScalarParameter() {parameterName=Display 2 Tiles})
  n35 = Distance(A=Frac(None=Multiply(A=@n33, B=@n34) {constB=2}), B=Constant() {r=0.5})
  NamedRerouteDeclaration(None=MaterialFunctionCall("Value (0-1)"=Multiply(A=@n35) {constB=2}) {materialFunction=&MF_UI_EaseCurves}.EaseInOutCubic) {name=Transition Collor Mute, nodeColor=<rgba 0 0 0 1>}
  n36 = VectorParameter() {group=Multiple Values, parameterName=Slot 3 - Color Theme 2, defaultValue=<rgba 0.2789 0.0931 0.8469 1>}
  NamedRerouteDeclaration(None=@n36.RGB) {name=Slot 3 - Color 1, nodeColor=<rgba 0 0 0 1>}
  n37 = VectorParameter() {group=Multiple Values, parameterName=Slot 3 - Color Theme 4, defaultValue=<rgba 0.0097 0.0056 0.0307 1>}
  NamedRerouteDeclaration(None=@n37.RGB) {name=Slot 3 - Color 2, nodeColor=<rgba 0 0 0 1>}
  NamedRerouteDeclaration(None=ScalarParameter() {defaultValue=0.5, desc="0 = State 1\r\n0.5 = State 2\r\n1 = State 3", parameterName=State Controller}) {name=State Controller, nodeColor=<rgba 0 0 0 1>}
  n38 = MaterialFunctionCall!(Value=Multiply(A=Time()) {constB=0.15}, Scale, Height) {materialFunction=/FortUILibrary/MaterialFunctions/Basics/MF_UI_PingPong}
  n39 = SmoothStep(Value=@n38.Result) {constMax=0.35, constMin=0.25}
  n40 = SmoothStep(Value=@n38.Result) {constMax=0.85, constMin=0.65}
  NamedRerouteDeclaration(None=Max(A=LinearInterpolate(Alpha=@n39) {constB=0.5}, B=LinearInterpolate(Alpha=@n40))) {name=Debug State Controller, nodeColor=<rgba 0 0 0 1>}
  n41 = LinearInterpolate(A=NamedRerouteUsage(), B=NamedRerouteUsage(), Alpha=SmoothStep(Value=NamedRerouteUsage()) {constMax=0.5})
  n42 = NamedRerouteUsage()
  n43 = SmoothStep(Value=NamedRerouteUsage()) {constMin=0.5}
  n44 = NamedRerouteUsage()
  n45 = NamedRerouteUsage()
  n46 = NamedRerouteUsage()
  NamedRerouteDeclaration(None=LinearInterpolate(
    A=LinearInterpolate(A=@n41, B=@n42, Alpha=@n43)
    B=LinearInterpolate(A=@n44, B=@n45, Alpha=@n46)
    Alpha=ScalarParameter() {parameterName=Display 2 Tiles}
   )) {name=Color 1 Controller, nodeColor=<rgba 0 0 0 1>}
  n47 = LinearInterpolate(A=NamedRerouteUsage(), B=NamedRerouteUsage(), Alpha=SmoothStep(Value=NamedRerouteUsage()) {constMax=0.5})
  n48 = NamedRerouteUsage()
  n49 = SmoothStep(Value=NamedRerouteUsage()) {constMin=0.5}
  n50 = NamedRerouteUsage()
  n51 = NamedRerouteUsage()
  n52 = NamedRerouteUsage()
  NamedRerouteDeclaration(None=LinearInterpolate(
    A=LinearInterpolate(A=@n47, B=@n48, Alpha=@n49)
    B=LinearInterpolate(A=@n50, B=@n51, Alpha=@n52)
    Alpha=ScalarParameter() {parameterName=Display 2 Tiles}
   )) {name=Color 2 Controller, nodeColor=<rgba 0 0 0 1>}
  NamedRerouteUsage()
 ]
}
