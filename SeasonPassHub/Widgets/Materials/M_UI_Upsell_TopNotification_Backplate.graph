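# M_UI_Upsell_TopNotification_Backplate is a translucent UI material for the backplate of the Season Pass Hub's upsell top notification.
# It slightly hue-shifts the themeable Offer Color parameter, then adds two brightness offsets to form pale and deeper tints.
# The normalized UI U coordinate blends these into a horizontal gradient, lighter at the left.
# The shared CoreUI rounded-mask function supplies the rounded-corner shape.
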Material /SeasonPassHub/Widgets/Materials/M_UI_Upsell_TopNotification_Backplate
graph {
 properties {
  bAutomaticallySetUsageInEditor=false
  bUsedWithStaticMesh=false
  blendMode=BLEND_Translucent
  materialDomain=MD_UI
  usedShadingModels=Default Lit
 }
 nodes [
  n1 = MaterialFunctionCall(
   Hue Shift Percentage=Constant() {r=0.02}
   Texture=VectorParameter() {group=Theming, parameterName=Offer Color, defaultValue=<rgba 0.8714 0.3467 0.0561 1>}.RGB
  ) {materialFunction=&HueShift}
  LinearInterpolate(
   A=Saturate(None=Add(A=@n1.Result) {constB=0.65})
   B=Saturate(None=Add(A=@n1.Result) {constB=0.2})
   Alpha=ComponentMask(None=MaterialFunctionCall() {materialFunction=&GetUserInterfaceUV}.Normalized UV) {r=true}
  )
  MaterialFunctionCall(Corner Radius=Constant() {r=1}) {materialFunction=&MF_UI_CoreUI_RoundedMask}
 ]
}
