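# UI-domain font material for the upsell top notification in the Season Pass Hub, computing a themed offer color.
# The Offer Color parameter, an orange by default in the Theming group, is multiplied by itself to deepen the tint.
# The squared color is then scaled by the Color Multiply scalar, so hue and brightness can be adjusted per material instance.
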Material /SeasonPassHub/Widgets/Materials/M_UI_Upsell_TopNotification_Font
graph {
 properties {bAutomaticallySetUsageInEditor=false, materialDomain=MD_UI, usedShadingModels=Default Lit}
 nodes [
  n1 = VectorParameter() {group=Theming, parameterName=Offer Color, defaultValue=<rgba 0.8714 0.3467 0.0561 1>}
  Multiply(A=Multiply(A=@n1.RGB, B=@n1.RGB), B=ScalarParameter() {defaultValue=0.5, parameterName=Color Multiply}) {constB=0.5}
 ]
}
