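# Season pass upsell widget that switches between an animated top notification pill and a rivalry billboard via IsBillboard.
# VM_CheckForCMSText pulls title and body from the pass theme's text table, hiding the widget if either is missing, empty or "-", otherwise playing intros.
# It reuses or creates a PassBillboardVM, through which helper functions broadcast billboard text, color and image.
# SetColor tints pill images and fonts via dynamic materials.
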WidgetBlueprint /SeasonPassHub/Widgets/WBP_Upsell_TopNotification
animations [Intro {
  binding_count=5
  length_seconds=1.25
  object=&self:Intro
  sequence {bindings [
    Text_Body {
     id=#1
     possessed_class=CommonTextBlock
     tracks [
      {
       type=MovieScene2DTransformTrack
       sections [{
         type=MovieScene2DTransformSection
         end_frame=25
         start_frame=0
         channels [
          Translation.X {type=MovieSceneScriptingFloatChannel, keys [[15,-40],[24,0]]}
          Translation.Y {type=MovieSceneScriptingFloatChannel, const [0,2]}
          Angle {type=MovieSceneScriptingFloatChannel, const [0,2]}
          Scale.X {type=MovieSceneScriptingFloatChannel, const [1,2]}
          Scale.Y {type=MovieSceneScriptingFloatChannel, const [1,2]}
          Shear.X {type=MovieSceneScriptingFloatChannel, const [0,2]}
          Shear.Y {type=MovieSceneScriptingFloatChannel, const [0,2]}
         ]
        }]
      }
      {
       type=MovieSceneFloatTrack
       display_name=Render Opacity
       sections [{type=MovieSceneFloatSection, end_frame=25, start_frame=0, channels [None {type=MovieSceneScriptingFloatChannel, keys [[15,0],[24,1]]}]}]
      }
     ]
    }
    Backer {
     id=#2
     possessed_class=Image
     tracks [{
       type=MovieSceneFloatTrack
       display_name=Render Opacity
       sections [{type=MovieSceneFloatSection, end_frame=23, start_frame=0, channels [None {type=MovieSceneScriptingFloatChannel, keys [[13,0],[22,1]]}]}]
      }]
    }
    Text_Title {
     id=#3
     possessed_class=CommonTextBlock
     tracks [
      {
       type=MovieScene2DTransformTrack
       sections [{
         type=MovieScene2DTransformSection
         end_frame=23
         start_frame=0
         channels [
          Translation.X {type=MovieSceneScriptingFloatChannel, keys [[13,-40],[22,0]]}
          Translation.Y {type=MovieSceneScriptingFloatChannel, const [0,2]}
          Angle {type=MovieSceneScriptingFloatChannel, const [0,2]}
          Scale.X {type=MovieSceneScriptingFloatChannel, const [1,2]}
          Scale.Y {type=MovieSceneScriptingFloatChannel, const [1,2]}
          Shear.X {type=MovieSceneScriptingFloatChannel, const [0,2]}
          Shear.Y {type=MovieSceneScriptingFloatChannel, const [0,2]}
         ]
        }]
      }
      {
       type=MovieSceneFloatTrack
       display_name=Render Opacity
       sections [{type=MovieSceneFloatSection, end_frame=23, start_frame=0, channels [None {type=MovieSceneScriptingFloatChannel, keys [[13,0],[22,1]]}]}]
      }
     ]
    }
    TopNotification {
     id=#4
     possessed_class=SizeBox
     tracks [{
       type=MovieSceneFloatTrack
       display_name=Max Desired Width
       sections [{type=MovieSceneFloatSection, end_frame=25, start_frame=0, channels [None {type=MovieSceneScriptingFloatChannel, keys [[13,88],[24,800]]}]}]
      }]
    }
    Section {
     id=#5
     possessed_class=Image
     tracks [
      {
       type=MovieScene2DTransformTrack
       sections [{
         type=MovieScene2DTransformSection
         end_frame=14
         start_frame=0
         channels [
          Translation.X {type=MovieSceneScriptingFloatChannel, const [0,2]}
          Translation.Y {type=MovieSceneScriptingFloatChannel, keys [[8,16],[13,0]]}
          Angle {type=MovieSceneScriptingFloatChannel, const [0,1]}
          Scale.X {type=MovieSceneScriptingFloatChannel, default=0.85, keys [[8,0.85],[13,1]]}
          Scale.Y {type=MovieSceneScriptingFloatChannel, keys [[8,0.85],[13,1]]}
          Shear.X {type=MovieSceneScriptingFloatChannel, const [0,1]}
          Shear.Y {type=MovieSceneScriptingFloatChannel, const [0,1]}
         ]
        }]
      }
      {
       type=MovieSceneFloatTrack
       display_name=Render Opacity
       sections [{type=MovieSceneFloatSection, end_frame=12, start_frame=0, channels [None {type=MovieSceneScriptingFloatChannel, keys [[8,0],[11,1]]}]}]
      }
     ]
    }
   ]}
 }]
graphs [
 EventGraph {nodes [
   n1 = Event PreConstruct()
   > n2 = Branch(Condition=@n1.IsDesignTime)
   > SetVisibility(InVisibility=SelfHitTestInvisible)
   > "SetText (Text)"(self=Get Text_Title() -> (Text_Title:Common Text Object Reference), InText="INVTEXT(\"Title Text\")")
   > n3 = "SetText (Text)"(self=Get Text_Body() -> (Text_Body:Common Text Object Reference), InText="INVTEXT(\"Body Text Test\")")
   SetActiveWidget(
    execute=[@n4.then, @n3.then]
    self=Get OfferSwitch() -> (OfferSwitch:Common Visibility Switcher Object Reference)
    Widget=Select(
     Option 0:Widget Object Reference=Get TopNotification() -> (TopNotification:Size Box Object Reference)
     Option 1:Widget Object Reference=Get WBP_SeasonPass_RivalryBillboard() -> (WBP_SeasonPass_RivalryBillboard:WBP Season Pass Rivalry Billboard Object Reference)
     Index=Get IsBillboard() -> (IsBillboard:bool)
    ) -> (ReturnValue:Widget Object Reference)
   )
   > SetHideInputActionWithKeyboard!(
    execute
    self:Common Button Base Object Reference=Get Button_Hitbox!(self:WBP Season Pass Rivalry Billboard Object Reference=Get WBP_SeasonPass_RivalryBillboard()) -> (Button_Hitbox:WBP UIKit Button Generic Object Reference)
    bInHideInputActionWithKeyboard:bool=true
   ) -> (then)
   n4 = VM_CheckForCMSText!(execute=@n2.else, self, SelectedPass:Season Pass VM Object Reference) -> (then)
   Event Construct()
  ]}
 SetBillboardColorText {nodes [
   n1 = SetBillboardColorText!() -> (then, EntryType:ENexusUpsellEntryType Enum=Free, BillboardColor:Linear Color Structure) {is=title K2Node_FunctionEntry}
   > Set with Broadcast PromptText!(
    execute
    PromptText:Text=Format Text!(
     Format="LOCTEXT(\"<{StyleName}> {Body}</>  \")"
     StyleName:String=Get GoldBody_Syle() -> (GoldBody_Syle:String)
     Body:Text=Get BodyText() -> (BodyText:Text)
    ) -> (Result)
    self:Pass Billboard VM Object Reference=Get PassBillboardVM() -> (PassBillboardVM:Pass Billboard VM Object Reference)
   ) -> (then, Output_Get:Text)
   > Set with Broadcast BillboardColor!(execute, BillboardColor:Linear Color Structure=@n1.BillboardColor, self:Pass Billboard VM Object Reference=Get PassBillboardVM()) -> (then, Output_Get:Linear Color Structure)
   > SetLogoText!(execute, self, LogoText:Text=Get Title() -> (Title:Text)) -> (then)
  ]}
 SetColor {nodes [
   n1 = SetColor!() -> (then, Color:Linear Color Structure) {is=title K2Node_FunctionEntry}
   > n2 = GetDynamicMaterial(self=Get Backer() -> (Backer:Image Object Reference))
   > SetVectorParameterValue(self=@n2, ParameterName=Offer Color, Value=@n1)
   > n3 = GetDynamicMaterial(self=Get Section() -> (Section:Image Object Reference))
   > SetVectorParameterValue(self=@n3, ParameterName=Offer Color, Value=@n1)
   > n4 = GetDynamicFontMaterial!(execute, self:Text Object Reference=Get Text_Body()) -> (then, ReturnValue:Material Instance Dynamic Object Reference)
   > SetVectorParameterValue(self=@n4, ParameterName=Offer Color, Value=@n1)
   > n5 = GetDynamicFontMaterial(self=Get Text_Title())
   > SetVectorParameterValue(self=@n5, ParameterName=Offer Color, Value=@n1)
  ]}
 SetImage {nodes [
   n1 = SetImage!() -> (then, SetBillboardImage:Soft Object Reference) {is=title K2Node_FunctionEntry}
   > Set with Broadcast BillboardImage!(execute, BillboardImage:Soft Object Reference=@n1, self:Pass Billboard VM Object Reference=Get PassBillboardVM()) -> (then, Output_Get:Soft Object Reference)
  ]}
 SetLogoText {nodes [
   n1 = SetLogoText!() -> (then, LogoText:Text) {is=title K2Node_FunctionEntry}
   > Set with Broadcast LogoText!(execute, LogoText:Text=@n1, self:Pass Billboard VM Object Reference=Get PassBillboardVM()) -> (then, Output_Get:Text)
  ]}
 VM_CheckForCMSText {nodes [
   n1 = VM_CheckForCMSText!() -> (then, SelectedPass:Season Pass VM Object Reference) {is=title K2Node_FunctionEntry}
   > n2 = Is Valid(InputObject=@n1)
   SetVisibility(execute=@n2.Is Not Valid, InVisibility=Hidden)
   n3 = Get WBP_SeasonPass_RivalryBillboard()
   n4 = Get!(execute:Exec=@n2.Is Valid, self:WBP Season Pass Rivalry Billboard Object Reference=@n3) -> (then:Exec, else:Exec, PassBillboardVM:Pass Billboard VM Object Reference)
   > n5 = Set PassBillboardVM(PassBillboardVM:Pass Billboard VM Object Reference=@n4) -> (Output_Get:Pass Billboard VM Object Reference)
   n6 = Break Season Pass Theme Data!(SeasonPassThemeData:"Season Pass Theme Data Structure (by ref)"=GetThemeData!(self:Season Pass VM Object Reference=Get SelectedPass() -> (SelectedPass:Season Pass VM Object Reference)) -> (ReturnValue:"Season Pass Theme Data Structure (by ref)")) -> (
    TitleColor:Linear Color Structure
    TitleMaterial:Soft Object Reference
    TitleOutline:Font Outline Settings Structure
    SeasonColor:Linear Color Structure
    ProgressionColor:Linear Color Structure
    OverlayColor:Linear Color Structure
    VaultBackgroundMaterial:Soft Object Reference
    GiftboxVaultBackgroundMaterial:Soft Object Reference
    UIBackgroundMaterial:Soft Object Reference
    UpsellBackgroundMaterial:Soft Object Reference
    CrewCategoryBackgroundMaterial:Soft Object Reference
    CustomLogo:Soft Object Reference
    CustomSeasonTitle:Text
    CustomSeasonTitleColor:Linear Color Structure
    Upsell2BackgrounMaterial:Soft Object Reference
    ImageReferenceTable:Fort Image Reference Table Structure
    ColorReferenceTable:Fort Color Reference Table Structure
    TextReferenceTable:Fort Text Reference Table Structure
   ) {is=title K2Node_BreakStruct}
   n7 = TryLookupText!(Table:"Fort Text Reference Table Structure (by ref)"=@n6.TextReferenceTable, Name:String=UpsellTopPill_Body) -> (OutTextReference:Text, ReturnValue:bool)
   n8 = OR Boolean(
    A=TextIsEmpty(InText=@n7)
    B="Equal, Case Insensitive (String)"(A=Get Source String from Text!(Text:Text=@n7) -> (ReturnValue:String) {is=title K2Node_CallFunction}, B=-)
   )
   n9 = TryLookupText(Table=@n6.TextReferenceTable, Name=UpsellTopPill_Title)
   n10 = OR Boolean(A=TextIsEmpty(InText=@n9), B="Equal, Case Insensitive (String)"(A=Get Source String from Text(Text=@n9), B=-))
   n11 = AND Boolean(A=AND Boolean(A=NOT Boolean(A=@n10), B=@n9.ReturnValue), B=AND Boolean(A=NOT Boolean(A=@n8), B=@n7.ReturnValue))
   SetVisibility(
    execute=[@n14.then, @n5.then]
    InVisibility=Select(Option 0:ESlateVisibility Enum=Hidden, Option 1:ESlateVisibility Enum=SelfHitTestInvisible, Index=@n11) -> (ReturnValue:ESlateVisibility Enum)
   )
   > Set Title(Title:Text=@n9) -> (Output_Get:Text)
   > Set BodyText(BodyText:Text=@n7) -> (Output_Get:Text)
   > "SetText (Text)"(self=Get Text_Title(), InText=@n9)
   > "SetText (Text)"(self=Get Text_Body(), InText=@n7)
   > Branch(Condition=@n11)
   > PlayAnimation(InAnimation=Get Intro())
   n12 = Get WBP_SeasonPass_RivalryBillboard()
   > SetBillboard!(execute, self:WBP Season Pass Rivalry Billboard Object Reference=@n12, IsValidUIFeature:bool=true) -> (then)
   > IntroAnim(self:WBP Season Pass Rivalry Billboard Object Reference=@n12)
   n13 = Construct Object from Class!(
    execute=@n4.else
    Class=/SeasonPassHub/InterstitialVMs/PassBillboardVM.PassBillboardVM_C
    self
    UIFeature:String
    BillboardColor:Linear Color Structure=(R=0,G=0,B=0,A=0)
    LogoText:Text
    PromptText:Text
    BillboardImage:Soft Object Reference
    LogoImage:Soft Object Reference
    InputAction:Data Table Row Handle Structure
    ShadowColor:Linear Color Structure=(R=0,G=0,B=0,A=0)
   ) -> (then, ReturnValue:Pass Billboard VM Object Reference)
   > SetPassBillboardVM!(execute, self:WBP Season Pass Rivalry Billboard Object Reference=@n3, Viewmodel:Pass Billboard VM Object Reference=@n13) -> (then)
   > n14 = Set PassBillboardVM(PassBillboardVM=@n13)
  ]}
]
variables [
 BodyText {type=string, default=""}
 Bonus_Style {type=string, default=BonusTitle}
 BPBody_Style {type=string, default=BPBody}
 BundleBody_Style {type=string, default=BundleBody}
 GoldBody_Syle {type=string, default=GoldBody}
 IsBillboard {type=boolean}
 PassBillboardVM {type=PassBillboardVM_C, default=null}
 Title {type=string, default=""}
 Width Override {type=number, default=800}
]
widgets {
 info {inheritedWidgetCount=0, parentClass=&UserWidget, rootWidgetClass=&CommonVisibilitySwitcher, widgetCount=11}
 widgets [
  OfferSwitch {widgetClassPath=&CommonVisibilitySwitcher, bIsVariable=true, widget=OfferSwitch}
  TopNotification {
   widgetClassPath=&SizeBox
   bIsVariable=true
   parent=OfferSwitch
   widget=TopNotification
   properties {bOverride_MaxDesiredWidth=true, bOverride_MinDesiredWidth=true, heightOverride=100, maxDesiredWidth=800, minDesiredWidth=600}
   slot {class=CommonVisibilitySwitcherSlot}
  }
  GridPanel_56 {
   widgetClassPath=&GridPanel
   parent=TopNotification
   widget=GridPanel_56
   properties {columnFill [0,1], rowFill [0]}
   slot {class=SizeBoxSlot}
  }
  Backer {
   widgetClassPath=&Image
   bIsVariable=true
   parent=GridPanel_56
   widget=Backer
   properties {brush {
     bIsDynamicallyLoaded=false
     drawAs=Image
     imageType=FullColor
     mirroring=NoMirror
     resourceName=None
     resourceObject=/SeasonPassHub/Widgets/Materials/M_UI_Upsell_TopNotification_Backplate
     tiling=NoTile
     imageSize=<32 32>
     margin=<ltrb 0 0 0 0>
     outlineSettings {
      bUseBrushTransparency=false
      roundingType=HalfHeightRadius
      width=0
      color {colorUseRule=UseColor_Specified, specifiedColor=<rgba 0 0 0 0>}
      cornerRadii=<0 0 0 0>
     }
     tintColor {colorUseRule=UseColor_Specified, specifiedColor=<rgba 1 1 1 1>}
     uVRegion {bIsValid=false, max=<0 0>, min=<0 0>}
    }}
   slot {class=GridSlot, properties {columnSpan=2, padding=<ltrb 4 4 4 4>}}
  }
  Section {
   widgetClassPath=&Image
   bIsVariable=true
   parent=GridPanel_56
   widget=Section
   properties {brush {
     bIsDynamicallyLoaded=false
     drawAs=Image
     imageType=FullColor
     mirroring=NoMirror
     resourceName=None
     resourceObject=/SeasonPass/Widgets/Materials/M_UI_Upsell_TopNotification_Portrait
     tiling=NoTile
     imageSize=<88 88>
     margin=<ltrb 0 0 0 0>
     outlineSettings {
      bUseBrushTransparency=false
      roundingType=HalfHeightRadius
      width=0
      color {colorUseRule=UseColor_Specified, specifiedColor=<rgba 0 0 0 0>}
      cornerRadii=<0 0 0 0>
     }
     tintColor {colorUseRule=UseColor_Specified, specifiedColor=<rgba 1 1 1 1>}
     uVRegion {bIsValid=false, max=<0 0>, min=<0 0>}
    }}
   slot {class=GridSlot, properties {horizontalAlignment=HAlign_Center, verticalAlignment=VAlign_Center}}
  }
  VerticalBox_Text {
   widgetClassPath=&VerticalBox
   bIsVariable=true
   parent=GridPanel_56
   widget=VerticalBox_Text
   slot {class=GridSlot, properties {column=1, verticalAlignment=VAlign_Center, padding=<ltrb 0 8 32 8>}}
  }
  SizeBox_0 {
   widgetClassPath=&SizeBox
   parent=VerticalBox_Text
   widget=SizeBox_0
   properties {bOverride_MaxDesiredHeight=true, maxDesiredHeight=64}
   slot {class=VerticalBoxSlot, properties {verticalAlignment=VAlign_Center}}
  }
  Text_Title {
   widgetClassPath=&CommonTextBlock
   bIsVariable=true
   parent=SizeBox_0
   widget=Text_Title
   properties {
    autoWrapText=true
    mobileFontSizeMultiplier=1.14
    scrollOrientation=Orient_Vertical
    scrollStyle=/FortUILibrary/Styles/Typography/Foundation/TextScrollStyle-NoFade.TextScrollStyle-NoFade_C
    text=Title Text
    textTransformPolicy=ToUpper
    font {
     bForceMonospaced=false
     bMaterialIsStencil=false
     fontMaterial=/SeasonPassHub/Widgets/Materials/MI_UI_Upsell_TopNotification_Font
     fontObject=&HeadingNow
     letterSpacing=0
     monospacedWidth=1
     size=15.75
     skewAmount=0
     typefaceFontName=75 Medium
     outlineSettings {
      bApplyOutlineToDropShadows=false
      bMiteredCorners=false
      bSeparateFillAlpha=false
      outlineMaterial=null
      outlineSize=0
      outlineColor=<rgba 0 0 0 1>
     }
    }
    shadowOffset=<0 0>
   }
   slot {class=SizeBoxSlot}
  }
  SizeBox_1 {
   widgetClassPath=&SizeBox
   parent=VerticalBox_Text
   widget=SizeBox_1
   properties {bOverride_MaxDesiredHeight=true, maxDesiredHeight=80}
   slot {class=VerticalBoxSlot, properties {verticalAlignment=VAlign_Top}}
  }
  Text_Body {
   widgetClassPath=&CommonTextBlock
   bIsVariable=true
   parent=SizeBox_1
   widget=Text_Body
   properties {
    autoWrapText=true
    mobileFontSizeMultiplier=1.14
    scrollOrientation=Orient_Vertical
    scrollStyle=&TextScrollStyle-NoFade_C
    text="Body Text Test\r\nBody Text Test\r\nBody Text Test"
    font {
     bForceMonospaced=false
     bMaterialIsStencil=false
     fontMaterial=/SeasonPassHub/Widgets/Materials/M_UI_Upsell_TopNotification_Font
     fontObject=&HeadingNow
     letterSpacing=0
     monospacedWidth=1
     size=15.75
     skewAmount=0
     typefaceFontName=74 Regular
     outlineSettings {
      bApplyOutlineToDropShadows=false
      bMiteredCorners=false
      bSeparateFillAlpha=false
      outlineMaterial=null
      outlineSize=0
      outlineColor=<rgba 0 0 0 1>
     }
    }
    shadowOffset=<0 0>
   }
   slot {class=SizeBoxSlot}
  }
  WBP_SeasonPass_RivalryBillboard {
   widgetClassPath=/SeasonPassHub/Widgets/Reboot/WBP_SeasonPass_RivalryBillboard
   bIsVariable=true
   parent=OfferSwitch
   widget=WBP_SeasonPass_RivalryBillboard
   properties {shadowVisibility=SelfHitTestInvisible}
   slot {class=CommonVisibilitySwitcherSlot, properties {horizontalAlignment=HAlign_Center, verticalAlignment=VAlign_Center}}
  }
 ]
}
