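# Full-screen season pass upsell screen offering pass, bundle and Crew tiles with their benefits, purchase legal text and gifting.
# Initialize derives entries from config overrides or offer ownership; SetEntries spawns purchase buttons into a grid, tints background materials and staggers intros.
# Focused entries switch backgrounds and benefits, with an UpsellReworkActive flag choosing between legacy and reworked layouts.
# Theme images load async or by URL; compact UI adjusts layout, and activation handles music and analytics.
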WidgetBlueprint /SeasonPassHub/Widgets/CrewUpsellScreenMultiView
animations [Intro {
  binding_count=10
  length_seconds=0.9
  object=&self:Intro
  sequence {bindings [
    LeftBenefits {
     id=#1
     possessed_class=GridPanel
     tracks [
      {
       type=MovieSceneFloatTrack
       display_name=Render Opacity
       sections [{type=MovieSceneFloatSection, end_frame=18, start_frame=0, channels [None {type=MovieSceneScriptingFloatChannel, keys [[6,0],[10,1]]}]}]
      }
      {
       type=MovieScene2DTransformTrack
       sections [{
         type=MovieScene2DTransformSection
         end_frame=18
         start_frame=0
         channels [
          Translation.X {type=MovieSceneScriptingFloatChannel, const [0,1]}
          Translation.Y {type=MovieSceneScriptingFloatChannel, keys [[6,-24],[10,0]]}
         ]
        }]
      }
     ]
    }
    ScreenBg {
     id=#2
     possessed_class=Image
     tracks [{
       type=MovieSceneColorTrack
       display_name=Color and Opacity
       sections [{
         type=MovieSceneColorSection
         end_frame=18
         start_frame=0
         channels [
          Color.R {type=MovieSceneScriptingFloatChannel, keys [[0,0],[18,1]]}
          Color.G {type=MovieSceneScriptingFloatChannel, keys [[0,0],[18,1]]}
          Color.B {type=MovieSceneScriptingFloatChannel, keys [[0,0],[18,1]]}
          Color.A {type=MovieSceneScriptingFloatChannel, const [1,2]}
         ]
        }]
      }]
    }
    RightBenefitsLabel {
     id=#3
     possessed_class=CommonTextBlock
     tracks [{
       type=MovieSceneFloatTrack
       display_name=Render Opacity
       sections [{type=MovieSceneFloatSection, end_frame=18, start_frame=0, channels [None {type=MovieSceneScriptingFloatChannel, keys [[10,0],[14,1]]}]}]
      }]
    }
    UpsellBenefitsGroup {
     id=#4
     possessed_class=GridPanel
     tracks [
      {
       type=MovieScene2DTransformTrack
       sections [{
         type=MovieScene2DTransformSection
         end_frame=18
         start_frame=0
         channels [
          Translation.X {type=MovieSceneScriptingFloatChannel, default=0}
          Translation.Y {type=MovieSceneScriptingFloatChannel, keys [[6,-24],[10,0]]}
         ]
        }]
      }
      {
       type=MovieSceneFloatTrack
       display_name=Render Opacity
       sections [{type=MovieSceneFloatSection, end_frame=18, start_frame=0, channels [None {type=MovieSceneScriptingFloatChannel, keys [[6,0],[12,1]]}]}]
      }
     ]
    }
    RightBenefits {
     id=#5
     possessed_class=GridPanel
     tracks [
      {
       type=MovieSceneFloatTrack
       display_name=Render Opacity
       sections [{type=MovieSceneFloatSection, end_frame=18, start_frame=0, channels [None {type=MovieSceneScriptingFloatChannel, keys [[6,0],[10,1]]}]}]
      }
      {
       type=MovieScene2DTransformTrack
       sections [{
         type=MovieScene2DTransformSection
         end_frame=18
         start_frame=0
         channels [
          Translation.X {type=MovieSceneScriptingFloatChannel, const [0,1]}
          Translation.Y {type=MovieSceneScriptingFloatChannel, keys [[6,-24],[10,0]]}
         ]
        }]
      }
     ]
    }
    LeftBenefitsLabel {
     id=#6
     possessed_class=CommonTextBlock
     tracks [{
       type=MovieSceneFloatTrack
       display_name=Render Opacity
       sections [{type=MovieSceneFloatSection, end_frame=18, start_frame=0, channels [None {type=MovieSceneScriptingFloatChannel, keys [[10,0],[14,1]]}]}]
      }]
    }
    LeftBenefitsLabel_1 {
     id=#7
     possessed_class=CommonTextBlock
     tracks [{
       type=MovieSceneFloatTrack
       display_name=Render Opacity
       sections [{type=MovieSceneFloatSection, end_frame=18, start_frame=0, channels [None {type=MovieSceneScriptingFloatChannel, keys [[10,0],[14,1]]}]}]
      }]
    }
    Div {
     id=#8
     possessed_class=Image
     tracks [{
       type=MovieSceneFloatTrack
       display_name=Render Opacity
       sections [{type=MovieSceneFloatSection, end_frame=18, start_frame=0, channels [None {type=MovieSceneScriptingFloatChannel, keys [[12,0],[14,1]]}]}]
      }]
    }
    UpsellBenefits {
     id=#9
     possessed_class=DynamicEntryBox
     tracks [
      {
       type=MovieSceneFloatTrack
       display_name=Render Opacity
       sections [{type=MovieSceneFloatSection, end_frame=18, start_frame=0, channels [None {type=MovieSceneScriptingFloatChannel, keys [[6,0],[11,1]]}]}]
      }
      {
       type=MovieScene2DTransformTrack
       sections [{
         type=MovieScene2DTransformSection
         end_frame=18
         start_frame=0
         channels [
          Translation.X {type=MovieSceneScriptingFloatChannel, const [0,2]}
          Translation.Y {type=MovieSceneScriptingFloatChannel, keys [[6,-24],[10,0]]}
         ]
        }]
      }
     ]
    }
    UpsellItems {
     id=#10
     possessed_class=DynamicEntryBox
     tracks [
      {
       type=MovieSceneFloatTrack
       display_name=Render Opacity
       sections [{type=MovieSceneFloatSection, end_frame=18, start_frame=0, channels [None {type=MovieSceneScriptingFloatChannel, keys [[3,0],[8,1]]}]}]
      }
      {
       type=MovieScene2DTransformTrack
       sections [{
         type=MovieScene2DTransformSection
         end_frame=18
         start_frame=0
         channels [
          Translation.X {type=MovieSceneScriptingFloatChannel, const [0,2]}
          Translation.Y {type=MovieSceneScriptingFloatChannel, keys [[3,120],[8,0]]}
         ]
        }]
      }
     ]
    }
   ]}
 }]
graphs [
 BP_GetDesiredFocusTarget {nodes [
   Get Desired Focus Target()
   n1 = Get WBP_UpsellScreen_BP() -> (WBP_UpsellScreen_BP:WBP Upsell Screen BP Object Reference)
   n2 = GetAllChildren(self=Select(
     Option 0:Panel Widget Object Reference=Get UpsellItemGroup() -> (UpsellItemGroup:Grid Panel Object Reference)
     Option 1:Panel Widget Object Reference=Get UpsellItemGroup!(self:WBP Upsell Screen BP Object Reference=@n1) -> (UpsellItemGroup)
     Index=Get UpsellReworkActive() -> (UpsellReworkActive:bool)
    ) -> (ReturnValue:Panel Widget Object Reference))
   > n3 = Branch(Condition=Is Not Empty!(TargetArray:Array of Widget Object References=@n2) -> (ReturnValue:bool))
   > n4 = Set UpsellItems(UpsellItems:Array of Widget Object References=@n2) -> (Output_Get:Array of Widget Object References)
   > n5 = For Each Loop(Array:Array of Widget Object References=@n4) -> (Array Element:Widget Object Reference)
   > n6 = Cast To WBP_PassPurchaseButton!(execute, Object=@n5) -> (then, CastFailed, AsWBP Pass Purchase Button:WBP Pass Purchase Button Object Reference) {is=title K2Node_DynamicCast}
   > Branch(Condition=Get IsFocused!(self:WBP Pass Purchase Button Object Reference=@n6) -> (IsFocused:bool))
   > Return Node!(execute, ReturnValue:Widget Object Reference=@n5)
   n7 = Cast To WBP_PassPurchaseButton_Small!(execute=@n6.CastFailed, Object=@n5) -> (then, CastFailed, AsWBP Pass Purchase Button Small:WBP Pass Purchase Button Small Object Reference) {is=title K2Node_DynamicCast}
   > Branch(Condition=Get IsFocused(self:WBP Pass Purchase Button Small Object Reference=@n7))
   > Return Node(ReturnValue=@n7)
   Return Node(
    execute=@n5.Completed
    ReturnValue="Get (a copy)"(Array:Array of Widget Object References=Get UpsellItems() -> (UpsellItems:Array of Widget Object References)) -> (Output:Widget Object Reference)
   )
   Return Node(execute=@n3.else, ReturnValue=Get Button_Close() -> (Button_Close:WBP UIKit Button Quiet Object Reference))
  ]}
 BP_GetDesiredInputConfig {nodes [
   Get Desired Input Config() {is=title K2Node_FunctionEntry}
   > Return Node!(
    execute
    ReturnValue_bIgnoreMoveInput:bool=False
    ReturnValue_bIgnoreLookInput:bool=False
    ReturnValue_InputMode:ECommonInputMode Enum=Menu
    ReturnValue_MouseCaptureMode:EMouseCaptureMode Enum=NoCapture
    ReturnValue_MouseLockMode:EMouseLockMode Enum=DoNotLock
    ReturnValue_bHideCursorDuringViewportCapture:bool=True
   )
  ]}
 ClearFocusItems {nodes [
   ClearFocusItems!() -> (then, Button:Common Button Base Object Reference) {is=title K2Node_FunctionEntry}
   > n1 = Sequence!(execute) -> (then_0, then_1, then_2:Exec, then_3:Exec)
   > SetVisibility(
    self=[
     Get LeftBenefitsLabel() -> (LeftBenefitsLabel:Common Text Object Reference)
     Get Div() -> (Div:Image Object Reference)
     Get RightBenefitsLabel() -> (RightBenefitsLabel:Common Text Object Reference)
     Get LeftBenefitsLabel_1() -> (LeftBenefitsLabel_1:Common Text Object Reference)
    ]
    InVisibility=Collapsed
   )
   > SetVisibility(
    self=[
     Get UpsellBenefitsGroup() -> (UpsellBenefitsGroup:Horizontal Box Object Reference)
     Get SingleLineBenefits() -> (SingleLineBenefits:Vertical Box Object Reference)
     Get LeftBenefits() -> (LeftBenefits:Vertical Box Object Reference)
     Get RightBenefits() -> (RightBenefits:Vertical Box Object Reference)
    ]
    InVisibility=Collapsed
   )
   > SetVisibility(self=Get ContentWarningContainer() -> (ContentWarningContainer:Size Box Object Reference), InVisibility=Collapsed)
   n2 = For Each Loop(
    Exec=@n1.then_1
    Array:Array of Crew Upsell Multi View Benefit Object References=Get BenefitsGroupLeft() -> (BenefitsGroupLeft:Array of Crew Upsell Multi View Benefit Object References)
   ) -> (Array Element:Crew Upsell Multi View Benefit Object Reference)
   > Set Visibility(Visibility=Collapsed, self=@n2)
   n3 = For Each Loop(Exec=@n1.then_2, Array=Get BenefitsGroupRight() -> (BenefitsGroupRight:Array of Crew Upsell Multi View Benefit Object References))
   > Set Visibility(Visibility=Collapsed, self=@n3)
  ]}
 CrewOverlayGlow {nodes [
   CrewOverlayGlow() {is=title K2Node_FunctionEntry}
   > n1 = GetDynamicMaterial(self=Get Crew_Background() -> (Crew_Background:Epic CMSImage Object Reference))
   > SetScalarParameterValue(self=@n1, ParameterName=Overlay_Glow, Value=0.65)
   > SetScalarParameterValue(self=@n1, ParameterName=Overlay_Intensity, Value=0.95)
   > SetScalarParameterValue(self=@n1, ParameterName=Overlay_Size, Value=0.85)
   > SetScalarParameterValue(self=@n1, ParameterName=Overlay_X_Offset, Value=0.035)
   > SetScalarParameterValue(self=@n1, ParameterName=Overlay_Y_Offset, Value=0.14)
  ]}
 EventGraph {nodes [
   n1 = "On Clicked (Button_Close)"!() -> (OutputDelegate, then, Button:Common Button Base Object Reference) {is=title K2Node_ComponentBoundEvent}
   > n2 = Fire Event Element Interacted(Element=@n1.Button, AdditionalInfo=Get Analytics Additional Info() -> (Analytics Additional Info:Flow Analytics Additional Info Structure))
   DeactivateWidget(execute=[@n2.then, @n6.then])
   > ClearPurchaseUXTraceId!(
    execute
    self:Fort Mcp Context Object Reference=Get FortMcpContext!() -> (ReturnValue:Fort Mcp Context Object Reference) {is=title K2Node_GetBlueprintContextEx}
    DebugSourceContext:String
   ) -> (then)
   > Set Analytics Additional Info(Analytics Additional Info:Flow Analytics Additional Info Structure=Make Flow Analytics Additional Info!() -> (FlowAnalyticsAdditionalInfo:Flow Analytics Additional Info Structure) {is=title K2Node_MakeStruct}) -> (Output_Get:Flow Analytics Additional Info Structure)
   > Call On Closed!(execute:Exec, self:Crew Upsell Screen Multi View Object Reference) -> (then:Exec) {is=title K2Node_CallDelegate}
   n3 = "On Clicked (Button_GiftPass)"!() -> (OutputDelegate, then, Button:Common Button Base Object Reference) {is=title K2Node_ComponentBoundEvent}
   > Fire Event Element Interacted(Element=@n3.Button, AdditionalInfo=Get Analytics Additional Info())
   > ShowGiftingScreen(self:Season Pass VM Object Reference=Get SeasonPassVM() -> (SeasonPassVM:Season Pass VM Object Reference))
   "On Clicked (Button_SeeBenefits)"!() -> (OutputDelegate, then, Button:Common Button Base Object Reference) {is=title K2Node_ComponentBoundEvent}
   > n4 = PushPrimaryContentClassAsync!(execute:Exec, WidgetClass:Common Activatable Widget Class Reference=/Crew/UI/CrewInfoContainer_B.CrewInfoContainer_B_C) -> (then:Exec, OnPrimaryContentPushed:Exec, ActivatableWidget:Common Activatable Widget Object Reference) {is=title K2Node_AsyncAction}
   n5 = Cast To CrewInfoContainer_B!(execute=@n4.OnPrimaryContentPushed, Object=@n4) -> (then, CastFailed, AsCrew Info Container B:Crew Info Container B Object Reference) {is=title K2Node_DynamicCast}
   > PageSelected!(execute, self:Crew Info Container B Object Reference=@n5, PageToGo:E_CrewContainer_Page Enum=NewEnumerator1, IsBenefitsRoot:bool=true) -> (then)
   n6 = Close Screen() {is=title K2Node_CustomEvent}
   n7 = SetUpsellGrid!() -> (OutputDelegate, then, isBundle:bool=~) {is=title K2Node_CustomEvent}
   n8 = GetElementThemeData!(self:Season Pass Element Base VM Object Reference=Get SeasonPassVM()) -> (
    ReturnValue_ImageReferenceTable:Fort Image Reference Table Structure
    ReturnValue_ColorReferenceTable:Fort Color Reference Table Structure
    ReturnValue_TextReferenceTable:Fort Text Reference Table Structure
   )
   n9 = TryLookupImage!(
    Table:"Fort Image Reference Table Structure (by ref)"=@n8
    Name:String=Select(Option 0:String=BPUpsellGrid, Option 1:String=BunddleGrid, Index=@n7.isBundle) -> (ReturnValue:String)
   ) -> (
    OutImageReference_Image:Soft Object Reference
    OutImageReference_ImageUrl:String
    OutImageReference_MaterialTextureParamName:Name
    ReturnValue:bool
   )
   > n10 = Async Load Asset(Asset=@n9)
   n11 = Cast To Texture2D(execute=@n10.Completed, Object=@n10)
   > SetImage!(
    execute
    self:WBP CMSClient Widget Object Reference=Get WBP_CMSClientWidget!(self:WBP Upsell Screen BP Object Reference=Get WBP_UpsellScreen_BP()) -> (WBP_CMSClientWidget:WBP CMSClient Widget Object Reference)
    Image:Texture 2D Object Reference=@n11
    URL:String=@n9.OutImageReference_ImageUrl
   ) -> (then)
   n12 = SetBPUpsellBGImages!() -> (OutputDelegate, then, isBundle:bool=~) {is=title K2Node_CustomEvent}
   > n13 = Sequence!(execute) -> (then_0, then_1)
   n14 = TryLookupImage(Table=@n8, Name=Select(Option 0=BPUpsellBG, Option 1=BPUpsellBGBundle, Index=@n12.isBundle))
   > n15 = Branch(Condition=IsEmpty(InString=@n14.OutImageReference_ImageUrl))
   n16 = Async Load Asset(execute=[@n19.OnFail, @n15.then], Asset=@n14)
   n17 = Cast To Texture2D(execute=@n16.Completed, Object=@n16)
   > Set BPMainTexture(BPMainTexture:Texture 2D Object Reference=@n17) -> (Output_Get:Texture 2D Object Reference)
   > n18 = GetDynamicMaterial(self=Get IMG_BG() -> (IMG_BG:Image Object Reference))
   > SetTextureParameterValue(self=@n18, ParameterName=MainTexture, Value=Get BPMainTexture() -> (BPMainTexture:Texture 2D Object Reference))
   n19 = DownloadImage!(execute:Exec=@n15.else, URL:String=@n14.OutImageReference_ImageUrl) -> (then:Exec, OnSuccess:Exec, OnFail:Exec, Texture:Texture 2DDynamic Object Reference) {is=title K2Node_AsyncAction}
   n20 = GetDynamicMaterial(execute=@n19.OnSuccess, self=Get IMG_BG())
   > SetTextureParameterValue(self=@n20, ParameterName=MainTexture, Value=@n19)
   n21 = TryLookupImage(Table=@n8, Name=UpsellCrewMain)
   n22 = Async Load Asset(execute=@n13.then_1, Asset=@n21)
   n23 = Cast To Texture2D(execute=@n22.Completed, Object=@n22)
   > n24 = Set BPMainTexture_Crew(BPMainTexture_Crew:Texture 2D Object Reference=@n23) -> (Output_Get:Texture 2D Object Reference)
   > SetCrewIMG!(
    execute
    self:WBP Upsell Screen BP Object Reference=Get WBP_UpsellScreen_BP()
    Texture:Texture 2D Object Reference=@n24
    URL:String=@n21.OutImageReference_ImageUrl
   ) -> (then)
   > CrewOverlayGlow() {is=title K2Node_CallFunction}
   n25 = Event PreConstruct()
   > n26 = Branch(Condition=@n25.IsDesignTime)
   PlayAnimation(execute=@n26.else, InAnimation=Get Intro())
   > SetVisibility(self=Get UpsellBenefitsContainer() -> (UpsellBenefitsContainer:Grid Panel Object Reference), InVisibility=Hidden)
   Event On Activated()
   > n27 = Sequence!(execute) -> (then_0, then_1, then_2:Exec, then_3:Exec, then_4:Exec)
   > n28 = Branch(Condition=IsValid(Object=Get SeasonPassVM()))
   Set SeasonPassVM(
    execute=@n28.else
    SeasonPassVM:Season Pass VM Object Reference=Get SelectedPass!(self:Season Pass Selection VM Object Reference=Get SeasonPassSelectionVM() -> (SeasonPassSelectionVM:Season Pass Selection VM Object Reference)) -> (SelectedPass:Season Pass VM Object Reference)
   ) -> (Output_Get:Season Pass VM Object Reference)
   DelayUntilNextTick!(execute=@n27.then_1) -> (then)
   > PlayUpsellMusic!(execute, self, bPlay:bool=true) -> (then)
   n29 = Get CrewVM() -> (CrewVM:Crew VM Object Reference)
   n30 = GeneratePurchaseTraceId!(
    execute=@n27.then_2
    OfferId:String=GetEpicOfferId!(self:Fort Simple Offer VM Object Reference=Get CrewOffer!(self:Crew VM Object Reference=@n29) -> (CrewOffer:Fort Simple Offer VM Object Reference)) -> (ReturnValue:String)
   ) -> (then, AnalyticsAdditionalInfo:Flow Analytics Additional Info Structure)
   > n31 = Set Analytics Additional Info(Analytics Additional Info=@n30)
   > Fire Event Sub View Opened!(execute, View:Widget Object Reference=Self-Reference(), AdditionalInfo:Flow Analytics Additional Info Structure=@n31, bIsOverlay:bool=false) -> (then)
   Event On Deactivated()
   > Parent: On Deactivated() {is=title K2Node_CallParentFunction}
   > PlayUpsellMusic(bPlay=false)
   Event Construct()
   > SetupBenefitGroupArrays()
   > SetCompactUIStyle() {is=title K2Node_CallFunction}
   > Set Text(self=Get LegalText() -> (LegalText:Common Text Object Reference))
  ]}
 GetBenefits {nodes [
   n1 = GetBenefits!() -> (then, EntryType:ENexusUpsellEntryType Enum=Free) {is=title K2Node_FunctionEntry}
   n2 = Get SeasonPassVM()
   n3 = Get CrewVM()
   > Return Node!(
    execute
    Benefits:Nexus Upsell Benefits Structure=Select!(
     Free:Nexus Upsell Benefits Structure=GetBenefitsForUpsellType!(self:Season Pass VM Object Reference=@n2, UpsellType:ENexusUpsellEntryType Enum=Free) -> (BenefitData:Nexus Upsell Benefits Structure, ReturnValue:Fort Simple Offer VM Object Reference)
     Paid:Nexus Upsell Benefits Structure=GetBenefitsForUpsellType(self=@n2, UpsellType=Paid)
     Bundle:Nexus Upsell Benefits Structure=GetBenefitsForUpsellType(self=@n2, UpsellType=Bundle)
     Crew:Nexus Upsell Benefits Structure=Get UpsellBenefits!(self:Crew VM Object Reference=@n3) -> (UpsellBenefits:Nexus Upsell Benefits Structure)
     Index:ENexusUpsellEntryType Enum=@n1
    ) -> (ReturnValue:Nexus Upsell Benefits Structure)
   )
  ]}
 GetFrontendFlowInterruptConfig {nodes [
   GetFrontendFlowInterruptConfig() {is=title K2Node_FunctionEntry}
   > Return Node!(execute, ReturnValue:Frontend Flow Interrup Config Structure)
  ]}
 GetPrimaryContentSetup {nodes [
   GetPrimaryContentSetup() {is=title K2Node_FunctionEntry}
   > Return Node(ReturnValue:Primary Content Setup Structure=Make Primary Content Setup!(
     bShowBottomBar:bool=false
     bShowChatWidget:bool=false
     bEnableNewTextChat:bool=False
     bShowTopBar:bool=false
     bTabSupportsInterface:bool=False
    ) -> (PrimaryContentSetup:Primary Content Setup Structure) {is=title K2Node_MakeStruct})
  ]}
 HandleContentWarningInfoBox {nodes [
   n1 = HandleContentWarningInfoBox!() -> (then, InText:"Text (by ref)") {is=title K2Node_FunctionEntry}
   > n2 = Branch(Condition=TextIsEmpty(InText=@n1))
   n3 = Get WBP_UIKit_InfoBox_ContentWarning() -> (WBP_UIKit_InfoBox_ContentWarning:WBP UIKit Info Box Object Reference)
   > SetVisibility(self=@n3, InVisibility=Collapsed)
   SetText(execute=@n2.else, self:WBP UIKit Info Box Object Reference=@n3, InText:Text=Get InText() -> (InText:"Text (by ref)"))
   > SetVisibility(self=@n3, InVisibility=SelfHitTestInvisible)
  ]}
 Initialize {nodes [
   Initialize() {is=title K2Node_FunctionEntry}
   > SetUpsellRework()
   > Clear(TargetArray:Array of Nexus Upsell Entry Structures=Get Default Entries() -> (Default Entries:Array of Nexus Upsell Entry Structures))
   n1 = Break Crew Upsell Config!(CrewUpsellConfig:"Crew Upsell Config Structure (by ref)"=Get Config() -> (Config:Crew Upsell Config Structure)) -> (
    OverrideUpsellPass:String
    PassBackgroundUrl:Array of Motd Image Structures
    CrewBackgroundUrl:Array of Motd Image Structures
    bUsePassBackgroundMaterial:bool=~
    bUseCrewBackgroundMaterial:bool=~
    bCrewOnRightInUpsell:bool=~
    bCrew50PercentInUpsell:bool=~
    bCrew30PercentInUpsell:bool=~
    bCrewAsSliver:bool=~
    bCrewOnlyInUpsell:bool=~
    bPassOnlyInUpsell:bool=~
    bAnimateCrewBG:bool=~
    bUseNewUpsellFormat:bool=~
    bAllowToggleBundle:bool=~
    bDefaultShowBundle:bool=~
    EntryOverrides:Array of Nexus Upsell Entry Structures
   ) {is=title K2Node_BreakStruct}
   > n2 = Branch(Condition=Is Empty!(TargetArray:Array of Nexus Upsell Entry Structures=@n1.EntryOverrides) -> (ReturnValue:bool))
   n3 = IsValid(Object=Get PassBundleOffer!(self:Season Pass VM Object Reference=Get SeasonPassVM()) -> (PassBundleOffer:Season Pass Bundle Offer VM Object Reference))
   > Add(
    TargetArray:Array of Nexus Upsell Entry Structures=Get Default Entries()
    NewItem:"Nexus Upsell Entry Structure (by ref)"=Make Nexus Upsell Entry!(EntryType:ENexusUpsellEntryType Enum=Select!(Option 0:ENexusUpsellEntryType Enum=Free, Option 1:ENexusUpsellEntryType Enum=Paid, Index:bool=@n3) -> (ReturnValue:ENexusUpsellEntryType Enum)) -> (NexusUpsellEntry:Nexus Upsell Entry Structure) {is=title K2Node_MakeStruct}
   ) {is=title K2Node_CallArrayFunction}
   n4 = Get SeasonPassVM()
   > n5 = Branch(Condition=Get bIsOwned!(self:Fort Simple Offer VM Object Reference=Get PassBundleOffer(self=@n4)) -> (bIsOwned:bool))
   Add(execute=[@n6.then, @n5.then], TargetArray=Get Default Entries(), NewItem=Make Nexus Upsell Entry(EntryType=Crew))
   > SetEntries!(execute, self, Upsell Entries:Array of Nexus Upsell Entry Structures=Get Default Entries()) -> (then)
   n6 = Add(
    execute=@n5.else
    TargetArray=Get Default Entries()
    NewItem=Make Nexus Upsell Entry(EntryType=Select(Option 0=Paid, Option 1=Bundle, Index=@n3))
   )
   SetEntries(execute=@n2.else, Upsell Entries=@n1.EntryOverrides)
  ]}
 OnClosed {nodes [OnClosed() {is=title K2Node_FunctionEntry}]}
 PlayUpsellMusic {nodes [PlayUpsellMusic!() -> (then, bPlay:bool=~) {is=title K2Node_FunctionEntry}]}
 SeasonPassVMChanged {nodes [
   SeasonPassVMChanged!() -> (then, SeasonPassVM:Season Pass VM Object Reference) {is=title K2Node_FunctionEntry}
   > Initialize() {is=title K2Node_CallFunction}
  ]}
 SetCompactUIStyle {nodes [
   SetCompactUIStyle() {is=title K2Node_FunctionEntry}
   > Branch(Condition=IsUsingCompactUI())
   > SetScrollingEnabled!(
    execute
    self:Common Text Object Reference=Get Text(self:WBP Real Price Legal Text Object Reference=Get WBP_RealPrice_LegalText() -> (WBP_RealPrice_LegalText:WBP Real Price Legal Text Object Reference)) -> (Text:Common Text Object Reference)
    bInIsScrollingEnabled:bool=false
   ) -> (then)
   n1 = Get LegalText()
   > SetScrollingEnabled(self=@n1, bInIsScrollingEnabled=true)
   > SetSize!(
    execute
    self:Vertical Box Slot Object Reference=SlotAsVerticalBoxSlot(Widget=@n1)
    InSize_Value:float=1
    InSize_SizeRule:ESlateSizeRule Enum=Fill
   ) -> (then)
   n2 = Get SizeBox_Content() -> (SizeBox_Content:Size Box Object Reference)
   > ClearWidthOverride(self:Size Box Object Reference=[@n2, Get ContentWarningContainer()])
   > ClearMaxAspectRatio(self:Size Box Object Reference=@n2)
   n3 = SlotAsOverlaySlot(Widget=@n2)
   > SetVerticalAlignment(self=@n3, InVerticalAlignment=VAlign_Center)
   > SetPadding!(execute, self:Overlay Slot Object Reference=@n3, InPadding:Margin Structure) -> (then)
   > SetPadding(self:Size Box Slot Object Reference=SlotAsSizeBoxSlot!(Widget:Widget Object Reference=Get VB_Content() -> (VB_Content:Vertical Box Object Reference)) -> (ReturnValue:Size Box Slot Object Reference))
   > n4 = Sequence!(execute) -> (then_0, then_1, then_2)
   > Branch(Condition=CompareToScreenAspectRatio!(Comparison:EAspectRatioComparison Enum=GreaterThanOrEqualTo, AspectRatio:EAspectRatio Enum=Ratio_16_9) -> (ReturnValue:bool))
   > SetPadding!(
    execute
    self:Vertical Box Slot Object Reference=SlotAsVerticalBoxSlot(Widget=Get Overlay_Benefits() -> (Overlay_Benefits:Overlay Object Reference))
    InPadding_Left:float=0
    InPadding_Top:float=24
    InPadding_Right:float=0
    InPadding_Bottom:float=0
   ) -> (then)
   > n5 = GetSafeZonePadding!(execute) -> (then, SafePadding:Vector 4 Structure, SafePaddingScale:Vector 2D Structure, SpillOverPadding:Vector 4 Structure)
   > Branch(Condition="float >= float"(A=BreakVector2D(InVec=@n5.SafePaddingScale).Y, B=0.025))
   > Set PassSwitcher_TopPadding(PassSwitcher_TopPadding:float=0) -> (Output_Get:float)
   Branch(execute=@n4.then_1, Condition=CompareToScreenAspectRatio(Comparison=LessThanOrEqualTo, AspectRatio=Ratio_23_16))
   > Set SB_Legal_HeightOverride(SB_Legal_HeightOverride:float=165) -> (Output_Get:float)
   > Set PassTilesPadding(PassTilesPadding:double=16) -> (Output_Get:double)
   SetGridSlotPadding!(
    execute=@n4.then_2
    Widget:Widget Object Reference=Get Pass_Switcher() -> (Pass_Switcher:Common Visibility Switcher Object Reference)
    Padding_Left:float=0
    Padding_Top:float=Get PassSwitcher_TopPadding() -> (PassSwitcher_TopPadding:float)
    Padding_Right:float=0
    Padding_Bottom:float=0
   ) -> (then)
   > SetHeightOverride(
    self=Get SB_Legal() -> (SB_Legal:Size Box Object Reference)
    InHeightOverride=Get SB_Legal_HeightOverride() -> (SB_Legal_HeightOverride:float)
   )
  ]}
 SetConfig {nodes [
   n1 = SetConfig!() -> (then, Config:Crew Upsell Config Structure) {is=title K2Node_FunctionEntry}
   > Set Config(Config:Crew Upsell Config Structure=@n1) -> (Output_Get:Crew Upsell Config Structure)
   > Is Valid(InputObject=Get SeasonPassVM())
   > Initialize()
  ]}
 SetEntries {nodes [
   SetEntries!() -> (then, Upsell Entries:Array of Nexus Upsell Entry Structures) {is=title K2Node_FunctionEntry}
   n1 = Get UpsellReworkActive()
   > SetActiveWidget(
    self=Get Pass_Switcher()
    Widget=Select(Option 0:Widget Object Reference=Get SizeBox_Content(), Option 1:Widget Object Reference=Get WBP_UpsellScreen_BP(), Index=@n1) -> (ReturnValue:Widget Object Reference)
   )
   > SetActiveWidget(
    self=Get BGPassSwitch() -> (BGPassSwitch:Common Visibility Switcher Object Reference)
    Widget=Select(Option 0=Get ScreenBg() -> (ScreenBg:Image Object Reference), Option 1=Get IMG_BG(), Index=@n1)
   )
   > Set TotalEntries(TotalEntries:int=Length(TargetArray:Array of Nexus Upsell Entry Structures=Get Upsell Entries() -> (Upsell Entries:Array of Nexus Upsell Entry Structures))) -> (Output_Get:int=0)
   > Set AnimationDelayTotal(AnimationDelayTotal:double=0) -> (Output_Get:double)
   > n2 = Branch(Condition=@n1)
   n3 = Get UpsellItemGroup(self=Get WBP_UpsellScreen_BP())
   > ClearChildren(self=@n3)
   > SetColumnFill!(execute, self:Grid Panel Object Reference=@n3, ColumnIndex:int=2, Coefficient:float=0) -> (then)
   > SetColumnFill(self=@n3, ColumnIndex=3, Coefficient=0)
   > n4 = For Each Loop(Array:Array of Nexus Upsell Entry Structures=Get Upsell Entries()) -> (Array Element:Nexus Upsell Entry Structure)
   > Set BenefitData(BenefitData:Nexus Upsell Entry Structure=@n4) -> (Output_Get:Nexus Upsell Entry Structure)
   > n5 = Set ListIndex(ListIndex:int=@n4.Array Index) -> (Output_Get:int=0)
   > SetColumnFill(self=Get UpsellItemGroup(self=Get WBP_UpsellScreen_BP()), ColumnIndex=@n5, Coefficient=1)
   > n6 = Construct Object from Class!(
    execute
    Class=/SeasonPassHub/Widgets/UpsellBattlePass/WBP_PassPurchaseButton_Small.WBP_PassPurchaseButton_Small_C
    self
    SeasonPassVM:Season Pass VM Object Reference
    Style:Common Button Style Class Reference=/FortUILibrary/UIKit/Buttons/Blocks/ButtonStyle_UIKit_Empty.ButtonStyle_UIKit_Empty_C
    bApplyAlphaOnDisable:bool=True
    bLocked:bool=False
    bSelectable:bool=False
    bShouldSelectUponReceivingFocus:bool=False
    bInteractableWhenSelected:bool=False
    bToggleable:bool=False
    bTriggerClickedAfterSelection:bool=False
    bDisplayInputActionWhenNotInteractable:bool=True
    bHideInputActionWithKeyboard:bool=False
    bShouldUseFallbackDefaultInputAction:bool=True
    bRequiresHold:bool=False
    HoldData:Common UIHold Data Class Reference
    ClickMethod:EButtonClickMethod Enum=DownAndUp
    InputPriority:int=0
    TriggeringInputAction:Data Table Row Handle Structure
   ) -> (then, ReturnValue:WBP Pass Purchase Button Small Object Reference)
   > n7 = Set UpsellItem_Small(UpsellItem_Small:WBP Pass Purchase Button Small Object Reference=@n6) -> (Output_Get:WBP Pass Purchase Button Small Object Reference)
   > Set IsAnimatedIn!(execute, IsAnimatedIn:bool=true, self:WBP Pass Purchase Button Small Object Reference=@n7) -> (then, Output_Get:bool)
   n8 = Get UpsellItem_Small() -> (UpsellItem_Small:WBP Pass Purchase Button Small Object Reference)
   > AddChild(self=Get UpsellItemGroup(self=Get WBP_UpsellScreen_BP()), Content=@n8)
   > SetGridSlotAlignment!(
    execute
    Widget:Widget Object Reference=@n8
    Horizontal:EHorizontalAlignment Enum=HAlign_Center
    Vertical:EVerticalAlignment Enum=VAlign_Bottom
   ) -> (then)
   > SetGridSlotColumnAndRow!(
    execute
    Widget:Widget Object Reference=Get UpsellItem_Small()
    Row:int=0
    Row Span:int=1
    Column:int=Get ListIndex() -> (ListIndex:int=0)
    Column Span:int=1
   ) -> (then)
   n9 = Get ListIndex() -> (ListIndex=0)
   n10 = Get PassTilesPadding() -> (PassTilesPadding:double)
   n11 = int - int(A=Get TotalEntries() -> (TotalEntries:int=0))
   n12 = Get ListIndex() -> (ListIndex=0)
   n13 = Make Margin(
    Left=SelectFloat(A=0, B=@n10, bPickA="Equal (Integer)"(A=@n9))
    Top=0
    Right=SelectFloat(A=0, B=@n10, bPickA="Equal (Integer)"(A=@n12, B=@n11))
    Bottom=0
   )
   > SetGridSlotPadding!(execute, Widget=Get UpsellItem_Small(), Padding:Margin Structure=@n13) -> (then)
   > Initialize!(
    execute
    self:WBP Pass Purchase Button Small Object Reference=Get UpsellItem_Small()
    SeasonPassVM:Season Pass VM Object Reference=Get SeasonPassVM()
    BenefitData:Nexus Upsell Entry Structure=Get BenefitData() -> (BenefitData:Nexus Upsell Entry Structure)
    ListIndex:int=Get ListIndex() -> (ListIndex=0)
   ) -> (then)
   n14 = Create Event()
   > Bind Event to On Focused(self:WBP Pass Purchase Button Small Object Reference=Get UpsellItem_Small(), Delegate=@n14) {is=title K2Node_AddDelegate}
   n15 = Create Event()
   > Bind Event to Close Screen(self:WBP Pass Purchase Button Small Object Reference=Get UpsellItem_Small(), Delegate=@n15) {is=title K2Node_AddDelegate}
   > n16 = GetDynamicMaterial(self=Get IMG_BG())
   n17 = Get SeasonPassVM()
   n18 = Select!(
    Free:Nexus Upsell Benefits Structure=GetBenefitsForUpsellType(self=@n17, UpsellType=Free)
    Paid:Nexus Upsell Benefits Structure=GetBenefitsForUpsellType(self=@n17, UpsellType=Paid)
    Bundle:Nexus Upsell Benefits Structure=GetBenefitsForUpsellType(self=@n17, UpsellType=Bundle)
    Crew:Nexus Upsell Benefits Structure=Get UpsellBenefits(self=Get CrewVM())
    Index:ENexusUpsellEntryType Enum=Break Nexus Upsell Entry!(NexusUpsellEntry:"Nexus Upsell Entry Structure (by ref)"=Get BenefitData()) -> (EntryType:ENexusUpsellEntryType Enum, OverrideName:Text, OverrideImageUrl:String) {is=title K2Node_BreakStruct}
   ) -> (ReturnValue:Nexus Upsell Benefits Structure)
   n19 = Get UpsellReworkActive()
   n20 = GetUpsellThemeColor!(
    Benefits:"Nexus Upsell Benefits Structure (by ref)"=@n18
    ThemeName:String=Select!(Option 0:String=Tile Color 1, Option 1:String=BG Color 1, Index:bool=@n19) -> (ReturnValue:String)
    DefaultColor:Linear Color Structure=(R=0.5755,G=0.5755,B=0.5755,A=1)
   ) -> (ReturnValue_SpecifiedColor:Linear Color Structure, ReturnValue_ColorUseRule:ESlateColorStylingMode Enum)
   n21 = Get ListIndex() -> (ListIndex=0)
   n22 = Select!(Option 0:Name=Slot 1 - Color Theme 2, Option 1:Name=Slot 2 - Color Theme 2, Option 2:Name=Slot 3 - Color Theme 2, Index:int=@n21) -> (ReturnValue:Name)
   > SetVectorParameterValue(self=@n16, ParameterName=@n22, Value=@n20)
   n23 = GetUpsellThemeColor(
    Benefits=@n18
    ThemeName=Select!(Option 0:String=Tile Color 2, Option 1:String=BG Color 2, Index:bool=@n19) -> (ReturnValue:String)
    DefaultColor=(R=0.1875,G=0.1875,B=0.1875,A=1)
   )
   n24 = Select!(Option 0:Name=Slot 1 - Color Theme 4, Option 1:Name=Slot 2 - Color Theme 4, Option 2:Name=Slot 3 - Color Theme 4, Index:int=@n21) -> (ReturnValue:Name)
   > SetVectorParameterValue(self=@n16, ParameterName=@n24, Value=@n23)
   > n25 = Set AnimationDelayTotal(AnimationDelayTotal=float + float(A=Get AnimationDelayTotal() -> (AnimationDelayTotal:double), B=Get AnimationDelayAmt() -> (AnimationDelayAmt:float)))
   n26 = Get UpsellItem_Small()
   > PlayAnimation(
    self=@n26
    InAnimation=Get IntroAnimation!(self:WBP Pass Purchase Button Small Object Reference=@n26) -> (IntroAnimation:Widget Animation Object Reference)
    StartAtTime=@n25
   )
   SetInitialDesiredFocusTileAndBenefits(execute=@n4.Completed)
   > SetBPUpsellBGImages!(execute, self, isBundle:bool=false) -> (then) {is=title K2Node_CallFunction}
   > SetUpsellGrid!(execute, self, isBundle:bool=false) -> (then) {is=title K2Node_CallFunction}
   ClearChildren(execute=@n2.else, self=Get UpsellItemGroup!() -> (UpsellItemGroup))
   > SetColumnFill(self=Get UpsellItemGroup(), ColumnIndex=2, Coefficient=0)
   > SetColumnFill(self=Get UpsellItemGroup(), ColumnIndex=3, Coefficient=0)
   > n27 = For Each Loop(Array=Get Upsell Entries())
   > Set BenefitData(BenefitData=@n27)
   > n28 = Set ListIndex(ListIndex=@n27.Array Index) -> (Output_Get=0)
   > SetColumnFill(self=Get UpsellItemGroup(), ColumnIndex=@n28, Coefficient=1)
   > n29 = Construct Object from Class(
    Class=&/WBP_PassPurchaseButton.WBP_PassPurchaseButton_C
    Style=&ButtonStyle_UIKit_Empty_C
    bApplyAlphaOnDisable=True
    bLocked=False
    bSelectable=False
    bShouldSelectUponReceivingFocus=False
    bInteractableWhenSelected=False
    bToggleable=False
    bTriggerClickedAfterSelection=False
    bDisplayInputActionWhenNotInteractable=True
    bHideInputActionWithKeyboard=False
    bShouldUseFallbackDefaultInputAction=True
    bRequiresHold=False
    ClickMethod=DownAndUp
    InputPriority=0
   ) -> (ReturnValue:WBP Pass Purchase Button Object Reference)
   > n30 = Set UpsellItem(UpsellItem:WBP Pass Purchase Button Object Reference=@n29) -> (Output_Get:WBP Pass Purchase Button Object Reference)
   > Set IsAnimatedIn(IsAnimatedIn=true, self:WBP Pass Purchase Button Object Reference=@n30)
   > AddChild(self=Get UpsellItemGroup(), Content=@n29)
   > SetGridSlotAlignment(Widget=Get UpsellItem() -> (UpsellItem:WBP Pass Purchase Button Object Reference), Horizontal=HAlign_Center, Vertical=VAlign_Bottom)
   > SetGridSlotColumnAndRow(Widget=Get UpsellItem(), Row=0, Row Span=1, Column=Get ListIndex() -> (ListIndex=0), Column Span=1)
   > SetGridSlotPadding(Widget=Get UpsellItem(), Padding=@n13)
   > Initialize(
    self:WBP Pass Purchase Button Object Reference=Get UpsellItem()
    SeasonPassVM=Get SeasonPassVM()
    BenefitData=Get BenefitData()
    ListIndex=Get ListIndex() -> (ListIndex=0)
   )
   > Bind Event to On Focused(self:WBP Pass Purchase Button Object Reference=Get UpsellItem(), Delegate=@n14)
   > Bind Event to Close Screen(self:WBP Pass Purchase Button Object Reference=Get UpsellItem(), Delegate=@n15)
   > n31 = GetDynamicMaterial(self=Get ScreenBg())
   > SetVectorParameterValue(self=@n31, ParameterName=@n22, Value=@n20)
   > SetVectorParameterValue(self=@n31, ParameterName=@n24, Value=@n23)
   > n32 = Set AnimationDelayTotal(AnimationDelayTotal=float + float(A=Get AnimationDelayTotal(), B=Get AnimationDelayAmt()))
   n33 = Get UpsellItem()
   > PlayAnimation(self=@n33, InAnimation=Get IntroAnimation(self:WBP Pass Purchase Button Object Reference=@n33), StartAtTime=@n32)
   SetInitialDesiredFocusTileAndBenefits(execute=@n27.Completed)
  ]}
 SetFocusedBenefits {nodes [
   SetFocusedBenefits!() -> (
    then
    Benefits:Array of Nexus Upsell Benefits Group Structures
    Legalese:Text
    ListIndex:int
    EntryType:ENexusUpsellEntryType Enum=Free
    OfferVM:Fort Simple Offer VM Object Reference
   ) {is=title K2Node_FunctionEntry}
   > n1 = Sequence()
   > n2 = Sequence!(execute) -> (then_0, then_1)
   > ClearFocusItems!(execute, self, Button:Common Button Base Object Reference) -> (then) {is=title K2Node_CallFunction}
   > n3 = Sequence()
   > Branch(Condition=Get UpsellReworkActive())
   > SetVisibility(self=Get WBP_Upsell_TopNotification() -> (WBP_Upsell_TopNotification:WBP Upsell Top Notification Object Reference), InVisibility=Collapsed)
   n4 = "Equal (Enum)"(A:ENexusUpsellEntryType Enum=Crew, B:ENexusUpsellEntryType Enum=Get EntryType() -> (EntryType:ENexusUpsellEntryType Enum=Free))
   > SetActiveWidget(
    self=Get BGPassSwitch()
    Widget=Select!(
     Option 0:Widget Object Reference=Get IMG_BG()
     Option 1:Widget Object Reference=Get CrewBG() -> (CrewBG:Overlay Object Reference)
     Index:bool=@n4
    ) -> (ReturnValue:Widget Object Reference)
   )
   n5 = Get WBP_UpsellScreen_BP()
   > SetActiveWidget(
    self=Get Switch_Pass!(self:WBP Upsell Screen BP Object Reference=@n5) -> (Switch_Pass:Common Visibility Switcher Object Reference)
    Widget=Select(
     Option 0=Get BR_Pass!(self:WBP Upsell Screen BP Object Reference=@n5) -> (BR_Pass:Horizontal Box Object Reference)
     Option 1=Get Crew_Pass!(self:WBP Upsell Screen BP Object Reference=@n5) -> (Crew_Pass:Overlay Object Reference)
     Index=@n4
    )
   )
   > n6 = Switch on ENexusUpsellEntryType!(execute, Selection:ENexusUpsellEntryType Enum=Get EntryType() -> (EntryType=Free)) -> (Free:Exec, Paid:Exec, Bundle:Exec, Crew:Exec) {is=title K2Node_SwitchEnum}
   SetUpsellGrid(execute=@n6.Bundle, isBundle=true)
   > SetBPUpsellBGImages(isBundle=true)
   SetUpsellGrid(execute=@n6.Paid, isBundle=false)
   > SetBPUpsellBGImages(isBundle=false)
   n7 = GetConfigDataBoolWithDefault!(EntryName:String=CrewRework, bDefaultValue:bool=true) -> (ReturnValue:bool)
   n8 = "Equal (Enum)"(A=Crew, B=Get EntryType() -> (EntryType=Free))
   SetVisibility(
    execute=@n3.then_1
    self=Get Button_SeeBenefits() -> (Button_SeeBenefits:WBP UIKit Button Quiet Object Reference)
    InVisibility=Select(Option 0:ESlateVisibility Enum=Collapsed, Option 1:ESlateVisibility Enum=SelfHitTestInvisible, Index=AND Boolean(A=@n8, B=@n7)) -> (ReturnValue:ESlateVisibility Enum)
   )
   Set Visibility(execute=@n2.then_1, Visibility=HitTestInvisible, self=Get UpsellBenefitsContainer())
   > n9 = Branch(Condition=Get UpsellReworkActive())
   n10 = Get SeasonPassVM()
   n11 = Select!(
    Free:Nexus Upsell Benefits Structure
    Paid:Nexus Upsell Benefits Structure=GetBenefitsForUpsellType(self=@n10, UpsellType=Paid)
    Bundle:Nexus Upsell Benefits Structure=GetBenefitsForUpsellType(self=@n10, UpsellType=Bundle)
    Crew:Nexus Upsell Benefits Structure
    Index:ENexusUpsellEntryType Enum=Get EntryType() -> (EntryType=Free)
   ) -> (ReturnValue:Nexus Upsell Benefits Structure)
   > SetCMSLabelData!(
    execute
    self:WBP CMSClient Widget Object Reference=Get WBP_CMSClientWidget(self=Get WBP_UpsellScreen_BP())
    LabelData:Array of Nexus Upsell Benefits Group Structures=Get Benefits() -> (Benefits:Array of Nexus Upsell Benefits Group Structures)
    GridTextColor:Linear Color Structure=GetUpsellThemeColor(Benefits=@n11, ThemeName=BenefitTextColor, DefaultColor=(R=0.5755,G=0.5755,B=0.5755,A=1))
    Debug:bool=false
   ) -> (then)
   n12 = Get EntryType() -> (EntryType=Free)
   > SetColors!(
    execute
    self:WBP UI Offer Details Object Reference=Get WBP_UI_Offer_Details!(self:WBP Upsell Screen BP Object Reference=Get WBP_UpsellScreen_BP()) -> (WBP_UI_Offer_Details:WBP UI Offer Details Object Reference)
    UpsellType:ENexusUpsellEntryType Enum=@n12
   ) -> (then)
   > Branch(Condition="Not Equal (Enum)"!(A:ENexusUpsellEntryType Enum=@n12, B:ENexusUpsellEntryType Enum=Crew) -> (ReturnValue:bool))
   n13 = Get Benefits()
   n14 = Branch(execute=@n9.else, Condition="integer > integer"(A=Length(TargetArray:Array of Nexus Upsell Benefits Group Structures=@n13), B=1))
   > SetVisibility(self=[Get LeftBenefitsLabel(), Get RightBenefits(), Get LeftBenefits(), Get Div(), Get RightBenefitsLabel()], InVisibility=HitTestInvisible)
   > "SetText (Text)"(
    self=Get LeftBenefitsLabel()
    InText="Get (a copy)"!(Array:Array of Nexus Upsell Benefits Group Structures=Get Benefits(), Dimension 1) -> (Output_Label:Text, Output_Benefits:Array of Nexus Simple Benefit Structures)
   )
   > "SetText (Text)"(self=Get RightBenefitsLabel(), InText="Get (a copy)"(Array=Get Benefits(), Dimension 1=1))
   > SetGridSlotColumnAndRow(Widget=Get LeftBenefitsLabel(), Row=0, Row Span=1, Column=0, Column Span=1)
   > n15 = For Each Loop!(Exec, Array:Array of Nexus Upsell Benefits Group Structures=Get Benefits()) -> (LoopBody, Array Element_Label:Text, Array Element_Benefits:Array of Nexus Simple Benefit Structures, Array Index, Completed)
   > n16 = For Each Loop!(Exec, Array:Array of Nexus Simple Benefit Structures=@n15.Array Element_Benefits) -> (LoopBody, Array Element:Nexus Simple Benefit Structure, Array Index, Completed)
   n17 = Get BenefitsGroupLeft()
   > Branch(Condition="integer < integer"(A=@n16.Array Index, B=Length(TargetArray:Array of Crew Upsell Multi View Benefit Object References=@n17)))
   n18 = GetUpsellThemeColor(
    Benefits=GetBenefits!(EntryType:ENexusUpsellEntryType Enum=Get EntryType() -> (EntryType=Free)) -> (Benefits:Nexus Upsell Benefits Structure) {is=title K2Node_CallFunction}
    ThemeName=Offer Color
    DefaultColor=(R=0.6514,G=0.3916,B=1,A=1)
   )
   n19 = "Get (a copy)"!(
    Array:Array of Crew Upsell Multi View Benefit Object References=Select!(
     Option 0:Array of Crew Upsell Multi View Benefit Object References=Get BenefitsGroupLeft()
     Option 1:Array of Crew Upsell Multi View Benefit Object References=Get BenefitsGroupRight()
     Index:int=@n15.Array Index
    ) -> (ReturnValue:Array of Crew Upsell Multi View Benefit Object References)
    Dimension 1=@n16.Array Index
   ) -> (Output:Crew Upsell Multi View Benefit Object Reference)
   > SetBenefit!(
    execute
    self:Crew Upsell Multi View Benefit Object Reference=@n19
    Benefit:"Nexus Simple Benefit Structure (by ref)"=@n16
    ShowText:"Boolean (by ref)"=true
    Color:Linear Color Structure=@n18
   ) -> (then)
   > Set Visibility(Visibility=SelfHitTestInvisible, self=@n19)
   > SetColor!(execute, self:WBP Upsell Top Notification Object Reference=Get WBP_Upsell_TopNotification(), Color:Linear Color Structure=@n18) -> (then)
   SetVisibility(execute=@n14.else, self=[Get SingleLineBenefits(), Get UpsellBenefitsGroup(), Get LeftBenefitsLabel_1()], InVisibility=HitTestInvisible)
   > "SetText (Text)"(
    self=Get LeftBenefitsLabel_1()
    InText="Get (a copy)"!(Array:Array of Nexus Upsell Benefits Group Structures=Get Benefits(), Dimension 1) -> (Output_Label:Text, Output_Benefits:Array of Nexus Simple Benefit Structures)
   )
   n20 = Get Benefits()
   > Set TotalEntries(TotalEntries=Length(TargetArray:Array of Nexus Simple Benefit Structures="Get (a copy)"(Array=@n20).Output_Benefits)) -> (Output_Get=0)
   > ClearChildren(self=Get UpsellBenefitsGroup())
   > n21 = For Each Loop(Array="Get (a copy)"(Array=Get Benefits()).Output_Benefits)
   > Set BenefitData(BenefitData:Nexus Simple Benefit Structure=@n21) -> (Output_Get:Nexus Simple Benefit Structure)
   > Set BenefitIndex(BenefitIndex:int=@n21.Array Index) -> (Output_Get:int=0)
   > n22 = Construct Object from Class!(execute, Class=&/CrewUpsellMultiViewBenefit.CrewUpsellMultiViewBenefit_C, self) -> (then, ReturnValue:Crew Upsell Multi View Benefit Object Reference)
   > n23 = Set BenefitItem(BenefitItem:Crew Upsell Multi View Benefit Object Reference=@n22) -> (Output_Get:Crew Upsell Multi View Benefit Object Reference)
   > AddChild(self=Get UpsellBenefitsGroup(), Content=@n23)
   n24 = "Equal (Integer)"(A=Get BenefitIndex() -> (BenefitIndex:int=0))
   n25 = "Equal (Integer)"(A=Get BenefitIndex() -> (BenefitIndex=0), B=int - int(A=Get TotalEntries() -> (TotalEntries=0)))
   > SetPadding!(
    execute
    self:User Widget Object Reference=Get BenefitItem() -> (BenefitItem:Crew Upsell Multi View Benefit Object Reference)
    InPadding:Margin Structure=Make Margin(Left=SelectFloat(A=0, B=16, bPickA=@n24), Top=0, Right=SelectFloat(A=0, B=16, bPickA=@n25), Bottom=0)
   ) -> (then)
   n26 = Get TotalEntries() -> (TotalEntries=0)
   n27 = int - int(A=1760, B=int * int(A=int - int(A=@n26), B=32))
   n28 = Get TotalEntries() -> (TotalEntries=0)
   > SetMaxDesiredWidth!(
    execute
    self:Size Box Object Reference=Get RootSizeBox!(self:Crew Upsell Multi View Benefit Object Reference=Get BenefitItem()) -> (RootSizeBox:Size Box Object Reference)
    InMaxDesiredWidth:float="To Float (Integer)"(InInt=int / int(A=@n27, B=@n28))
   ) -> (then)
   n29 = Get EntryType() -> (EntryType=Free)
   > SetBenefit(
    self=Get BenefitItem()
    Benefit=Get BenefitData() -> (BenefitData:Nexus Simple Benefit Structure)
    ShowText=true
    Color=GetUpsellThemeColor(Benefits=GetBenefits(EntryType=@n29), ThemeName=Offer Color, DefaultColor=(R=0.5776,G=0.5776,B=0.5776,A=1))
   )
   > SetVisibility(self=Get ContentWarningContainer(), InVisibility=SelfHitTestInvisible)
   n30 = Sequence!(execute=@n1.then_1) -> (then_0, then_1, then_2:Exec)
   > Set Text(Text=Get Legalese() -> (Legalese:Text), self=Get LegalText())
   > UpdateRealPriceLegalText!(execute, self, OfferVM:Fort Simple Offer VM Object Reference=Get OfferVM() -> (OfferVM:Fort Simple Offer VM Object Reference)) -> (then)
   n31 = GetDynamicMaterial(execute=@n30.then_1, self=Get ScreenBg())
   > n32 = GetDynamicMaterial(self=Get IMG_BG())
   n33 = Get UpsellItemGroup()
   n34 = "Equal (Integer)"(A=Length(TargetArray:Array of Widget Object References=GetAllChildren(self=@n33)), B=2)
   > SetScalarParameterValue(
    self=[@n32, @n31]
    ParameterName=State Controller
    Value=Select!(Option 0:float=0, Option 1:float=SelectFloat(A=1, B=0.5, bPickA=@n34), Option 2:float=1, Index=Get ListIndex() -> (ListIndex=0)) -> (ReturnValue:float)
   )
   n35 = Branch(execute=@n30.then_2, Condition=Get UpsellReworkActive())
   n36 = Get WBP_UpsellScreen_BP()
   > n37 = For Each Loop(Array:Array of Widget Object References=GetAllChildren(self=Get UpsellItemGroup!(self:WBP Upsell Screen BP Object Reference=@n36) -> (UpsellItemGroup))) -> (Array Element:Widget Object Reference)
   > n38 = Cast To WBP_PassPurchaseButton_Small(Object=@n37)
   n39 = Get List Index!(self:WBP Pass Purchase Button Small Object Reference=@n38) -> (List Index:int=0)
   n40 = Get ListIndex() -> (ListIndex=0)
   > Branch(Condition=AND Boolean(A="Not Equal (Integer)"(A=@n39, B=@n40), B=Get IsFocused(self=@n38)))
   > SetFocusState!(execute, self:WBP Pass Purchase Button Small Object Reference=@n38, isFocused:bool=false) -> (then)
   n41 = For Each Loop(Exec=@n35.else, Array=GetAllChildren(self=Get UpsellItemGroup!() -> (UpsellItemGroup)))
   > n42 = Cast To WBP_PassPurchaseButton(Object=@n41)
   n43 = Get List Index(self:WBP Pass Purchase Button Object Reference=@n42) -> (List Index=0)
   n44 = Get ListIndex() -> (ListIndex=0)
   > Branch(Condition=AND Boolean(A=Get IsFocused(self:WBP Pass Purchase Button Object Reference=@n42), B="Not Equal (Integer)"(A=@n43, B=@n44)))
   > SetFocusState(self:WBP Pass Purchase Button Object Reference=@n42, isFocused=false)
   Branch(execute=@n1.then_2, Condition=Get UpsellReworkActive())
   > RefreshRivalryBillboard!(
    execute
    self:WBP Upsell Screen BP Object Reference=Get WBP_UpsellScreen_BP()
    Benefits:Array of Nexus Upsell Benefits Group Structures=Get Benefits()
    Legalese:Text=Get Legalese()
    ListIndex:int=Get ListIndex() -> (ListIndex=0)
    EntryType:ENexusUpsellEntryType Enum=Get EntryType() -> (EntryType=Free)
    OfferVM:Fort Simple Offer VM Object Reference=Get OfferVM()
   ) -> (then)
   Reroute Node(InputPin:Linear Color Structure="To Color (SlateColor)"!(Value:"Slate Color Structure (by ref)"=GetUpsellThemeColor!(Benefits=@n11, ThemeName=BG Color 1, DefaultColor=(R=1,G=0.651,B=0.1569,A=1)) -> (ReturnValue:Slate Color Structure)) -> (ReturnValue:Linear Color Structure)) -> (OutputPin:Linear Color Structure)
  ]}
 SetInitialDesiredFocusTileAndBenefits {nodes [
   SetInitialDesiredFocusTileAndBenefits() {is=title K2Node_FunctionEntry}
   n1 = GetDesiredFocusTarget!(self:Common Activatable Widget Object Reference) -> (ReturnValue:Widget Object Reference)
   > n2 = Cast To WBP_PassPurchaseButton(Object=@n1)
   > SetFocusState(self=@n2, isFocused=true)
   > SetFocusedBenefits!(
    execute
    self
    Benefits:Array of Nexus Upsell Benefits Group Structures=Get Benefits!(self:WBP Pass Purchase Button Object Reference=@n2) -> (Benefits)
    Legalese:Text=Get LegalText!(self:WBP Pass Purchase Button Object Reference=@n2) -> (LegalText:Text)
    ListIndex:int=Get List Index(self=@n2) -> (List Index=0)
    EntryType:ENexusUpsellEntryType Enum=Get!(self:WBP Pass Purchase Button Object Reference=@n2) -> (Benefit Data_EntryType:ENexusUpsellEntryType Enum, Benefit Data_OverrideName:Text, Benefit Data_OverrideImageUrl:String)
    OfferVM:Fort Simple Offer VM Object Reference=Get Offer!(self:WBP Pass Purchase Button Object Reference=@n2) -> (Offer:Fort Simple Offer VM Object Reference)
   ) -> (then) {is=title K2Node_CallFunction}
   n3 = Cast To WBP_PassPurchaseButton_Small(execute=@n2.CastFailed, Object=@n1)
   > SetFocusState(self:WBP Pass Purchase Button Small Object Reference=@n3, isFocused=true)
   > SetFocusedBenefits(
    Benefits=Get Benefits(self:WBP Pass Purchase Button Small Object Reference=@n3)
    Legalese=Get LegalText(self:WBP Pass Purchase Button Small Object Reference=@n3)
    ListIndex=Get List Index(self:WBP Pass Purchase Button Small Object Reference=@n3) -> (List Index=0)
    EntryType=Get(self:WBP Pass Purchase Button Small Object Reference=@n3)
    OfferVM=Get Offer(self:WBP Pass Purchase Button Small Object Reference=@n3)
   )
  ]}
 SetUpsellRework {nodes [
   SetUpsellRework() {is=title K2Node_FunctionEntry}
   n1 = GetSeasonPassTypeId!(self:Season Pass VM Object Reference=Get SeasonPassVM()) -> (ReturnValue:String)
   > Set UpsellReworkActive(UpsellReworkActive:bool=AND Boolean(A="Equal Exactly (String)"(A=@n1, B=br), B=GetConfigDataBoolWithDefault(EntryName=present_pass_upsell_s41, bDefaultValue=true))) -> (Output_Get:bool)
   > Return Node!(execute)
  ]}
 SetupBenefitGroupArrays {nodes [
   SetupBenefitGroupArrays() {is=title K2Node_FunctionEntry}
   > Add(
    TargetArray:Array of Crew Upsell Multi View Benefit Object References=Get BenefitsGroupLeft()
    NewItem:"Crew Upsell Multi View Benefit Object Reference (by ref)"=Get BenefitGroupLeftWidget0() -> (BenefitGroupLeftWidget0:Crew Upsell Multi View Benefit Object Reference)
   )
   > Add(
    TargetArray=Get BenefitsGroupLeft()
    NewItem=Get BenefitGroupLeftWidget1() -> (BenefitGroupLeftWidget1:Crew Upsell Multi View Benefit Object Reference)
   )
   > Add(
    TargetArray=Get BenefitsGroupLeft()
    NewItem=Get BenefitGroupLeftWidget2() -> (BenefitGroupLeftWidget2:Crew Upsell Multi View Benefit Object Reference)
   )
   > Add(
    TargetArray=Get BenefitsGroupLeft()
    NewItem=Get BenefitGroupLeftWidget3() -> (BenefitGroupLeftWidget3:Crew Upsell Multi View Benefit Object Reference)
   )
   > Add(
    TargetArray=Get BenefitsGroupRight()
    NewItem=Get BenefitGroupRightWidget0() -> (BenefitGroupRightWidget0:Crew Upsell Multi View Benefit Object Reference)
   )
   > Add(
    TargetArray=Get BenefitsGroupRight()
    NewItem=Get BenefitGroupRightWidget1() -> (BenefitGroupRightWidget1:Crew Upsell Multi View Benefit Object Reference)
   )
   > Add(
    TargetArray=Get BenefitsGroupRight()
    NewItem=Get BenefitGroupRightWidget2() -> (BenefitGroupRightWidget2:Crew Upsell Multi View Benefit Object Reference)
   )
   > Add(
    TargetArray=Get BenefitsGroupRight()
    NewItem=Get BenefitGroupRightWidget3() -> (BenefitGroupRightWidget3:Crew Upsell Multi View Benefit Object Reference)
   )
  ]}
 UpdateRealPriceLegalText {nodes [
   n1 = UpdateRealPriceLegalText!() -> (then, OfferVM:Fort Simple Offer VM Object Reference) {is=title K2Node_FunctionEntry}
   > Set with Broadcast Price!(
    execute
    Price:int="To Integer (Integer64)"!(InInt:Integer64=GetPrice!(self:Fort Base Offer VM Object Reference=@n1) -> (ReturnValue:Integer64=0)) -> (ReturnValue:int=0)
    self:WBP Real Price Legal Text Object Reference=Get WBP_RealPrice_LegalText()
   ) -> (then, Output_Get:int=0)
   n2 = Get OfferVM()
   > Set with Broadcast IsPriceShowing!(
    execute
    IsPriceShowing:bool=NOT Boolean(A=Get bIsOwned(self=@n2))
    self:WBP Real Price Legal Text Object Reference=Get WBP_RealPrice_LegalText()
   ) -> (then, Output_Get:bool)
   > Set with Broadcast Currency!(
    execute
    Currency:EFortPriceCurrency Enum=GetCurrencyType!(self:Fort Base Offer VM Object Reference=Get OfferVM()) -> (ReturnValue:EFortPriceCurrency Enum=Mtx)
    self:WBP Real Price Legal Text Object Reference=Get WBP_RealPrice_LegalText()
   ) -> (then, Output_Get:EFortPriceCurrency Enum=Mtx)
  ]}
]
variables [
 Analytics Additional Info {type=FlowAnalyticsAdditionalInfo, default {}}
 AnimationDelayAmt {type=number, default=-0.2}
 AnimationDelayTotal {type=number}
 BenefitItems {type="array<NexusUpsellEntry>", default []}
 BenefitsGroupLeft {type="array<CrewUpsellMultiViewBenefit_C>", default []}
 BenefitsGroupRight {type="array<CrewUpsellMultiViewBenefit_C>", default []}
 BPMainTexture {type=Texture2D}
 BPMainTexture_Crew {type=Texture2D}
 Config {
  type=CrewUpsellConfig
  default {
   bAllowToggleBundle=true
   bAnimateCrewBG=true
   bCrew30PercentInUpsell=false
   bCrew50PercentInUpsell=false
   bCrewAsSliver=false
   bCrewOnRightInUpsell=false
   bCrewOnlyInUpsell=false
   bDefaultShowBundle=true
   bPassOnlyInUpsell=false
   bUseCrewBackgroundMaterial=true
   bUseNewUpsellFormat=true
   bUsePassBackgroundMaterial=true
   overrideUpsellPass=""
   crewBackgroundUrl []
   entryOverrides []
   passBackgroundUrl []
  }
 }
 OnClosed {}
 PassTilesPadding {type=number, default=24}
 SelectedPassChangedInSubscreen {type=boolean}
 UpsellMusicComponent {type=AudioComponent, default=null}
 UpsellReworkActive {type=boolean}
]
widgets {
 info {
  inheritedWidgetCount=0
  namedSlotCount=1
  parentClass=&CommonActivatableWidget
  rootWidgetClass=&Overlay
  widgetCount=58
 }
 widgets [
  Overlay_81 {
   widgetClassPath=&Overlay
   widget=Overlay_81
   properties {navigation=&self:WidgetTree.Overlay_81.WidgetNavigation_1}
  }
  BGPassSwitch {
   widgetClassPath=&CommonVisibilitySwitcher
   bIsVariable=true
   parent=Overlay_81
   widget=BGPassSwitch
   slot {class=OverlaySlot, properties {horizontalAlignment=HAlign_Fill, verticalAlignment=VAlign_Fill}}
  }
  ScreenBg {
   widgetClassPath=&Image
   bIsVariable=true
   parent=BGPassSwitch
   widget=ScreenBg
   properties {brush {
     bIsDynamicallyLoaded=false
     drawAs=Image
     imageType=FullColor
     mirroring=NoMirror
     resourceName=None
     resourceObject=/SeasonPassHub/Widgets/Materials/M_UI_Upsell_Background
     tiling=NoTile
     imageSize=<32 32>
     margin=<ltrb 0 0 0 0>
     outlineSettings {
      bUseBrushTransparency=false
      roundingType=HalfHeightRadius
      width=0
      color {colorUseRule=UseColor_Specified, specifiedColor=<rgba 0 0 0 0>}
      cornerRadii=<0 0 0 0>
     }
     tintColor {colorUseRule=UseColor_Specified, specifiedColor=<rgba 1 1 1 1>}
     uVRegion {bIsValid=false, max=<0 0>, min=<0 0>}
    }}
   slot {class=CommonVisibilitySwitcherSlot}
  }
  IMG_BG {
   widgetClassPath=&Image
   bIsVariable=true
   parent=BGPassSwitch
   widget=IMG_BG
   properties {brush {
     bIsDynamicallyLoaded=false
     drawAs=Image
     imageType=FullColor
     mirroring=NoMirror
     resourceName=None
     resourceObject=/SeasonPassHub/Widgets/UpsellBattlePass/Materials/MI_UI_Upsell_BackgroundBP
     tiling=NoTile
     imageSize=<32 32>
     margin=<ltrb 0 0 0 0>
     outlineSettings {
      bUseBrushTransparency=false
      roundingType=HalfHeightRadius
      width=0
      color {colorUseRule=UseColor_Specified, specifiedColor=<rgba 0 0 0 0>}
      cornerRadii=<0 0 0 0>
     }
     tintColor {colorUseRule=UseColor_Specified, specifiedColor=<rgba 1 1 1 1>}
     uVRegion {bIsValid=false, max=<0 0>, min=<0 0>}
    }}
   slot {class=CommonVisibilitySwitcherSlot}
  }
  CrewBG {widgetClassPath=&Overlay, bIsVariable=true, parent=BGPassSwitch, widget=CrewBG, slot {class=CommonVisibilitySwitcherSlot}}
  Crew_Background {
   widgetClassPath=/Script/EpicCMSUIFramework.EpicCMSImage
   bIsVariable=true
   parent=CrewBG
   widget=Crew_Background
   properties {
    bMatchImageSize=false
    materialTextureParamName=CMS_Crew_Bento
    brush {
     bIsDynamicallyLoaded=false
     drawAs=Image
     imageType=FullColor
     mirroring=NoMirror
     resourceName=None
     resourceObject=/Crew/UI/Materials/M_UI_CrewUpsell_50x50_US
     tiling=NoTile
     imageSize=<32 32>
     margin=<ltrb 0 0 0 0>
     outlineSettings {
      bUseBrushTransparency=false
      roundingType=HalfHeightRadius
      width=0
      color {colorUseRule=UseColor_Specified, specifiedColor=<rgba 0 0 0 0>}
      cornerRadii=<0 0 0 0>
     }
     tintColor {colorUseRule=UseColor_Specified, specifiedColor=<rgba 1 1 1 1>}
     uVRegion {bIsValid=false, max=<0 0>, min=<0 0>}
    }
   }
   slot {class=OverlaySlot, properties {horizontalAlignment=HAlign_Fill, verticalAlignment=VAlign_Fill}}
  }
  Bolt {
   widgetClassPath=&Image
   bIsVariable=true
   parent=CrewBG
   widget=Bolt
   properties {brush {
     bIsDynamicallyLoaded=false
     drawAs=Image
     imageType=FullColor
     mirroring=NoMirror
     resourceName=None
     resourceObject=/SeasonPassHub/Widgets/UpsellBattlePass/Materials/MI_UI_Bolt_Crew
     tiling=NoTile
     imageSize=<32 32>
     margin=<ltrb 0 0 0 0>
     outlineSettings {
      bUseBrushTransparency=false
      roundingType=HalfHeightRadius
      width=0
      color {colorUseRule=UseColor_Specified, specifiedColor=<rgba 0 0 0 0>}
      cornerRadii=<0 0 0 0>
     }
     tintColor {colorUseRule=UseColor_Specified, specifiedColor=<rgba 1 1 1 1>}
     uVRegion {bIsValid=false, max=<0 0>, min=<0 0>}
    }}
   slot {class=OverlaySlot, properties {horizontalAlignment=HAlign_Fill, verticalAlignment=VAlign_Fill}}
  }
  WBP_CaptureForPostBufferUpdate {
   widgetClassPath=&WBP_CaptureForPostBufferUpdate
   bIsVariable=true
   parent=Overlay_81
   slot {class=OverlaySlot, properties {horizontalAlignment=HAlign_Fill, verticalAlignment=VAlign_Fill}}
  }
  SafeZone {
   widgetClassPath=&SafeZone
   bIsVariable=true
   parent=Overlay_81
   widget=SafeZone
   slot {class=OverlaySlot, properties {horizontalAlignment=HAlign_Fill, verticalAlignment=VAlign_Fill}}
  }
  GridPanel_0 {
   widgetClassPath=&GridPanel
   parent=SafeZone
   widget=GridPanel_0
   properties {bOverride_Cursor=true, columnFill [0,1,0], rowFill [0,1,0]}
   slot {class=SafeZoneSlot, properties {padding=<ltrb 48 24 48 24>}}
  }
  WBP_PlayerInfo {
   widgetClassPath=/Game/UI/Frontend/ItemShop/MultiProduct/WBP_PlayerInfo
   bIsVariable=true
   parent=GridPanel_0
   widget=WBP_PlayerInfo
   slot {class=GridSlot, properties {column=2, horizontalAlignment=HAlign_Right, layer=100, verticalAlignment=VAlign_Top}}
  }
  WBP_Upsell_TopNotification {
   widgetClassPath=&/WBP_Upsell_TopNotification
   bIsVariable=true
   parent=GridPanel_0
   widget=WBP_Upsell_TopNotification
   properties {width Override=750}
   slot {class=GridSlot, properties {columnSpan=3, horizontalAlignment=HAlign_Center, layer=100, verticalAlignment=VAlign_Top}}
  }
  SB_Legal {
   widgetClassPath=&SizeBox
   bIsVariable=true
   parent=GridPanel_0
   widget=SB_Legal
   slot {class=GridSlot, properties {columnSpan=3, layer=100, row=2, verticalAlignment=VAlign_Bottom}}
  }
  LegalandActionBar {widgetClassPath=&HorizontalBox, parent=SB_Legal, widget=LegalandActionBar, slot {class=SizeBoxSlot}}
  VB_Legal {
   widgetClassPath=&VerticalBox
   bIsVariable=true
   parent=LegalandActionBar
   widget=VB_Legal
   slot {class=HorizontalBoxSlot, properties {verticalAlignment=VAlign_Bottom, size {sizeRule=Fill, value=1}}}
  }
  LegalText {
   widgetClassPath=&CommonTextBlock
   bIsVariable=true
   parent=VB_Legal
   widget=LegalText
   properties {
    autoWrapText=true
    bIsScrollingEnabled=false
    mobileFontSizeMultiplier=1.4
    scrollOrientation=Orient_Vertical
    scrollStyle=&TextScrollStyle-Base_C
    text=Legal Text Goes Here really long hsuih dfoih oiehoihjsdg iohdfsg iongfoidh gfdoihdgfohdfgio uhdfsoiu hdfgs oijre ijoeir oierj greoij oeirgj oigerjoierjfierh oiurefhoiure hoiu erhroer
    colorAndOpacity {colorUseRule=UseColor_Specified, specifiedColor=<rgba 1 1 1 0.5>}
    font {
     bForceMonospaced=false
     bMaterialIsStencil=false
     fontMaterial=null
     fontObject=&NotoSans
     letterSpacing=0
     monospacedWidth=1
     size=24
     skewAmount=0
     typefaceFontName=NotoSans-Regular
     outlineSettings {
      bApplyOutlineToDropShadows=false
      bMiteredCorners=false
      bSeparateFillAlpha=false
      outlineMaterial=null
      outlineSize=0
      outlineColor=<rgba 0 0 0 1>
     }
    }
    shadowOffset=<0 0>
   }
   slot {class=VerticalBoxSlot}
  }
  WBP_RealPrice_LegalText {
   widgetClassPath=/FortUILibrary/UIKit/Price/WBP_RealPrice_LegalText
   bIsVariable=true
   parent=VB_Legal
   widget=WBP_RealPrice_LegalText
   properties {autoWrap=true, renderOpacity=0.5, wrapTextAt=0, textColor {colorUseRule=UseColor_Specified, specifiedColor=<rgba 1 1 1 1>}}
   slot {class=VerticalBoxSlot, properties {padding=<ltrb 0 8 0 0>}}
  }
  WBP_ActionBarContainer_1 {
   widgetClassPath=/Game/UI/Frontend/BottomBar/WBP_ActionBarContainer
   bIsVariable=true
   parent=LegalandActionBar
   widget=WBP_ActionBarContainer_1
   slot {class=HorizontalBoxSlot, properties {verticalAlignment=VAlign_Bottom, padding=<ltrb 20 0 0 0>}}
  }
  HorizontalBox_1 {widgetClassPath=&HorizontalBox, namedSlotHost=WBP_ActionBarContainer_1, widget=HorizontalBox_1}
  Button_GiftPass {
   widgetClassPath=/FortUILibrary/UIKit/Buttons/WBP_UIKit_Button_Quiet
   bIsVariable=true
   parent=HorizontalBox_1
   widget=Button_GiftPass
   properties {
    bIsFocusable=false
    text=Gift The Pass
    triggeringInputAction {dataTable=&InputActionDataTable, rowName=GenericFaceButtonTop}
   }
   slot {class=HorizontalBoxSlot, properties {padding=<ltrb 0 0 24 0>}}
  }
  Button_SeeBenefits {
   widgetClassPath=&WBP_UIKit_Button_Quiet
   bIsVariable=true
   parent=HorizontalBox_1
   widget=Button_SeeBenefits
   properties {minWidth=184, text=See Benefits, triggeringInputAction {dataTable=/SeasonPass/Data/SeasonPassInputActionDataTable, rowName=SeeBenefits}}
   slot {class=HorizontalBoxSlot, properties {horizontalAlignment=HAlign_Right, verticalAlignment=VAlign_Top, padding=<ltrb 0 0 24 0>}}
  }
  Button_Close {
   widgetClassPath=&WBP_UIKit_Button_Quiet
   bIsVariable=true
   parent=HorizontalBox_1
   widget=Button_Close
   properties {bIsFocusable=false, text=Close, triggeringInputAction {dataTable=&InputActionDataTable, rowName=GenericBack}}
   slot {class=HorizontalBoxSlot}
  }
  WBP_AthenaBottomBar_WithNotifications {
   widgetClassPath=/Game/UI/Frontend/BottomBar/WBP_AthenaBottomBar_WithNotifications
   bIsVariable=true
   parent=HorizontalBox_1
   widget=WBP_AthenaBottomBar_WithNotifications
   slot {class=HorizontalBoxSlot, properties {padding=<ltrb 24 0 0 0>}}
  }
  Pass_Switcher {
   widgetClassPath=&CommonVisibilitySwitcher
   bIsVariable=true
   parent=GridPanel_0
   widget=Pass_Switcher
   slot {class=GridSlot, properties {columnSpan=3, rowSpan=2}}
  }
  SizeBox_Content {
   widgetClassPath=&SizeBox
   bIsVariable=true
   parent=Pass_Switcher
   widget=SizeBox_Content
   properties {
    bOverride_HeightOverride=true
    bOverride_MaxAspectRatio=true
    bOverride_WidthOverride=true
    heightOverride=940
    maxAspectRatio=2.5
    maxDesiredHeight=900
    maxDesiredWidth=1760
    minAspectRatio=0.667
    minDesiredHeight=450
    minDesiredWidth=880
    widthOverride=1760
   }
   slot {class=CommonVisibilitySwitcherSlot, properties {padding=<ltrb 120 96 120 0>}}
  }
  VB_Content {
   widgetClassPath=&VerticalBox
   bIsVariable=true
   parent=SizeBox_Content
   widget=VB_Content
   properties {bOverride_Cursor=true, flowDirectionPreference=Culture}
   slot {class=SizeBoxSlot, properties {padding=<ltrb 0 -60 0 -40>}}
  }
  SB_PassButtons {
   widgetClassPath=&SizeBox
   bIsVariable=true
   parent=VB_Content
   widget=SB_PassButtons
   properties {bOverride_HeightOverride=true, heightOverride=526, widthOverride=1000}
   slot {class=VerticalBoxSlot, properties {horizontalAlignment=HAlign_Center, verticalAlignment=VAlign_Bottom, padding=<ltrb 0 36 0 0>}}
  }
  UpsellItemGroup {
   widgetClassPath=&GridPanel
   bIsVariable=true
   parent=SB_PassButtons
   widget=UpsellItemGroup
   properties {columnFill [1], rowFill [1]}
   slot {class=SizeBoxSlot, properties {horizontalAlignment=HAlign_Center, verticalAlignment=VAlign_Bottom}}
  }
  UpsellItems {
   widgetClassPath=&DynamicEntryBox
   bIsVariable=true
   parent=UpsellItemGroup
   widget=UpsellItems
   properties {
    entryHorizontalAlignment=HAlign_Center
    entryVerticalAlignment=VAlign_Bottom
    entryWidgetClass=&WBP_PassPurchaseButton_C
    entrySpacing=<32 0>
   }
   slot {class=GridSlot, properties {columnSpan=1, horizontalAlignment=HAlign_Center, rowSpan=1, verticalAlignment=VAlign_Bottom}}
  }
  Overlay_Benefits {
   widgetClassPath=&Overlay
   bIsVariable=true
   parent=VB_Content
   widget=Overlay_Benefits
   slot {class=VerticalBoxSlot, properties {horizontalAlignment=HAlign_Center, padding=<ltrb 0 48 0 0>}}
  }
  UpsellBenefitsContainer {
   widgetClassPath=&GridPanel
   bIsVariable=true
   parent=Overlay_Benefits
   widget=UpsellBenefitsContainer
   properties {bOverride_Cursor=true, columnFill [1], rowFill [1]}
   slot {class=OverlaySlot, properties {horizontalAlignment=HAlign_Fill, verticalAlignment=VAlign_Fill}}
  }
  BenefitGroups {
   widgetClassPath=&HorizontalBox
   bIsVariable=true
   parent=UpsellBenefitsContainer
   widget=BenefitGroups
   properties {bOverride_Cursor=true}
   slot {class=GridSlot, properties {columnSpan=1, horizontalAlignment=HAlign_Center, rowSpan=1}}
  }
  SingleLineBenefits {
   widgetClassPath=&VerticalBox
   bIsVariable=true
   parent=BenefitGroups
   widget=SingleLineBenefits
   properties {visibility=Collapsed}
   slot {class=HorizontalBoxSlot, properties {horizontalAlignment=HAlign_Center}}
  }
  LeftBenefitsLabel_1 {
   widgetClassPath=&CommonTextBlock
   bIsVariable=true
   parent=SingleLineBenefits
   widget=LeftBenefitsLabel_1
   properties {
    text=Earn Up To:
    colorAndOpacity {colorUseRule=UseColor_Specified, specifiedColor=<rgba 1 1 1 0.8>}
    font {
     bForceMonospaced=false
     bMaterialIsStencil=false
     fontMaterial=null
     fontObject=null
     letterSpacing=0
     monospacedWidth=1
     size=24
     skewAmount=0
     typefaceFontName=None
     outlineSettings {
      bApplyOutlineToDropShadows=false
      bMiteredCorners=false
      bSeparateFillAlpha=false
      outlineMaterial=null
      outlineSize=0
      outlineColor=<rgba 0 0 0 1>
     }
    }
    shadowOffset=<0 0>
   }
   slot {class=VerticalBoxSlot, properties {horizontalAlignment=HAlign_Left, verticalAlignment=VAlign_Top, padding=<ltrb 10 0 0 24>}}
  }
  UpsellBenefitsGroup {
   widgetClassPath=&HorizontalBox
   bIsVariable=true
   parent=SingleLineBenefits
   widget=UpsellBenefitsGroup
   properties {bOverride_Cursor=true}
   slot {class=VerticalBoxSlot, properties {verticalAlignment=VAlign_Top}}
  }
  UpsellBenefits {
   widgetClassPath=&DynamicEntryBox
   bIsVariable=true
   parent=UpsellBenefitsGroup
   widget=UpsellBenefits
   properties {
    entryWidgetClass=&CrewUpsellMultiViewBenefit_C
    numDesignerPreviewEntries=4
    entrySpacing=<32 0>
   }
   slot {class=HorizontalBoxSlot, properties {horizontalAlignment=HAlign_Center, verticalAlignment=VAlign_Top, size {sizeRule=Fill, value=1}}}
  }
  LeftBenefits {
   widgetClassPath=&VerticalBox
   bIsVariable=true
   parent=BenefitGroups
   widget=LeftBenefits
   properties {bOverride_Cursor=true}
   slot {class=HorizontalBoxSlot, properties {horizontalAlignment=HAlign_Left}}
  }
  LeftBenefitsLabel {
   widgetClassPath=&CommonTextBlock
   bIsVariable=true
   parent=LeftBenefits
   widget=LeftBenefitsLabel
   properties {
    text=Earn Up To:
    colorAndOpacity {colorUseRule=UseColor_Specified, specifiedColor=<rgba 1 1 1 0.8>}
    font {
     bForceMonospaced=false
     bMaterialIsStencil=false
     fontMaterial=null
     fontObject=&NotoSans
     letterSpacing=0
     monospacedWidth=1
     size=24
     skewAmount=0
     typefaceFontName=NotoSans-Regular
     outlineSettings {
      bApplyOutlineToDropShadows=false
      bMiteredCorners=false
      bSeparateFillAlpha=false
      outlineMaterial=null
      outlineSize=0
      outlineColor=<rgba 0 0 0 1>
     }
    }
    shadowOffset=<0 0>
   }
   slot {class=VerticalBoxSlot, properties {verticalAlignment=VAlign_Top, padding=<ltrb 10 0 0 12>}}
  }
  HorizontalBox_0 {
   widgetClassPath=&HorizontalBox
   bIsVariable=true
   parent=LeftBenefits
   widget=HorizontalBox_0
   slot {class=VerticalBoxSlot, properties {horizontalAlignment=HAlign_Left, padding=<ltrb 0 8 0 0>}}
  }
  VerticalBox_5 {
   widgetClassPath=&VerticalBox
   parent=HorizontalBox_0
   widget=VerticalBox_5
   properties {bOverride_Cursor=true}
   slot {class=HorizontalBoxSlot, properties {horizontalAlignment=HAlign_Left, padding=<ltrb 0 0 0 8>}}
  }
  BenefitGroupLeftWidget0 {
   widgetClassPath=&/CrewUpsellMultiViewBenefit
   bIsVariable=true
   parent=VerticalBox_5
   widget=BenefitGroupLeftWidget0
   slot {class=VerticalBoxSlot, properties {padding=<ltrb 0 0 0 8>}}
  }
  BenefitGroupLeftWidget1 {
   widgetClassPath=&CrewUpsellMultiViewBenefit
   bIsVariable=true
   parent=VerticalBox_5
   widget=BenefitGroupLeftWidget1
   slot {class=VerticalBoxSlot, properties {padding=<ltrb 0 0 0 8>}}
  }
  VerticalBox_4 {
   widgetClassPath=&VerticalBox
   parent=HorizontalBox_0
   widget=VerticalBox_4
   properties {bOverride_Cursor=true}
   slot {class=HorizontalBoxSlot, properties {horizontalAlignment=HAlign_Left, padding=<ltrb 8 0 0 8>}}
  }
  BenefitGroupLeftWidget2 {
   widgetClassPath=&CrewUpsellMultiViewBenefit
   bIsVariable=true
   parent=VerticalBox_4
   widget=BenefitGroupLeftWidget2
   slot {class=VerticalBoxSlot, properties {padding=<ltrb 0 0 0 8>}}
  }
  BenefitGroupLeftWidget3 {
   widgetClassPath=&CrewUpsellMultiViewBenefit
   bIsVariable=true
   parent=VerticalBox_4
   widget=BenefitGroupLeftWidget3
   slot {class=VerticalBoxSlot, properties {padding=<ltrb 0 0 0 8>}}
  }
  Div {
   widgetClassPath=&Image
   bIsVariable=true
   parent=BenefitGroups
   widget=Div
   properties {brush {
     bIsDynamicallyLoaded=false
     drawAs=Image
     imageType=NoImage
     mirroring=NoMirror
     resourceName=None
     resourceObject=null
     tiling=NoTile
     imageSize=<2 110>
     margin=<ltrb 0 0 0 0>
     outlineSettings {
      bUseBrushTransparency=false
      roundingType=HalfHeightRadius
      width=0
      color {colorUseRule=UseColor_Specified, specifiedColor=<rgba 0 0 0 0>}
      cornerRadii=<0 0 0 0>
     }
     tintColor {colorUseRule=UseColor_Specified, specifiedColor=<rgba 0.6939 0.6939 0.6939 0.2>}
     uVRegion {bIsValid=false, max=<0 0>, min=<0 0>}
    }}
   slot {class=HorizontalBoxSlot, properties {horizontalAlignment=HAlign_Center, verticalAlignment=VAlign_Center, padding=<ltrb 32 12 32 0>}}
  }
  RightBenefits {
   widgetClassPath=&VerticalBox
   bIsVariable=true
   parent=BenefitGroups
   widget=RightBenefits
   properties {bOverride_Cursor=true}
   slot {class=HorizontalBoxSlot, properties {horizontalAlignment=HAlign_Right}}
  }
  RightBenefitsLabel {
   widgetClassPath=&CommonTextBlock
   bIsVariable=true
   parent=RightBenefits
   widget=RightBenefitsLabel
   properties {
    text=Instantly Claim:
    colorAndOpacity {colorUseRule=UseColor_Specified, specifiedColor=<rgba 1 1 1 0.8>}
    font {
     bForceMonospaced=false
     bMaterialIsStencil=false
     fontMaterial=null
     fontObject=&NotoSans
     letterSpacing=0
     monospacedWidth=1
     size=24
     skewAmount=0
     typefaceFontName=NotoSans-Regular
     outlineSettings {
      bApplyOutlineToDropShadows=false
      bMiteredCorners=false
      bSeparateFillAlpha=false
      outlineMaterial=null
      outlineSize=0
      outlineColor=<rgba 0 0 0 1>
     }
    }
    shadowOffset=<0 0>
   }
   slot {class=VerticalBoxSlot, properties {verticalAlignment=VAlign_Top, padding=<ltrb 2 0 0 12>}}
  }
  HorizontalBox_6 {
   widgetClassPath=&HorizontalBox
   bIsVariable=true
   parent=RightBenefits
   widget=HorizontalBox_6
   slot {class=VerticalBoxSlot, properties {horizontalAlignment=HAlign_Left, padding=<ltrb -8 8 0 0>}}
  }
  VerticalBox_7 {
   widgetClassPath=&VerticalBox
   parent=HorizontalBox_6
   widget=VerticalBox_7
   properties {bOverride_Cursor=true}
   slot {class=HorizontalBoxSlot, properties {horizontalAlignment=HAlign_Left, padding=<ltrb 0 0 0 8>}}
  }
  BenefitGroupRightWidget0 {
   widgetClassPath=&CrewUpsellMultiViewBenefit
   bIsVariable=true
   parent=VerticalBox_7
   widget=BenefitGroupRightWidget0
   slot {class=VerticalBoxSlot, properties {padding=<ltrb 0 0 0 8>}}
  }
  BenefitGroupRightWidget2 {
   widgetClassPath=&CrewUpsellMultiViewBenefit
   bIsVariable=true
   parent=VerticalBox_7
   widget=BenefitGroupRightWidget2
   slot {class=VerticalBoxSlot, properties {padding=<ltrb 0 0 0 8>}}
  }
  VerticalBox_6 {
   widgetClassPath=&VerticalBox
   parent=HorizontalBox_6
   widget=VerticalBox_6
   properties {bOverride_Cursor=true}
   slot {class=HorizontalBoxSlot, properties {horizontalAlignment=HAlign_Left, padding=<ltrb 8 0 0 8>}}
  }
  BenefitGroupRightWidget1 {
   widgetClassPath=&CrewUpsellMultiViewBenefit
   bIsVariable=true
   parent=VerticalBox_6
   widget=BenefitGroupRightWidget1
   slot {class=VerticalBoxSlot, properties {padding=<ltrb 0 0 0 8>}}
  }
  BenefitGroupRightWidget3 {
   widgetClassPath=&CrewUpsellMultiViewBenefit
   bIsVariable=true
   parent=VerticalBox_6
   widget=BenefitGroupRightWidget3
   slot {class=VerticalBoxSlot, properties {padding=<ltrb 0 0 0 8>}}
  }
  WBP_UpsellScreen_BP {
   widgetClassPath=/SeasonPassHub/Widgets/UpsellBattlePass/WBP_UpsellScreen_BP
   bIsVariable=true
   parent=Pass_Switcher
   widget=WBP_UpsellScreen_BP
   properties {bSupportsActivationFocus=false}
   slot {class=CommonVisibilitySwitcherSlot}
  }
  ContentWarningContainer {
   widgetClassPath=&SizeBox
   bIsVariable=true
   parent=GridPanel_0
   widget=ContentWarningContainer
   properties {bOverride_WidthOverride=true, widthOverride=1252}
   slot {class=GridSlot, properties {columnSpan=3, horizontalAlignment=HAlign_Center, row=1, verticalAlignment=VAlign_Bottom}}
  }
  WBP_UIKit_InfoBox_ContentWarning {
   widgetClassPath=/FortUILibrary/UIKit/InfoBox/WBP_UIKit_InfoBox
   bIsVariable=true
   parent=ContentWarningContainer
   widget=WBP_UIKit_InfoBox_ContentWarning
   properties {
    iconMaterial=/FortUILibrary/Textures/MaterialInstances/MI_UIKit_Icon_AgeGatedAlert
    iconTopAlignment=true
    previewText="Disclaimer text here Disclaimer text here Disclaimer text here Disclaimer text here Disclaimer text here Disclaimer text here "
    visibility=Collapsed
   }
   slot {class=SizeBoxSlot, properties {padding=<ltrb 0 0 0 12>}}
  }
 ]
}
