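# Activatable Crew upsell screen whose switcher holds a loading spinner and four layouts: 9010, 7030, 5050 and MultiView.
# Config comes from the interstitial view model, else the Crew view model's CMS or experiment config.
# The spinner can wait for Season Pass offer data, deactivating on timeout; activation pushes config to every layout, choosing one by config flags or an index override.
# Closing a layout deactivates the screen and advances interstitials.
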WidgetBlueprint /SeasonPassHub/Widgets/CrewUpsellScreen
graphs [
 ActivateConfig {nodes [
   ActivateConfig!() -> (then, Config:Crew Upsell Config Structure) {is=title K2Node_FunctionEntry}
   > SetConfig!(
    execute
    self:Crew Upsell Screen 9010 Object Reference=Get CrewUpsellScreen9010() -> (CrewUpsellScreen9010:Crew Upsell Screen 9010 Object Reference)
    Config:Crew Upsell Config Structure=Get Config() -> (Config:Crew Upsell Config Structure)
   ) -> (then)
   > SetConfig(
    self:Crew Upsell Screen 7030 Object Reference=Get CrewUpsellScreen7030() -> (CrewUpsellScreen7030:Crew Upsell Screen 7030 Object Reference)
    Config=Get Config()
   )
   > SetConfig(
    self:Crew Upsell Screen 5050 Object Reference=Get CrewUpsellScreen5050() -> (CrewUpsellScreen5050:Crew Upsell Screen 5050 Object Reference)
    Config=Get Config()
   )
   > SetConfig(
    self:Crew Upsell Screen Multi View Object Reference=Get CrewUpsellScreenMultiView() -> (CrewUpsellScreenMultiView:Crew Upsell Screen Multi View Object Reference)
    Config=Get Config()
   )
   > SetActiveWidget!(execute, self:Self Object Reference, Config:Crew Upsell Config Structure=Get Config()) -> (then)
   > Clear and Invalidate Timer by Handle(Handle=Get TimeoutHandle() -> (TimeoutHandle:Timer Handle Structure))
  ]}
 BP_GetDesiredFocusTarget {nodes [
   Get Desired Focus Target()
   > Return Node!(
    execute
    ReturnValue:Widget Object Reference=GetActiveWidget!(self:Widget Switcher Object Reference=Get Switcher() -> (Switcher:Common Activatable Widget Switcher Object Reference)) -> (ReturnValue:Widget Object Reference)
   )
  ]}
 CrewVMChanged {nodes [
   n1 = CrewVMChanged!() -> (then, CrewVM:Crew VM Object Reference) {is=title K2Node_FunctionEntry}
   > SetConfig(
    self:Self Object Reference
    Config=GetConfig!(
     self
     CrewVM:Crew VM Object Reference=@n1
     InterstitialVM:BPUpsell Interstitial VM Object Reference=Get BPUpsellInterstitialVM() -> (BPUpsellInterstitialVM:BPUpsell Interstitial VM Object Reference)
    ) -> (Config:Crew Upsell Config Structure)
   )
  ]}
 EventGraph {nodes [
   n1 = "On Closed (CrewUpsellScreen9010)"() {is=title K2Node_ComponentBoundEvent}
   DeactivateWidget(execute=[@n1.then, @n2.then, @n3.then, @n4.then])
   > DismissInterstitial!(execute, self:Fort Interstitial VM Object Reference=Get BPUpsellInterstitialVM(), bShouldShowNext:bool=true) -> (then)
   n2 = "On Closed (CrewUpsellScreen7030)"() {is=title K2Node_ComponentBoundEvent}
   n3 = "On Closed (CrewUpsellScreen5050)"() {is=title K2Node_ComponentBoundEvent}
   n4 = "On Closed (CrewUpsellScreenMultiView)"() {is=title K2Node_ComponentBoundEvent}
   n5 = Timeout() {is=title K2Node_CustomEvent}
   > PrintString!(
    execute
    InString=CrewUpsellScreen timed out while waiting for pass offer data.
    bPrintToScreen
    bPrintToLog
    TextColor=(R=1,G=0,B=0,A=1)
    Duration=10
    Key
   ) -> (then)
   > DeactivateWidget()
   n6 = Test() {is=title K2Node_CustomEvent}
   TryActivateConfig!(execute=[@n6.then, @n10.else, @n8.then], self, Offer:Fort Simple Offer VM Object Reference) -> (then)
   Event On Activated()
   > n7 = Set Timer by Event(
    Delegate=@n5
    Time=GetConfigDataFloatWithDefault!(EntryName:String=CrewUpsellWaitTimeout, DefaultValue:float=60) -> (ReturnValue:float)
    bLooping=false
   )
   > n8 = Set TimeoutHandle(TimeoutHandle:Timer Handle Structure=@n7) -> (Output_Get:Timer Handle Structure)
   Event On Deactivated()
   > Clear and Invalidate Timer by Handle(Handle=Get TimeoutHandle())
   Event Construct()
   > SetActiveWidget!(
    execute
    self:Widget Switcher Object Reference=Get Switcher()
    Widget:Widget Object Reference=Get CrewUpsellSpinner() -> (CrewUpsellSpinner:Crew Upsell Spinner Object Reference)
   ) -> (then)
   n9 = GetConfigDataFloatWithDefault(EntryName=CrewUpsellTestWait, DefaultValue=20)
   n10 = Branch(Condition="float > float"(A=@n9))
   > Set Timer by Event(Delegate=@n6, Time=@n9, bLooping=false)
  ]}
 GetConfig {nodes [
   n1 = GetConfig!() -> (then, CrewVM:Crew VM Object Reference, InterstitialVM:BPUpsell Interstitial VM Object Reference) {is=title K2Node_FunctionEntry}
   > n2 = Is Valid(InputObject=@n1.InterstitialVM)
   Return Node!(
    execute=@n2.Is Not Valid
    Config:Crew Upsell Config Structure=Select(
     Option 0:Crew Upsell Config Structure=Get CMSUpsellConfig!(self:Crew VM Object Reference=@n1) -> (CMSUpsellConfig:Crew Upsell Config Structure)
     Option 1:Crew Upsell Config Structure=Get ExperimentUpsellConfig!(self:Crew VM Object Reference=@n1) -> (ExperimentUpsellConfig:Crew Upsell Config Structure)
     Index=Get UseExperiment() -> (UseExperiment:bool)
    ) -> (ReturnValue:Crew Upsell Config Structure)
   )
   > Return Node(Config=Get Config!(self:BPUpsell Interstitial VM Object Reference=@n1.InterstitialVM) -> (Config))
  ]}
 GetPrimaryContentSetup {nodes [
   GetPrimaryContentSetup() {is=title K2Node_FunctionEntry}
   > Return Node!(
    execute
    ReturnValue:Primary Content Setup Structure=Make Primary Content Setup!(
     bShowBottomBar:bool=false
     bShowChatWidget:bool=false
     bEnableNewTextChat:bool=False
     bShowTopBar:bool=false
     bTabSupportsInterface:bool=False
    ) -> (PrimaryContentSetup:Primary Content Setup Structure) {is=title K2Node_MakeStruct}
   )
  ]}
 InterstitialChanged {nodes [
   n1 = InterstitialChanged!() -> (then, Interstitial:BPUpsell Interstitial VM Object Reference) {is=title K2Node_FunctionEntry}
   > SetConfig(Config=GetConfig(CrewVM=Get CrewVM() -> (CrewVM:Crew VM Object Reference), InterstitialVM=@n1) {is=title K2Node_CallFunction})
  ]}
 SetActiveWidget {nodes [
   n1 = SetActiveWidget!() -> (then, Config:Crew Upsell Config Structure) {is=title K2Node_FunctionEntry}
   n2 = GetConfigDataIntWithDefault!(EntryName:String=OverrideUpsellScreen, DefaultValue:int=-1) -> (ReturnValue:int=0)
   n3 = Get Switcher()
   > n4 = Branch(Condition="In Range (Integer)"!(
     Value:int=@n2
     Min:int=0
     Max:int=GetNumWidgets!(self:Widget Switcher Object Reference=@n3) -> (ReturnValue:int=0)
     InclusiveMin:bool=true
     InclusiveMax:bool=false
    ) -> (ReturnValue:bool))
   > SetActiveWidgetIndex(self=Get Switcher(), Index=@n2)
   n5 = Break Crew Upsell Config!(CrewUpsellConfig:"Crew Upsell Config Structure (by ref)"=@n1) -> (
    OverrideUpsellPass:String
    PassBackgroundUrl:Array of Motd Image Structures
    CrewBackgroundUrl:Array of Motd Image Structures
    bUsePassBackgroundMaterial:bool=~
    bUseCrewBackgroundMaterial:bool=~
    bCrewOnRightInUpsell:bool=~
    bCrew50PercentInUpsell:bool=~
    bCrew30PercentInUpsell:bool=~
    bCrewAsSliver:bool=~
    bCrewOnlyInUpsell:bool=~
    bPassOnlyInUpsell:bool=~
    bAnimateCrewBG:bool=~
    bUseNewUpsellFormat:bool=~
    bAllowToggleBundle:bool=~
    bDefaultShowBundle:bool=~
    EntryOverrides:Array of Nexus Upsell Entry Structures
   ) {is=title K2Node_BreakStruct}
   n6 = Select(
    Option 0:Widget Object Reference=Get CrewUpsellScreen9010()
    Option 1:Widget Object Reference=Get CrewUpsellScreen7030()
    Index=@n5.bCrew30PercentInUpsell
   ) -> (ReturnValue:Widget Object Reference)
   n7 = Get CrewUpsellScreen5050()
   SetActiveWidget(
    execute=@n4.else
    self=Get Switcher()
    Widget=Select(
     Option 0=Select(Option 0=@n6, Option 1=@n7, Index=@n5.bCrew50PercentInUpsell)
     Option 1=Get CrewUpsellScreenMultiView()
     Index=@n5.bUseNewUpsellFormat
    )
   ) {is=title K2Node_CallFunction}
  ]}
 SetConfig {nodes [
   n1 = SetConfig!() -> (then, Config:Crew Upsell Config Structure) {is=title K2Node_FunctionEntry}
   > Set Config(Config:Crew Upsell Config Structure=@n1) -> (Output_Get:Crew Upsell Config Structure)
  ]}
 TryActivateConfig {nodes [
   TryActivateConfig!() -> (then, Offer:Fort Simple Offer VM Object Reference) {is=title K2Node_FunctionEntry}
   > n1 = Branch(Condition=GetConfigDataBoolWithDefault!(EntryName:String=CrewUpsellWaitEnabled, bDefaultValue:bool=true) -> (ReturnValue:bool))
   n2 = GetActiveWidget(self=Get Switcher())
   n3 = Get CrewUpsellSpinner()
   > Branch(Condition=AND Boolean(A=IsActivated!(self:Common Activatable Widget Object Reference) -> (ReturnValue:bool), B="Equal (Object)"(A=@n2, B=@n3)))
   > n4 = Is Valid(InputObject=Get SeasonPassVM() -> (SeasonPassVM:Season Pass VM Object Reference))
   n5 = Get SeasonPassVM()
   n6 = Get PassOffer!(self:Season Pass VM Object Reference=@n5) -> (PassOffer:Fort Simple Offer VM Object Reference)
   > n7 = Branch(Condition=OR Boolean(A=IsValid(Object=@n6), B=Get bIsDevPass!(self:Season Pass VM Object Reference=@n5) -> (bIsDevPass:bool)))
   ActivateConfig!(execute=[@n7.then, @n4.Is Not Valid, @n1.else], self, Config:Crew Upsell Config Structure=Get Config!() -> (Config)) -> (then) {is=title K2Node_CallFunction}
  ]}
]
variables [
 Config {
  type=CrewUpsellConfig
  default {
   bAllowToggleBundle=true
   bAnimateCrewBG=true
   bCrew30PercentInUpsell=false
   bCrew50PercentInUpsell=false
   bCrewAsSliver=false
   bCrewOnRightInUpsell=false
   bCrewOnlyInUpsell=false
   bDefaultShowBundle=true
   bPassOnlyInUpsell=false
   bUseCrewBackgroundMaterial=true
   bUseNewUpsellFormat=true
   bUsePassBackgroundMaterial=true
   overrideUpsellPass=""
   crewBackgroundUrl []
   entryOverrides []
   passBackgroundUrl []
  }
 }
 TimeoutHandle {type=TimerHandle}
 UseExperiment {type=boolean}
]
widgets {
 info {
  inheritedWidgetCount=0
  parentClass=&CommonActivatableWidget
  rootWidgetClass=&CommonActivatableWidgetSwitcher
  widgetCount=6
 }
 widgets [
  Switcher {widgetClassPath=&CommonActivatableWidgetSwitcher, bIsVariable=true, widget=Switcher, properties {transitionDuration=0}}
  CrewUpsellSpinner {
   widgetClassPath=&/CrewUpsellSpinner
   bIsVariable=true
   parent=Switcher
   widget=CrewUpsellSpinner
   slot {class=WidgetSwitcherSlot}
  }
  CrewUpsellScreen5050 {
   widgetClassPath=&/CrewUpsellScreen5050
   bIsVariable=true
   parent=Switcher
   widget=CrewUpsellScreen5050
   properties {bIsFocusable=true}
   slot {class=WidgetSwitcherSlot}
  }
  CrewUpsellScreen9010 {
   widgetClassPath=&/CrewUpsellScreen9010
   bIsVariable=true
   parent=Switcher
   widget=CrewUpsellScreen9010
   properties {bIsFocusable=true}
   slot {class=WidgetSwitcherSlot}
  }
  CrewUpsellScreen7030 {
   widgetClassPath=&/CrewUpsellScreen7030
   bIsVariable=true
   parent=Switcher
   widget=CrewUpsellScreen7030
   properties {bIsFocusable=true}
   slot {class=WidgetSwitcherSlot}
  }
  CrewUpsellScreenMultiView {
   widgetClassPath=&/CrewUpsellScreenMultiView
   bIsVariable=true
   parent=Switcher
   widget=CrewUpsellScreenMultiView
   properties {bIsFocusable=true}
   slot {class=WidgetSwitcherSlot}
  }
 ]
}
