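# Season pass upsell screen that layers CrewUpsellScreen over a full-size item preview using the item shop scene changer tag.
# Activation activates the preview and binds to the Crew screen's deactivation, which deactivates this widget and removes the binding; deactivation also deactivates the preview.
# GetPrimaryContentSetup hides the top bar, bottom bar and chat.
# GetUpsellPass finds the pass view model by ID, falling back to "battlepassdata" when empty.
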WidgetBlueprint /SeasonPassHub/Widgets/BattlePassUpsell
graphs [
 EventGraph {nodes [
   n1 = CustomEvent() {is=title K2Node_CustomEvent}
   > DeactivateWidget()
   n2 = Get CrewUpsellScreen() -> (CrewUpsellScreen:Crew Upsell Screen Object Reference)
   > Unbind Event from On Widget Deactivated(self:Common Activatable Widget Object Reference=@n2, Delegate=@n1) {is=title K2Node_RemoveDelegate}
   Event On Activated()
   n3 = Get FortItemPreviewWidget() -> (FortItemPreviewWidget:WBP Item Preview Object Reference)
   > ActivateWidget(self=@n3)
   > Bind Event to On Widget Deactivated(self:Common Activatable Widget Object Reference=@n2, Delegate=@n1) {is=title K2Node_AddDelegate}
   Event On Deactivated()
   > DeactivateWidget(self=@n3)
  ]}
 GetPrimaryContentSetup {nodes [
   GetPrimaryContentSetup() {is=title K2Node_FunctionEntry}
   > Return Node!(
    execute
    ReturnValue:Primary Content Setup Structure=Make Primary Content Setup!(
     bShowBottomBar:bool=false
     bShowChatWidget:bool=false
     bEnableNewTextChat:bool=False
     bShowTopBar:bool=false
     bTabSupportsInterface:bool=False
    ) -> (PrimaryContentSetup:Primary Content Setup Structure) {is=title K2Node_MakeStruct}
   )
  ]}
 GetUpsellPass {nodes [
   n1 = GetUpsellPass!() -> (then, PassSystem:Season Pass System VM Object Reference, PassId:String) {is=title K2Node_FunctionEntry}
   n2 = IsEmpty(InString=@n1.PassId)
   > Return Node!(
    execute
    Result:Season Pass VM Object Reference=FindSeasonPass!(
     self:Season Pass System VM Object Reference=@n1
     TemplateId:String=Select(Option 0:String=@n1.PassId, Option 1:String=battlepassdata, Index=@n2) -> (ReturnValue:String)
    ) -> (ReturnValue:Season Pass VM Object Reference)
   )
  ]}
]
variables []
widgets {
 info {inheritedWidgetCount=0, parentClass=&CommonActivatableWidget, rootWidgetClass=&Overlay, widgetCount=3}
 widgets [
  Overlay_0 {widgetClassPath=&Overlay, widget=Overlay_0}
  FortItemPreviewWidget {
   widgetClassPath=/FortniteCoreUI/Frontend/ItemPreview/WBP_ItemPreview
   bIsVariable=true
   parent=Overlay_0
   widget=FortItemPreviewWidget
   properties {
    bSupportsActivationFocus=false
    itemViewObject=&self:WidgetTree.FortItemPreviewWidget.ItemViewObject
    previewSceneChangerTag {tagName=UI.ItemPreview.SceneChanger.ItemShop}
   }
   slot {class=OverlaySlot, properties {horizontalAlignment=HAlign_Fill, verticalAlignment=VAlign_Fill}}
  }
  CrewUpsellScreen {
   widgetClassPath=&/CrewUpsellScreen
   bIsVariable=true
   parent=Overlay_0
   widget=CrewUpsellScreen
   properties {bAutoActivate=true, useExperiment=true}
   slot {class=OverlaySlot, properties {horizontalAlignment=HAlign_Fill, verticalAlignment=VAlign_Fill}}
  }
 ]
}
