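# Widget Blueprint interstitial wrapper that hosts the Crew info container as an activatable screen presenting Crew subscription benefits.
# On construct it binds the container's deactivation and Crew purchase events to one handler.
# That handler deactivates the wrapper and dismisses the CrewBenefitsInterstitial view model, requesting that the next interstitial show.
# It directs focus to the container and hides the top bar, bottom bar and chat in its primary content setup.
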WidgetBlueprint /SeasonPassHub/Widgets/CrewBenefitsInterstitialWrapper
graphs [
 BP_GetDesiredFocusTarget {nodes [
   Get Desired Focus Target()
   > Return Node!(execute, ReturnValue:Widget Object Reference=Get CrewInfoContainer() -> (CrewInfoContainer:Crew Info Container Object Reference))
  ]}
 EventGraph {nodes [
   n1 = Pop Self() {is=title K2Node_CustomEvent}
   > DeactivateWidget()
   > DismissInterstitial!(
    execute
    self:Fort Interstitial VM Object Reference=Get CrewBenefitsInterstitialVM() -> (CrewBenefitsInterstitialVM:Crew Benefits Interstitial VM Object Reference)
    bShouldShowNext:bool=true
   ) -> (then)
   Event Construct()
   n2 = Get CrewInfoContainer()
   > Bind Event to On Widget Deactivated(self:Common Activatable Widget Object Reference=@n2, Delegate=@n1) {is=title K2Node_AddDelegate}
   > Bind Event to Crew Purchased(self:Crew Info Container Object Reference=@n2, Delegate=@n1) {is=title K2Node_AddDelegate}
  ]}
 GetPrimaryContentSetup {nodes [
   GetPrimaryContentSetup() {is=title K2Node_FunctionEntry}
   > Return Node(ReturnValue:Primary Content Setup Structure=Make Primary Content Setup!(
     bShowBottomBar:bool=false
     bShowChatWidget:bool=false
     bEnableNewTextChat:bool=False
     bShowTopBar:bool=false
     bTabSupportsInterface:bool=False
    ) -> (PrimaryContentSetup:Primary Content Setup Structure) {is=title K2Node_MakeStruct})
  ]}
]
variables []
widgets {
 info {
  inheritedWidgetCount=0
  parentClass=&CommonActivatableWidget
  rootWidgetClass=/Crew/UI/CrewInfoContainer.CrewInfoContainer_C
  widgetCount=1
 }
 widgets [CrewInfoContainer {widgetClassPath=/Crew/UI/CrewInfoContainer, bIsVariable=true, widget=CrewInfoContainer, properties {bAutoActivate=true}}]
}
