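# Full-screen Season Pass Hub interstitial that shows the Crew upsell screen over an item preview using the Item Shop preview scene.
# The auto-activating upsell screen deactivates this activatable widget when it deactivates, and the item preview follows this widget's activation.
# GetPrimaryContentSetup hides the top bar, bottom bar and chat.
# GetUpsellPass looks up the season pass named by the upsell config's override pass ID.
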WidgetBlueprint /SeasonPassHub/Widgets/InterstitialUpsell
graphs [
 EventGraph {nodes [
   "On Widget Deactivated (CrewUpsellScreen)"() {is=title K2Node_ComponentBoundEvent}
   > DeactivateWidget()
   Event On Activated()
   n1 = Get FortItemPreviewWidget() -> (FortItemPreviewWidget:WBP Item Preview Object Reference)
   > ActivateWidget(self=@n1)
   Event On Deactivated()
   > DeactivateWidget(self=@n1)
  ]}
 GetPrimaryContentSetup {nodes [
   GetPrimaryContentSetup() {is=title K2Node_FunctionEntry}
   > Return Node!(
    execute
    ReturnValue:Primary Content Setup Structure=Make Primary Content Setup!(
     bShowBottomBar:bool=false
     bShowChatWidget:bool=false
     bEnableNewTextChat:bool=False
     bShowTopBar:bool=false
     bTabSupportsInterface:bool=False
    ) -> (PrimaryContentSetup:Primary Content Setup Structure) {is=title K2Node_MakeStruct}
   )
  ]}
 GetUpsellPass {nodes [
   n1 = GetUpsellPass!() -> (then, PassSystem:Season Pass System VM Object Reference, Config:Crew Upsell Config Structure) {is=title K2Node_FunctionEntry}
   > Return Node!(
    execute
    Result:Season Pass VM Object Reference=FindSeasonPass!(
     self:Season Pass System VM Object Reference=@n1
     TemplateId:String=Break Crew Upsell Config!(CrewUpsellConfig:"Crew Upsell Config Structure (by ref)"=@n1.Config) -> (
      OverrideUpsellPass:String
      PassBackgroundUrl:Array of Motd Image Structures
      CrewBackgroundUrl:Array of Motd Image Structures
      bUsePassBackgroundMaterial:bool=~
      bUseCrewBackgroundMaterial:bool=~
      bCrewOnRightInUpsell:bool=~
      bCrew50PercentInUpsell:bool=~
      bCrew30PercentInUpsell:bool=~
      bCrewAsSliver:bool=~
      bCrewOnlyInUpsell:bool=~
      bPassOnlyInUpsell:bool=~
      bAnimateCrewBG:bool=~
      bUseNewUpsellFormat:bool=~
      bAllowToggleBundle:bool=~
      bDefaultShowBundle:bool=~
      EntryOverrides:Array of Nexus Upsell Entry Structures
     ) {is=title K2Node_BreakStruct}
    ) -> (ReturnValue:Season Pass VM Object Reference)
   )
  ]}
]
variables []
widgets {
 info {inheritedWidgetCount=0, parentClass=&CommonActivatableWidget, rootWidgetClass=&Overlay, widgetCount=3}
 widgets [
  Overlay_0 {widgetClassPath=&Overlay, widget=Overlay_0}
  FortItemPreviewWidget {
   widgetClassPath=/FortniteCoreUI/Frontend/ItemPreview/WBP_ItemPreview
   bIsVariable=true
   parent=Overlay_0
   widget=FortItemPreviewWidget
   properties {
    itemViewObject=&self:WidgetTree.FortItemPreviewWidget.ItemViewObject
    previewSceneChangerTag {tagName=UI.ItemPreview.SceneChanger.ItemShop}
   }
   slot {class=OverlaySlot, properties {horizontalAlignment=HAlign_Fill, verticalAlignment=VAlign_Fill}}
  }
  CrewUpsellScreen {
   widgetClassPath=&/CrewUpsellScreen
   bIsVariable=true
   parent=Overlay_0
   widget=CrewUpsellScreen
   properties {bAutoActivate=true, useExperiment=true}
   slot {class=OverlaySlot, properties {horizontalAlignment=HAlign_Fill, verticalAlignment=VAlign_Fill}}
  }
 ]
}
