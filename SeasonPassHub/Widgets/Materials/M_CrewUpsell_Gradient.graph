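# UI material, by name the gradient for the Season Pass hub's Crew upsell, drawing a tintable vertical fade with alpha-composite blending.
# Squared vertical normalized UV drives two separate top-to-bottom lerps, one for color strength and one for translucency.
# A Color parameter times vertex color is premultiplied by the first lerp, and both results scale with vertex alpha and Opacity.
# The separate ramps let color contribution and background occlusion fade independently.
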Material /SeasonPassHub/Widgets/Materials/M_CrewUpsell_Gradient
graph {
 properties {
  bAutomaticallySetUsageInEditor=false
  bUsedWithStaticMesh=false
  blendMode=BLEND_AlphaComposite
  materialDomain=MD_UI
  usedShadingModels=Default Lit
 }
 nodes [
  n1 = VertexColor()
  n2 = Multiply(A=@n1.A, B=ScalarParameter() {defaultValue=1, parameterName=Opacity, sliderMax=1})
  n3 = ScalarParameter() {defaultValue=0.8, parameterName=Top Multiplied Transparency, sliderMax=1}
  n4 = ScalarParameter() {defaultValue=0.075, parameterName=Bottom Multiplied Transparency, sliderMax=1}
  n5 = Power(
   Base=ComponentMask(None=MaterialFunctionCall() {materialFunction=&GetUserInterfaceUV}.Normalized UV) {g=true}
   Exp=Constant() {r=2}
  ) {constExponent=1}
  Multiply(
   A=Multiply(A=VectorParameter() {parameterName=Color, defaultValue=<rgba 1 1 1 1>}.RGB, B=@n1)
   B=Multiply(A=@n2, B=LinearInterpolate(A=@n3, B=@n4, Alpha=@n5) {constA=0.8, constB=0.075})
  )
  Multiply(
   A=@n2
   B=LinearInterpolate(
    A=ScalarParameter() {defaultValue=1, parameterName=Top Translucency, sliderMax=1}
    B=ScalarParameter() {defaultValue=0.25, parameterName=Bottom Translucency, sliderMax=1}
    Alpha=@n5
   ) {constA=1, constB=0.25}
  )
 ]
}
