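# Translucent UI material for the Battle Pass target offer blur effect, drawing a soft, tinted radial glow.
# Color comes from the Color parameter plus a faint noise term, with MF_UI_Random generating noise from UVs.
# Alpha is an inverted distance from the center of the normalized UI UV, shaped by an EaseInSine curve and scaled by GlowOpacity.
# The blur look is purely procedural, with no scene or texture sampling in the graph.
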Material /SeasonPassHub/Widgets/Materials/M_UI_Battle_Pass_Target_Offer_Blur_Effect
graph {
 properties {bAutomaticallySetUsageInEditor=false, blendMode=BLEND_Translucent, materialDomain=MD_UI, usedShadingModels=Default Lit}
 nodes [
  n1 = NamedRerouteUsage()
  NamedRerouteDeclaration(None=Add(A=Multiply(A=@n1) {constB=0.1}, B=VectorParameter() {parameterName=Color, defaultValue=<rgba 0.2006 0 1 1>}.RGB)) {name=FinalTexture, nodeColor=<rgba 0.2422 1 0 1>}
  n2 = MaterialFunctionCall() {materialFunction=&GetUserInterfaceUV}
  n3 = Constant() {r=0.5}
  n4 = OneMinus(None=Multiply(A=Distance(A=@n2.Normalized UV, B=@n3)) {constB=2})
  NamedRerouteDeclaration(None=Multiply(
    A=MaterialFunctionCall!("Value (0-1)"=@n4) {materialFunction=/FortUILibrary/UGC/MaterialFunctions/MF_UGCUI_EaseCurves}.EaseInSine
    B=ScalarParameter() {defaultValue=0.8, parameterName=GlowOpacity}
   )) {name=FinalAlpha, nodeColor=<rgba 1 0 0.5625 1>}
  NamedRerouteDeclaration(None=MaterialFunctionCall!(UV=TextureCoordinate(), Seed) {materialFunction=&MF_UI_Random}.Result) {name=NoiseTexture, nodeColor=<rgba 0 1 0.2031 1>}
  NamedRerouteUsage()
  NamedRerouteUsage()
 ]
}
