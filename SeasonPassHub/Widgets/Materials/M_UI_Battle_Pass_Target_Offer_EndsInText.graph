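# UI-domain material for the "ends in" text of a Season Pass hub Battle Pass target offer, judging by its name and path.
# Its graph is a single lerp between a Color vector parameter, gold-orange by default, and constant white.
# The lerp alpha is unwired and stays at its default, and the graph shows no link from the lerp to a material output.
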Material /SeasonPassHub/Widgets/Materials/M_UI_Battle_Pass_Target_Offer_EndsInText
graph {
 properties {bAutomaticallySetUsageInEditor=false, materialDomain=MD_UI, usedShadingModels=Default Lit}
 nodes [LinearInterpolate(
   A=VectorParameter() {group=Color, parameterName=Color, defaultValue=<rgba 0.776 0.4301 0.0382 1>}.RGB
   B=Constant4Vector() {constant=<rgba 1 1 1 1>}.RGB
  )]
}
