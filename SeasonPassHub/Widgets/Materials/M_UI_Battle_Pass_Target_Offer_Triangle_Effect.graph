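# Translucent UI material that draws an animated triangle effect, named for the Battle Pass target offer widget.
# Three channels of a triangle texture pan at separate speeds, then are smoothstepped and thresholded into layers that are clamped and summed.
# UVs support a localization flip and fractional rotation, and feed ping-pong fade gradients and a flow mask.
# Named reroutes organize the composition, Color tint and final mask scaled by MaterialOpacity.
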Material /SeasonPassHub/Widgets/Materials/M_UI_Battle_Pass_Target_Offer_Triangle_Effect
graph {
 properties {bAutomaticallySetUsageInEditor=false, blendMode=BLEND_Translucent, materialDomain=MD_UI, usedShadingModels=Default Lit}
 nodes [
  n1 = ScalarParameter() {defaultValue=0.12, parameterName=Speed3}
  n2 = Constant()
  n3 = TextureSampleParameter2D(UVs=Panner(Coordinate=NamedRerouteUsage(), Speed=AppendVector(A=@n1, B=@n2)) {speedX=0.5}) {parameterName=TextureMask, samplerType=SAMPLERTYPE_LinearColor, texture=/Game/UI/Foundation/Items/Textures/T-Card-Hero-Anim-Triangle}
  n4 = Multiply(A=MaterialFunctionCall(Alpha=@n3.R, Min=NamedRerouteUsage()) {materialFunction=&SmoothStep}.Result)
  NamedRerouteDeclaration(None=Multiply(A=Subtract(A=@n4) {constB=0.65}) {constB=8}) {name=Texture03, nodeColor=<rgba 1 0 0.3516 1>}
  n5 = Constant2Vector() {g=1, r=-1}
  n6 = ScalarParameter() {group=Localization, parameterName=Flip, sliderMax=1}
  n7 = Constant2Vector() {g=1, r=1}
  n8 = MaterialFunctionCall!(
   Rotation_Fraction=ScalarParameter() {parameterName=DirectionRotationFraction}
   TexCoord=MaterialFunctionCall(UVs=TextureCoordinate(), Texture Scale=LinearInterpolate(A=@n7, B=@n5, Alpha=@n6) {constAlpha=0}) {materialFunction=&ScaleUVsByCenter}.UVs
  ) {materialFunction=/FortUILibrary/Foundation/Art_ToolKit/AngledBoxes/MF_UI_Rotator_Fractions}
  NamedRerouteDeclaration(None=ComponentMask(None=@n8.Result) {r=true}) {name=FlowMask, nodeColor=<rgba 0.3594 1 0 1>}
  n9 = NamedRerouteUsage()
  n10 = NamedRerouteUsage()
  n11 = Add(A=Add(A=Clamp(None=@n9), B=Clamp(None=@n10)), B=Clamp(None=NamedRerouteUsage()))
  n12 = NamedRerouteUsage()
  n13 = NamedRerouteUsage()
  n14 = Add(A=SmoothStep(Value=OneMinus(None=@n13)) {constMax=0.9, constMin=0.2}) {constB=0.1}
  n15 = Saturate(None=Multiply(A=Multiply(A=@n12, B=@n14)) {constB=3})
  NamedRerouteDeclaration(None=Multiply(A=LinearInterpolate(B=@n11, Alpha=@n15), B=ScalarParameter() {defaultValue=0.35, parameterName=MaterialOpacity})) {name=FinalMask, nodeColor=<rgba 1 0.375 0 1>}
  n16 = VectorParameter() {parameterName=Color, defaultValue=<rgba 0.7605 0.4233 0.1221 1>}
  NamedRerouteDeclaration(None=LinearInterpolate(A=@n16.RGB, B=@n16.RGB, Alpha=NamedRerouteUsage())) {name=Final Texture, nodeColor=<rgba 0.0547 0 1 1>}
  n17 = Clamp(None=NamedRerouteUsage())
  n18 = Clamp(None=NamedRerouteUsage()) {maxDefault=0.5}
  n19 = NamedRerouteUsage()
  NamedRerouteDeclaration(None=Add(A=Add(A=@n17, B=@n18), B=Clamp(None=@n19) {maxDefault=0.05})) {name=FinalComposition, nodeColor=<rgba 0.2656 1 0 1>}
  n20 = NamedRerouteDeclaration(None=@n8.Result) {name=UV, nodeColor=<rgba 1 0.6094 0 1>}
  n21 = MaterialFunctionCall!(Value=@n20, Scale, Height) {materialFunction=/FortUILibrary/MaterialFunctions/Basics/MF_UI_PingPong}
  n22 = ScalarParameter() {defaultValue=1.5}
  NamedRerouteDeclaration(None=Saturate(None=Multiply(A=@n21.Result, B=@n22) {constB=1.5})) {name=HorizontalFade, nodeColor=<rgba 1 0 0.75 1>}
  n23 = NamedRerouteDeclaration(None=ComponentMask(None=@n20) {r=true}) {name=HorizontalUVGradient, nodeColor=<rgba 1 0 0.375 1>}
  n24 = Add(A=ComponentMask(None=@n21.Result) {r=true}, B=Multiply(A=SmoothStep(Value=@n23) {constMin=0.5}, B=Constant() {r=0.8}))
  n25 = ComponentMask(None=@n21.Result) {g=true}
  NamedRerouteDeclaration(None=OneMinus(None=AppendVector(A=@n24, B=@n25))) {name=HorizontalFadeInverse, nodeColor=<rgba 0 1 0.9297 1>}
  n26 = ScalarParameter() {defaultValue=0.06, parameterName=Speed1}
  n27 = Constant()
  n28 = TextureSampleParameter2D(UVs=Panner(Coordinate=NamedRerouteUsage(), Speed=AppendVector(A=@n26, B=@n27)) {speedX=0.25}) {parameterName=TextureMask, samplerType=SAMPLERTYPE_LinearColor, texture=&T-Card-Hero-Anim-Triangle}
  n29 = Multiply(A=MaterialFunctionCall(Alpha=@n28.B, Min=NamedRerouteUsage()) {materialFunction=&SmoothStep}.Result)
  NamedRerouteDeclaration(None=Multiply(A=Subtract(A=@n29) {constB=0.65}) {constB=8}) {name=Texture01, nodeColor=<rgba 0.2891 0 1 1>}
  n30 = ScalarParameter() {defaultValue=0.08, parameterName=Speed2}
  n31 = Constant()
  n32 = TextureSampleParameter2D(UVs=Panner(Coordinate=NamedRerouteUsage(), Speed=AppendVector(A=@n30, B=@n31)) {speedX=0.15}) {parameterName=TextureMask, samplerType=SAMPLERTYPE_LinearColor, texture=&T-Card-Hero-Anim-Triangle}
  n33 = Multiply(A=MaterialFunctionCall(Alpha=@n32.G, Min=NamedRerouteUsage()) {materialFunction=&SmoothStep}.Result)
  NamedRerouteDeclaration(None=Multiply(A=Subtract(A=@n33) {constB=0.65}) {constB=8}) {name=Texture02, nodeColor=<rgba 0.0312 1 0 1>}
  NamedRerouteUsage()
  NamedRerouteUsage()
 ]
}
