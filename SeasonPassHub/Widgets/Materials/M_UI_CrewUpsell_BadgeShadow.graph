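# UI modulate-blend material that draws a soft elliptical shadow, named for the badge in the Season Pass hub's Crew upsell widget.
# Normalized UI UVs are centered, scaled separately per axis and turned into a smoothstepped radial distance.
# That mask fades a Shadow Color parameter, processed by linear burn, slight lightening and adjustable desaturation, out to white so the edges leave the background untouched.
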Material /SeasonPassHub/Widgets/Materials/M_UI_CrewUpsell_BadgeShadow
graph {
 properties {
  bAutomaticallySetUsageInEditor=false
  bUsedWithStaticMesh=false
  blendMode=BLEND_Modulate
  materialDomain=MD_UI
  usedShadingModels=Default Lit
 }
 nodes [
  n1 = MaterialFunctionCall!(Blend=Constant() {r=0.77}, Base=VectorParameter() {parameterName=Shadow Color}.RGB) {materialFunction=/Engine/Functions/Engine_MaterialFunctions03/Blends/Blend_LinearBurn}
  n2 = ConstantBiasScale(None=MaterialFunctionCall() {materialFunction=&GetUserInterfaceUV}.Normalized UV) {bias=-0.5, scale=2}
  n3 = ComponentMask(None=@n2) {r=true}
  n4 = ScalarParameter() {defaultValue=0.3, parameterName=Scale X}
  n5 = ComponentMask(None=@n2) {g=true}
  n6 = ScalarParameter() {defaultValue=0.9, parameterName=Scale Y}
  n7 = Length(None=AppendVector(A=Multiply(A=@n3, B=@n4), B=Multiply(A=@n5, B=@n6))) {is=class MaterialExpressionLength}
  LinearInterpolate(
   A=Desaturation(None=LinearInterpolate(A=@n1.Result) {constAlpha=0.125}, Fraction=ScalarParameter() {parameterName=Desaturation})
   Alpha=SmoothStep(Value=Saturate(None=@n7))
  ) {constA=1}
  VertexColor()
 ]
}
