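# UI material for a radial badge shadow on the Season Pass hub's Crew upsell widget, named for the Porridge variant, using Modulate blending.
# Normalized UI UVs are re-centred, and their saturated length is smoothstepped into a soft circle, dark at the middle and white at the edge.
# Vertex alpha blends white toward that gradient, so the multiply fades out with widget opacity.
# Separate branches tint the gradient by Shadow Colour and invert it.
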Material /SeasonPassHub/Widgets/Materials/M_UI_CrewUpsell_BadgeShadow_Porridge
graph {
 properties {
  bAutomaticallySetUsageInEditor=false
  bUsedWithStaticMesh=false
  blendMode=BLEND_Modulate
  materialDomain=MD_UI
  usedShadingModels=Default Lit
 }
 nodes [
  n1 = ConstantBiasScale(None=MaterialFunctionCall() {materialFunction=&GetUserInterfaceUV}.Normalized UV) {bias=-0.5, scale=2}
  n2 = SmoothStep(Value=Saturate(None=Length(None=@n1) {is=class MaterialExpressionLength}))
  LinearInterpolate(B=@n2, Alpha=VertexColor().A) {constA=1}
  LinearInterpolate(A=VectorParameter() {parameterName=Shadow Colour, defaultValue=<rgba 1 0 0.8676 1>}.RGB, Alpha=@n2)
  OneMinus(None=@n2)
 ]
}
