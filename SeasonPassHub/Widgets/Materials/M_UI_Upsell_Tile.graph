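# Translucent UI material for a Season Pass Hub upsell tile, with a themed gradient, animated Crew bento pattern, hover response and rounded-corner mask.
# An eased, offset radial gradient drives the color blend and hover overlay, composited with Lighten blends.
# The pattern pans over rotated, scaled UVs and is weighted toward the side edges.
# Random UV sampling under a sine-driven density mask generates time offsets that vary foreground tile motion.
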Material /SeasonPassHub/Widgets/Materials/M_UI_Upsell_Tile
graph {
 properties {
  bAutomaticallySetUsageInEditor=false
  bUsedWithStaticMesh=false
  blendMode=BLEND_Translucent
  materialDomain=MD_UI
  usedShadingModels=Default Lit
 }
 nodes [
  n1 = NamedRerouteUsage()
  n2 = NamedRerouteUsage()
  n3 = NamedRerouteUsage()
  n4 = VectorParameter() {group=Theming, parameterName=Color Theme 2, defaultValue=<rgba 0.0003 0.3467 0.9734 1>}
  LinearInterpolate(
   A=MaterialFunctionCall!(Blend=NamedRerouteUsage(), Base=NamedRerouteUsage()) {materialFunction=/Engine/Functions/Engine_MaterialFunctions03/Blends/Blend_Lighten}.Result
   B=LinearInterpolate(
    A=MaterialFunctionCall(Blend=@n1, Base=@n2) {materialFunction=&Blend_Lighten}.Result
    B=Power(Base=@n3) {constExponent=1.2}
    Alpha=NamedRerouteUsage()
   )
   Alpha=Floor(None=OneMinus(None=@n4.A))
  )
  n5 = ScalarParameter() {defaultValue=10, group=Layout, parameterName=TopLeftCornerRadius, sortPriority=4}
  n6 = ScalarParameter() {defaultValue=10, group=Layout, parameterName=BottomLeftCornerRadius, sortPriority=7}
  n7 = ScalarParameter() {defaultValue=10, group=Layout, parameterName=TopRightCornerRadius, sortPriority=5}
  n8 = ScalarParameter() {defaultValue=10, group=Layout, parameterName=BottomRightCornerRadius, sortPriority=6}
  n9 = ScalarParameter() {
   defaultValue=1
   desc="If enabled reads corner radius values in pixels.\r\nbool"
   group=Layout
   parameterName=UseValueInPixels
   sliderMax=1
   sortPriority=1
  }
  NamedRerouteDeclaration(None=MaterialFunctionCall(
    Corner Radius=MaterialFunctionCall(R=@n5, G=@n6, B=@n7, A=@n8) {materialFunction=&AppendMany}.RGBA
    Radius in Pixels=@n9
   ) {materialFunction=&MF_UI_CoreUI_RoundedMask}.Shape Mask) {name=Shape Mask, nodeColor=<rgba 0 0 0 1>}
  n10 = Add(A=TextureCoordinate(), B=Constant2Vector() {g=0.75})
  n11 = Constant2Vector() {g=3, r=2}
  n12 = MaterialFunctionCall("Value (0-1)"=MaterialFunctionCall(UV=MaterialFunctionCall(UVs=@n10, Texture Scale=@n11) {materialFunction=&ScaleUVsByCenter}.UVs) {materialFunction=&MF_UI_RadialGradient}.Smooth) {materialFunction=&MF_UI_EaseCurves}
  NamedRerouteDeclaration(None=@n12.EaseInOutSine) {name=Base Gradient, nodeColor=<rgba 0 0 0 1>}
  n13 = TextureCoordinate()
  NamedRerouteDeclaration(None=Rotator(
    Coordinate=MaterialFunctionCall(UV=@n13) {materialFunction=&MF_UI_UVScaling}.Scaled to Fit
    Time=ScalarParameter() {defaultValue=0.08, group=Global, parameterName=Tile_Angle}
   ) {speed=6.2832}) {name=Screen_Angle, nodeColor=<rgba 0 0 0 1>}
  NamedRerouteDeclaration(None=ScalarParameter() {defaultValue=1.675, group=Tiles_Foreground, parameterName=Tile-Scale-Foreground}) {name=FG_Tile_Scale, nodeColor=<rgba 0 0 0 1>}
  NamedRerouteDeclaration(None=TextureObjectParameter() {parameterName=CMS_Crew_Bento, texture=/Game/Athena/UI/Subscriptions/Art/Textures/T_CrewUpsell_Bento}) {name=Texture Object, nodeColor=<rgba 0 0 0 1>}
  n14 = Multiply(A=NamedRerouteUsage(), B=AppendVector(A=Constant() {r=0.5}, B=NamedRerouteUsage()))
  n15 = AppendVector(
   A=ScalarParameter() {defaultValue=0.25, group=Tiles_Foreground, parameterName=FG_Time_Offset_X-Tiling}
   B=ScalarParameter() {defaultValue=1, group=Tiles_Foreground, parameterName=FG_Time_Offset_Y-Tiling}
  )
  n16 = ScalarParameter() {defaultValue=-0.5, group=Tiles_Foreground, parameterName=FG_Time_Offset_X-Offsets}
  n17 = Round(None=ScalarParameter() {defaultValue=4, group=Tiles_Foreground, parameterName=FG_Time_Offset_Y-Offsets})
  n18 = MaterialFunctionCall(Float2=Add(A=ScalarParameter() {defaultValue=0.0625, group=Tiles_Foreground, parameterName=FG_Time_Mask_Offset}, B=@n14)) {materialFunction=&BreakOutFloat2Components}
  n19 = ConstantBiasScale(None=Add(A=MaterialFunctionCall(Value=@n18.G) {materialFunction=&LinearSine}.Linear Sine) {constB=-0.5}) {bias=-0.125, scale=12}
  n20 = MaterialFunctionCall!(
   PrimaryTextureCoordinates=Add(A=Multiply(A=@n14, B=@n15), B=AppendVector(A=@n16, B=@n17))
   2ndOrderTextureCoordinates
   RandomSeed
   NoiseDensityBias=LinearInterpolate(
    A=ScalarParameter() {defaultValue=1, group=Tiles_Foreground, parameterName=FG_Time_Mask_Upper_Range}
    B=ScalarParameter() {defaultValue=0.125, group=Tiles_Foreground, parameterName=FG_Time_Mask_Lower_Range}
    Alpha=Saturate(None=@n19)
   ) {constA=1, constB=0.5}
   Bypass Noise Influence Crosstalk
  ) {materialFunction=/Game/Effects/Fort_Effects/Materials/_Functions/MF_RandomUVSample}
  NamedRerouteDeclaration(None=Multiply(
    A=ConstantBiasScale(None=@n20.NoiseWithRange) {bias=0.5, scale=-2}
    B=ScalarParameter() {defaultValue=0.05, group=Tiles_Foreground, parameterName=Tile_FG_Time_Factor}
   )) {name=FG_Time_Offset, nodeColor=<rgba 0 0 0 1>}
  n21 = NamedRerouteUsage()
  n22 = NamedRerouteUsage()
  n23 = Panner(Coordinate=Multiply(A=Constant2Vector() {g=1, r=0.25}, B=Multiply(A=@n21, B=@n22)), Time=Multiply(A=Time(), B=NamedRerouteUsage())) {speedX=0.125}
  n24 = NamedRerouteUsage()
  n25 = ComponentMask(None=TextureCoordinate()) {r=true}
  n26 = Constant() {r=0.5}
  n27 = Power(Base=Multiply(A=Distance(A=@n25, B=@n26)) {constB=2}) {constExponent=2.5}
  n28 = LinearInterpolate(B=ScalarParameter() {defaultValue=0.5, parameterName=Crew Tile Hover Opacity}, Alpha=NamedRerouteUsage()) {constA=1, constB=0.65}
  NamedRerouteDeclaration(None=Multiply(A=TextureSample(UVs=@n23, Tex=@n24).A, B=Multiply(A=@n27, B=@n28)) {constB=0.5}) {name=Tiles, nodeColor=<rgba 0 0 0 1>}
  n29 = Multiply(
   A=NamedRerouteUsage()
   B=Min(A=@n12.EaseInOutSine, B=ScalarParameter() {defaultValue=0.98, parameterName=Hover Overlay Opacity}) {constB=0.5}
  )
  n30 = NamedRerouteUsage()
  n31 = LinearInterpolate(Alpha=@n30) {constA=1, constB=0.95}
  NamedRerouteDeclaration(None=Multiply(A=Power(Base=@n29, Exp=@n31), B=@n30) {constB=0}) {name=Hover Overlay, nodeColor=<rgba 0 0 0 1>}
  NamedRerouteDeclaration(None=VectorParameter() {group=Theming, parameterName=Color Theme 2, defaultValue=<rgba 0.0003 0.3467 0.9734 1>}.RGB) {name=Color 01, nodeColor=<rgba 0 0 0 1>}
  NamedRerouteDeclaration(None=VectorParameter() {group=Theming, parameterName=Color Theme 4, defaultValue=<rgba 0.0024 0.0052 0.0103 1>}.RGB) {name=Color 02, nodeColor=<rgba 0 0 0 1>}
  NamedRerouteDeclaration(None=ScalarParameter() {group=Hover, parameterName=Hover}) {name=Hover, nodeColor=<rgba 0 0 0 1>}
  n32 = NamedRerouteUsage()
  NamedRerouteDeclaration(None=LinearInterpolate(A=NamedRerouteUsage(), B=NamedRerouteUsage(), Alpha=Multiply(A=@n32) {constB=0.5})) {name=Colored Gradient, nodeColor=<rgba 0 0 0 1>}
  NamedRerouteUsage()
 ]
}
