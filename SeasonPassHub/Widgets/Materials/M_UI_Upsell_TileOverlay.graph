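# UI material for a rounded-corner overlay on Season Pass Hub upsell tiles, with a hover state.
# A Hover scalar blends color from black to mid grey and opacity from 0.35 to zero, both masked by the tile shape.
# Alpha-composite blending lets this shift from a darkening tint to an additive highlight.
# The shape comes from MF_UI_CoreUI_RoundedMask, with per-corner radii in optional pixel units and only bottom corners rounded by default.
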Material /SeasonPassHub/Widgets/Materials/M_UI_Upsell_TileOverlay
graph {
 properties {
  bAutomaticallySetUsageInEditor=false
  bUsedWithStaticMesh=false
  blendMode=BLEND_AlphaComposite
  materialDomain=MD_UI
  usedShadingModels=Default Lit
 }
 nodes [
  Multiply(
   A=LinearInterpolate(A=Constant3Vector(), B=Constant3Vector() {constant=<rgba 0.491 0.491 0.491 1>}, Alpha=NamedRerouteUsage())
   B=NamedRerouteUsage()
  )
  Multiply(A=LinearInterpolate(A=Constant() {r=0.35}, B=Constant(), Alpha=NamedRerouteUsage()), B=NamedRerouteUsage())
  n1 = ScalarParameter() {group=Layout, parameterName=TopLeftCornerRadius, sortPriority=4}
  n2 = ScalarParameter() {defaultValue=10, group=Layout, parameterName=BottomLeftCornerRadius, sortPriority=7}
  n3 = ScalarParameter() {group=Layout, parameterName=TopRightCornerRadius, sortPriority=5}
  n4 = ScalarParameter() {defaultValue=10, group=Layout, parameterName=BottomRightCornerRadius, sortPriority=6}
  n5 = ScalarParameter() {
   defaultValue=1
   desc="If enabled reads corner radius values in pixels.\r\nbool"
   group=Layout
   parameterName=UseValueInPixels
   sliderMax=1
   sortPriority=1
  }
  NamedRerouteDeclaration(None=MaterialFunctionCall(
    Corner Radius=MaterialFunctionCall(R=@n1, G=@n2, B=@n3, A=@n4) {materialFunction=&AppendMany}.RGBA
    Radius in Pixels=@n5
   ) {materialFunction=&MF_UI_CoreUI_RoundedMask}.Shape Mask) {name=Shape Mask, nodeColor=<rgba 0 0 0 1>}
  NamedRerouteDeclaration(None=ScalarParameter() {group=Hover, parameterName=Hover}) {name=Hover, nodeColor=<rgba 0 0 0 1>}
 ]
}
