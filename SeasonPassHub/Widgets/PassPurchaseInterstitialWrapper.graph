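# PassPurchaseInterstitialWrapper is an activatable widget that presents the Season Pass purchase screen as an interstitial.
# On Construct it binds the screen's Back to Hub event to Back, which dismisses the interstitial viewmodel, requesting the next one, then deactivates the widget.
# GetPass looks up a season pass viewmodel by template ID.
# GetPrimaryContentSetup hides the top bar, bottom bar and chat.
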WidgetBlueprint /SeasonPassHub/Widgets/PassPurchaseInterstitialWrapper
graphs [
 EventGraph {nodes [
   n1 = Back() {is=title K2Node_CustomEvent}
   > DismissInterstitial!(
    execute
    self:Fort Interstitial VM Object Reference=Get PassPurchaseInterstitialVM() -> (PassPurchaseInterstitialVM:Pass Purchase Interstitial VM Object Reference)
    bShouldShowNext:bool=true
   ) -> (then)
   > DeactivateWidget()
   Event Construct()
   > Bind Event to Back to Hub(
    self:Season Pass Purchase Screen Object Reference=Get SeasonPass_PurchaseScreen() -> (SeasonPass_PurchaseScreen:Season Pass Purchase Screen Object Reference)
    Delegate=@n1
   ) {is=title K2Node_AddDelegate}
  ]}
 GetPass {nodes [
   n1 = GetPass!() -> (then, Pass System:Season Pass System VM Object Reference, TemplateId:String) {is=title K2Node_FunctionEntry}
   > Return Node!(
    execute
    ReturnValue:Season Pass VM Object Reference=FindSeasonPass!(self:Season Pass System VM Object Reference=@n1, TemplateId:String=@n1.TemplateId) -> (ReturnValue:Season Pass VM Object Reference)
   )
  ]}
 GetPrimaryContentSetup {nodes [
   GetPrimaryContentSetup() {is=title K2Node_FunctionEntry}
   > Return Node(ReturnValue:Primary Content Setup Structure=Make Primary Content Setup!(
     bShowBottomBar:bool=false
     bShowChatWidget:bool=false
     bEnableNewTextChat:bool=False
     bShowTopBar:bool=false
     bTabSupportsInterface:bool=False
    ) -> (PrimaryContentSetup:Primary Content Setup Structure) {is=title K2Node_MakeStruct})
  ]}
]
variables []
widgets {
 info {
  inheritedWidgetCount=0
  parentClass=&CommonActivatableWidget
  rootWidgetClass=/SeasonPass/Widgets/SeasonPass_PurchaseScreen.SeasonPass_PurchaseScreen_C
  widgetCount=1
 }
 widgets [SeasonPass_PurchaseScreen {
   widgetClassPath=/SeasonPass/Widgets/SeasonPass_PurchaseScreen
   bIsVariable=true
   widget=SeasonPass_PurchaseScreen
   properties {bAutoActivate=true}
  }]
}
