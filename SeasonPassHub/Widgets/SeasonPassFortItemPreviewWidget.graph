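# SeasonPassFortItemPreviewWidget is a Season Pass Hub item preview widget derived from the FortniteCoreUI frontend WBP_ItemPreview.
# It declares no widgets or variables of its own and has an empty event graph, relying on the parent for preview behavior.
# Its only logic is a GetPrimaryContentSetup function that returns a Primary Content Setup structure left at default values.
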WidgetBlueprint /SeasonPassHub/Widgets/SeasonPassFortItemPreviewWidget
graphs [
 EventGraph {links [], nodes []}
 GetPrimaryContentSetup {nodes [
   GetPrimaryContentSetup() {is=title K2Node_FunctionEntry}
   > Return Node!(execute, ReturnValue:Primary Content Setup Structure)
  ]}
]
variables []
widgets {
 info {
  inheritedWidgetCount=0
  parentClass=/FortniteCoreUI/Frontend/ItemPreview/WBP_ItemPreview.WBP_ItemPreview_C
  rootWidgetClass=null
  widgetCount=0
 }
 widgets []
}
