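# WBP_SeasonPassScreen is the season pass hub screen, scrolling vertically between Battle Pass and Crew landing pages via swipe and page nav.
# Activation restores page, focus, camera and item preview, or routes external launches to rewards or other screens, with an upsell A/B test.
# Requested backgrounds load asynchronously, discard stale results, wait for the CMS texture, then transition through a dynamic material.
# UMG animations drive intro and page transitions, partly through material parameter collections.
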WidgetBlueprint /SeasonPassHub/Widgets/Reboot/WBP_SeasonPassScreen
animations [
 AnimMoveUp {
  binding_count=1
  length_seconds=0.25
  object=&self:AnimMoveUp
  sequence {bindings [SeasonPassesScreen_C_57 {
     id=#1
     possessed_class=SeasonPassesScreen_C
     tracks [{
       type=MovieSceneBoolTrack
       display_name=Is Vertical Transitioning
       sections [{type=MovieSceneBoolSection, end_frame=5, start_frame=0, channels [None {type=MovieSceneScriptingBoolChannel, keys [[0,true],[5,false]]}]}]
      }]
    }]}
 }
 AnimMoveDown {
  binding_count=1
  length_seconds=0.25
  object=&self:AnimMoveDown
  sequence {bindings [SeasonPassesScreen_C_58 {
     id=#2
     possessed_class=SeasonPassesScreen_C
     tracks [{
       type=MovieSceneBoolTrack
       display_name=Is Vertical Transitioning
       sections [{
         type=MovieSceneBoolSection
         end_frame=5
         start_frame=0
         channels [None {type=MovieSceneScriptingBoolChannel, keys [[0,true],[5,false],[5.0003,false]]}]
        }]
      }]
    }]}
 }
 IntroOnActivated {
  binding_count=6
  length_seconds=1
  object=&self:IntroOnActivated
  sequence {
   bindings [
    CrewNavWrapper {
     id=#3
     possessed_class=Overlay
     tracks [
      {
       type=MovieSceneFloatTrack
       display_name=Render Opacity
       sections [{type=MovieSceneFloatSection, channels [None {type=MovieSceneScriptingFloatChannel, keys [[0,0],[33,0],[60,1]]}]}]
      }
      {
       type=MovieScene2DTransformTrack
       sections [{
         type=MovieScene2DTransformSection
         end_frame=60
         start_frame=0
         channels [
          Scale.X {type=MovieSceneScriptingFloatChannel, keys [[0,1.2],[33,1.2],[60,1]]}
          Scale.Y {type=MovieSceneScriptingFloatChannel, keys [[0,1.2],[33,1.2],[60,1]]}
         ]
        }]
      }
     ]
    }
    PageNav_Bottom {
     id=#4
     possessed_class=WBP_SeasonPassPageNav_C
     tracks [
      {
       type=MovieSceneFloatTrack
       display_name=Render Opacity
       sections [{type=MovieSceneFloatSection, channels [None {type=MovieSceneScriptingFloatChannel, keys [[0,0],[33,0],[60,1]]}]}]
      }
      {
       type=MovieScene2DTransformTrack
       sections [{
         type=MovieScene2DTransformSection
         end_frame=60
         start_frame=0
         channels [
          Scale.X {type=MovieSceneScriptingFloatChannel, keys [[0,1.2],[33,1.2],[60,1]]}
          Scale.Y {type=MovieSceneScriptingFloatChannel, keys [[0,1.2],[33,1.2],[60,1]]}
         ]
        }]
      }
     ]
    }
    TopRight {
     id=#5
     possessed_class=VerticalBox
     tracks [
      {
       type=MovieSceneFloatTrack
       display_name=Render Opacity
       sections [{type=MovieSceneFloatSection, channels [None {type=MovieSceneScriptingFloatChannel, keys [[0,0],[33,0],[60,1]]}]}]
      }
      {
       type=MovieScene2DTransformTrack
       sections [{
         type=MovieScene2DTransformSection
         end_frame=60
         start_frame=0
         channels [
          Scale.X {type=MovieSceneScriptingFloatChannel, keys [[0,1.2],[33,1.2],[60,1]]}
          Scale.Y {type=MovieSceneScriptingFloatChannel, keys [[0,1.2],[33,1.2],[60,1]]}
         ]
        }]
      }
     ]
    }
    Legal {
     id=#6
     possessed_class=HorizontalBox
     tracks [
      {
       type=MovieSceneFloatTrack
       display_name=Render Opacity
       sections [{type=MovieSceneFloatSection, channels [None {type=MovieSceneScriptingFloatChannel, keys [[0,0],[33,0],[60,1]]}]}]
      }
      {
       type=MovieScene2DTransformTrack
       sections [{
         type=MovieScene2DTransformSection
         end_frame=60
         start_frame=0
         channels [
          Scale.X {type=MovieSceneScriptingFloatChannel, keys [[0,1.2],[33,1.2],[60,1]]}
          Scale.Y {type=MovieSceneScriptingFloatChannel, keys [[0,1.2],[33,1.2],[60,1]]}
         ]
        }]
      }
     ]
    }
    PageNav_Top {
     id=#7
     possessed_class=WBP_SeasonPassPageNav_C
     tracks [
      {
       type=MovieSceneFloatTrack
       display_name=Render Opacity
       sections [{type=MovieSceneFloatSection, channels [None {type=MovieSceneScriptingFloatChannel, keys [[0,0],[33,0],[60,1]]}]}]
      }
      {
       type=MovieScene2DTransformTrack
       sections [{
         type=MovieScene2DTransformSection
         end_frame=60
         start_frame=0
         channels [
          Scale.X {type=MovieSceneScriptingFloatChannel, keys [[0,1.2],[33,1.2],[60,1]]}
          Scale.Y {type=MovieSceneScriptingFloatChannel, keys [[0,1.2],[33,1.2],[60,1]]}
         ]
        }]
      }
     ]
    }
    OverlayMain {
     id=#8
     possessed_class=Overlay
     tracks [{
       type=MovieSceneFloatTrack
       display_name=Render Opacity
       sections [{type=MovieSceneFloatSection, end_frame=60, start_frame=0, channels [None {type=MovieSceneScriptingFloatChannel, const [1,2]}]}]
      }]
    }
   ]
   display_rate {numerator=60}
   tracks [{
     type=MovieSceneMaterialParameterCollectionTrack
     display_name=MPC_UI_Pass_Properties
     sections [{
       type=MovieSceneParameterSection
       channels [
        Anim_IntroSequence1 {type=MovieSceneScriptingFloatChannel, default=0, keys [[0,0],[12,0.5],[51,1]]}
        Anim_IntroSequence2 {type=MovieSceneScriptingFloatChannel, default=0, keys [[0,0],[12,0],[51,1]]}
        Anim_IntroSequence3 {type=MovieSceneScriptingFloatChannel, default=0, keys [[0,0],[20,0],[42,1]]}
        Anim_IntroSequence4 {type=MovieSceneScriptingFloatChannel, default=0, keys [[0,0],[24,0],[42,1]]}
        Anim_IntroSequence5 {type=MovieSceneScriptingFloatChannel, default=0, keys [[0,0],[20,0],[42,1]]}
       ]
      }]
    }]
  }
 }
 PageTransition {
  binding_count=6
  length_seconds=0.5
  object=&self:PageTransition
  sequence {bindings [
    Legal {
     id=#9
     possessed_class=HorizontalBox
     tracks [
      {
       type=MovieSceneFloatTrack
       display_name=Render Opacity
       sections [{type=MovieSceneFloatSection, end_frame=10, start_frame=0, channels [None {type=MovieSceneScriptingFloatChannel, keys [[2,0],[10,1]]}]}]
      }
      {
       type=MovieScene2DTransformTrack
       sections [{
         type=MovieScene2DTransformSection
         end_frame=10
         start_frame=0
         channels [
          Translation.X {type=MovieSceneScriptingFloatChannel, const [0,1]}
          Translation.Y {type=MovieSceneScriptingFloatChannel, keys [[2,40],[10,0]]}
          Angle {type=MovieSceneScriptingFloatChannel, const [0,1]}
          Scale.X {type=MovieSceneScriptingFloatChannel, const [1,1]}
          Scale.Y {type=MovieSceneScriptingFloatChannel, const [1,1]}
          Shear.X {type=MovieSceneScriptingFloatChannel, const [0,1]}
          Shear.Y {type=MovieSceneScriptingFloatChannel, const [0,1]}
         ]
        }]
      }
     ]
    }
    TopRight {
     id=#10
     possessed_class=VerticalBox
     tracks [
      {
       type=MovieSceneFloatTrack
       display_name=Render Opacity
       sections [{type=MovieSceneFloatSection, end_frame=10, start_frame=0, channels [None {type=MovieSceneScriptingFloatChannel, keys [[2,0],[10,1]]}]}]
      }
      {
       type=MovieScene2DTransformTrack
       sections [{
         type=MovieScene2DTransformSection
         end_frame=10
         start_frame=0
         channels [
          Translation.X {type=MovieSceneScriptingFloatChannel, keys [[2,100],[10,0]]}
          Translation.Y {type=MovieSceneScriptingFloatChannel, const [0,2]}
          Angle {type=MovieSceneScriptingFloatChannel, const [0,1]}
          Scale.X {type=MovieSceneScriptingFloatChannel, const [1,1]}
          Scale.Y {type=MovieSceneScriptingFloatChannel, const [1,1]}
          Shear.X {type=MovieSceneScriptingFloatChannel, const [0,1]}
          Shear.Y {type=MovieSceneScriptingFloatChannel, const [0,1]}
         ]
        }]
      }
     ]
    }
    PageNav_Bottom {
     id=#11
     possessed_class=WBP_SeasonPassPageNav_C
     tracks [
      {
       type=MovieSceneFloatTrack
       display_name=Render Opacity
       sections [{type=MovieSceneFloatSection, end_frame=10, start_frame=0, channels [None {type=MovieSceneScriptingFloatChannel, keys [[2,0],[10,1]]}]}]
      }
      {
       type=MovieScene2DTransformTrack
       sections [{
         type=MovieScene2DTransformSection
         end_frame=10
         start_frame=0
         channels [
          Translation.X {type=MovieSceneScriptingFloatChannel, const [0,1]}
          Translation.Y {type=MovieSceneScriptingFloatChannel, keys [[2,40],[10,0]]}
          Angle {type=MovieSceneScriptingFloatChannel, const [0,1]}
          Scale.X {type=MovieSceneScriptingFloatChannel, const [1,1]}
          Scale.Y {type=MovieSceneScriptingFloatChannel, const [1,1]}
          Shear.X {type=MovieSceneScriptingFloatChannel, const [0,1]}
          Shear.Y {type=MovieSceneScriptingFloatChannel, const [0,1]}
         ]
        }]
      }
     ]
    }
    CrewNavWrapper {
     id=#12
     possessed_class=Overlay
     tracks [
      {
       type=MovieSceneFloatTrack
       display_name=Render Opacity
       sections [{type=MovieSceneFloatSection, end_frame=10, start_frame=0, channels [None {type=MovieSceneScriptingFloatChannel, keys [[2,0],[10,1]]}]}]
      }
      {
       type=MovieScene2DTransformTrack
       sections [{
         type=MovieScene2DTransformSection
         end_frame=10
         start_frame=0
         channels [
          Translation.X {type=MovieSceneScriptingFloatChannel, const [0,1]}
          Translation.Y {type=MovieSceneScriptingFloatChannel, keys [[0,100],[10,0]]}
          Angle {type=MovieSceneScriptingFloatChannel, const [0,1]}
          Scale.X {type=MovieSceneScriptingFloatChannel, const [1,1]}
          Scale.Y {type=MovieSceneScriptingFloatChannel, const [1,1]}
          Shear.X {type=MovieSceneScriptingFloatChannel, const [0,1]}
          Shear.Y {type=MovieSceneScriptingFloatChannel, const [0,1]}
         ]
        }]
      }
     ]
    }
    BattlePassNavWrapper {
     id=#13
     possessed_class=Overlay
     tracks [
      {
       type=MovieSceneFloatTrack
       display_name=Render Opacity
       sections [{type=MovieSceneFloatSection, end_frame=10, start_frame=0, channels [None {type=MovieSceneScriptingFloatChannel, keys [[2,0],[10,1]]}]}]
      }
      {
       type=MovieScene2DTransformTrack
       sections [{
         type=MovieScene2DTransformSection
         end_frame=10
         start_frame=0
         channels [
          Translation.X {type=MovieSceneScriptingFloatChannel, const [0,2]}
          Translation.Y {type=MovieSceneScriptingFloatChannel, keys [[2,100],[10,0]]}
          Angle {type=MovieSceneScriptingFloatChannel, const [0,1]}
          Scale.X {type=MovieSceneScriptingFloatChannel, const [1,1]}
          Scale.Y {type=MovieSceneScriptingFloatChannel, const [1,1]}
          Shear.X {type=MovieSceneScriptingFloatChannel, const [0,1]}
          Shear.Y {type=MovieSceneScriptingFloatChannel, const [0,1]}
         ]
        }]
      }
     ]
    }
    PageNav_Top {
     id=#14
     possessed_class=WBP_SeasonPassPageNav_C
     tracks [
      {
       type=MovieSceneFloatTrack
       display_name=Render Opacity
       sections [{type=MovieSceneFloatSection, end_frame=10, start_frame=0, channels [None {type=MovieSceneScriptingFloatChannel, keys [[2,0],[10,1]]}]}]
      }
      {
       type=MovieScene2DTransformTrack
       sections [{
         type=MovieScene2DTransformSection
         end_frame=10
         start_frame=0
         channels [
          Translation.X {type=MovieSceneScriptingFloatChannel, const [0,1]}
          Translation.Y {type=MovieSceneScriptingFloatChannel, keys [[2,40],[10,0]]}
          Angle {type=MovieSceneScriptingFloatChannel, const [0,1]}
          Scale.X {type=MovieSceneScriptingFloatChannel, const [1,1]}
          Scale.Y {type=MovieSceneScriptingFloatChannel, const [1,1]}
          Shear.X {type=MovieSceneScriptingFloatChannel, const [0,1]}
          Shear.Y {type=MovieSceneScriptingFloatChannel, const [0,1]}
         ]
        }]
      }
     ]
    }
   ]}
 }
]
graphs [
 BP_GetDesiredFocusTarget {nodes [
   Get Desired Focus Target()
   n1 = IsVisible(self=Get CrewNavWrapper() -> (CrewNavWrapper:Overlay Object Reference))
   n2 = Get BattlePassPage() -> (BattlePassPage:WBP Season Pass Page Object Reference)
   n3 = Get CrewLandingPage() -> (CrewLandingPage:WBP Crew Landing Page Object Reference)
   > Return Node!(
    execute
    ReturnValue:Widget Object Reference=GetDesiredFocusTarget!(self:Common Activatable Widget Object Reference=Select(Option 0:Common Activatable Widget Object Reference=@n2, Option 1:Common Activatable Widget Object Reference=@n3, Index=@n1) -> (ReturnValue:Common Activatable Widget Object Reference)) -> (ReturnValue:Widget Object Reference)
   )
  ]}
 BP_GetDesiredInputConfig {nodes [
   Get Desired Input Config() {is=title K2Node_FunctionEntry}
   > Return Node(ReturnValue:UIInput Config Structure=Make UIInput Config!(
     bIgnoreMoveInput:bool=False
     bIgnoreLookInput:bool=False
     InputMode:ECommonInputMode Enum=All
     MouseCaptureMode:EMouseCaptureMode Enum=CaptureDuringMouseDown
     MouseLockMode:EMouseLockMode Enum=DoNotLock
     bHideCursorDuringViewportCapture:bool=True
    ) -> (UIInputConfig:UIInput Config Structure) {is=title K2Node_MakeStruct})
  ]}
 BattlePassIsNotSelected {nodes [
   Inputs!()
   n1 = Get SeasonPassSelectionVM() -> (SeasonPassSelectionVM:Season Pass Selection VM Object Reference)
   Outputs!(ReturnValue:bool="Not Equal (Object)"(
     A:Object Reference=Get SelectedPass!(self:Season Pass Selection VM Object Reference=@n1) -> (SelectedPass:Season Pass VM Object Reference)
     B:Object Reference=Get BattlePassData() -> (BattlePassData:Season Pass VM Object Reference)
    ) -> (ReturnValue:bool=~) {is=title K2Node_PromotableOperator})
  ]}
 CF_ShowIfPassValid {nodes [
   n1 = CF_ShowIfPassValid!() -> (then, Pass VM:Season Pass VM Object Reference) {is=title K2Node_FunctionEntry}
   > Return Node!(
    execute
    Visibility:ESlateVisibility Enum=Select(Option 0:ESlateVisibility Enum=Collapsed, Option 1:ESlateVisibility Enum=SelfHitTestInvisible, Index=IsValid(Object=@n1)) -> (ReturnValue:ESlateVisibility Enum)
   )
  ]}
 CachePendingBackgroundParameters {nodes [
   n1 = CachePendingBackgroundParameters!() -> (
    then
    BackgroundAsset:Soft Object Reference
    Direction:EVaultWorldTransitionDirection Enum=Forward
    FocalPointValue:double=~
    IsHorizontalOrientation:bool=~
   ) {is=title K2Node_FunctionEntry}
   > Set PendingIsHorizontalOrientation(PendingIsHorizontalOrientation:bool=@n1.IsHorizontalOrientation) -> (Output_Get:bool)
   > Set PendingDirection(PendingDirection:EVaultWorldTransitionDirection Enum=GetDesiredPendingDirection!(PendingDirection:EVaultWorldTransitionDirection Enum=Get Direction() -> (Direction:EVaultWorldTransitionDirection Enum=Forward)) -> (Result:EVaultWorldTransitionDirection Enum=Forward)) -> (Output_Get:EVaultWorldTransitionDirection Enum=Forward)
   > Set RequestedBackgroundSoftPath(RequestedBackgroundSoftPath:Soft Object Path Structure=To Soft Object Path!(SoftObjectReference:Soft Object Reference=Get BackgroundAsset() -> (BackgroundAsset:Soft Object Reference)) -> (ReturnValue:Soft Object Path Structure)) -> (Output_Get:Soft Object Path Structure)
   > Set PendingFocalPointValue(PendingFocalPointValue:double=Get FocalPointValue() -> (FocalPointValue:double)) -> (Output_Get:double)
   > Set PendingLoadedBackground(PendingLoadedBackground:Object Reference) -> (Output_Get:Object Reference)
   > Set IsBackgroundLoaded(IsBackgroundLoaded:bool=false) -> (Output_Get:bool)
  ]}
 ControllerPageScroll {nodes [
   n1 = ControllerPageScroll!() -> (then, Navigation:EUINavigation Enum) {is=title K2Node_FunctionEntry}
   > ScrollToNextPage!(
    execute
    self
    Up:bool=Select!(
     Left:bool=false
     Right:bool=false
     Up:bool=true
     Down:bool=false
     Next:bool=true
     Previous:bool=false
     Invalid:bool=false
     Index:EUINavigation Enum=@n1
    ) -> (ReturnValue:bool)
   ) -> (then)
   > Return Node!(execute, ReturnValue:Widget Object Reference=GetDesiredFocusTarget())
  ]}
 Decide If Showing Upsell A/B Test {nodes [
   Inputs!() -> (Exec:Exec)
   > n1 = Is Valid(InputObject=Get BattlePassData())
   n2 = Get bDoesOwnsPass!(self:Season Pass VM Object Reference=Get BattlePassData()) -> (bDoesOwnsPass:bool)
   n3 = GetConfigDataBoolWithDefault!(EntryName:String=HasSeenPassUpsellTest, bDefaultValue:bool=false) -> (ReturnValue:bool)
   > n4 = Branch(Condition=AND Boolean!(
     A=NOT Boolean(A=@n3)
     B=GetConfigDataBoolWithDefault(EntryName=present_pass_purchase_screen.enabled, bDefaultValue=false)
     C:bool=NOT Boolean(A=@n2)
    ) -> (ReturnValue))
   > SetUserSettingBool!(execute, Name:String=HasSeenPassUpsellTest, bValue:bool=true) -> (then)
   > n5 = PushPrimaryContentClassAsync!(execute:Exec, WidgetClass:Common Activatable Widget Class Reference=/SeasonPassHub/Widgets/CrewUpsellScreen.CrewUpsellScreen_C) -> (then:Exec, OnPrimaryContentPushed:Exec, ActivatableWidget:Common Activatable Widget Object Reference) {is=title K2Node_AsyncAction}
   n6 = Cast To CrewUpsellScreen!(execute=@n5.OnPrimaryContentPushed, Object=@n5) -> (then, CastFailed, AsCrew Upsell Screen:Crew Upsell Screen Object Reference) {is=title K2Node_DynamicCast}
   > SetSeasonPassVM!(execute, self:Crew Upsell Screen Object Reference=@n6, Viewmodel:Season Pass VM Object Reference=Get BattlePassData()) -> (then)
   Outputs!(Else:Exec=[@n4.else, @n1.Is Not Valid])
  ]}
 EventGraph {nodes [
   QuickResetIntroAnimation()
   > Set Padding!(
    execute
    Padding:Margin Structure=Get CrewItemPreviewPadding() -> (CrewItemPreviewPadding:Margin Structure)
    self:User Widget Object Reference=Get SeasonPassFortItemPreviewWidget() -> (SeasonPassFortItemPreviewWidget:Season Pass Fort Item Preview Widget Object Reference)
   ) -> (then, Output_Get:Margin Structure)
   "On Widget Deactivated (BattlePassPage)"() {is=title K2Node_ComponentBoundEvent}
   > QuickResetIntroAnimation()
   n1 = "Navigate (PageNav_Top)"() {is=title K2Node_ComponentBoundEvent}
   ScrollToNextPage(execute=[@n1.then, @n17.then], Up=true)
   n2 = "On Swipe Down (SwipePanel)"() {is=title K2Node_ComponentBoundEvent}
   ScrollToNextPage(execute=[@n2.then, @n13.then], Up=false)
   n3 = "Entered Pass Subscreen (BattlePassPage)"() {is=title K2Node_ComponentBoundEvent}
   Set with Broadcast IsActivatingFromSubscreen!(execute=[@n3.then, @n5.then], IsActivatingFromSubscreen:bool=true) -> (then, Output_Get:bool)
   "Entered Crew Subscreen (CrewLandingPage)"() {is=title K2Node_ComponentBoundEvent}
   > n4 = Set HasEnteredCrewPassSubscreen!(execute, HasEnteredCrewPassSubscreen:bool=false, self:WBP Crew Landing Page Object Reference=Get CrewLandingPage()) -> (then, Output_Get:bool)
   n5 = Set IsActivatingFromCrew(execute=[@n4.then, @n16.then], IsActivatingFromCrew:bool=true) -> (Output_Get:bool)
   "Clear Preview (BattlePassPage)"() {is=title K2Node_ComponentBoundEvent}
   > Branch(Condition=IsActivated!(self:Common Activatable Widget Object Reference=Get SeasonPassFortItemPreviewWidget()) -> (ReturnValue:bool))
   > ClearSelectedCategory(self:Season Pass Selection VM Object Reference=Get SeasonPassSelectionVM())
   n6 = "Set Unlocked (BattlePassPage)"!() -> (OutputDelegate, then, Category:Season Pass Category VM Object Reference) {is=title K2Node_ComponentBoundEvent}
   > n7 = Create Widget!(
    execute:Exec
    Class:User Widget Class Reference=/SeasonPass/Widgets/Reboot/SetUnlock/WBP_UI_BPReboot_SetUnlocking.WBP_UI_BPReboot_SetUnlocking_C
    OwningPlayer:Player Controller Object Reference
    SeasonPassCategoryVM:Season Pass Category VM Object Reference=@n6.Category
   ) -> (then:Exec, ReturnValue:WBP UI BPReboot Set Unlocking Object Reference)
   > TryPushWidgetAsModalPanel!(execute, PushWidget:User Widget Object Reference=@n7) -> (then)
   > Set with Broadcast IsActivatingFromSubscreen(IsActivatingFromSubscreen=true)
   "On Billboard Clicked (RivalryButton)"() {is=title K2Node_ComponentBoundEvent}
   n8 = Get SeasonPassFortItemPreviewWidget()
   > ClearPreviewItem(self:Fort Item Preview Widget Object Reference=@n8)
   > DeactivateWidget(self=@n8)
   n9 = Get Pages!(self:Season Pass Category VM Object Reference=Get SelectedCategory!(self:Season Pass Selection VM Object Reference=Get SeasonPassSelectionVM()) -> (SelectedCategory:Season Pass Category VM Object Reference)) -> (Pages:Array of Season Pass Page VM Object References)
   n10 = "Get (a copy)"(
    Array:Array of Season Pass Track VM Object References=Get Tracks!(self:Season Pass Page VM Object Reference="Get (a copy)"(Array:Array of Season Pass Page VM Object References=@n9) -> (Output:Season Pass Page VM Object Reference)) -> (Tracks:Array of Season Pass Track VM Object References)
    Dimension 1=1
   ) -> (Output:Season Pass Track VM Object Reference)
   > n11 = Create Widget!(
    execute
    Class=/SeasonPass/Widgets/Reboot/WBP_Passes_StylesViewer.WBP_Passes_StylesViewer_C
    OwningPlayer
    SeasonPassVM:Season Pass VM Object Reference=Get SeasonPassVM() -> (SeasonPassVM:Season Pass VM Object Reference)
    SeasonPassRewardVM:Season Pass Reward VM Object Reference="Get (a copy)"(Array:Array of Season Pass Reward VM Object References=Get Rewards!(self:Season Pass Track VM Object Reference=@n10) -> (Rewards:Array of Season Pass Reward VM Object References)) -> (Output:Season Pass Reward VM Object Reference)
    FortVariantSelectorVM:Fort Variant Selector VM Object Reference
   ) -> (then, ReturnValue:WBP Passes Styles Viewer Object Reference)
   > PushContentWidget!(execute, PushWidget:Widget Object Reference=@n11) -> (then)
   n12 = "On External Media Changed (EpicCMSImage_BackgroundLoaderHelper)"!() -> (OutputDelegate, then, Texture:Texture 2D Object Reference) {is=title K2Node_ComponentBoundEvent}
   > Set PendingCMSTexture(PendingCMSTexture:Texture Object Reference=@n12.Texture) -> (Output_Get:Texture Object Reference)
   > Set IsCMSTextureLoaded(IsCMSTextureLoaded:bool=true) -> (Output_Get:bool)
   > Branch(Condition=Get IsBackgroundLoaded() -> (IsBackgroundLoaded:bool))
   > OnBackgroundReady()
   n13 = "Navigate (PageNav_Bottom)"() {is=title K2Node_ComponentBoundEvent}
   n14 = "Selected Crew Pass Tile (CrewLandingPage)"!() -> (OutputDelegate, then, SeasonPassVM:Season Pass VM Object Reference, ButtonSelected:WBP Crew Pass Button Object Reference) {is=title K2Node_ComponentBoundEvent}
   > OnCrewPassTileSelected!(
    execute
    self
    SelectedPass:Season Pass VM Object Reference=@n14.SeasonPassVM
    NewSelectedButton:WBP Crew Pass Button Object Reference=@n14.ButtonSelected
   ) -> (then)
   n15 = "Entered Pass Subscreen (CrewLandingPage)"!() -> (OutputDelegate, then, PassVM:Season Pass VM Object Reference) {is=title K2Node_ComponentBoundEvent}
   > Set SelectedPass!(
    execute
    SelectedPass:Season Pass VM Object Reference=@n15.PassVM
    self:Season Pass Selection VM Object Reference=Get SeasonPassSelectionVM()
   ) -> (then, Output_Get:Season Pass VM Object Reference)
   > n16 = Set HasEnteredCrewPassSubscreen(HasEnteredCrewPassSubscreen=true, self=Get CrewLandingPage())
   "On Widget Deactivated (CrewLandingPage)"() {is=title K2Node_ComponentBoundEvent}
   > QuickResetIntroAnimation()
   > PlaySound2D(Sound=/Game/Sounds/Fort_UI_Sounds/FrontEnd/Crew/UI_Crew_ScreenAppear_Cue)
   "On Navigate to Tiles Without Pass (BattlePassPage)"() {is=title K2Node_ComponentBoundEvent}
   > ScrollToNextPage(Up=false)
   > Set FallbackFocusWidgetWhenPassIsMissing!(
    execute
    FallbackFocusWidgetWhenPassIsMissing:Widget Object Reference=GetDesiredFocusTarget(self=Get CrewLandingPage())
    self:WBP Season Pass Page Object Reference=Get BattlePassPage()
   ) -> (then, Output_Get:Widget Object Reference)
   n17 = "On Swipe Up (SwipePanel)"() {is=title K2Node_ComponentBoundEvent}
   ExternalNavigationNonBP() {is=title K2Node_CustomEvent}
   > n18 = Is Valid(InputObject=Get SelectedPass(self=Get SeasonPassSelectionVM()))
   ScrollToNextPage(execute=@n18.Is Not Valid, Up=false)
   > SetFrontEndCamera!(
    execute
    self:Fort Front End Context Object Reference=Get FortFrontEndContext() -> (ReturnValue:Fort Front End Context Object Reference) {is=title K2Node_GetSubsystem}
    NewState:EFrontEndCamera Enum=CosmeticDisplay
    TransitionParams:View Target Transition Params Structure=(BlendTime=0,BlendFunction=VTBlend_Cubic,BlendExp=2,bLockOutgoing=False)
   ) -> (then)
   n19 = Get SeasonPassFortItemPreviewWidget()
   > SetZoomLevel!(execute, self:Fort Item Preview Screen Object Reference=@n19, NewZoomLevel:float=0, bSnap:bool=true, bZoomEvenIfNotAllowed:bool=false) -> (then)
   > ActivateWidget(self=@n19)
   > Set PlayIntroAnimNextScrollUp(PlayIntroAnimNextScrollUp:bool=true) -> (Output_Get:bool)
   > PlayIntroAnim()
   n20 = PushPrimaryContentClassAsync(execute=@n18.Is Valid, WidgetClass=/SeasonPass/Widgets/SeasonPassCategory.SeasonPassCategory_C)
   Set with Broadcast IsActivatingFromSubscreen(execute=@n20.OnPrimaryContentPushed, IsActivatingFromSubscreen=true)
   > Set IsActivatingFromCrew(IsActivatingFromCrew=true)
   n21 = RequestLoadBackground!() -> (
    OutputDelegate
    then
    BackgroundAsset:Soft Object Reference
    Direction:EVaultWorldTransitionDirection Enum=Forward
    FocalPointValue:double=~
    IsHorizontalOrientation:bool=~
   ) {is=title K2Node_CustomEvent}
   > CachePendingBackgroundParameters!(
    execute
    self
    BackgroundAsset:Soft Object Reference=@n21.BackgroundAsset
    Direction:EVaultWorldTransitionDirection Enum=@n21.Direction
    FocalPointValue:double=@n21.FocalPointValue
    IsHorizontalOrientation:bool=@n21.IsHorizontalOrientation
   ) -> (then) {is=title K2Node_CallFunction}
   > n22 = Async Load Asset(Asset=@n21.BackgroundAsset)
   n23 = GetSoftObjectPath(Object=@n22)
   n24 = Get RequestedBackgroundSoftPath() -> (RequestedBackgroundSoftPath:Soft Object Path Structure)
   Branch(
    execute=@n22.Completed
    Condition="Equal Exactly (String)"(A=BreakSoftObjectPath(InSoftObjectPath=@n23), B=BreakSoftObjectPath(InSoftObjectPath=@n24))
   )
   > Branch(Condition=Get bIsActive() -> (bIsActive:bool))
   > Set PendingLoadedBackground(PendingLoadedBackground=@n22)
   > Set IsBackgroundLoaded(IsBackgroundLoaded=true)
   > Branch(Condition=Get IsCMSTextureLoaded() -> (IsCMSTextureLoaded:bool))
   > OnBackgroundReady()
   OnBackgroundReady() {is=title K2Node_CustomEvent}
   > n25 = Sequence()
   > SetScalarParameterValue!(
    execute
    Collection:Material Parameter Collection Object Reference=/Game/Environments/FrontEnd/Blueprints/ItemPreview/VaultWorldTransitions/MaterialCollectionParameters/MPC_UI_Master_VaultWorld_Settings
    ParameterName:Name=Is Horizontal Orientation
    ParameterValue:float="To Float (Boolean)"(InBool=Get PendingIsHorizontalOrientation() -> (PendingIsHorizontalOrientation:bool))
   ) -> (then) {is=title K2Node_CallMaterialParameterCollectionFunction}
   > SetVectorParameterValue!(
    execute
    Collection:Material Parameter Collection Object Reference=&MPC_UI_Master_VaultWorld_Settings
    ParameterName:Name=FocalPointPosition
    ParameterValue_R:float=Get PendingFocalPointValue() -> (PendingFocalPointValue:double)
    ParameterValue_G:float=0
    ParameterValue_B:float=0
    ParameterValue_A:float=0
   ) -> (then) {is=title K2Node_CallMaterialParameterCollectionFunction}
   n26 = Cast To MaterialInterface!(execute=@n25.then_1, Object=Get PendingLoadedBackground() -> (PendingLoadedBackground:Object Reference)) -> (then, CastFailed, AsMaterial Interface:Material Interface Object Reference) {is=title K2Node_DynamicCast}
   > n27 = CreateDynamicMaterialInstance!(execute, Parent=@n26, OptionalName=None, CreationFlags=None) -> (then, ReturnValue)
   > SetTextureParameterValue(self=@n27, ParameterName=CMS_Crew_Bento, Value=Get PendingCMSTexture() -> (PendingCMSTexture:Texture Object Reference))
   > UpdateBackground!(
    execute
    self:Fort Item Preview Background Object Reference=Get FortItemPreviewBackground() -> (FortItemPreviewBackground:Fort Item Preview Background Object Reference)
    LoadedBackground:Object Reference=@n27
    TransitionDirection:EVaultWorldTransitionDirection Enum=Get PendingDirection() -> (PendingDirection:EVaultWorldTransitionDirection Enum=Forward)
    PrimaryColor:Linear Color Structure=(R=0,G=0,B=0,A=0)
    SecondaryColor:Linear Color Structure=(R=0,G=0,B=0,A=0)
   ) -> (then)
   UpdateBackground(
    execute=@n26.CastFailed
    self=Get FortItemPreviewBackground()
    LoadedBackground=Get PendingLoadedBackground()
    TransitionDirection=Get PendingDirection() -> (PendingDirection=Forward)
    PrimaryColor=(R=0,G=0,B=0,A=0)
    SecondaryColor=(R=0,G=0,B=0,A=0)
   )
   Show Purchase Screen() {is=title K2Node_CustomEvent}
   n28 = Get SelectedPass(self=Get SeasonPassSelectionVM())
   > n29 = Is Valid(InputObject=@n28)
   n30 = PushPrimaryContentClassAsync(execute=[@n35.then, @n29.Is Not Valid], WidgetClass=/Crew/UI/CrewInfoContainer.CrewInfoContainer_C)
   n31 = Set IsActivatingFromCrew(execute=@n30.OnPrimaryContentPushed, IsActivatingFromCrew=true)
   Set with Broadcast IsActivatingFromSubscreen(execute=[@n34.then, @n31.then], IsActivatingFromSubscreen=true)
   Set ExternalLaunchPass(execute=@n29.Is Valid, ExternalLaunchPass:Season Pass VM Object Reference=@n28) -> (Output_Get:Season Pass VM Object Reference)
   > n32 = PushPrimaryContentSoftClassAsync!(
    execute:Exec
    SoftWidgetClass:Common Activatable Widget Soft Class Reference=GetActivatableWidgetForAlias!(WidgetAlias:Gameplay Tag Structure=(TagName="UI.Alias.PassPurchase")) -> (ReturnValue:Common Activatable Widget Soft Class Reference)
   ) -> (then:Exec, OnPrimaryContentPushed:Exec, ActivatableWidget:Common Activatable Widget Object Reference) {is=title K2Node_AsyncAction}
   n33 = Get ExternalLaunchPass() -> (ExternalLaunchPass:Season Pass VM Object Reference)
   SetViewModelByClass!(
    execute=@n32.OnPrimaryContentPushed
    self:MVVM View Object Reference=GetExtension!(self:User Widget Object Reference=@n32, ExtensionType:User Widget Extension Class Reference=/Script/ModelViewViewModel.MVVMView) -> (ReturnValue:MVVM View Object Reference)
    NewValue:Notify Field Value Changed Interface=@n33
   ) -> (then, ReturnValue:bool)
   > Set SelectedPass(SelectedPass=@n33, self=Get SeasonPassSelectionVM())
   > n34 = Set ExternalLaunchPass()
   n35 = Show Crew Screen() {is=title K2Node_CustomEvent}
   Show Reward() {is=title K2Node_CustomEvent}
   > n36 = Is Valid(InputObject=Get ExternalLaunchReward() -> (ExternalLaunchReward:Season Pass Reward VM Object Reference))
   Set ExternalLaunchReward(
    execute=@n36.Is Not Valid
    ExternalLaunchReward:Season Pass Reward VM Object Reference=Get SelectedReward!(self:Season Pass Selection VM Object Reference=Get SeasonPassSelectionVM()) -> (SelectedReward:Season Pass Reward VM Object Reference)
   ) -> (Output_Get:Season Pass Reward VM Object Reference)
   > n37 = Branch(Condition=Battle Pass Is Not Selected!() -> (ReturnValue:bool=~) {is=title K2Node_MacroInstance})
   > n38 = DelayUntilNextFrame!(execute) -> (then)
   NavigateToReward!(
    execute=[@n39.OnPrimaryContentPushed, @n38.then]
    self:Season Pass Selection VM Object Reference=Get SeasonPassSelectionVM()
    Reward:Season Pass Reward VM Object Reference=Get ExternalLaunchReward()
   ) -> (then)
   > Set ExternalLaunchReward()
   > Set with Broadcast IsActivatingFromSubscreen(IsActivatingFromSubscreen=true)
   n39 = PushPrimaryContentClassAsync(execute=@n37.else, WidgetClass=/SeasonPass/Widgets/Reboot/WBP_RewardTrack_Screen.WBP_RewardTrack_Screen_C)
   n40 = OnPreviewItemShown!() -> (OutputDelegate, then, ItemDefinition:Fort Account Item Definition Object Reference) {is=title K2Node_CustomEvent}
   > UpdatePreviewWindowPadding!(execute, self, ItemDefinition:Fort Account Item Definition Object Reference=@n40.ItemDefinition) -> (then)
   > PlayCharacterNameAnimation(self:WBP Season Pass Character Name Object Reference=Get WBP_SeasonPass_CharacterName!(self:WBP Season Pass Page Object Reference=Get BattlePassPage()) -> (WBP_SeasonPass_CharacterName:WBP Season Pass Character Name Object Reference))
   Event On Activated()
   > Broadcast Field Value Changed!(execute, self:Widget Object Reference, FieldId:Field Notification Id Structure=(FieldName="FindBattlePassData")) -> (then)
   > SetHasSeenPassOrHUB(self:Season Pass System VM Object Reference=Get SeasonPassSystemVM() -> (SeasonPassSystemVM:Season Pass System VM Object Reference))
   > n41 = GetAndClearIsExternalNavigation!(execute, self:Season Pass Selection VM Object Reference=Get SeasonPassSelectionVM()) -> (then, ReturnValue:bool)
   > n42 = Branch(Condition=@n41)
   > n43 = Branch(Condition=Battle Pass Is Not Selected() -> (ReturnValue=~))
   > n44 = Is Valid(InputObject=Get ExternalLaunchPass())
   ExternalNavigationNonBP(execute=@n44.Is Not Valid) {is=title K2Node_CallFunction}
   n45 = Is Valid(exec=@n43.else, InputObject=Get SelectedReward(self=Get SeasonPassSelectionVM()))
   Set PageIndex(
    execute=[@n45.Is Not Valid, @n56.Else]
    PageIndex:int=Select!(Option 0:int=0, Option 1:int=1, Index:bool=Get IsActivatingFromCrew() -> (IsActivatingFromCrew:bool)) -> (ReturnValue:int)
   ) -> (Output_Get:int=0)
   > PlayIntroAnim()
   > n46 = SpawnSound2D!(
    execute
    Sound:Sound Base Object Reference=/Game/Sounds/Fort_UI_Sounds/FrontEnd/BattlePass/SC_BattlePass_ScreenAppear_Whoosh_Special_Cue
    VolumeMultiplier:float=1
    PitchMultiplier:float=1
    StartTime:float=0
    ConcurrencySettings:Sound Concurrency Object Reference
    bPersistAcrossLevelTransition:bool=false
    bAutoDestroy:bool=true
   ) -> (then, ReturnValue:Audio Component Object Reference)
   > Set ScreenInBattlePass(ScreenInBattlePass:Audio Component Object Reference=@n46) -> (Output_Get:Audio Component Object Reference)
   > ClearSelectedPage(self:Season Pass Selection VM Object Reference=Get SeasonPassSelectionVM())
   > DelayUntilNextTick!(execute) -> (then)
   > Branch(Condition=Get bIsActive())
   > PlayUpsellMusic!(execute, self, bPlay:bool=true) -> (then)
   > n47 = Branch(Condition=Get IsActivatingFromSubscreen() -> (IsActivatingFromSubscreen:bool))
   HandlePageChange(execute=[@n47.then, @n55.then])
   > RequestRefreshFocus(self:Common Activatable Widget Object Reference)
   > Set Padding(
    Padding=Select(
     Option 0:Margin Structure=Make Margin(Left=0, Top=0, Right=0, Bottom=0)
     Option 1:Margin Structure=Get CrewItemPreviewPadding()
     Index=Get IsActivatingFromCrew()
    ) -> (ReturnValue:Margin Structure)
    self=Get SeasonPassFortItemPreviewWidget()
   )
   > SetFrontEndCamera(
    self=Get FortFrontEndContext()
    NewState=CosmeticDisplay
    TransitionParams=(BlendTime=0,BlendFunction=VTBlend_Cubic,BlendExp=2,bLockOutgoing=False)
   )
   n48 = Get SeasonPassFortItemPreviewWidget()
   > SetZoomLevel(self=@n48, NewZoomLevel=0, bSnap=true, bZoomEvenIfNotAllowed=false)
   > Bind Event to On Preview Item Shown(self:Fort Item Preview Widget Object Reference=@n48, Delegate=Create Event()) {is=title K2Node_AddDelegate}
   > ActivateWidget(self=@n48)
   > n49 = Get!(execute:Exec) -> (then:Exec, else:Exec, SeasonPassSelectionVM:Season Pass Selection VM Object Reference)
   > n50 = SetPreviewData!(
    execute
    self:Fort Item Preview Widget Object Reference=@n48
    Data:"Fort Cosmetic Preview Data Structure (by ref)"=GetPreviewData!(self:Season Pass Selection VM Object Reference=@n49) -> (ReturnValue:"Fort Cosmetic Preview Data Structure (by ref)")
   ) -> (then)
   UpdateActionbarBackplateVisibility(execute=[@n50.then, @n49.else])
   > UpdateNexusFrontendMode!(execute, Target:Widget Object Reference=Self-Reference(), Is Activating:bool=true) -> (then)
   > Set IsActivatingFromCrew(IsActivatingFromCrew=false)
   > Set with Broadcast IsActivatingFromSubscreen(IsActivatingFromSubscreen=false)
   > SetVectorParameterValue!(
    execute
    Collection=/SeasonPass/MaterialFunctions/MPC_UI_Pass_Properties
    ParameterName=ThemeColor
    ParameterValue:"Linear Color Structure (by ref)"=MakeColor!(R:float=0, G:float=0, B:float=0, A:float=0) -> (ReturnValue:Linear Color Structure)
   ) -> (then)
   > Fire Event Main View Opened!(execute, View:Widget Object Reference=Self-Reference(), AdditionalInfo:Flow Analytics Additional Info Structure, bIsOverlay:bool=false) -> (then)
   > Is Valid(InputObject=Get SelectedPass(self=Get SeasonPassSelectionVM()))
   > n51 = Cast To FortPlayerController(Object=GetOwningPlayer())
   n52 = Get SeasonPassSelectionVM()
   n53 = Get Source String from Text!(Text:Text=GetDisplayName!(self:Season Pass VM Object Reference=Get SelectedPass(self=@n52)) -> (ReturnValue:Text)) -> (ReturnValue:String) {is=title K2Node_CallFunction}
   > Branch(Condition=NOT Boolean(A=IsEmpty(InString=@n53)))
   > "Fire Event Sub View Opened (Non-Localized)"!(
    execute
    FortPC:Fort Player Controller Object Reference=@n51
    ViewName:String=@n53
    AdditionalInfo:Flow Analytics Additional Info Structure
    bIsOverlay:bool=false
   ) -> (then)
   > n54 = StartListeningToEvent!(
    execute:Exec
    Target:Gameplay Event Router Component Object Reference=GetGlobalEventRouter()
    EventType:Script Struct Object Reference=/FortniteCoreUI/UI/Social/Chat/TextChat/BPS_TextChatActiveStateChange
    Context:Object Reference=GetOwningPlayer()
    EventBubblingRule:EEventBubblingRule Enum
   ) -> (
    then:Exec
    AsyncTaskProxy:Async Action Start Listening To Event Object Reference
    OnEventReceived:Exec
    Payload:BPS Text Chat Active State Change Structure
   )
   > Set ChatEventListener(ChatEventListener:Async Action Start Listening To Event Object Reference=@n54) -> (Output_Get:Async Action Start Listening To Event Object Reference)
   SetButtonPromptsVisibility!(
    execute=@n54.OnEventReceived
    self
    ShouldShow:bool=NOT Boolean(A=Break BPS Text Chat Active State Change!(BPS_TextChatActiveStateChange:"BPS Text Chat Active State Change Structure (by ref)"=@n54.Payload) -> (bActive_4_#15:bool=~) {is=title K2Node_BreakStruct})
   ) -> (then)
   n55 = Set SelectedPass(execute=@n47.else, SelectedPass=Get BattlePassData(), self=Get SeasonPassSelectionVM())
   Show Reward(execute=@n45.Is Valid) {is=title K2Node_CallFunction}
   n56 = Decide if Showing Upsell A/B Test!(Exec:Exec=@n42.else) -> (Else:Exec) {is=title K2Node_MacroInstance}
   Event On Deactivated()
   > n57 = Get!(execute) -> (then, else, ScreenInBattlePass:Audio Component Object Reference)
   > n58 = FadeOut!(
    execute
    self:Audio Component Object Reference=@n57
    FadeOutDuration:float=1
    FadeVolumeLevel:float=0
    FadeCurve:EAudioFaderCurve Enum=Linear
   ) -> (then)
   n59 = Get SeasonPassFortItemPreviewWidget()
   ClearPreviewItem(execute=[@n58.then, @n57.else], self=@n59)
   > Unbind all Events from On Preview Item Shown!(execute:Exec, self:Fort Item Preview Widget Object Reference=@n59) -> (then:Exec) {is=title K2Node_ClearDelegate}
   > DeactivateWidget(self=@n59)
   > UpdateNexusFrontendMode(Target=Self-Reference(), Is Activating=false)
   > PlayUpsellMusic(bPlay=false)
   > n60 = Branch(Condition=Get IsActivatingFromSubscreen())
   UpdateBackground(
    execute=@n60.else
    self=Get FortItemPreviewBackground()
    TransitionDirection=None
    PrimaryColor=(R=0,G=0,B=0,A=0)
    SecondaryColor=(R=0,G=0,B=0,A=0)
   )
   > SetRenderOpacity(self=Get OverlayMain() -> (OverlayMain:Overlay Object Reference), InOpacity=0)
   Event PreConstruct()
   > SetCompactUIStyle() {is=title K2Node_CallFunction}
   > Branch(Condition=IsCultureRightToLeft!(Culture:String=GetCurrentCulture!() -> (ReturnValue:String)) -> (ReturnValue:bool))
   > HandleRTL()
   > n61 = StartListeningToEvent(Target=GetGlobalEventRouter(), EventType=&BPS_TextChatActiveStateChange, Context=GetOwningPlayer())
   > Set ChatEventRef(ChatEventRef:Async Action Start Listening To Event Object Reference=@n61) -> (Output_Get:Async Action Start Listening To Event Object Reference)
   SetButtonPromptsVisibility(
    execute=@n61.OnEventReceived
    ShouldShow=NOT Boolean(A=Break BPS Text Chat Active State Change(BPS_TextChatActiveStateChange=@n61.Payload) -> (bActive_4_#15=~))
   )
   Event Destruct()
   > StopListeningToEvent(self:Async Action Start Listening To Event Object Reference=Get ChatEventListener() -> (ChatEventListener:Async Action Start Listening To Event Object Reference))
   > Set ChatEventListener()
   ScrollToNextPage(Up=true)
   "Nav Up But No Buttons Left (BattlePassPage)"() {is=title K2Node_ComponentBoundEvent}
   "Crew Page Activated (CrewLandingPage)"() {is=title K2Node_ComponentBoundEvent}
   "On Widget Activated (BattlePassPage)"() {is=title K2Node_ComponentBoundEvent}
  ]}
 FindBattlePassData {nodes [
   FindBattlePassData() {is=title K2Node_FunctionEntry}
   > Return Node!(
    execute
    BattlePassData:Season Pass VM Object Reference=FindSeasonPass!(self:Season Pass System VM Object Reference=Get SeasonPassSystemVM(), TemplateId:String=battlepassdata) -> (ReturnValue:Season Pass VM Object Reference)
   )
  ]}
 GetDesiredPendingDirection {nodes [
   GetDesiredPendingDirection!() -> (then, PendingDirection:EVaultWorldTransitionDirection Enum=Forward) {is=title K2Node_FunctionEntry}
   > n1 = Branch(Condition=Get SelectedPassChangedInSubscreen() -> (SelectedPassChangedInSubscreen:bool))
   > Set SelectedPassChangedInSubscreen(SelectedPassChangedInSubscreen:bool=false) -> (Output_Get:bool)
   > Return Node!(execute, Result:EVaultWorldTransitionDirection Enum=Forward)
   Return Node(execute=@n1.else, Result=Get PendingDirection() -> (PendingDirection=Forward))
  ]}
 GetNumPurchasableBattlePassRewards {nodes [
   GetNumPurchasableBattlePassRewards() {is=title K2Node_FunctionEntry}
   n1 = Get BattlePassData()
   > n2 = Is Valid(InputObject=@n1)
   Return Node!(execute=@n2.Is Not Valid, NumPurchasableRewards:int=0)
   > Return Node(NumPurchasableRewards=GetNumPurchaseableRewards!(self:Season Pass Element Base VM Object Reference=@n1) -> (ReturnValue:int=0))
  ]}
 GetPrimaryContentSetup {nodes [
   GetPrimaryContentSetup() {is=title K2Node_FunctionEntry}
   > Return Node!(execute, ReturnValue:Primary Content Setup Structure)
  ]}
 GetSeasonPassTemplateIds {nodes [
   n1 = GetSeasonPassTemplateIds!() -> (then, Benefits:Array of Crew Benefits Group VM Object References) {is=title K2Node_FunctionEntry}
   > n2 = For Each Loop(Array:Array of Crew Benefits Group VM Object References=@n1) -> (Array Element:Crew Benefits Group VM Object Reference)
   > n3 = For Each Loop(Array:Array of Crew Benefit VM Object References=Get Benefits!(self:Crew Benefits Group VM Object Reference=@n2) -> (Benefits:Array of Crew Benefit VM Object References)) -> (Array Element:Crew Benefit VM Object Reference)
   Return Node!(execute=@n3.Completed, TemplateIds:Array of Strings=Get Results() -> (Results:Array of Strings))
   > Branch(Condition="Equal (Enum)"(
     A:ECrewTileType Enum=GetType!(self:Crew Benefit VM Object Reference=@n3) -> (ReturnValue:ECrewTileType Enum=None)
     B:ECrewTileType Enum=BattlePass
    ))
   > Add(TargetArray=Get Results(), NewItem=GetTemplateId!(self:Crew Benefit VM Object Reference=@n3) -> (ReturnValue:String)) {is=title K2Node_CallArrayFunction}
  ]}
 GetTotalRewardsByCategory {nodes [
   n1 = Get SeasonPassSystemVM()
   n2 = IsValid(Object=@n1)
   n3 = FindSeasonPass(self=@n1, TemplateId=Inputs!() -> (TemplateId:String))
   Outputs!(ReturnValue:int=GetNumPurchaseableRewards(self=Select(Option 0:Season Pass VM Object Reference, Option 1:Season Pass VM Object Reference=@n3, Index=@n2) -> (ReturnValue:Season Pass VM Object Reference)) -> (ReturnValue=0))
  ]}
 HandleCrewBackgroundTransition {nodes [
   n1 = HandleCrewBackgroundTransition!() -> (then, TransitionDirection:EVaultWorldTransitionDirection Enum=Forward, IsHorizontalOrientation:bool=~) {is=title K2Node_FunctionEntry}
   n2 = Get CrewVM() -> (CrewVM:Crew VM Object Reference)
   > RequestLoadBackground!(
    execute
    self
    BackgroundAsset:Soft Object Reference=Select(
     Option 0:Soft Object Reference=/Crew/UI/Materials/M_UI_CrewUpsel_VW__US
     Option 1:Soft Object Reference=/Crew/UI/Materials/M_UI_CrewUpsel_VW_S
     Index=IsSubscribed!(self:Crew VM Object Reference=@n2) -> (ReturnValue:bool)
    ) -> (ReturnValue:Soft Object Reference)
    Direction:EVaultWorldTransitionDirection Enum=@n1
    FocalPointValue:double=-1
    IsHorizontalOrientation:bool=Get IsHorizontalOrientation() -> (IsHorizontalOrientation:bool)
   ) -> (then) {is=title K2Node_CallFunction}
   > PlaySound2D(Sound=Get CrewScreenAppear() -> (CrewScreenAppear:Sound Base Object Reference))
  ]}
 HandleNavigateToCrew {nodes [
   n1 = HandleNavigateToCrew!() -> (then, IsNavigateToCrewExternal:bool=~) {is=title K2Node_FunctionEntry}
   > Branch(Condition=@n1)
   > n2 = GetAndClearIsNavigateToCrewExternal!(execute, self:Season Pass Selection VM Object Reference=Get SeasonPassSelectionVM()) -> (then, ReturnValue:bool)
   > Branch(Condition=@n2)
   > Show Crew Screen() {is=title K2Node_CallFunction}
  ]}
 HandleNavigateToPurchase {nodes [
   n1 = HandleNavigateToPurchase!() -> (then, IsNavigateToPurchaseExternal:bool=~) {is=title K2Node_FunctionEntry}
   > Branch(Condition=@n1)
   > n2 = GetAndClearIsNavigateToPurchaseExternal!(execute, self:Season Pass Selection VM Object Reference=Get SeasonPassSelectionVM()) -> (then, ReturnValue:bool)
   > Branch(Condition=@n2)
   > Show Purchase Screen() {is=title K2Node_CallFunction}
  ]}
 HandleNavigateToReward {nodes [
   n1 = HandleNavigateToReward!() -> (then, IsNavigateToRewardExternal:bool=~) {is=title K2Node_FunctionEntry}
   > Branch(Condition=@n1)
   > n2 = GetAndClearIsNavigateToRewardExternal!(execute, self:Season Pass Selection VM Object Reference=Get SeasonPassSelectionVM()) -> (then, ReturnValue:bool)
   > Branch(Condition=@n2)
   > Show Reward()
  ]}
 HandlePageChange {nodes [
   HandlePageChange() {is=title K2Node_FunctionEntry}
   > n1 = Branch(Condition=Get IsActivatingFromSubscreen())
   n2 = Sequence!(execute=[@n16.then, @n1.then]) -> (then_0, then_1, then_2:Exec)
   > n3 = Switch on Int!(execute:Exec, Selection:int=Get PageIndex() -> (PageIndex:int=0)) -> (0:Exec, 1:Exec) {bHasDefaultPin=false}
   > Fire Event Sub View Opened!(execute, View:Widget Object Reference=Get BattlePassPage(), AdditionalInfo:Flow Analytics Additional Info Structure, bIsOverlay:bool=false) -> (then)
   > SetVisibility(self=[Get CrewNavWrapper(), Get PageNav_Top() -> (PageNav_Top:WBP Season Pass Page Nav Object Reference)], InVisibility=Collapsed)
   > Set Padding(Padding=Make Margin(Left=0, Top=0, Right=0, Bottom=0), self=Get SeasonPassFortItemPreviewWidget())
   > SetVisibility(
    self=[
     Get BattlePassNavWrapper() -> (BattlePassNavWrapper:Overlay Object Reference)
     Get LevelRequirements() -> (LevelRequirements:WBP Season Pass Level Requirements Object Reference)
     Get RemainingSetsToUnlock() -> (RemainingSetsToUnlock:WBP Season Pass Remaining Sets To Unlock Object Reference)
    ]
    InVisibility=SelfHitTestInvisible
   )
   > SetVisibility(self=Get PageNav_Bottom() -> (PageNav_Bottom:WBP Season Pass Page Nav Object Reference), InVisibility=Visible)
   > SetForcedCondensed!(
    execute
    self:WBP Season Pass Rewards Object Reference=Get SeasonPass_Rewards() -> (SeasonPass_Rewards:WBP Season Pass Rewards Object Reference)
    IsForced:bool=false
   ) -> (then)
   > n4 = Set SeasonPassVM!(execute, SeasonPassVM:Season Pass VM Object Reference=Get BattlePassData(), self:WBP Season Pass Page Object Reference=Get BattlePassPage()) -> (then, Output_Get:Season Pass VM Object Reference)
   n5 = Get SeasonPassSelectionVM()
   > Set SelectedPass(SelectedPass=@n4, self=@n5)
   > n6 = Branch(Condition=Get IsActivatingFromSubscreen())
   n7 = GetNumPurchasableBattlePassRewards!(execute=[@n10.then, @n6.then], self) -> (then, NumPurchasableRewards:int=0) {is=title K2Node_CallFunction}
   n8 = Get NumPurchaseableRewards!(self:Season Pass System VM Object Reference=Get SeasonPassSystemVM()) -> (NumPurchaseableRewards:int=0)
   > SetPageNav!(
    execute
    self
    Target:WBP Season Pass Page Nav Object Reference=Get PageNav_Bottom()
    Title:Text=Format Text(Format="LOCTEXT(\"Passes && FN Crew\")")
    Description:Text=Set Nav Rewards Value Text!(Value:int=int - int(A=@n8, B=@n7)) -> (ReturnValue:Text) {is=title K2Node_MacroInstance}
    ChevronPosition:E_ChevronPosition Enum=NewEnumerator1
   ) -> (then)
   n9 = Get BattlePassPage()
   > SetText!(
    execute
    self:Fort CTAButton Object Reference=Get BuyRewards_Button!(self:WBP Season Pass Header Object Reference=Get SeasonPass_Header!(self:WBP Season Pass Page Object Reference=@n9) -> (SeasonPass_Header:WBP Season Pass Header Object Reference)) -> (BuyRewards_Button:WBP UIKit Button Loud Object Reference)
    ButtonText:"Text (by ref)"=Format Text(Format="LOCTEXT(\"Buy Levels\")")
   ) -> (then)
   > Set IsBattlePass!(execute, IsBattlePass:bool=true, self:WBP Season Pass Header Object Reference=Get SeasonPass_Header(self=Get BattlePassPage())) -> (then, Output_Get:bool)
   > HandleSeasonPassBackgroundTransition!(execute, self, TransitionDirection:EVaultWorldTransitionDirection Enum=Backward, IsHorizontalOrientation:bool=false) -> (then)
   > RequestRefreshFocus()
   n10 = ClearSelectedCategory(execute=@n6.else, self=@n5)
   Fire Event Sub View Opened(execute=@n3.1, View=Get CrewLandingPage(), bIsOverlay=false)
   > VM_HandleThemeData!(execute, self, PassTheme:Season Pass Theme Data Structure) -> (then)
   > n11 = GetNumPurchasableBattlePassRewards() -> (NumPurchasableRewards=0)
   > SetPageNav(
    Target=Get PageNav_Top()
    Title=Format Text(Format="LOCTEXT(\"Battle Pass\")")
    Description=Set Nav Rewards Value Text(Value=@n11)
    ChevronPosition=NewEnumerator0
   )
   > SetVisibility(self=[Get BattlePassNavWrapper(), Get LevelRequirements(), Get RemainingSetsToUnlock(), Get PageNav_Bottom()], InVisibility=Collapsed)
   > Set Padding(Padding=Get CrewItemPreviewPadding(), self=Get SeasonPassFortItemPreviewWidget())
   > SetVisibility(self=Get CrewNavWrapper(), InVisibility=SelfHitTestInvisible)
   > SetForcedCondensed(self=Get SeasonPass_Rewards(), IsForced=true)
   > n12 = Branch(Condition=Get HasEnteredCrewPassSubscreen!(self:WBP Crew Landing Page Object Reference=Get CrewLandingPage()) -> (HasEnteredCrewPassSubscreen:bool))
   n13 = Get SeasonPassSelectionVM()
   > Set SelectedPassChangedInSubscreen(SelectedPassChangedInSubscreen="Not Equal (Object)"(A=Get CrewSelectedPass() -> (CrewSelectedPass:Season Pass VM Object Reference), B=Get SelectedPass(self=@n13)) -> (ReturnValue=~))
   > n14 = OpeningTile!(
    execute
    self:WBP Crew Landing Page Object Reference=Get CrewLandingPage()
    SeasonPassVM:Season Pass VM Object Reference=Get CrewSelectedPass()
    ButtonSelected:WBP Crew Pass Button Object Reference=Get CrewSelectedButton() -> (CrewSelectedButton:WBP Crew Pass Button Object Reference)
   ) -> (then)
   Set HasEnteredCrewPassSubscreen(execute=[@n14.then, @n15.then], HasEnteredCrewPassSubscreen=false, self=Get CrewLandingPage())
   > Set IsPageNav!(execute, IsPageNav:bool=true, self:WBP Crew Landing Page Object Reference=Get CrewLandingPage()) -> (then, Output_Get:bool)
   > SetVisibility(self=Get PageNav_Top(), InVisibility=Visible)
   > RequestRefreshFocus()
   OpeningTile(execute=@n12.else, self=Get CrewLandingPage())
   > n15 = HandleCrewBackgroundTransition!(execute, self, TransitionDirection:EVaultWorldTransitionDirection Enum=Forward, IsHorizontalOrientation:bool=false) -> (then) {is=title K2Node_CallFunction}
   UpdateActionbarBackplateVisibility(execute=@n2.then_1)
   UpdateUpsellMusic!(execute=@n2.then_2, self, ParamName:Name=None) -> (then)
   n16 = ClearSelectedPass(execute=@n1.else, self:Season Pass Selection VM Object Reference=Get SeasonPassSelectionVM())
  ]}
 HandleRTL {nodes [
   HandleRTL() {is=title K2Node_FunctionEntry}
   > Set FlowDirectionPreference!(
    execute
    FlowDirectionPreference:EFlowDirectionPreference Enum=Culture
    self:Widget Object Reference=Get SwipePanel() -> (SwipePanel:Fort Swipe Panel Object Reference)
   ) -> (then, Output_Get:EFlowDirectionPreference Enum=Inherit)
  ]}
 HandleSeasonPassBackgroundTransition {nodes [
   HandleSeasonPassBackgroundTransition!() -> (then, TransitionDirection:EVaultWorldTransitionDirection Enum=Forward, IsHorizontalOrientation:bool=~) {is=title K2Node_FunctionEntry}
   > n1 = Sequence!(execute) -> (then_0, then_1)
   n2 = Get SelectedPass(self=Get SeasonPassSelectionVM())
   > Is Valid(InputObject=@n2)
   > RequestLoadBackground(
    BackgroundAsset=Break Season Pass Theme Data!(SeasonPassThemeData:"Season Pass Theme Data Structure (by ref)"=GetThemeData!(self:Season Pass VM Object Reference=@n2) -> (ReturnValue:"Season Pass Theme Data Structure (by ref)")) -> (
     VaultBackgroundMaterial:Soft Object Reference
     GiftboxVaultBackgroundMaterial:Soft Object Reference
     CustomLogo:Soft Object Reference
     CustomSeasonTitle:Text
     CustomSeasonTitleColor:Linear Color Structure
     Upsell2BackgrounMaterial:Soft Object Reference
     ImageReferenceTable:Fort Image Reference Table Structure
     ColorReferenceTable:Fort Color Reference Table Structure
     TextReferenceTable:Fort Text Reference Table Structure
    ) {is=title K2Node_BreakStruct}
    Direction=Get TransitionDirection() -> (TransitionDirection:EVaultWorldTransitionDirection Enum=Forward)
    FocalPointValue=0
    IsHorizontalOrientation=Get IsHorizontalOrientation()
   )
   PlaySound2D(execute=@n1.then_1, Sound=Get SeasonPassScreenAppear() -> (SeasonPassScreenAppear:Sound Base Object Reference))
  ]}
 NavigateToContent {nodes [
   NavigateToContent!() -> (then, Navigation:EUINavigation Enum) {is=title K2Node_FunctionEntry}
   n1 = IsVisible(self=Get CrewNavWrapper())
   n2 = Get BattlePassPage()
   n3 = Get CrewLandingPage()
   > Return Node!(
    execute
    ReturnValue:Widget Object Reference=GetDesiredFocusTarget(self=Select(Option 0:Common Activatable Widget Object Reference=@n2, Option 1:Common Activatable Widget Object Reference=@n3, Index=@n1) -> (ReturnValue:Common Activatable Widget Object Reference))
   )
  ]}
 OnBackgroundURLChanged {nodes [
   n1 = OnBackgroundURLChanged!() -> (then, URL:String) {is=title K2Node_FunctionEntry}
   > Set PendingCMSTexture()
   > Set IsCMSTextureLoaded(IsCMSTextureLoaded=false)
   > SetMediaURL!(
    execute
    self:Epic CMSImage Object Reference=Get EpicCMSImage_BackgroundLoaderHelper() -> (EpicCMSImage_BackgroundLoaderHelper:Epic CMSImage Object Reference)
    MediaUrl:String=@n1
   ) -> (then)
  ]}
 OnCrewPassTileSelected {nodes [
   n1 = OnCrewPassTileSelected!() -> (then, SelectedPass:Season Pass VM Object Reference, NewSelectedButton:WBP Crew Pass Button Object Reference) {is=title K2Node_FunctionEntry}
   > n2 = Set SelectedPass(SelectedPass=@n1, self=Get SeasonPassSelectionVM())
   > Set CrewSelectedPass(CrewSelectedPass:Season Pass VM Object Reference=@n2) -> (Output_Get:Season Pass VM Object Reference)
   > n3 = Get!(execute) -> (then, else, CrewSelectedButton:WBP Crew Pass Button Object Reference)
   > Set WasAnyTileSelected(WasAnyTileSelected:bool=true) -> (Output_Get:bool)
   n4 = Get NewSelectedButton() -> (NewSelectedButton:WBP Crew Pass Button Object Reference)
   > n5 = Is Valid(InputObject=@n4)
   > n6 = Compare Int!(
    Exec:Exec
    Input:int=Get Index!(self:WBP Crew Pass Button Object Reference=@n4) -> (Index:int=0)
    CompareWith:int=Get Index(self=@n3) -> (Index=0)
   ) -> (">":Exec, "==":Exec, "<":Exec) {is=title K2Node_MacroInstance}
   n7 = Set BackgroundTransitionDirection(execute=@n6."<", BackgroundTransitionDirection:EVaultWorldTransitionDirection Enum=Backward) -> (Output_Get:EVaultWorldTransitionDirection Enum=Forward)
   n8 = Set CrewSelectedButton(execute=[@n7.then, @n12.then, @n11.then], CrewSelectedButton:WBP Crew Pass Button Object Reference=Get NewSelectedButton()) -> (Output_Get:WBP Crew Pass Button Object Reference)
   > Is Valid(InputObject=@n8)
   > n9 = Branch(Condition=Get IsCrewTile!(self:WBP Crew Pass Button Object Reference=@n8) -> (IsCrewTile:bool))
   n10 = Get BackgroundTransitionDirection() -> (BackgroundTransitionDirection:EVaultWorldTransitionDirection Enum=Forward)
   > HandleCrewBackgroundTransition(TransitionDirection=@n10, IsHorizontalOrientation=Get WasAnyTileSelected() -> (WasAnyTileSelected:bool))
   HandleSeasonPassBackgroundTransition(execute=@n9.else, TransitionDirection=@n10, IsHorizontalOrientation=true) {is=title K2Node_CallFunction}
   n11 = Set BackgroundTransitionDirection(execute=@n6."==", BackgroundTransitionDirection=None) -> (Output_Get=Forward)
   n12 = Set BackgroundTransitionDirection(execute=[@n6.">", @n5.Is Not Valid, @n13.then], BackgroundTransitionDirection=Forward) -> (Output_Get=Forward)
   n13 = Set WasAnyTileSelected(execute=@n3.else, WasAnyTileSelected=false)
  ]}
 OnExternalNavigationDeactivated {nodes [OnExternalNavigationDeactivated() {is=title K2Node_FunctionEntry}]}
 OnMouseWheel {nodes [
   n1 = OnMouseWheel!() -> (then, MyGeometry:Geometry Structure, MouseEvent:"Pointer Event Structure (by ref)") {is=title K2Node_FunctionEntry}
   > ScrollToNextPage(Up="float > float"(A=Get Wheel Delta!(Input:"Pointer Event Structure (by ref)"=@n1.MouseEvent) -> (ReturnValue:float) {is=title K2Node_CallFunction}))
   > Return Node(ReturnValue:Event Reply Structure=Handled())
  ]}
 PlayIntroAnim {nodes [
   PlayIntroAnim() {is=title K2Node_FunctionEntry}
   > PlayAnimationForward(InAnimation=Get IntroOnActivated() -> (IntroOnActivated:Widget Animation Object Reference))
   > Opening(self:WBP Season Pass Page Object Reference=Get BattlePassPage())
  ]}
 PlayUpsellMusic {nodes [PlayUpsellMusic!() -> (then, bPlay:bool=~) {is=title K2Node_FunctionEntry}]}
 QuickResetIntroAnimation {nodes [
   QuickResetIntroAnimation() {is=title K2Node_FunctionEntry}
   > SetRenderOpacity(
    self=[
     Get CrewNavWrapper()
     Get BattlePassNavWrapper()
     Get PageNav_Top()
     Get PageNav_Bottom()
     Get LegalandActionBar() -> (LegalandActionBar:Horizontal Box Object Reference)
     Get TopRight() -> (TopRight:Vertical Box Object Reference)
    ]
    InOpacity=0
   )
  ]}
 ScrollToNextPage {nodes [
   n1 = ScrollToNextPage!() -> (then, Up:bool=~) {is=title K2Node_FunctionEntry}
   > Set IsPageUp!(execute, IsPageUp:bool=@n1, self:WBP Season Pass Page Object Reference=Get BattlePassPage()) -> (then, Output_Get:bool)
   n2 = SelectInt(A=0, B=1, bPickA=Get Up() -> (Up:bool))
   > Branch(Condition="Not Equal (Integer)"(A=@n2, B=Get PageIndex() -> (PageIndex=0)))
   > Set PageIndex(PageIndex=@n2) -> (Output_Get=0)
   > HandlePageChange() {is=title K2Node_CallFunction}
   n3 = Get AnimMoveUp() -> (AnimMoveUp:Widget Animation Object Reference)
   n4 = Get AnimMoveDown() -> (AnimMoveDown:Widget Animation Object Reference)
   > n5 = Branch(Condition=OR Boolean(A=IsAnimationPlaying(InAnimation=@n4), B=IsAnimationPlaying(InAnimation=@n3)))
   Branch(
    execute=[@n9.then, @n5.then]
    Condition=AND Boolean!(A=Get PlayIntroAnimNextScrollUp() -> (PlayIntroAnimNextScrollUp:bool), B=Get Up()) -> (ReturnValue)
   )
   > PlayIntroAnim() {is=title K2Node_CallFunction}
   > Set PlayIntroAnimNextScrollUp(PlayIntroAnimNextScrollUp=false)
   n6 = Get Up()
   n7 = Branch(execute=@n5.else, Condition=@n6)
   > n8 = PlaySound2D(Sound=/Game/Sounds/Fort_UI_Sounds/FrontEnd/Crew/SC_UI_Crew_ScreenLeave_Cue)
   n9 = PlayAnimationForward(
    execute=[@n8.then, @n7.else]
    InAnimation=Select(Option 0:Widget Animation Object Reference=Get AnimMoveDown(), Option 1:Widget Animation Object Reference=Get AnimMoveUp(), Index=@n6) -> (ReturnValue:Widget Animation Object Reference)
    PlaybackSpeed=2
   )
  ]}
 SetButtonPromptsVisibility {nodes [
   n1 = SetButtonPromptsVisibility!() -> (then, ShouldShow:bool=~) {is=title K2Node_FunctionEntry}
   > SetVisibility(
    self=Get WBP_ActionBarContainer!(self:WBP Athena Bottom Bar Widget Object Reference=Get WBP_AthenaBottomBarWidget() -> (WBP_AthenaBottomBarWidget:WBP Athena Bottom Bar Widget Object Reference)) -> (WBP_ActionBarContainer:WBP Action Bar Container Object Reference)
    InVisibility=Select(Option 0:ESlateVisibility Enum=Collapsed, Option 1:ESlateVisibility Enum=SelfHitTestInvisible, Index=@n1) -> (ReturnValue:ESlateVisibility Enum)
   )
  ]}
 SetCategoryBillboard {nodes [
   n1 = SetCategoryBillboard!() -> (then, SelectedCategory:Season Pass Category VM Object Reference) {is=title K2Node_FunctionEntry}
   > n2 = Is Valid(InputObject=@n1)
   FortBlueprintLog!(execute=@n2.Is Not Valid, InString:String=VM_SelectedCategoryUpdated: Is NOT valid Category) -> (then)
   > UpdateVisibility!(
    execute
    self:WBP Season Pass Rivalry Billboard Object Reference=Get RivalryButton() -> (RivalryButton:WBP Season Pass Rivalry Billboard Object Reference)
    IsValidUIFeature:bool=false
   ) -> (then)
   FortBlueprintLog(execute=@n2.Is Valid, InString=VM_SelectedCategoryUpdated: Is valid Category)
   n3 = Break Season Pass Element Theme Data!(SeasonPassElementThemeData:"Season Pass Element Theme Data Structure (by ref)"=GetElementThemeData!(self:Season Pass Element Base VM Object Reference=Get SelectedCategory!() -> (SelectedCategory)) -> (ReturnValue:"Season Pass Element Theme Data Structure (by ref)")) -> (
    ImageReferenceTable:Fort Image Reference Table Structure
    ColorReferenceTable:Fort Color Reference Table Structure
    TextReferenceTable:Fort Text Reference Table Structure
   ) {is=title K2Node_BreakStruct}
   n4 = TryLookupColor!(Table:"Fort Color Reference Table Structure (by ref)"=@n3.ColorReferenceTable, Name:String=CharacterColor) -> (OutColor:Linear Color Structure, ReturnValue:bool)
   n5 = TryLookupText!(Table:"Fort Text Reference Table Structure (by ref)"=@n3.TextReferenceTable, Name:String=Category-UIFeature-CustomPromptText) -> (OutTextReference:Text, ReturnValue:bool)
   > n6 = Construct Object from Class!(
    execute
    Class=/SeasonPassHub/InterstitialVMs/PassBillboardVM.PassBillboardVM_C
    self
    UIFeature:String="To String (Text)"(InText=TryLookupText(Table=@n3.TextReferenceTable, Name=Category-UIFeature))
    BillboardColor:Linear Color Structure=Select(Option 0:Linear Color Structure=(R=0.0802,G=1,B=0,A=1), Option 1:Linear Color Structure=@n4, Index=@n4.ReturnValue) -> (ReturnValue:Linear Color Structure)
    LogoText:Text=TryLookupText(Table=@n3.TextReferenceTable, Name=Category-UIFeature-CustomText)
    PromptText:Text=Select(Option 0:Text="LOCTEXT(\"Unlock outfit style!\")", Option 1:Text=@n5, Index=@n5.ReturnValue) -> (ReturnValue:Text)
    BillboardImage:Soft Object Reference=Break Fort Image Reference!(FortImageReference:"Fort Image Reference Structure (by ref)"=TryLookupImage!(Table:"Fort Image Reference Table Structure (by ref)"=@n3, Name:String=BillboardImage) -> (OutImageReference:Fort Image Reference Structure, ReturnValue:bool)) -> (Image:Soft Object Reference, ImageUrl:String, MaterialTextureParamName:Name) {is=title K2Node_BreakStruct}
    LogoImage:Soft Object Reference=Break Fort Image Reference(FortImageReference=TryLookupImage(Table=@n3, Name=BillboardImage_Logo))
    InputAction_DataTable:Data Table Object Reference=&InputActionDataTable
    InputAction_RowName:Name=GenericFaceButtonTop
    ShadowColor:Linear Color Structure=(R=0,G=0,B=0,A=0)
   ) -> (then, ReturnValue:Pass Billboard VM Object Reference)
   > SetPassBillboardVM!(execute, self:WBP Season Pass Rivalry Billboard Object Reference=Get RivalryButton(), Viewmodel:Pass Billboard VM Object Reference=@n6) -> (then)
  ]}
 SetCompactUIStyle {nodes [
   SetCompactUIStyle() {is=title K2Node_FunctionEntry}
   > Branch(Condition=IsUsingCompactUI())
   > SetPadding!(
    execute
    self:Size Box Slot Object Reference=SlotAsSizeBoxSlot!(Widget:Widget Object Reference=Get OverlayMain()) -> (ReturnValue:Size Box Slot Object Reference)
    InPadding_Left:float=48
    InPadding_Top:float=24
    InPadding_Right:float=48
    InPadding_Bottom:float=24
   ) -> (then)
   n1 = Get SB_Content() -> (SB_Content:Size Box Object Reference)
   > ClearWidthOverride(self:Size Box Object Reference=@n1)
   > SetPadding(self:Overlay Slot Object Reference=SlotAsOverlaySlot(Widget=@n1), InPadding_Left=0, InPadding_Top=96, InPadding_Right=0, InPadding_Bottom=0)
   > SetPadding!(execute, self=SlotAsOverlaySlot(Widget=Get PageNav_Top()), InPadding:Margin Structure) -> (then)
   > Set CrewItemPreviewPadding(CrewItemPreviewPadding:Margin Structure=Make Margin(Left=720, Top=0, Right=0, Bottom=0)) -> (Output_Get:Margin Structure)
  ]}
 SetNavRewardsValueText {nodes [
   n1 = Inputs!() -> (Value:int)
   Outputs(ReturnValue:Text=Select(Option 0=Format Text!(Format="LOCTEXT(\"({NumRewards} to Claim)\")", NumRewards:int=@n1) -> (Result), Index="Equal (Integer)"(A=@n1)))
  ]}
 SetPageNav {nodes [
   n1 = SetPageNav!() -> (then, Target:WBP Season Pass Page Nav Object Reference, Title:Text, Description:Text, ChevronPosition:E_ChevronPosition Enum=NewEnumerator0) {is=title K2Node_FunctionEntry}
   > SetData!(
    execute
    self:WBP Season Pass Page Nav Object Reference=@n1
    PageNavData:F Page Nav Data Structure=Make F Page Nav Data!(
     Title_5_#16:Text=@n1.Title
     Description_7_#17:Text=@n1.Description
     ChevronPosition_2_#18:E_ChevronPosition Enum=@n1.ChevronPosition
    ) -> (F_PageNav_Data:F Page Nav Data Structure) {is=title K2Node_MakeStruct}
   ) -> (then)
  ]}
 ShowFTUEModal {nodes [
   Inputs!() -> (execute:Exec)
   > n1 = Branch(Condition=GetConfigDataBoolWithDefault(EntryName=HasSeenPassFTUEModal, bDefaultValue=false))
   Outputs!(then:Exec=[@n8.then, @n1.then])
   n2 = Construct Object from Class!(execute=@n1.else, Class=/SeasonPass/Widgets/Reboot/Data/DialogVM_SeasonPassFTUE.DialogVM_SeasonPassFTUE_C, self) -> (then, ReturnValue:Dialog VM Season Pass FTUE Object Reference)
   n3 = TryLookupImage(
    Table=GetThemeData!(self=Get SeasonPassVM()) -> (
     ReturnValue_TitleColor:Linear Color Structure
     ReturnValue_TitleMaterial:Soft Object Reference
     ReturnValue_TitleOutline:Font Outline Settings Structure
     ReturnValue_SeasonColor:Linear Color Structure
     ReturnValue_ProgressionColor:Linear Color Structure
     ReturnValue_OverlayColor:Linear Color Structure
     ReturnValue_VaultBackgroundMaterial:Soft Object Reference
     ReturnValue_GiftboxVaultBackgroundMaterial:Soft Object Reference
     ReturnValue_UIBackgroundMaterial:Soft Object Reference
     ReturnValue_UpsellBackgroundMaterial:Soft Object Reference
     ReturnValue_CrewCategoryBackgroundMaterial:Soft Object Reference
     ReturnValue_CustomLogo:Soft Object Reference
     ReturnValue_CustomSeasonTitle:Text
     ReturnValue_CustomSeasonTitleColor:Linear Color Structure
     ReturnValue_Upsell2BackgrounMaterial:Soft Object Reference
     ReturnValue_ImageReferenceTable:Fort Image Reference Table Structure
     ReturnValue_ColorReferenceTable:Fort Color Reference Table Structure
     ReturnValue_TextReferenceTable:Fort Text Reference Table Structure
    ).ReturnValue_ImageReferenceTable
    Name=FTUE
   )
   > n4 = Branch(Condition=@n3.ReturnValue)
   n5 = Get FTUEImageLoader() -> (FTUEImageLoader:Fort Lazy Image Object Reference)
   > SetBrushFromImageReference!(
    execute
    self:Fort Lazy Image Object Reference=@n5
    ImageReference:"Fort Image Reference Structure (by ref)"=@n3
    bMatchSize:bool=false
    DownloadPriority:ECMSDownloadPriority Enum=Normal
   ) -> (then)
   > n6 = SetDialogImage!(
    execute
    self:UI Kit Dialog Viewmodel Object Reference=@n2
    InImage:Texture 2D Object Reference=GetBrushResourceAsTexture2D!(Brush:"Slate Brush Structure (by ref)"=Get Brush!(self:Image Object Reference=@n5) -> (Brush:Slate Brush Structure)) -> (ReturnValue:Texture 2D Object Reference)
   ) -> (then)
   n7 = Create Widget!(
    execute=[@n6.then, @n4.else]
    Class=/SeasonPass/Widgets/Reboot/WBP_Dialog_BattlePassFTUE.WBP_Dialog_BattlePassFTUE_C
    OwningPlayer
    UIKitDialogViewModel:UI Kit Dialog Viewmodel Object Reference=@n2
   ) -> (then, ReturnValue:WBP Dialog Battle Pass FTUE Object Reference)
   > TryPushWidgetAsModalPanel(PushWidget=@n7)
   > n8 = SetUserSettingBool(Name=HasSeenPassFTUEModal, bValue=true)
  ]}
 UpdateActionbarBackplateVisibility {nodes [
   UpdateActionbarBackplateVisibility() {is=title K2Node_FunctionEntry}
   > SetHasPrompts!(
    execute
    self:WBP Action Bar Container Object Reference=Get WBP_ActionBarContainer(self=Get WBP_AthenaBottomBarWidget())
    HasPrompts:bool=OR Boolean(A=IsVisible(), B=IsVisible())
   ) -> (then)
  ]}
 UpdatePreferredPass {nodes [
   UpdatePreferredPass() {is=title K2Node_FunctionEntry}
   > n1 = Get!(execute) -> (then, else, SeasonPassSystemVM:Season Pass System VM Object Reference)
   > n2 = For Each Loop(Array:Array of Season Pass VM Object References=Get SeasonPasses!(self:Season Pass System VM Object Reference=@n1) -> (SeasonPasses:Array of Season Pass VM Object References)) -> (Array Element:Season Pass VM Object Reference)
   Branch(execute=@n2.Completed, Condition="integer > integer"(A=Get PassesOwned() -> (PassesOwned:int=0), B=1))
   > Set PreferredPass(PreferredPass:Season Pass VM Object Reference) -> (Output_Get:Season Pass VM Object Reference)
   Branch(execute=@n2.LoopBody, Condition=Get bDoesOwnsPass(self=@n2))
   > Set PreferredPass(PreferredPass=@n2)
   > Set PassesOwned(PassesOwned:int=int + int(A=Get PassesOwned() -> (PassesOwned=0))) -> (Output_Get:int=0)
  ]}
 UpdatePreviewWindowPadding {nodes [
   n1 = UpdatePreviewWindowPadding!() -> (then, ItemDefinition:Fort Account Item Definition Object Reference) {is=title K2Node_FunctionEntry}
   > n2 = Cast To AthenaCosmeticItemDefinition!(execute, Object=@n1) -> (then, CastFailed, AsAthena Cosmetic Item Definition:Athena Cosmetic Item Definition Object Reference) {is=title K2Node_DynamicCast}
   > n3 = Branch(Condition="Equal (Enum)"(
     A:EFortItemType Enum=GetItemType!(ItemDefinition:Item Definition Base Object Reference=@n2) -> (ReturnValue:EFortItemType Enum=WorldItem)
     B:EFortItemType Enum=CosmeticMimosa
    ))
   n4 = SlotAsHorizontalBoxSlot!(Widget:Widget Object Reference=Get SizeBox_PreviewWindow() -> (SizeBox_PreviewWindow:Size Box Object Reference)) -> (ReturnValue:Horizontal Box Slot Object Reference)
   n5 = Get!(self:Horizontal Box Slot Object Reference=@n4) -> (Padding_Left:float=~, Padding_Top:float=~, Padding_Right:float=~, Padding_Bottom:float=~)
   > SetPadding!(
    execute
    self:Horizontal Box Slot Object Reference=@n4
    InPadding_Left:float=@n5
    InPadding_Top:float=@n5.Padding_Top
    InPadding_Right:float=@n5.Padding_Right
    InPadding_Bottom:float=-110
   ) -> (then)
   SetPadding(execute=@n3.else, self=@n4, InPadding_Left=@n5, InPadding_Top=@n5.Padding_Top, InPadding_Right=@n5.Padding_Right, InPadding_Bottom=-510)
  ]}
 UpdateUpsellMusic {nodes [UpdateUpsellMusic!() -> (then, ParamName:Name) {is=title K2Node_FunctionEntry}]}
 VM_HandleThemeData {nodes [
   n1 = VM_HandleThemeData!() -> (then, PassTheme:Season Pass Theme Data Structure) {is=title K2Node_FunctionEntry}
   n2 = TryLookupImage(
    Table=Break Season Pass Theme Data!(SeasonPassThemeData=@n1) -> (
     TitleColor:Linear Color Structure
     TitleMaterial:Soft Object Reference
     TitleOutline:Font Outline Settings Structure
     SeasonColor:Linear Color Structure
     ProgressionColor:Linear Color Structure
     OverlayColor:Linear Color Structure
     VaultBackgroundMaterial
     GiftboxVaultBackgroundMaterial
     UIBackgroundMaterial:Soft Object Reference
     UpsellBackgroundMaterial:Soft Object Reference
     CrewCategoryBackgroundMaterial:Soft Object Reference
     CustomLogo
     CustomSeasonTitle
     CustomSeasonTitleColor
     Upsell2BackgrounMaterial
     ImageReferenceTable
     ColorReferenceTable
     TextReferenceTable
    ).ImageReferenceTable
    Name=Overlay
   )
   > n3 = Branch(Condition=@n2.ReturnValue)
   n4 = Get ThemedOverlay() -> (ThemedOverlay:Fort Lazy Image Object Reference)
   > n5 = SetBrushFromImageReference(self=@n4, ImageReference=@n2, bMatchSize=false, DownloadPriority=Normal)
   Set Visibility(
    execute=[@n5.then, @n3.else]
    Visibility=Select(Option 0:ESlateVisibility Enum=Collapsed, Option 1:ESlateVisibility Enum=SelfHitTestInvisible, Index=@n2.ReturnValue) -> (ReturnValue:ESlateVisibility Enum)
    self=@n4
   )
  ]}
 VM_SelectedCategoryChanged {nodes [
   n1 = VM_SelectedCategoryChanged!() -> (then, SelectedCategory:Object Reference) {is=title K2Node_FunctionEntry}
   > n2 = Is Valid(InputObject=@n1)
   SetScalarParameterValue(execute=@n2.Is Not Valid, Collection=&MPC_UI_Pass_Properties, ParameterName=ShowCharactersSet, ParameterValue=1)
   SetScalarParameterValue(execute=@n2.Is Valid, Collection=&MPC_UI_Pass_Properties, ParameterName=ShowCharactersSet, ParameterValue=0)
  ]}
]
variables [
 BattlePassData {type=SeasonPassVM, default=null}
 CachedTabIndex {type=integer}
 ChatEventListener {type=AsyncAction_StartListeningToEvent, default=null}
 ChatEventRef {type=AsyncAction_StartListeningToEvent, default=null}
 CrewItemPreviewPadding {type=Margin, default=<ltrb 810 0 0 0>}
 CrewLegalText {type=string, default=""}
 CrewScreenAppear {type=SoundBase, default=null}
 CrewSelectedButton {type=WBP_CrewPassButton_C, default=null}
 CrewSelectedPass {type=SeasonPassVM, default=null}
 ExternalLaunchPass {type=SeasonPassVM, default=null}
 ExternalLaunchReward {type=SeasonPassRewardVM, default=null}
 IsActivatingFromCrew {type=boolean}
 IsActivatingFromSubscreen {type=boolean}
 IsBackgroundLoaded {type=boolean}
 IsCMSTextureLoaded {type=boolean}
 IsVerticalTransitioning {type=boolean}
 OnExternalNavigationDeactivated {}
 PageIndex {type=integer}
 PendingCMSTexture {type=Texture, default=null}
 PendingDirection {type=EVaultWorldTransitionDirection, default=Forward}
 PendingFocalPointValue {type=number}
 PendingIsHorizontalOrientation {type=boolean}
 PendingLoadedBackground {type=Object, default=null}
 PlayIntroAnimNextScrollUp {type=boolean}
 PreferredPass {type=SeasonPassVM, default=null}
 RequestedBackgroundSoftPath {type=SoftObjectPath, default=""}
 ScreenInBattlePass {type=AudioComponent, default=null}
 Season Pass Header {type=WBP_SeasonPass_Header_C, default=null}
 SeasonPassScreenAppear {type=SoundBase, default=null}
 SelectedPassChangedInSubscreen {type=boolean}
]
widgets {
 info {
  inheritedWidgetCount=0
  namedSlotCount=1
  parentClass=&CommonActivatableWidget
  rootWidgetClass=/Script/FortniteUI.FortSwipePanel
  widgetCount=34
 }
 widgets [
  SwipePanel {
   widgetClassPath=&FortSwipePanel
   bIsVariable=true
   widget=SwipePanel
   properties {bConsumePointerInput=false, thresholdUnit=Inches, swipeThreshold=<0 0.25>}
  }
  Overlay_0 {widgetClassPath=&Overlay, parent=SwipePanel, widget=Overlay_0, slot {class=FortSwipePanelSlot}}
  ThemedOverlay {
   widgetClassPath=&FortLazyImage
   bIsVariable=true
   parent=Overlay_0
   widget=ThemedOverlay
   properties {
    visibility=Collapsed
    brush {
     bIsDynamicallyLoaded=false
     drawAs=Image
     imageType=NoImage
     mirroring=NoMirror
     resourceName=None
     resourceObject=null
     tiling=NoTile
     imageSize=<32 700>
     margin=<ltrb 0 0 0 0>
     outlineSettings {
      bUseBrushTransparency=false
      roundingType=HalfHeightRadius
      width=0
      color {colorUseRule=UseColor_Specified, specifiedColor=<rgba 0 0 0 0>}
      cornerRadii=<0 0 0 0>
     }
     tintColor {colorUseRule=UseColor_Specified, specifiedColor=<rgba 1 1 1 1>}
     uVRegion {bIsValid=false, max=<0 0>, min=<0 0>}
    }
   }
   slot {class=OverlaySlot, properties {horizontalAlignment=HAlign_Fill, verticalAlignment=VAlign_Fill}}
  }
  EpicCMSImage_BackgroundLoaderHelper {
   widgetClassPath=/Script/EpicCMSUIFramework.EpicCMSImage
   bIsVariable=true
   parent=Overlay_0
   widget=EpicCMSImage_BackgroundLoaderHelper
   properties {visibility=Collapsed}
   slot {class=OverlaySlot}
  }
  WBP_CaptureForPostBufferUpdate {
   widgetClassPath=&WBP_CaptureForPostBufferUpdate
   bIsVariable=true
   parent=Overlay_0
   slot {class=OverlaySlot, properties {horizontalAlignment=HAlign_Fill, verticalAlignment=VAlign_Fill}}
  }
  SafeZone {
   widgetClassPath=&SafeZone
   bIsVariable=true
   parent=Overlay_0
   widget=SafeZone
   slot {class=OverlaySlot, properties {horizontalAlignment=HAlign_Fill, verticalAlignment=VAlign_Fill}}
  }
  Overlay_1 {widgetClassPath=&Overlay, parent=SafeZone, widget=Overlay_1, slot {class=SafeZoneSlot}}
  SB_Content {
   widgetClassPath=&SizeBox
   bIsVariable=true
   parent=Overlay_1
   widget=SB_Content
   properties {bOverride_WidthOverride=true, widthOverride=1920}
   slot {class=OverlaySlot, properties {horizontalAlignment=HAlign_Fill, verticalAlignment=VAlign_Fill}}
  }
  OverlayMain {
   widgetClassPath=&Overlay
   bIsVariable=true
   parent=SB_Content
   widget=OverlayMain
   slot {class=SizeBoxSlot, properties {padding=<ltrb 80 80 80 32>}}
  }
  SeasonPassFortItemPreviewWidget {
   widgetClassPath=/SeasonPassHub/Widgets/SeasonPassFortItemPreviewWidget
   bIsVariable=true
   parent=OverlayMain
   widget=SeasonPassFortItemPreviewWidget
   properties {
    bCanDisplayUserAccessories=false
    bPreviewActionsEnabled=false
    bSupportsActivationFocus=false
    itemViewObject=&self:WidgetTree.SeasonPassFortItemPreviewWidget.ItemViewObject
    visibility=Visible
    previewSceneChangerTag {tagName=UI.ItemPreview.SceneChanger.MPItemShop}
    screenTag {tagName=UI.ItemPreview.Screen.BattlePass.Landing}
   }
   slot {class=OverlaySlot, properties {horizontalAlignment=HAlign_Fill, verticalAlignment=VAlign_Fill, padding=<ltrb 0 0 0 48>}}
  }
  Grid_Main {
   widgetClassPath=&GridPanel
   bIsVariable=true
   parent=OverlayMain
   widget=Grid_Main
   properties {flowDirectionPreference=Culture, columnFill [1], rowFill [1]}
   slot {class=OverlaySlot, properties {horizontalAlignment=HAlign_Fill, verticalAlignment=VAlign_Fill}}
  }
  Widget_CameraFrame {
   widgetClassPath=&SizeBox
   bIsVariable=true
   parent=Grid_Main
   widget=Widget_CameraFrame
   slot {class=GridSlot, properties {columnSpan=1, rowSpan=1}}
  }
  HorizontalBox_0 {widgetClassPath=&HorizontalBox, parent=Widget_CameraFrame, widget=HorizontalBox_0, slot {class=SizeBoxSlot}}
  SizeBox_PreviewWindow {
   widgetClassPath=&SizeBox
   bIsVariable=true
   parent=HorizontalBox_0
   widget=SizeBox_PreviewWindow
   properties {bOverride_HeightOverride=true, bOverride_WidthOverride=true, heightOverride=960, widthOverride=1744}
   slot {class=HorizontalBoxSlot}
  }
  FortItemPreviewBackground {
   widgetClassPath=/Script/FortniteUI.FortItemPreviewBackground
   bIsVariable=true
   parent=HorizontalBox_0
   widget=FortItemPreviewBackground
   properties {previewSceneChangerTag {tagName=UI.ItemPreview.SceneChanger.MPItemShop}}
   slot {class=HorizontalBoxSlot}
  }
  BattlePassNavWrapper {
   widgetClassPath=&Overlay
   bIsVariable=true
   parent=Grid_Main
   widget=BattlePassNavWrapper
   properties {navigation=&self:WidgetTree.BattlePassNavWrapper.WidgetNavigation_0, visibility=Collapsed}
   slot {class=GridSlot, properties {columnSpan=1, rowSpan=1}}
  }
  BattlePassPage {
   widgetClassPath=&/WBP_SeasonPassPage
   bIsVariable=true
   parent=BattlePassNavWrapper
   widget=BattlePassPage
   slot {class=OverlaySlot, properties {horizontalAlignment=HAlign_Fill, verticalAlignment=VAlign_Fill}}
  }
  CrewNavWrapper {
   widgetClassPath=&Overlay
   bIsVariable=true
   parent=Grid_Main
   widget=CrewNavWrapper
   properties {navigation=&self:WidgetTree.CrewNavWrapper.WidgetNavigation_0}
   slot {class=GridSlot, properties {columnSpan=1, rowSpan=1}}
  }
  CrewLandingPage {
   widgetClassPath=&/WBP_CrewLandingPage
   bIsVariable=true
   parent=CrewNavWrapper
   widget=CrewLandingPage
   properties {bSetVisibilityOnActivated=true, bSetVisibilityOnDeactivated=true}
   slot {class=OverlaySlot, properties {horizontalAlignment=HAlign_Fill, verticalAlignment=VAlign_Fill}}
  }
  TopRight {
   widgetClassPath=&VerticalBox
   bIsVariable=true
   parent=Grid_Main
   widget=TopRight
   slot {class=GridSlot, properties {columnSpan=1, rowSpan=1, verticalAlignment=VAlign_Top, padding=<ltrb 0 60 0 0>}}
  }
  ProgressStatus {
   widgetClassPath=&VerticalBox
   bIsVariable=true
   parent=TopRight
   widget=ProgressStatus
   slot {class=VerticalBoxSlot, properties {horizontalAlignment=HAlign_Right}}
  }
  SeasonPass_Rewards {
   widgetClassPath=/SeasonPass/Widgets/Reboot/WBP_SeasonPass_Rewards
   bIsVariable=true
   parent=ProgressStatus
   widget=SeasonPass_Rewards
   slot {class=VerticalBoxSlot}
  }
  LevelRequirements {
   widgetClassPath=&/WBP_SeasonPass_LevelRequirements
   bIsVariable=true
   parent=TopRight
   widget=LevelRequirements
   slot {class=VerticalBoxSlot, properties {horizontalAlignment=HAlign_Right, padding=<ltrb 0 12 0 0>}}
  }
  RemainingSetsToUnlock {
   widgetClassPath=/SeasonPass/Widgets/Reboot/WBP_SeasonPass_RemainingSetsToUnlock
   bIsVariable=true
   parent=TopRight
   widget=RemainingSetsToUnlock
   slot {class=VerticalBoxSlot, properties {horizontalAlignment=HAlign_Right, verticalAlignment=VAlign_Center, padding=<ltrb 0 8 0 0>}}
  }
  RivalryButton {
   widgetClassPath=&/WBP_SeasonPass_RivalryBillboard
   bIsVariable=true
   parent=TopRight
   widget=RivalryButton
   slot {class=VerticalBoxSlot, properties {horizontalAlignment=HAlign_Right, padding=<ltrb 0 20 0 0>}}
  }
  LegalandActionBar {
   widgetClassPath=&HorizontalBox
   bIsVariable=true
   parent=Grid_Main
   widget=LegalandActionBar
   slot {class=GridSlot, properties {verticalAlignment=VAlign_Bottom}}
  }
  LegalText {
   widgetClassPath=&CommonTextBlock
   bIsVariable=true
   parent=LegalandActionBar
   widget=LegalText
   properties {
    bIsScrollingEnabled=false
    text=Legal text goes here
    wrapTextAt=950
    colorAndOpacity {colorUseRule=UseColor_Specified, specifiedColor=<rgba 1 1 1 0.6>}
    font {
     bForceMonospaced=false
     bMaterialIsStencil=false
     fontMaterial=null
     fontObject=&HeadingNow
     letterSpacing=0
     monospacedWidth=1
     size=15.75
     skewAmount=0
     typefaceFontName=74 Regular
     outlineSettings {
      bApplyOutlineToDropShadows=false
      bMiteredCorners=false
      bSeparateFillAlpha=false
      outlineMaterial=null
      outlineSize=0
      outlineColor=<rgba 0 0 0 1>
     }
    }
    shadowOffset=<0 0>
   }
   slot {class=HorizontalBoxSlot, properties {verticalAlignment=VAlign_Center, size {sizeRule=Fill, value=1}}}
  }
  Spacer_329 {
   widgetClassPath=&Spacer
   parent=LegalandActionBar
   widget=Spacer_329
   slot {class=HorizontalBoxSlot, properties {verticalAlignment=VAlign_Center, size {sizeRule=Fill, value=1}}}
  }
  Spacer {
   widgetClassPath=&Spacer
   parent=LegalandActionBar
   widget=Spacer
   slot {class=HorizontalBoxSlot, properties {verticalAlignment=VAlign_Center, size {sizeRule=Fill, value=1}}}
  }
  PageNav_Top {
   widgetClassPath=&/WBP_SeasonPassPageNav
   bIsVariable=true
   parent=OverlayMain
   widget=PageNav_Top
   slot {class=OverlaySlot, properties {horizontalAlignment=HAlign_Center, padding=<ltrb 0 60 0 0>}}
  }
  PageNav_Bottom {
   widgetClassPath=&WBP_SeasonPassPageNav
   bIsVariable=true
   parent=OverlayMain
   widget=PageNav_Bottom
   properties {renderTransform {angle=0, scale=<1 1>, shear=<0 0>, translation=<0 20>}}
   slot {class=OverlaySlot, properties {horizontalAlignment=HAlign_Center, verticalAlignment=VAlign_Bottom}}
  }
  WBP_AthenaBottomBarWidget {
   widgetClassPath=/Game/UI/Frontend/BottomBar/WBP_AthenaBottomBarWidget
   bIsVariable=true
   parent=Overlay_0
   widget=WBP_AthenaBottomBarWidget
   slot {class=OverlaySlot, properties {horizontalAlignment=HAlign_Right, verticalAlignment=VAlign_Bottom}}
  }
  WBP_AthenaBottomBar_WithNotifications {
   widgetClassPath=/Game/UI/Frontend/BottomBar/WBP_AthenaBottomBar_WithNotifications
   bIsVariable=true
   namedSlotHost=WBP_AthenaBottomBarWidget
   widget=WBP_AthenaBottomBar_WithNotifications
  }
  FTUEImageLoader {
   widgetClassPath=&FortLazyImage
   bIsVariable=true
   parent=Overlay_0
   widget=FTUEImageLoader
   properties {renderOpacity=0}
   slot {class=OverlaySlot}
  }
 ]
}
