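# SeasonPassButton is a season pass UI button widget derived from WBP_UIKit_ButtonGeneric_Base, with no widgets of its own.
# On Construct in a mobile game, it collapses the text block, resizes the icon and rounds the corners of its enhanced tab background.
# On Enabled, On Disabled and a ChangeEnabledStyle custom event switch the whole widget between full and reduced opacity.
# Its ShowText animation has no bindings.
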WidgetBlueprint /SeasonPass/Widgets/SeasonPassButton
animations [ShowText {binding_count=0, length_seconds=0.1787, object=&self:ShowText, sequence {bindings []}}]
graphs [EventGraph {nodes [
   n1 = ChangeEnabledStyle!() -> (OutputDelegate, then, Enabled:bool=~) {is=title K2Node_CustomEvent}
   > n2 = Branch(Condition=@n1.Enabled)
   SetColorAndOpacity(execute=[@n5.then, @n2.then], self:User Widget Object Reference, InColorAndOpacity:Linear Color Structure=(R=1,G=1,B=1,A=1))
   SetColorAndOpacity(execute=[@n4.then, @n2.else], InColorAndOpacity=(R=1,G=1,B=1,A=0.4))
   Event Construct()
   > Parent: Construct()
   > Branch(Condition=IsMobileGame())
   > SetVisibility(self=Get Block_Text() -> (Block_Text:WBP UIKit Block Text Object Reference), InVisibility=Collapsed)
   > SetIconSize!(execute, self:WBP UIKit Button Generic Base Object Reference, Icon:Vector 2D Structure=(X=64,Y=64)) -> (then)
   > n3 = Cast To WBP_UIKit_BlockBackground_Tab_Enhanced!(execute, Object=Get Block_Background() -> (Block_Background:WBP UIKit Block Background Base Object Reference)) -> (then, CastFailed, AsWBP UIKit Block Background Tab Enhanced:WBP UIKit Block Background Tab Enhanced Object Reference) {is=title K2Node_DynamicCast}
   > SetCornerRadiuses!(execute, self:WBP UIKit Block Background Tab Enhanced Object Reference=@n3, CornerRadius:float=20, RadiusInPixels:bool=true) -> (then)
   n4 = Event On Disabled() {is=title K2Node_Event}
   n5 = Event On Enabled() {is=title K2Node_Event}
  ]}]
variables []
widgets {
 info {
  inheritedWidgetCount=0
  parentClass=/FortUILibrary/Systems/ModularButton/WBP_UIKit_ButtonGeneric_Base.WBP_UIKit_ButtonGeneric_Base_C
  rootWidgetClass=null
  widgetCount=0
 }
 widgets []
}
