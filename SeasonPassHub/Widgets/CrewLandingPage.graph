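# Activatable landing page for the Crew subscription in the Season Pass hub, with a header, horizontal pass tiles and subscription warnings.
# CrewVM benefit groups are filtered to BattlePass entries for the tile box, the first double-width, and clicks are re-broadcast through dispatchers.
# Mouse-and-keyboard input gets scroll arrows, while focus starts on the first tile for subscribers, otherwise the header.
# Mobile layout adjusts clipping and aspect-ratio-based padding.
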WidgetBlueprint /SeasonPassHub/Widgets/CrewLandingPage
animations [TilesIntro {
  binding_count=2
  length_seconds=0.35
  object=&self:TilesIntro
  sequence {bindings [Overlay_Passes {
     id=#1
     possessed_class=Overlay
     tracks [
      {
       type=MovieSceneFloatTrack
       display_name=Render Opacity
       sections [{type=MovieSceneFloatSection, end_frame=6, start_frame=-1, channels [None {type=MovieSceneScriptingFloatChannel, keys [[2,0],[5.0003,1]]}]}]
      }
      {
       type=MovieScene2DTransformTrack
       sections [{
         type=MovieScene2DTransformSection
         end_frame=6
         start_frame=-1
         channels [
          Translation.X {type=MovieSceneScriptingFloatChannel, const [0,2]}
          Translation.Y {type=MovieSceneScriptingFloatChannel, keys [[2,140],[5.0003,0]]}
          Angle {type=MovieSceneScriptingFloatChannel, const [0,1]}
          Scale.X {type=MovieSceneScriptingFloatChannel, const [1,1]}
          Scale.Y {type=MovieSceneScriptingFloatChannel, const [1,1]}
          Shear.X {type=MovieSceneScriptingFloatChannel, const [0,1]}
          Shear.Y {type=MovieSceneScriptingFloatChannel, const [0,1]}
         ]
        }]
      }
     ]
    }]}
 }]
graphs [
 BP_GetDesiredFocusTarget {nodes [
   Get Desired Focus Target()
   n1 = Get Crew_Header() -> (Crew_Header:Crew Header Object Reference)
   n2 = Get CrewVM() -> (CrewVM:Crew VM Object Reference)
   n3 = GetAllEntries!(self:Dynamic Entry Box Base Object Reference=Get Passes() -> (Passes:Dynamic Entry Box Object Reference)) -> (ReturnValue:Array of User Widget Object References)
   > Return Node!(
    execute
    ReturnValue:Widget Object Reference=Select(
     Option 0:Widget Object Reference=GetDesiredFocusTarget!(self:Common Activatable Widget Object Reference=@n1) -> (ReturnValue:Widget Object Reference)
     Option 1:Widget Object Reference="Get (a copy)"(Array:Array of User Widget Object References=@n3) -> (Output:User Widget Object Reference)
     Index=IsSubscribed!(self:Crew VM Object Reference=@n2) -> (ReturnValue:bool)
    ) -> (ReturnValue:Widget Object Reference)
   )
  ]}
 CrewPassButtonClick {nodes [CrewPassButtonClick!() -> (then, PassVM:Season Pass VM Object Reference) {is=title K2Node_FunctionEntry}]}
 CrewPassTileSelected {nodes [
   n1 = CrewPassTileSelected!() -> (then, PassButton:Crew Pass Button Object Reference) {is=title K2Node_FunctionEntry}
   > Set SelectedPassButton(SelectedPassButton:Crew Pass Button Object Reference=@n1) -> (Output_Get:Crew Pass Button Object Reference)
  ]}
 EnteredCrewSubscreen {nodes [EnteredCrewSubscreen() {is=title K2Node_FunctionEntry}]}
 EventGraph {nodes [
   "On Clicked (ScrollRight)"!() -> (OutputDelegate, then, Button:Common Button Base Object Reference) {is=title K2Node_ComponentBoundEvent}
   n1 = Get ScrollBox() -> (ScrollBox:Athena Scroll Box Object Reference)
   n2 = GetScrollOffsetOfEnd!(self:Scroll Box Object Reference=@n1) -> (ReturnValue:float)
   n3 = GetScrollOffset!(self:Scroll Box Object Reference=@n1) -> (ReturnValue:float)
   n4 = float + float(A=@n3, B=Get ButtonScrollAmount() -> (ButtonScrollAmount:double))
   n5 = Get ScrollBox()
   > SetScrollOffset!(execute, self:Scroll Box Object Reference=@n5, NewScrollOffset:float=SelectFloat(A=@n2, B=@n4, bPickA="float >= float"(A=@n4, B=@n2))) -> (then)
   > n6 = Branch(Condition="Equal (Float)"(A=@n3, B=@n2))
   > n7 = SetVisibility(self=Get ScrollRight() -> (ScrollRight:WBP UIKit Button Regular Object Reference), InVisibility=Collapsed)
   SetVisibility(
    execute=[@n7.then, @n6.else]
    self=Get ScrollLEft() -> (ScrollLEft:WBP UIKit Button Regular Object Reference)
    InVisibility=SelfHitTestInvisible
   )
   "On Clicked (ScrollLEft)"!() -> (OutputDelegate, then, Button:Common Button Base Object Reference) {is=title K2Node_ComponentBoundEvent}
   n8 = GetScrollOffset(self=Get ScrollBox())
   n9 = float - float(A=@n8, B=Get ButtonScrollAmount())
   > SetScrollOffset(self=@n5, NewScrollOffset=SelectFloat(A=0, B=@n9, bPickA="float <= float"(A=@n9) -> (ReturnValue=~)))
   > n10 = Branch(Condition="Equal (Float)"(A=@n8))
   > n11 = SetVisibility(self=Get ScrollLEft(), InVisibility=Collapsed)
   SetVisibility(execute=[@n11.then, @n10.else], self=Get ScrollRight(), InVisibility=SelfHitTestInvisible)
   n12 = CrewPassButtonClicked!() -> (OutputDelegate, then, SeasonPassVM:Season Pass VM Object Reference) {is=title K2Node_CustomEvent}
   > Call Crew Pass Button Click!(execute:Exec, self:Crew Landing Page Object Reference, PassVM:Season Pass VM Object Reference=@n12.SeasonPassVM) -> (then:Exec) {is=title K2Node_CallDelegate}
   OpeningTile() {is=title K2Node_CustomEvent}
   > PlayAnimation(InAnimation=Get TilesIntro() -> (TilesIntro:Widget Animation Object Reference))
   > Opening(self:Crew Header Object Reference=Get Crew_Header())
   OnViewCrewHeaderClicked() {is=title K2Node_CustomEvent}
   > Call Entered Crew Subscreen!(execute:Exec, self:Crew Landing Page Object Reference) -> (then:Exec) {is=title K2Node_CallDelegate}
   Event Construct()
   n13 = Get CommonInputSubsystem()
   > Bind Event to On Input Method Changed(self=@n13, Delegate=Create Event())
   > HandleInputChanged!(execute, self, Input:ECommonInputType Enum=GetCurrentInputType(self=@n13)) -> (then)
   > SetMobileStyle() {is=title K2Node_CallFunction}
   > Bind Event to Header Entered Subscreen(self:Crew Header Object Reference=Get Crew_Header(), Delegate=Create Event()) {is=title K2Node_AddDelegate}
   > SetStatusIndicator!(
    execute
    self:WBP UIKit Info Box Object Reference=Get MultipleSubError() -> (MultipleSubError:WBP UIKit Info Box Object Reference)
    Icon:Object Reference=/FortUILibrary/Textures/MaterialInstances/MI_UIKit_Icon_ExclamationMark
    StatusType:E_UI_StatusIndicator_Type Enum=NewEnumerator4
    Config:E_UI_StatusIndicator_Configuration Enum=NewEnumerator0
   ) -> (then)
   Event On Activated()
   > DelayUntilNextTick!(execute) -> (then)
   > DelayUntilNextTick()
   n14 = GetScrollOffsetOfEnd(self=Get ScrollBox())
   > SetVisibility(
    self=Get ScrollRight()
    InVisibility=Select(Option 0:ESlateVisibility Enum=Collapsed, Option 1:ESlateVisibility Enum=SelfHitTestInvisible, Index="float > float"(A=@n14)) -> (ReturnValue:ESlateVisibility Enum)
   )
  ]}
 HandleInputChanged {nodes [
   n1 = HandleInputChanged!() -> (then, Input:ECommonInputType Enum) {is=title K2Node_FunctionEntry}
   > n2 = Branch(Condition="Equal (Enum)"(A=@n1, B=MouseAndKeyboard))
   n3 = Get MKBScrollArrows() -> (MKBScrollArrows:Horizontal Box Object Reference)
   > SetVisibility(self=@n3, InVisibility=SelfHitTestInvisible)
   SetVisibility(execute=@n2.else, self=@n3, InVisibility=Collapsed)
  ]}
 HandleNewPasses {nodes [
   HandleNewPasses!() -> (then, Passes:Array of Season Pass VM Object References) {is=title K2Node_FunctionEntry}
   n1 = Get BenefitsGroups!(self:Crew VM Object Reference=Get CrewVM()) -> (BenefitsGroups:Array of Crew Benefits Group VM Object References)
   > Branch(Condition=Is Not Empty!(TargetArray:Array of Crew Benefits Group VM Object References=@n1) -> (ReturnValue:bool))
   > n2 = Set Benefits(Benefits:Array of Crew Benefits Group VM Object References=@n1) -> (Output_Get:Array of Crew Benefits Group VM Object References)
   > SetupPassesList!(execute, self, Benefits:Array of Crew Benefits Group VM Object References=@n2) -> (then)
  ]}
 NavigateToHeaderButton {nodes [
   NavigateToHeaderButton!() -> (then, Navigation:EUINavigation Enum) {is=title K2Node_FunctionEntry}
   > Return Node(ReturnValue=GetDesiredFocusTarget(self=Get Crew_Header()))
  ]}
 NavigateToPassTile {nodes [
   NavigateToPassTile!() -> (then, Navigation:EUINavigation Enum) {is=title K2Node_FunctionEntry}
   > n1 = Get!(execute:Exec) -> (then:Exec, else:Exec, SelectedPassButton:Crew Pass Button Object Reference)
   > Return Node(ReturnValue=@n1)
   n2 = Get Passes()
   Return Node(execute=@n1.else, ReturnValue="Get (a copy)"(Array=GetAllEntries(self=@n2)))
  ]}
 SetCrewSubsciptionPlatforms {nodes [
   n1 = SetCrewSubsciptionPlatforms!() -> (then, Subbed App Stores:Array of EAppStore Enums=DebugStore) {is=title K2Node_FunctionEntry}
   n2 = Length(TargetArray:Array of EAppStore Enums=@n1)
   > SetVisibility(self=Get MultipleSubError(), InVisibility=Select(Option 0=Collapsed, Option 1=SelfHitTestInvisible, Index="integer > integer"(A=@n2, B=1)))
  ]}
 SetMobileStyle {nodes [
   SetMobileStyle() {is=title K2Node_FunctionEntry}
   > Branch(Condition=IsMobileGame())
   > SetClipping!(execute, self:Widget Object Reference=Get ScrollBox(), InClipping:EWidgetClipping Enum=ClipToBounds) -> (then)
   n1 = GetViewportSize!() -> (ReturnValue_X:double, ReturnValue_Y:double)
   n2 = float / float(A=@n1, B=@n1.ReturnValue_Y)
   > SetPadding!(
    execute
    self:Vertical Box Slot Object Reference=SlotAsVerticalBoxSlot(Widget=Get Overlay_Passes() -> (Overlay_Passes:Overlay Object Reference))
    InPadding_Left:float=0
    InPadding_Top:float=SelectFloat(A=24, B=256, bPickA="float >= float"(A=@n2, B=1.34))
    InPadding_Right:float=0
    InPadding_Bottom:float=0
   ) -> (then)
   n3 = GetViewportSize()
   n4 = float / float(A=@n3, B=@n3.ReturnValue_Y)
   > SetPadding(
    self=SlotAsVerticalBoxSlot(Widget=Get SizeBox_Message() -> (SizeBox_Message:Size Box Object Reference))
    InPadding_Left=0
    InPadding_Top=SelectFloat(A=-80, B=24, bPickA="float >= float"(A=@n4, B=1.78))
    InPadding_Right=0
    InPadding_Bottom=0
   )
   > SetPadding(
    self:Scroll Box Slot Object Reference=SlotAsScrollBoxSlot!(Widget:Widget Object Reference=Get Passes()) -> (ReturnValue:Scroll Box Slot Object Reference)
    InPadding_Left=8
    InPadding_Top=8
    InPadding_Right=8
    InPadding_Bottom=8
   )
  ]}
 SetupPassesList {nodes [
   n1 = SetupPassesList!() -> (then, Benefits:Array of Crew Benefits Group VM Object References) {is=title K2Node_FunctionEntry}
   > Set FirstEntry(FirstEntry:bool=true) -> (Output_Get:bool)
   > Reset!(execute, self:Dynamic Entry Box Object Reference=Get Passes(), bDeleteWidgets:bool=true) -> (then)
   > n2 = For Each Loop(Array:Array of Crew Benefits Group VM Object References=@n1) -> (Array Element:Crew Benefits Group VM Object Reference)
   > n3 = For Each Loop(Array:Array of Crew Benefit VM Object References=Get Benefits!(self:Crew Benefits Group VM Object Reference=@n2) -> (Benefits:Array of Crew Benefit VM Object References)) -> (Array Element:Crew Benefit VM Object Reference)
   > Branch(Condition="Equal (Enum)"(
     A:ECrewTileType Enum=GetType!(self:Crew Benefit VM Object Reference=@n3) -> (ReturnValue:ECrewTileType Enum=None)
     B:ECrewTileType Enum=BattlePass
    ))
   > n4 = Create Entry!(execute, self:Dynamic Entry Box Object Reference=Get Passes()) -> (then, ReturnValue:User Widget Object Reference)
   > n5 = Cast To CrewPassButton!(execute, Object=@n4) -> (then, CastFailed, AsCrew Pass Button:Crew Pass Button Object Reference) {is=title K2Node_DynamicCast}
   > SetBenefit!(execute, self:Crew Pass Button Object Reference=@n5, Benefit:Crew Benefit VM Object Reference=@n3) -> (then)
   > Set Double Width!(execute, self:Crew Pass Button Object Reference=@n5, Is Double Width:bool=Get FirstEntry() -> (FirstEntry:bool)) -> (then) {is=title K2Node_CallFunction}
   > Set FirstEntry(FirstEntry=false)
   > Bind Event to Crew Pass Clicked(self:Crew Pass Button Object Reference=@n5, Delegate=Create Event()) {is=title K2Node_AddDelegate}
   > Bind Event to Crew Pass Focused(self:Crew Pass Button Object Reference=@n5, Delegate=Create Event()) {is=title K2Node_AddDelegate}
  ]}
]
variables [
 ButtonScrollAmount {type=number, default=1600}
 CrewPassButtonClick {}
 EnteredCrewSubscreen {}
 SelectedPassButton {type=CrewPassButton_C, default=null}
]
widgets {
 info {inheritedWidgetCount=0, parentClass=&CommonActivatableWidget, rootWidgetClass=&VerticalBox, widgetCount=15}
 widgets [
  VerticalBox_0 {widgetClassPath=&VerticalBox, widget=VerticalBox_0}
  Spacer_150 {
   widgetClassPath=&Spacer
   parent=VerticalBox_0
   widget=Spacer_150
   slot {class=VerticalBoxSlot, properties {size {sizeRule=Fill, value=1}}}
  }
  Crew_Header {widgetClassPath=&/Crew_Header, bIsVariable=true, parent=VerticalBox_0, widget=Crew_Header, slot {class=VerticalBoxSlot}}
  Spacer {
   widgetClassPath=&Spacer
   parent=VerticalBox_0
   widget=Spacer
   slot {class=VerticalBoxSlot, properties {size {sizeRule=Fill, value=1}}}
  }
  FortVisualAttachment_0 {
   widgetClassPath=&FortVisualAttachment
   parent=VerticalBox_0
   widget=FortVisualAttachment_0
   properties {contentAnchor=<1 1>}
   slot {class=VerticalBoxSlot, properties {horizontalAlignment=HAlign_Right, verticalAlignment=VAlign_Bottom}}
  }
  SizeBox_Message {
   widgetClassPath=&SizeBox
   bIsVariable=true
   parent=FortVisualAttachment_0
   widget=SizeBox_Message
   properties {bOverride_WidthOverride=true, widthOverride=592}
   slot {class=SizeBoxSlot}
  }
  VerticalBox_Errors {widgetClassPath=&VerticalBox, parent=SizeBox_Message, widget=VerticalBox_Errors, slot {class=SizeBoxSlot}}
  MultipleSubError {
   widgetClassPath=/FortUILibrary/UIKit/InfoBox/WBP_UIKit_InfoBox
   bIsVariable=true
   parent=VerticalBox_Errors
   widget=MultipleSubError
   properties {
    previewText="You are currently subscribed on multiple platforms. To avoid being billed multiple times, we recommend canceling extra subscriptions. See cancellation info."
    showIcon=false
    showStatusIndicator=true
    use Preview Text=true
    visibility=Collapsed
   }
   slot {class=VerticalBoxSlot}
  }
  CrewEntitlementMessage {
   widgetClassPath=/Crew/UI/CrewEntitlementMessage
   bIsVariable=true
   parent=VerticalBox_Errors
   widget=CrewEntitlementMessage
   properties {padding=<ltrb 0 24 0 0>}
   slot {class=VerticalBoxSlot}
  }
  Overlay_Passes {
   widgetClassPath=&Overlay
   bIsVariable=true
   parent=VerticalBox_0
   widget=Overlay_Passes
   properties {navigation=&self:WidgetTree.Overlay_Passes.WidgetNavigation_0}
   slot {class=VerticalBoxSlot, properties {padding=<ltrb 0 40 0 50>}}
  }
  ScrollBox {
   widgetClassPath=/Script/FortniteUI.AthenaScrollBox
   bIsVariable=true
   parent=Overlay_Passes
   widget=ScrollBox
   properties {
    bAllowRightClickDragScrolling=false
    clipping=Inherit
    consumeMouseWheel=Never
    orientation=Orient_Horizontal
    scrollBarVisibility=Collapsed
    widgetStyle {
     barThickness=9
     bottomShadowBrush {
      bIsDynamicallyLoaded=false
      drawAs=NoDrawType
      imageType=FullColor
      mirroring=NoMirror
      resourceName=../../../Engine/Content/Slate/Common/ScrollBoxShadowBottom.png
      resourceObject=null
      tiling=NoTile
      imageSize=<16 8>
      margin=<ltrb 0.5 0 0.5 1>
      outlineSettings {
       bUseBrushTransparency=false
       roundingType=HalfHeightRadius
       width=0
       color {colorUseRule=UseColor_Specified, specifiedColor=<rgba 0 0 0 0>}
       cornerRadii=<0 0 0 0>
      }
      tintColor {colorUseRule=UseColor_Specified, specifiedColor=<rgba 1 1 1 1>}
      uVRegion {bIsValid=false, max=<0 0>, min=<0 0>}
     }
     horizontalScrolledContentPadding=<ltrb 0 0 1 0>
     leftShadowBrush {
      bIsDynamicallyLoaded=false
      drawAs=NoDrawType
      imageType=FullColor
      mirroring=NoMirror
      resourceName=../../../Engine/Content/Slate/Common/ScrollBoxShadowLeft.png
      resourceObject=null
      tiling=NoTile
      imageSize=<8 16>
      margin=<ltrb 1 0.5 0 0.5>
      outlineSettings {
       bUseBrushTransparency=false
       roundingType=HalfHeightRadius
       width=0
       color {colorUseRule=UseColor_Specified, specifiedColor=<rgba 0 0 0 0>}
       cornerRadii=<0 0 0 0>
      }
      tintColor {colorUseRule=UseColor_Specified, specifiedColor=<rgba 1 1 1 1>}
      uVRegion {bIsValid=false, max=<0 0>, min=<0 0>}
     }
     rightShadowBrush {
      bIsDynamicallyLoaded=false
      drawAs=NoDrawType
      imageType=FullColor
      mirroring=NoMirror
      resourceName=../../../Engine/Content/Slate/Common/ScrollBoxShadowRight.png
      resourceObject=null
      tiling=NoTile
      imageSize=<8 16>
      margin=<ltrb 0 0.5 1 0.5>
      outlineSettings {
       bUseBrushTransparency=false
       roundingType=HalfHeightRadius
       width=0
       color {colorUseRule=UseColor_Specified, specifiedColor=<rgba 0 0 0 0>}
       cornerRadii=<0 0 0 0>
      }
      tintColor {colorUseRule=UseColor_Specified, specifiedColor=<rgba 1 1 1 1>}
      uVRegion {bIsValid=false, max=<0 0>, min=<0 0>}
     }
     topShadowBrush {
      bIsDynamicallyLoaded=false
      drawAs=NoDrawType
      imageType=FullColor
      mirroring=NoMirror
      resourceName=../../../Engine/Content/Slate/Common/ScrollBoxShadowTop.png
      resourceObject=null
      tiling=NoTile
      imageSize=<16 8>
      margin=<ltrb 0.5 1 0.5 0>
      outlineSettings {
       bUseBrushTransparency=false
       roundingType=HalfHeightRadius
       width=0
       color {colorUseRule=UseColor_Specified, specifiedColor=<rgba 0 0 0 0>}
       cornerRadii=<0 0 0 0>
      }
      tintColor {colorUseRule=UseColor_Specified, specifiedColor=<rgba 1 1 1 1>}
      uVRegion {bIsValid=false, max=<0 0>, min=<0 0>}
     }
     verticalScrolledContentPadding=<ltrb 0 0 0 1>
    }
   }
   slot {class=OverlaySlot, properties {horizontalAlignment=HAlign_Fill, verticalAlignment=VAlign_Bottom}}
  }
  Passes {
   widgetClassPath=&DynamicEntryBox
   bIsVariable=true
   parent=ScrollBox
   widget=Passes
   properties {
    entryVerticalAlignment=VAlign_Bottom
    entryWidgetClass=&/CrewPassButton.CrewPassButton_C
    navigation=&self:WidgetTree.Passes.WidgetNavigation_0
    entrySpacing=<24 0>
   }
   slot {class=ScrollBoxSlot}
  }
  MKBScrollArrows {
   widgetClassPath=&HorizontalBox
   bIsVariable=true
   parent=Overlay_Passes
   widget=MKBScrollArrows
   slot {class=OverlaySlot, properties {horizontalAlignment=HAlign_Fill, verticalAlignment=VAlign_Center, padding=<ltrb 25 0 25 0>}}
  }
  ScrollLEft {
   widgetClassPath=&WBP_UIKit_Button_Regular
   bIsVariable=true
   parent=MKBScrollArrows
   widget=ScrollLEft
   properties {iconMaterial=/Crew/UI/Materials/MI_UI_ScrollArrowLeft, text="", visibility=Collapsed}
   slot {class=HorizontalBoxSlot}
  }
  ScrollRight {
   widgetClassPath=&WBP_UIKit_Button_Regular
   bIsVariable=true
   parent=MKBScrollArrows
   widget=ScrollRight
   properties {iconMaterial=/Crew/UI/Materials/MI_UI_ScrollArrowRight, text="", visibility=Collapsed}
   slot {class=HorizontalBoxSlot, properties {horizontalAlignment=HAlign_Right, size {sizeRule=Fill, value=1}}}
  }
 ]
}
