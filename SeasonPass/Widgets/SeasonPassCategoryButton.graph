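# Season Pass category tile widget showing a category's title, background image, reward progress, claimable count and lock or time-lock status.
# Setter functions update reward counts and theme color, feeding UpdateProgressVisual, which switches to a completion highlight when owned equals total.
# An invisible hitbox button sets the selection view model's category on hover or click; click also pushes the SeasonPassCategory screen and logs analytics.
# Hover animates a background material parameter and focuses the outline.
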WidgetBlueprint /SeasonPass/Widgets/SeasonPassCategoryButton
animations [OnHover {
  binding_count=1
  length_seconds=0.2
  object=&self:OnHover
  sequence {bindings [LIBackground {
     id=#1
     possessed_class=FortLazyImage
     tracks [{
       type=MovieSceneWidgetMaterialTrack
       sections [{
         type=MovieSceneParameterSection
         end_frame=4
         start_frame=0
         channels [Hover Overlay {type=MovieSceneScriptingFloatChannel, keys [[0,0],[4,1]]}]
        }]
      }]
    }]}
 }]
graphs [
 EventGraph {nodes [
   "On Unhovered (Button_Hitbox)"!() -> (OutputDelegate, then, Button:Common Button Base Object Reference) {is=title K2Node_ComponentBoundEvent}
   n1 = Get Block_Outline() -> (Block_Outline:WBP UIKit Block Outline Object Reference)
   > RequestTransition!(execute, self:UIKit Block Object Reference=@n1, State:EUIKitBlockVisualState Enum=Unfocused) -> (then)
   n2 = Get OnHover() -> (OnHover:Widget Animation Object Reference)
   > PlayAnimation(InAnimation=@n2, PlayMode=Reverse)
   "On Hovered (Button_Hitbox)"!() -> (OutputDelegate, then, Button:Common Button Base Object Reference) {is=title K2Node_ComponentBoundEvent}
   > Set SelectedCategory!(
    execute
    SelectedCategory:Season Pass Category VM Object Reference=Get SeasonPassCategoryVM() -> (SeasonPassCategoryVM:Season Pass Category VM Object Reference)
    self:Season Pass Selection VM Object Reference=Get SeasonPassSelectionVM() -> (SeasonPassSelectionVM:Season Pass Selection VM Object Reference)
   ) -> (then, Output_Get:Season Pass Category VM Object Reference)
   > RequestTransition(self=@n1, State=Focused)
   > PlayAnimation(InAnimation=@n2)
   > PlaySound2D(Sound=Get HoveredSound() -> (HoveredSound:Sound Base Object Reference))
   "On Clicked (Button_Hitbox)"!() -> (OutputDelegate, then, Button:Common Button Base Object Reference) {is=title K2Node_ComponentBoundEvent}
   > PlaySound2D(Sound=Get PressedSound() -> (PressedSound:Sound Base Object Reference))
   > Set SelectedCategory(SelectedCategory=Get SeasonPassCategoryVM(), self=Get SeasonPassSelectionVM())
   > Call Season Pass Category Entered!(execute:Exec, self:Season Pass Category Button Object Reference) -> (then:Exec) {is=title K2Node_CallDelegate}
   > n3 = PushPrimaryContentClassAsync!(execute:Exec, WidgetClass:Common Activatable Widget Class Reference=&/SeasonPassCategory.SeasonPassCategory_C) -> (then:Exec, OnPrimaryContentPushed:Exec, ActivatableWidget:Common Activatable Widget Object Reference) {is=title K2Node_AsyncAction}
   n4 = GetCategoryPageTitle!(self:Season Pass Category VM Object Reference=Get SeasonPassCategoryVM()) -> (ReturnValue:Text)
   Fire Event Element Interacted(
    execute=@n3.OnPrimaryContentPushed
    Element=@n3
    AdditionalInfo="Add String (Non-Localized)"!(
     Info:"Flow Analytics Additional Info Structure (by ref)"=Make Flow Analytics Additional Info!() -> (FlowAnalyticsAdditionalInfo:Flow Analytics Additional Info Structure) {is=title K2Node_MakeStruct}
     Key:String=CategoryTitle
     Value:String=Get Source String from Text!(Text:Text=@n4) -> (ReturnValue:String) {is=title K2Node_CallFunction}
    ) -> (ReturnValue:Flow Analytics Additional Info Structure)
   )
   n5 = Set Double Width!() -> (OutputDelegate, then, Is Double Width:bool=~) {is=title K2Node_CustomEvent}
   > Set IsDoubleWidth(IsDoubleWidth:bool=@n5.Is Double Width) -> (Output_Get:bool)
   > SetWidthOverride(self=Get SizeBox() -> (SizeBox:Size Box Object Reference), InWidthOverride=SelectFloat(A=592, B=312, bPickA=@n5.Is Double Width))
  ]}
 LockStateChanged {nodes [
   n1 = LockStateChanged!() -> (then, bIsLocked:bool=~) {is=title K2Node_FunctionEntry}
   > SetLockedState!(execute, self, Is Locked:bool=Select(Option 0:bool=false, Option 1:bool=true, Index=@n1) -> (ReturnValue:bool)) -> (then)
  ]}
 SeasonPassCategoryEntered {nodes [SeasonPassCategoryEntered() {is=title K2Node_FunctionEntry}]}
 SetImage {nodes [
   n1 = SetImage!() -> (then, Image:Soft Object Reference) {is=title K2Node_FunctionEntry}
   > SetBrushFromLazyDisplayAsset!(
    execute
    self:Common Lazy Image Object Reference=Get LIBackground() -> (LIBackground:Fort Lazy Image Object Reference)
    LazyObject:"Soft Object Reference (by ref)"=@n1
    bMatchTextureSize:bool=false
   ) -> (then)
  ]}
 SetLockedState {nodes [
   SetLockedState!() -> (then, Is Locked:bool=~) {is=title K2Node_FunctionEntry}
   n1 = Get SeasonPassCategoryVM()
   > SetVisibility(
    self=Get Status_Timed() -> (Status_Timed:WBP UIKit Status Indicator Object Reference)
    InVisibility=Select(
     Option 0:ESlateVisibility Enum=Collapsed
     Option 1:ESlateVisibility Enum=Visible
     Index=Get bIsTimeLocked!(self:Season Pass Element Base VM Object Reference=@n1) -> (bIsTimeLocked:bool)
    ) -> (ReturnValue:ESlateVisibility Enum)
   )
   n2 = Get bIsTimeLocked(self=Get SeasonPassCategoryVM())
   n3 = Get SeasonPassCategoryVM()
   > SetVisibility(
    self=Get Status_LockedIcon() -> (Status_LockedIcon:Image Object Reference)
    InVisibility=Select(
     Option 0=Collapsed
     Option 1=Select(Option 0=Visible, Option 1=Collapsed, Index=@n2)
     Index=Get bIsLocked!(self:Season Pass Element Base VM Object Reference=@n3) -> (bIsLocked:bool)
    )
   )
   n4 = Get SeasonPassCategoryVM()
   > SetVisibility(
    self=Get ProgressHorizontalBox() -> (ProgressHorizontalBox:Horizontal Box Object Reference)
    InVisibility=Select(Option 0=Visible, Option 1=Collapsed, Index=Get bIsLocked(self=@n4))
   )
   > n5 = GetDynamicMaterial(self=Get LIBackground())
   n6 = Get SeasonPassCategoryVM()
   > SetScalarParameterValue(self=@n5, ParameterName=IsLocked, Value=SelectFloat(A=1, B=0, bPickA=Get bIsLocked(self=@n6)))
  ]}
 SetOwnedRewardsCount {nodes [
   n1 = SetOwnedRewardsCount!() -> (then, Owned:int) {is=title K2Node_FunctionEntry}
   > Set OwnedRewardsCount(OwnedRewardsCount:int=@n1) -> (Output_Get:int=0)
   > UpdateProgressVisual()
  ]}
 SetProgressionColor {nodes [
   n1 = SetProgressionColor!() -> (then, Color:Linear Color Structure) {is=title K2Node_FunctionEntry}
   > Set ProgressionColor(ProgressionColor:Linear Color Structure=@n1) -> (Output_Get:Linear Color Structure)
   > UpdateProgressVisual()
  ]}
 SetRewardsToClaim {nodes [
   n1 = SetRewardsToClaim!() -> (then, ToClaim:int) {is=title K2Node_FunctionEntry}
   n2 = OR Boolean(
    A=Get bIsLocked(self=Get SeasonPassCategoryVM())
    B="Equal (Integer)"(A=Get OwnedRewardsCount() -> (OwnedRewardsCount:int=0), B=Get TotalRewardsCount() -> (TotalRewardsCount:int=0))
   )
   > SetVisibility(
    self=Get O_Status_ToClaim() -> (O_Status_ToClaim:Overlay Object Reference)
    InVisibility=Select(Option 0=Collapsed, Option 1=Select(Option 0=Visible, Option 1=Collapsed, Index=@n2), Index="integer > integer"(A=@n1))
   )
   > SetText!(
    execute
    self:WBP UIKit Highlight Base Object Reference=Get Status_ToClaim() -> (Status_ToClaim:WBP UIKit Highlight Object Reference)
    Text:Text=Format Text!(Format="LOCTEXT(\"{num} to Claim\")", num:int=Get ToClaim() -> (ToClaim:int=0)) -> (Result)
   ) -> (then)
  ]}
 SetThemeData {nodes [
   n1 = SetThemeData!() -> (then, ThemeData:Season Pass Theme Data Structure) {is=title K2Node_FunctionEntry}
   > Set ProgressionColor(ProgressionColor=Break Season Pass Theme Data!(SeasonPassThemeData:"Season Pass Theme Data Structure (by ref)"=@n1) -> (
     ProgressionColor:Linear Color Structure
     VaultBackgroundMaterial:Soft Object Reference
     GiftboxVaultBackgroundMaterial:Soft Object Reference
     UIBackgroundMaterial:Soft Object Reference
     UpsellBackgroundMaterial:Soft Object Reference
     CrewCategoryBackgroundMaterial:Soft Object Reference
     CustomLogo:Soft Object Reference
     CustomSeasonTitle:Text
     CustomSeasonTitleColor:Linear Color Structure
     Upsell2BackgrounMaterial:Soft Object Reference
     ImageReferenceTable:Fort Image Reference Table Structure
     ColorReferenceTable:Fort Color Reference Table Structure
     TextReferenceTable:Fort Text Reference Table Structure
    ) {is=title K2Node_BreakStruct})
   > UpdateProgressVisual()
  ]}
 SetTitleSubtitle {nodes [
   SetTitleSubtitle!() -> (then, TextIn:Text) {is=title K2Node_FunctionEntry}
   > SetText!(
    execute
    self:Rich Text Block Object Reference=Get TitleSubtitle() -> (TitleSubtitle:Common Rich Text Block Object Reference)
    InText:"Text (by ref)"=Format Text!(
     Format="LOCTEXT(\"{Title}\\r\\n<text color=\\\"FFFFFFA6\\\" fontface=\\\"75 Medium\\\">{Subtitle}</>\")"
     Title:Text=Get TitleText() -> (TitleText:Text)
     Subtitle:Text=Get SubtitleText() -> (SubtitleText:Text)
    ) -> (Result)
   ) -> (then)
  ]}
 SetTotalRewardsCount {nodes [
   n1 = SetTotalRewardsCount!() -> (then, Total:int) {is=title K2Node_FunctionEntry}
   > Set TotalRewardsCount(TotalRewardsCount:int=@n1) -> (Output_Get:int=0)
   > UpdateProgressVisual()
  ]}
 UpdateProgressVisual {nodes [
   UpdateProgressVisual() {is=title K2Node_FunctionEntry}
   > n1 = Sequence!(execute) -> (then_0, then_1, then_2:Exec)
   Set Text(
    execute=@n1.then_1
    Text="To Text (Integer)"(Value=Get OwnedRewardsCount() -> (OwnedRewardsCount=0))
    self=Get ProgressTextX() -> (ProgressTextX:Common Text Object Reference)
   )
   n2 = "To String (Integer)"(InInt=Get TotalRewardsCount() -> (TotalRewardsCount=0))
   > Set Text(Text="To Text (String)"(InString=Append(A=/, B=@n2)), self=Get ProgressTextOf() -> (ProgressTextOf:Common Text Object Reference))
   > SetText!(
    execute
    self:WBP UIKit Highlight Base Object Reference=Get Highlight_ProgressComplete() -> (Highlight_ProgressComplete:WBP UIKit Highlight Object Reference)
    Text:Text=Format Text!(Format="LOCTEXT(\"{X}/{X}\")", X:String=@n2) -> (Result)
   ) -> (then)
   n3 = "Equal (Integer)"(A=Get OwnedRewardsCount() -> (OwnedRewardsCount=0), B=Get TotalRewardsCount() -> (TotalRewardsCount=0))
   SetColorAndOpacity!(
    execute=@n1.then_2
    self=Get ProgressTextX()
    InColorAndOpacity_SpecifiedColor:Linear Color Structure=SelectColor(
     A=Get CompletionColor() -> (CompletionColor:Linear Color Structure)
     B=Get ProgressionColor() -> (ProgressionColor:Linear Color Structure)
     bPickA=@n3
    )
    InColorAndOpacity_ColorUseRule:ESlateColorStylingMode Enum
   ) -> (then)
   > SetColorAndOpacity(self=Get ProgressTextOf(), InColorAndOpacity_SpecifiedColor=SelectColor(A=Get CompletionColor(), B=(R=1,G=1,B=1,A=1), bPickA=@n3))
   > SetActiveWidgetIndex(
    self:Common Visibility Switcher Object Reference=Get ProgressVisibilitySwitcher() -> (ProgressVisibilitySwitcher:Common Visibility Switcher Object Reference)
    Index=SelectInt(A=1, B=0, bPickA=@n3)
   )
   n4 = Get TotalRewardsCount() -> (TotalRewardsCount=0)
   n5 = Get OwnedRewardsCount() -> (OwnedRewardsCount=0)
   SetProgress!(
    execute=@n1.then_0
    self:WBP UIKit Progress Bar Base Object Reference=Get ProgressBar() -> (ProgressBar:WBP UIKit Progress Bar Object Reference)
    Progress:double=float / float(A="To Float (Integer)"(InInt=@n5), B="To Float (Integer)"(InInt=@n4))
   ) -> (then)
   > SetFillColors!(
    execute
    self:WBP UIKit Progress Bar Base Object Reference=Get ProgressBar()
    ColorA:Linear Color Structure=Get ProgressionColor()
    ColorB:Linear Color Structure=(R=1,G=1,B=1,A=1)
    ColorC:Linear Color Structure=(R=1,G=1,B=1,A=1)
   ) -> (then)
  ]}
]
variables [
 CompletionColor {type=LinearColor, default=<rgba 0.0395 1 0.2307 1>}
 HoveredSound {type=SoundBase, default=null}
 IsDoubleWidth {type=boolean}
 OwnedRewardsCount {type=integer}
 PressedSound {type=SoundBase, default=null}
 ProgressionColor {type=LinearColor, default=<rgba 1 0 0.7096 1>}
 SeasonPassCategoryEntered {}
 SubtitleText {type=string, default=""}
 TitleText {type=string, default=""}
 TotalRewardsCount {type=integer}
]
widgets {
 info {inheritedWidgetCount=0, parentClass=&CommonActivatableWidget, rootWidgetClass=&SizeBox, widgetCount=19}
 widgets [
  SizeBox {
   widgetClassPath=&SizeBox
   bIsVariable=true
   widget=SizeBox
   properties {
    bOverride_MinDesiredHeight=true
    bOverride_WidthOverride=true
    maxDesiredWidth=312
    minDesiredHeight=192
    minDesiredWidth=312
    widthOverride=312
   }
  }
  Overlay_1 {widgetClassPath=&Overlay, parent=SizeBox, widget=Overlay_1, slot {class=SizeBoxSlot}}
  LIBackground {
   widgetClassPath=&FortLazyImage
   bIsVariable=true
   parent=Overlay_1
   widget=LIBackground
   properties {brush {
     bIsDynamicallyLoaded=false
     drawAs=Image
     imageType=FullColor
     mirroring=NoMirror
     resourceName=None
     resourceObject=/SeasonPass/Materials/M_UI_Nexus_CategoryButton
     tiling=NoTile
     imageSize=<32 32>
     margin=<ltrb 0 0 0 0>
     outlineSettings {
      bUseBrushTransparency=false
      roundingType=HalfHeightRadius
      width=0
      color {colorUseRule=UseColor_Specified, specifiedColor=<rgba 0 0 0 0>}
      cornerRadii=<0 0 0 0>
     }
     tintColor {colorUseRule=UseColor_Specified, specifiedColor=<rgba 1 1 1 1>}
     uVRegion {bIsValid=false, max=<0 0>, min=<0 0>}
    }}
   slot {class=OverlaySlot, properties {horizontalAlignment=HAlign_Fill, verticalAlignment=VAlign_Fill}}
  }
  Block_Outline {
   widgetClassPath=/FortUILibrary/UIKit/Buttons/Blocks/WBP_UIKit_Block_Outline
   bIsVariable=true
   parent=Overlay_1
   widget=Block_Outline
   properties {bottomLeftCornerRadius=30, bottomRightCornerRadius=30, radiusInPixels=true, topLeftCornerRadius=30, topRightCornerRadius=30}
   slot {class=OverlaySlot, properties {horizontalAlignment=HAlign_Fill, verticalAlignment=VAlign_Fill}}
  }
  VerticalBox_45 {
   widgetClassPath=&VerticalBox
   parent=Overlay_1
   widget=VerticalBox_45
   properties {visibility=HitTestInvisible}
   slot {class=OverlaySlot, properties {horizontalAlignment=HAlign_Fill, verticalAlignment=VAlign_Fill, padding=<ltrb 12 8 12 8>}}
  }
  Status {
   widgetClassPath=&Overlay
   parent=VerticalBox_45
   widget=Status
   slot {class=VerticalBoxSlot, properties {horizontalAlignment=HAlign_Left, verticalAlignment=VAlign_Top, padding=<ltrb 0 0 8 8>}}
  }
  Status_Timed {
   widgetClassPath=/FortUILibrary/UIKit/StatusIndicator/WBP_UIKit_StatusIndicator
   bIsVariable=true
   parent=Status
   widget=Status_Timed
   properties {type=Attention, icon=/FortUILibrary/Textures/MaterialInstances/MI_UIKit_Icon_UpdatePending, visibility=Collapsed}
   slot {class=OverlaySlot}
  }
  O_Status_ToClaim {
   widgetClassPath=&Overlay
   bIsVariable=true
   parent=Status
   widget=O_Status_ToClaim
   properties {visibility=Visible}
   slot {class=OverlaySlot}
  }
  Status_ToClaim {
   widgetClassPath=/FortUILibrary/UIKit/Highlight/WBP_UIKit_Highlight
   bIsVariable=true
   parent=O_Status_ToClaim
   widget=Status_ToClaim
   properties {highlightSize=Compact, highlightText=2 to Claim, highlightType=Attention}
   slot {class=OverlaySlot}
  }
  Status_LockedIcon {
   widgetClassPath=&Image
   bIsVariable=true
   parent=Status
   widget=Status_LockedIcon
   properties {brush {
     bIsDynamicallyLoaded=false
     drawAs=Image
     imageType=FullColor
     mirroring=NoMirror
     resourceName=None
     resourceObject=/SeasonPass/Materials/M_UI_Icon_Locked
     tiling=NoTile
     imageSize=<32 32>
     margin=<ltrb 0 0 0 0>
     outlineSettings {
      bUseBrushTransparency=false
      roundingType=HalfHeightRadius
      width=0
      color {colorUseRule=UseColor_Specified, specifiedColor=<rgba 0 0 0 0>}
      cornerRadii=<0 0 0 0>
     }
     tintColor {colorUseRule=UseColor_Specified, specifiedColor=<rgba 1 1 1 1>}
     uVRegion {bIsValid=false, max=<0 0>, min=<0 0>}
    }}
   slot {class=OverlaySlot, properties {verticalAlignment=VAlign_Fill}}
  }
  Spacer_114 {
   widgetClassPath=&Spacer
   parent=VerticalBox_45
   widget=Spacer_114
   slot {class=VerticalBoxSlot, properties {size {sizeRule=Fill, value=1}}}
  }
  TitleSubtitle {
   widgetClassPath=&CommonRichTextBlock
   bIsVariable=true
   parent=VerticalBox_45
   widget=TitleSubtitle
   properties {
    bOverrideDefaultStyle=true
    scrollStyle=&TextScrollStyle-Base_C
    text="Bonus\r\n<text color=\"FFFFFFA6\" fontface=\"75 Medium\">Rewards</>"
    textTransformPolicy=ToUpper
    defaultTextStyleOverride {
     overflowPolicy=Clip
     transformPolicy=ToUpper
     colorAndOpacity {colorUseRule=UseColor_Specified, specifiedColor=<rgba 1 1 1 1>}
     font {
      bForceMonospaced=false
      bMaterialIsStencil=false
      fontMaterial=null
      fontObject=&HeadingNow
      letterSpacing=10
      monospacedWidth=1
      size=20.25
      skewAmount=0
      typefaceFontName=86 Bold
      outlineSettings {
       bApplyOutlineToDropShadows=false
       bMiteredCorners=false
       bSeparateFillAlpha=false
       outlineMaterial=null
       outlineSize=0
       outlineColor=<rgba 0 0 0 1>
      }
     }
     highlightColor {colorUseRule=UseColor_Specified, specifiedColor=<rgba 0 0 0 1>}
     highlightShape {
      bIsDynamicallyLoaded=false
      drawAs=Image
      imageType=NoImage
      mirroring=NoMirror
      resourceName=None
      resourceObject=null
      tiling=NoTile
      imageSize=<32 32>
      margin=<ltrb 0 0 0 0>
      outlineSettings {
       bUseBrushTransparency=false
       roundingType=HalfHeightRadius
       width=0
       color {colorUseRule=UseColor_Specified, specifiedColor=<rgba 0 0 0 0>}
       cornerRadii=<0 0 0 0>
      }
      tintColor {colorUseRule=UseColor_Specified, specifiedColor=<rgba 1 1 1 1>}
      uVRegion {bIsValid=false, max=<0 0>, min=<0 0>}
     }
     selectedBackgroundColor {colorUseRule=UseColor_Foreground, specifiedColor=<rgba 1 0 1 1>}
     shadowColorAndOpacity=<rgba 0 0 0 1>
     shadowOffset=<0 0>
     strikeBrush {
      bIsDynamicallyLoaded=false
      drawAs=Image
      imageType=NoImage
      mirroring=NoMirror
      resourceName=None
      resourceObject=null
      tiling=NoTile
      imageSize=<32 32>
      margin=<ltrb 0 0 0 0>
      outlineSettings {
       bUseBrushTransparency=false
       roundingType=HalfHeightRadius
       width=0
       color {colorUseRule=UseColor_Specified, specifiedColor=<rgba 0 0 0 0>}
       cornerRadii=<0 0 0 0>
      }
      tintColor {colorUseRule=UseColor_Specified, specifiedColor=<rgba 1 1 1 1>}
      uVRegion {bIsValid=false, max=<0 0>, min=<0 0>}
     }
     underlineBrush {
      bIsDynamicallyLoaded=false
      drawAs=Image
      imageType=NoImage
      mirroring=NoMirror
      resourceName=None
      resourceObject=null
      tiling=NoTile
      imageSize=<32 32>
      margin=<ltrb 0 0 0 0>
      outlineSettings {
       bUseBrushTransparency=false
       roundingType=HalfHeightRadius
       width=0
       color {colorUseRule=UseColor_Specified, specifiedColor=<rgba 0 0 0 0>}
       cornerRadii=<0 0 0 0>
      }
      tintColor {colorUseRule=UseColor_Specified, specifiedColor=<rgba 1 1 1 1>}
      uVRegion {bIsValid=false, max=<0 0>, min=<0 0>}
     }
    }
   }
   slot {class=VerticalBoxSlot, properties {verticalAlignment=VAlign_Bottom, padding=<ltrb 0 0 0 3>}}
  }
  ProgressVisibilitySwitcher {
   widgetClassPath=&CommonVisibilitySwitcher
   bIsVariable=true
   parent=VerticalBox_45
   widget=ProgressVisibilitySwitcher
   slot {class=VerticalBoxSlot, properties {verticalAlignment=VAlign_Bottom, padding=<ltrb 0 0 0 3>}}
  }
  ProgressHorizontalBox {
   widgetClassPath=&HorizontalBox
   bIsVariable=true
   parent=ProgressVisibilitySwitcher
   widget=ProgressHorizontalBox
   slot {class=CommonVisibilitySwitcherSlot}
  }
  ProgressBar {
   widgetClassPath=/FortUILibrary/UIKit/ProgressBar/WBP_UIKit_ProgressBar
   bIsVariable=true
   parent=ProgressHorizontalBox
   widget=ProgressBar
   properties {
    height=16
    linearMaterial=/SeasonPass/Materials/ProgressBar/MI_UI_NexusProgerssBar
    outline_PX=6
    progressValue=0.5
    useCircular=false
    useOutline=true
    colorA=<rgba 1 0.6849 0 1>
    colorB=<rgba 1 0.9857 0.9484 1>
    emptyColor=<rgba 0.3573 0.3573 0.3573 0.305>
   }
   slot {class=HorizontalBoxSlot, properties {verticalAlignment=VAlign_Center, padding=<ltrb 0 0 16 0>, size {sizeRule=Fill, value=1}}}
  }
  ProgressTextX {
   widgetClassPath=&CommonTextBlock
   bIsVariable=true
   parent=ProgressHorizontalBox
   widget=ProgressTextX
   properties {
    text=0
    font {
     bForceMonospaced=false
     bMaterialIsStencil=false
     fontMaterial=null
     fontObject=&NotoSans
     letterSpacing=0
     monospacedWidth=1
     size=24
     skewAmount=0
     typefaceFontName=NotoSans-Regular
     outlineSettings {
      bApplyOutlineToDropShadows=false
      bMiteredCorners=false
      bSeparateFillAlpha=false
      outlineMaterial=null
      outlineSize=0
      outlineColor=<rgba 0 0 0 1>
     }
    }
    shadowOffset=<0 0>
   }
   slot {class=HorizontalBoxSlot}
  }
  ProgressTextOf {
   widgetClassPath=&CommonTextBlock
   bIsVariable=true
   parent=ProgressHorizontalBox
   widget=ProgressTextOf
   properties {
    text=/100
    font {
     bForceMonospaced=false
     bMaterialIsStencil=false
     fontMaterial=null
     fontObject=&NotoSans
     letterSpacing=0
     monospacedWidth=1
     size=24
     skewAmount=0
     typefaceFontName=NotoSans-Regular
     outlineSettings {
      bApplyOutlineToDropShadows=false
      bMiteredCorners=false
      bSeparateFillAlpha=false
      outlineMaterial=null
      outlineSize=0
      outlineColor=<rgba 0 0 0 1>
     }
    }
    shadowOffset=<0 0>
   }
   slot {class=HorizontalBoxSlot}
  }
  Highlight_ProgressComplete {
   widgetClassPath=&WBP_UIKit_Highlight
   bIsVariable=true
   parent=ProgressVisibilitySwitcher
   widget=Highlight_ProgressComplete
   properties {highlightSize=Compact, highlightText=100/100, highlightType=Social}
   slot {class=CommonVisibilitySwitcherSlot, properties {horizontalAlignment=HAlign_Left, verticalAlignment=VAlign_Center}}
  }
  Button_Hitbox {
   widgetClassPath=&WBP_UIKit_Button_Generic
   bIsVariable=true
   parent=Overlay_1
   widget=Button_Hitbox
   properties {
    bHideInputAction=true
    bSimulateHoverOnTouchInput=false
    contentHorizontalAlignment=HAlign_Fill
    contentVerticalAlignment=VAlign_Fill
    secondText=""
    touchMethod=PreciseTap
    useBackgroundBlock=false
    useIconBlock=false
    useTextBlock=false
    visibility=Visible
   }
   slot {class=OverlaySlot, properties {horizontalAlignment=HAlign_Fill, verticalAlignment=VAlign_Fill}}
  }
 ]
}
