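# Season pass header widget showing badge, season and pass titles, purchase, buy-rewards and gift buttons, plus Crew access text.
# Theme data restyles titles; ownership switches purchase to buy-rewards; subscription state sets Crew text.
# Buttons push purchase content, open gifting, or show a buy-levels (BR pass) or buy-rewards modal, firing dispatchers.
# An Intro animation staggers elements in, and compact UI adjusts sizing and padding by aspect ratio and safe zone.
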WidgetBlueprint /SeasonPassHub/Widgets/SeasonPass_Header
animations [Intro {
  binding_count=6
  length_seconds=1
  object=&self:Intro
  sequence {bindings [
    GIftPass_Button {
     id=#1
     possessed_class=WBP_UIKit_Button_Regular_C
     tracks [{
       type=MovieSceneFloatTrack
       display_name=Render Opacity
       sections [{
         type=MovieSceneFloatSection
         end_frame=20
         start_frame=0
         channels [None {type=MovieSceneScriptingFloatChannel, keys [[0,0],[10,0],[15.0003,1],[20,1]]}]
        }]
      }]
    }
    Text_Season {
     id=#2
     possessed_class=CommonTextBlock
     tracks [
      {
       type=MovieScene2DTransformTrack
       sections [{
         type=MovieScene2DTransformSection
         end_frame=15
         start_frame=0
         channels [
          Translation.X {type=MovieSceneScriptingFloatChannel, const [0,2]}
          Translation.Y {type=MovieSceneScriptingFloatChannel, keys [[3,80],[10,0]]}
          Angle {type=MovieSceneScriptingFloatChannel}
          Scale.X {type=MovieSceneScriptingFloatChannel}
          Scale.Y {type=MovieSceneScriptingFloatChannel}
          Shear.X {type=MovieSceneScriptingFloatChannel}
          Shear.Y {type=MovieSceneScriptingFloatChannel}
         ]
        }]
      }
      {
       type=MovieSceneFloatTrack
       display_name=Render Opacity
       sections [{type=MovieSceneFloatSection, end_frame=15, start_frame=0, channels [None {type=MovieSceneScriptingFloatChannel, keys [[3,0],[10,1]]}]}]
      }
     ]
    }
    Text_Header {
     id=#3
     possessed_class=CommonTextBlock
     tracks [
      {
       type=MovieScene2DTransformTrack
       sections [{
         type=MovieScene2DTransformSection
         end_frame=18
         start_frame=0
         channels [
          Translation.X {type=MovieSceneScriptingFloatChannel, const [0,2]}
          Translation.Y {type=MovieSceneScriptingFloatChannel, keys [[2,80],[12,0]]}
          Angle {type=MovieSceneScriptingFloatChannel}
          Scale.X {type=MovieSceneScriptingFloatChannel}
          Scale.Y {type=MovieSceneScriptingFloatChannel}
          Shear.X {type=MovieSceneScriptingFloatChannel}
          Shear.Y {type=MovieSceneScriptingFloatChannel}
         ]
        }]
      }
      {
       type=MovieSceneFloatTrack
       display_name=Render Opacity
       sections [{type=MovieSceneFloatSection, end_frame=18, start_frame=0, channels [None {type=MovieSceneScriptingFloatChannel, keys [[2,0],[12,1]]}]}]
      }
     ]
    }
    PassButtonStates {
     id=#4
     possessed_class=CommonVisibilitySwitcher
     tracks [{
       type=MovieSceneFloatTrack
       display_name=Render Opacity
       sections [{type=MovieSceneFloatSection, end_frame=20, start_frame=0, channels [None {type=MovieSceneScriptingFloatChannel, keys [[10,0],[15,1]]}]}]
      }]
    }
    SeasonPassBadgeIcon {
     id=#5
     possessed_class=SeasonPassBadgeIcon_C
     tracks [
      {
       type=MovieScene2DTransformTrack
       sections [{
         type=MovieScene2DTransformSection
         end_frame=13
         start_frame=0
         channels [
          Translation.X {type=MovieSceneScriptingFloatChannel, const [0,2]}
          Translation.Y {type=MovieSceneScriptingFloatChannel, keys [[0,80],[8.0003,0]]}
          Angle {type=MovieSceneScriptingFloatChannel}
          Scale.X {type=MovieSceneScriptingFloatChannel}
          Scale.Y {type=MovieSceneScriptingFloatChannel}
          Shear.X {type=MovieSceneScriptingFloatChannel}
          Shear.Y {type=MovieSceneScriptingFloatChannel}
         ]
        }]
      }
      {
       type=MovieSceneFloatTrack
       display_name=Render Opacity
       sections [{type=MovieSceneFloatSection, end_frame=10, start_frame=0, channels [None {type=MovieSceneScriptingFloatChannel, keys [[0,0],[8,1]]}]}]
      }
     ]
    }
    Text_Description {
     id=#6
     possessed_class=CommonTextBlock
     tracks [{
       type=MovieSceneFloatTrack
       display_name=Render Opacity
       sections [{type=MovieSceneFloatSection, end_frame=18, start_frame=0, channels [None {type=MovieSceneScriptingFloatChannel, keys [[8,0],[13,1]]}]}]
      }]
    }
   ]}
 }]
graphs [
 ApplyTitleMaterial {nodes [
   n1 = ApplyTitleMaterial!() -> (then, Material:Soft Object Reference) {is=title K2Node_FunctionEntry}
   > Branch(Condition=IsValidSoftObjectReference(SoftObjectReference=@n1))
   > n2 = LoadAsset_Blocking(Asset=@n1)
   > n3 = Cast To MaterialInstance!(execute, Object=@n2) -> (then, CastFailed, AsMaterial Instance:Material Instance Object Reference) {is=title K2Node_DynamicCast}
   > SetFontMaterial!(
    execute
    self:Text Object Reference=Get Text_Header() -> (Text_Header:Common Text Object Reference)
    InMaterial:Material Interface Object Reference=@n3
   ) -> (then)
  ]}
 ApplyTitleOutline {nodes [
   n1 = ApplyTitleOutline!() -> (then, OutlineSettings:Font Outline Settings Structure) {is=title K2Node_FunctionEntry}
   n2 = Break Slate Font Info!(SlateFontInfo:"Slate Font Info Structure (by ref)"=Get Font!(self:Text Object Reference=Get Text_Header()) -> (Font:Slate Font Info Structure)) -> (
    FontObject:Object Reference
    FontMaterial:Object Reference
    OutlineSettings:Font Outline Settings Structure
    TypefaceFontName:Name
    Size:float=~
    LetterSpacing:int
    SkewAmount:float=~
    bForceMonospaced:bool=~
    bMaterialIsStencil:bool=~
    MonospacedWidth:float=~
   ) {is=title K2Node_BreakStruct}
   > SetFont(
    self=Get Text_Header()
    InFontInfo=MakeSlateFontInfo!(
     FontObject:Object Reference=@n2
     FontMaterial:Object Reference=@n2.FontMaterial
     OutlineSettings:Font Outline Settings Structure=@n1
     TypefaceFontName:Name=@n2.TypefaceFontName
     Size:float=@n2.Size
     LetterSpacing:int=@n2.LetterSpacing
     SkewAmount:float=@n2.SkewAmount
     bForceMonospaced:bool=@n2.bForceMonospaced
     bMaterialIsStencil:bool=@n2.bMaterialIsStencil
     MonospacedWidth:float=@n2.MonospacedWidth
    ) -> (ReturnValue:Slate Font Info Structure)
   )
  ]}
 BP_GetDesiredFocusTarget {nodes [
   Get Desired Focus Target()
   n1 = Get PassButtonStates() -> (PassButtonStates:Common Visibility Switcher Object Reference)
   > Return Node!(
    execute
    ReturnValue:Widget Object Reference=Select(
     Option 0:Widget Object Reference=Get GIftPass_Button() -> (GIftPass_Button:WBP UIKit Button Regular Object Reference)
     Option 1:Widget Object Reference=GetActiveWidget!(self:Common Visibility Switcher Object Reference=@n1) -> (ReturnValue:Widget Object Reference)
     Index=IsVisible(self=@n1)
    ) -> (ReturnValue:Widget Object Reference)
   )
  ]}
 CrewSubVisibility {nodes [
   n1 = CrewSubVisibility!() -> (then, Ownership:ESeasonPassPurchaseContext Enum=None) {is=title K2Node_FunctionEntry}
   > Return Node!(
    execute
    Result:ESlateVisibility Enum=Select(
     Option 0:ESlateVisibility Enum=Collapsed
     Option 1:ESlateVisibility Enum=SelfHitTestInvisible
     Index="Equal (Enum)"(A:ESeasonPassPurchaseContext Enum=@n1, B:ESeasonPassPurchaseContext Enum=Subscription)
    ) -> (ReturnValue:ESlateVisibility Enum)
   )
  ]}
 EventGraph {nodes [
   n1 = "On Clicked (PurchasePass_Button)"!() -> (OutputDelegate, then, Button:Common Button Base Object Reference) {is=title K2Node_ComponentBoundEvent}
   > Fire Event Element Interacted(Element=@n1.Button)
   > n2 = PushPrimaryContentSoftClassAsync!(
    execute:Exec
    SoftWidgetClass:Common Activatable Widget Soft Class Reference=GetActivatableWidgetForAlias!(WidgetAlias:Gameplay Tag Structure=(TagName="UI.Alias.PassPurchase")) -> (ReturnValue:Common Activatable Widget Soft Class Reference)
   ) -> (then:Exec, OnPrimaryContentPushed:Exec, ActivatableWidget:Common Activatable Widget Object Reference) {is=title K2Node_AsyncAction}
   SetViewModelByClass!(
    execute=@n2.OnPrimaryContentPushed
    self:MVVM View Object Reference=GetExtension!(self:User Widget Object Reference=@n2, ExtensionType:User Widget Extension Class Reference=/Script/ModelViewViewModel.MVVMView) -> (ReturnValue:MVVM View Object Reference)
    NewValue:Notify Field Value Changed Interface=Get SeasonPassVM() -> (SeasonPassVM:Season Pass VM Object Reference)
   ) -> (then, ReturnValue:bool)
   > Call Header Entered Subscreen!(execute:Exec, self:Season Pass Header Object Reference) -> (then:Exec) {is=title K2Node_CallDelegate}
   n3 = "On Clicked (GIftPass_Button)"!() -> (OutputDelegate, then, Button:Common Button Base Object Reference) {is=title K2Node_ComponentBoundEvent}
   > Fire Event Element Interacted(Element=@n3.Button)
   > Call Header Entered Subscreen()
   > ShowGiftingScreen(self:Season Pass VM Object Reference=Get SeasonPassVM())
   n4 = "On Clicked (BuyRewards_Button)"!() -> (OutputDelegate, then, Button:Common Button Base Object Reference) {is=title K2Node_ComponentBoundEvent}
   > Fire Event Element Interacted(Element=@n4.Button)
   > Call Has Entered Buy Rewards Modal!(execute:Exec, self:Season Pass Header Object Reference) -> (then:Exec) {is=title K2Node_CallDelegate}
   n5 = Get SeasonPassVM()
   > n6 = Branch(Condition="Equal Exactly (String)"(A=GetSeasonPassTypeId!(self:Season Pass VM Object Reference=@n5) -> (ReturnValue:String), B=br))
   > n7 = Create Widget!(
    execute:Exec
    Class:User Widget Class Reference=/SeasonPass/Widgets/Reboot/WBP_RewardTrack_BuyLevels.WBP_RewardTrack_BuyLevels_C
    OwningPlayer:Player Controller Object Reference
    SeasonPassVM:Season Pass VM Object Reference=Get SeasonPassVM()
   ) -> (then:Exec, ReturnValue:WBP Reward Track Buy Levels Object Reference)
   > TryPushWidgetAsModalPanel!(execute, PushWidget:User Widget Object Reference=@n7) -> (then)
   n8 = Create Widget!(
    execute=@n6.else
    Class=/SeasonPass/Widgets/SeasonPass_BuyRewards.SeasonPass_BuyRewards_C
    OwningPlayer
    SeasonPassVM=Get SeasonPassVM()
    SeasonPassCategoryVM:Season Pass Category VM Object Reference
   ) -> (then, ReturnValue:Season Pass Buy Rewards Object Reference)
   > TryPushWidgetAsModalPanel(PushWidget=@n8)
   Opening() {is=title K2Node_CustomEvent}
   > PlayAnimation(InAnimation=Get Intro(), PlaybackSpeed=2)
   Event OnInitialized()
   > SetCompactUIStyle() {is=title K2Node_CallFunction}
  ]}
 GetShouldDisplayBuyRewards {nodes [
   n1 = GetShouldDisplayBuyRewards!() -> (then, GetTotalOwnedRewards:int, GetNumCanBeClaimedRewards:int, "GetTotalRewards ":int) {is=title K2Node_FunctionEntry}
   > Return Node!(execute, IsEnabled:bool=false)
   "integer >= integer"(A=int + int(A=@n1, B=@n1.GetNumCanBeClaimedRewards), B=@n1."GetTotalRewards ")
  ]}
 HandleCrewSubAccessText {nodes [
   n1 = HandleCrewSubAccessText!() -> (then, PurchaseContext:ESeasonPassPurchaseContext Enum=None) {is=title K2Node_FunctionEntry}
   > SetVisibility(
    self=Get SeasonPass_CrewPassAccess() -> (SeasonPass_CrewPassAccess:Season Pass Crew Pass Access Object Reference)
    InVisibility=Select!(
     None:ESlateVisibility Enum=Hidden
     Purchase:ESlateVisibility Enum=Hidden
     Subscription:ESlateVisibility Enum=Visible
     Index:ESeasonPassPurchaseContext Enum=@n1
    ) -> (ReturnValue)
   )
  ]}
 HandleGiftPassButtonVisibility {nodes [
   HandleGiftPassButtonVisibility!() -> (then, DoesOwnsPass:bool=~, IsPassCompleted:bool=~) {is=title K2Node_FunctionEntry}
   n1 = Get IsPassCompleted() -> (IsPassCompleted:bool)
   n2 = Get DoesOwnsPass() -> (DoesOwnsPass:bool)
   > Return Node!(
    execute
    Visibility:ESlateVisibility Enum=Select!(Option 0:ESlateVisibility Enum=Collapsed, Option 1:ESlateVisibility Enum=SelfHitTestInvisible, Index:bool=OR Boolean(A=@n2, B=@n1)) -> (ReturnValue)
   )
  ]}
 HasEnteredBuyRewardsModal {nodes [HasEnteredBuyRewardsModal() {is=title K2Node_FunctionEntry}]}
 HeaderEnteredSubscreen {nodes [HeaderEnteredSubscreen() {is=title K2Node_FunctionEntry}]}
 IsCrewMember {nodes [
   n1 = IsCrewMember!() -> (then, IsCrewMember:bool=~) {is=title K2Node_FunctionEntry}
   > Set Is Crew Member(Is Crew Member:bool=@n1) -> (Output_Get:bool)
  ]}
 OwnershipPassChanged {nodes [
   n1 = OwnershipPassChanged!() -> (then, bOwnsPass:bool=~) {is=title K2Node_FunctionEntry}
   > SetActiveWidgetIndex(self:Common Visibility Switcher Object Reference=Get PassButtonStates(), Index=SelectInt(A=1, B=0, bPickA=@n1))
  ]}
 SetCompactUIStyle {nodes [
   SetCompactUIStyle() {is=title K2Node_FunctionEntry}
   > Branch(Condition=IsUsingCompactUI())
   > SetWidthOverride(self=Get SizeBox() -> (SizeBox:Size Box Object Reference), InWidthOverride=688)
   > Branch(Condition=CompareToScreenAspectRatio!(Comparison:EAspectRatioComparison Enum=GreaterThan, AspectRatio:EAspectRatio Enum=Ratio_4_3) -> (ReturnValue:bool))
   > n1 = GetSafeZonePadding!(execute) -> (then, SafePadding:Vector 4 Structure, SafePaddingScale:Vector 2D Structure, SpillOverPadding:Vector 4 Structure)
   n2 = "float > float"(A=BreakVector2D(InVec=@n1.SafePaddingScale).Y, B=0.01)
   > Set Size!(
    execute
    Size:Vector 2D Structure=Select(Option 0:Vector 2D Structure=(X=120,Y=120), Option 1:Vector 2D Structure=(X=96,Y=96), Index=@n2) -> (ReturnValue:Vector 2D Structure)
    self:Season Pass Badge Icon Object Reference=Get SeasonPassBadgeIcon() -> (SeasonPassBadgeIcon:Season Pass Badge Icon Object Reference)
   ) -> (then, Output_Get:Vector 2D Structure)
   > SetVerticalAlignment(
    self:Size Box Slot Object Reference=SlotAsSizeBoxSlot!(Widget:Widget Object Reference=Get VerticalBox_55() -> (VerticalBox_55:Vertical Box Object Reference)) -> (ReturnValue:Size Box Slot Object Reference)
    InVerticalAlignment=VAlign_Bottom
   )
   > SetPadding!(
    execute
    self:Vertical Box Slot Object Reference=SlotAsVerticalBoxSlot(Widget=Get CustomPassSwitcher() -> (CustomPassSwitcher:Common Visibility Switcher Object Reference))
    InPadding_Left:float=0
    InPadding_Top:float=SelectFloat(A=8, B=16, bPickA=@n2)
    InPadding_Right:float=0
    InPadding_Bottom:float=0
   ) -> (then)
   > Branch(Condition=@n2)
   > SetPadding(
    self=SlotAsVerticalBoxSlot(Widget=Get WrapBox() -> (WrapBox:Wrap Box Object Reference))
    InPadding_Left=0
    InPadding_Top=12
    InPadding_Right=0
    InPadding_Bottom=0
   )
   > SetPadding(
    self=SlotAsVerticalBoxSlot(Widget=Get SeasonPass_CrewPassAccess())
    InPadding_Left=0
    InPadding_Top=12
    InPadding_Right=0
    InPadding_Bottom=0
   )
  ]}
 SetThemeData {nodes [
   n1 = SetThemeData!() -> (then, ThemeData:Season Pass Theme Data Structure) {is=title K2Node_FunctionEntry}
   > ApplyCustomThemeData!(
    execute
    ThemeData:Season Pass Theme Data Structure=@n1
    Switcher:Common Visibility Switcher Object Reference=Get CustomPassSwitcher()
    LazyImage:Fort Lazy Image Object Reference=Get CustomPassImage() -> (CustomPassImage:Fort Lazy Image Object Reference)
    Text:Common Text Object Reference=Get CustomPassName() -> (CustomPassName:Common Text Object Reference)
   ) -> (then)
   n2 = Break Season Pass Theme Data!(SeasonPassThemeData:"Season Pass Theme Data Structure (by ref)"=@n1) -> (
    TitleColor:Linear Color Structure
    TitleMaterial:Soft Object Reference
    TitleOutline:Font Outline Settings Structure
    SeasonColor:Linear Color Structure
    VaultBackgroundMaterial:Soft Object Reference
    GiftboxVaultBackgroundMaterial:Soft Object Reference
    UIBackgroundMaterial:Soft Object Reference
    UpsellBackgroundMaterial:Soft Object Reference
    CrewCategoryBackgroundMaterial:Soft Object Reference
    CustomLogo:Soft Object Reference
    CustomSeasonTitle:Text
    CustomSeasonTitleColor:Linear Color Structure
    Upsell2BackgrounMaterial:Soft Object Reference
    ImageReferenceTable:Fort Image Reference Table Structure
    ColorReferenceTable:Fort Color Reference Table Structure
    TextReferenceTable:Fort Text Reference Table Structure
   ) {is=title K2Node_BreakStruct}
   > Set ColorAndOpacity!(execute, ColorAndOpacity:Slate Color Structure=Make Slate Color(SpecifiedColor=@n2), self:Text Object Reference=Get Text_Header()) -> (then, Output_Get:Slate Color Structure)
   > ApplyTitleMaterial!(execute, self, Material:Soft Object Reference=@n2.TitleMaterial) -> (then) {is=title K2Node_CallFunction}
   > ApplyTitleOutline!(execute, self, OutlineSettings:Font Outline Settings Structure=@n2.TitleOutline) -> (then) {is=title K2Node_CallFunction}
   > Set ColorAndOpacity(ColorAndOpacity=Make Slate Color(SpecifiedColor=@n2.SeasonColor), self=Get Text_Season() -> (Text_Season:Common Text Object Reference))
  ]}
 UpdateCrewSubscriptionState {nodes [
   n1 = UpdateCrewSubscriptionState!() -> (then, SubscriptionState:EMcpSubscriptionState Enum=Inactive) {is=title K2Node_FunctionEntry}
   n2 = MakeLiteralText(Value="LOCTEXT(\"Pass unlocked through FN Crew\")")
   n3 = Format Text!(
    Format="LOCTEXT(\"Pass unlocked through FN Crew until {Date}\")"
    Date:Text=GetSubscriptionEndDate!(self:Crew VM Object Reference=Get CrewVM() -> (CrewVM:Crew VM Object Reference)) -> (ReturnValue:Text)
   ) -> (Result)
   > "SetText (Text)"(
    self=Get Text_Description!(self:Season Pass Crew Pass Access Object Reference=Get SeasonPass_CrewPassAccess()) -> (Text_Description:Common Text Object Reference)
    InText=Select!(
     Inactive:Text=@n2
     Active:Text=@n2
     ActiveWithExternalBilling:Text=@n2
     NotAutoRenewable:Text=@n3
     Canceled:Text=@n3
     PaymentProcessError:Text=@n3
     PlatformAuthError:Text=@n2
     RewardsDelayed:Text=@n2
     Unavailable:Text
     Unknown:Text=@n2
     Index:EMcpSubscriptionState Enum=@n1
    ) -> (ReturnValue:Text)
   )
  ]}
]
variables [
 HasEnteredBuyRewardsModal {}
 HeaderEnteredSubscreen {}
 Is Crew Member {type=boolean}
]
widgets {
 info {inheritedWidgetCount=0, parentClass=&CommonActivatableWidget, rootWidgetClass=&SizeBox, widgetCount=18}
 widgets [
  SizeBox {
   widgetClassPath=&SizeBox
   bIsVariable=true
   widget=SizeBox
   properties {bOverride_WidthOverride=true, heightOverride=480, widthOverride=592}
  }
  VerticalBox_55 {widgetClassPath=&VerticalBox, bIsVariable=true, parent=SizeBox, widget=VerticalBox_55, slot {class=SizeBoxSlot}}
  SeasonPassBadgeIcon {
   widgetClassPath=/SeasonPass/Widgets/SeasonPassBadgeIcon
   bIsVariable=true
   parent=VerticalBox_55
   widget=SeasonPassBadgeIcon
   properties {size=<140 140>}
   slot {class=VerticalBoxSlot, properties {horizontalAlignment=HAlign_Center, verticalAlignment=VAlign_Center}}
  }
  CustomPassSwitcher {
   widgetClassPath=&CommonVisibilitySwitcher
   bIsVariable=true
   parent=VerticalBox_55
   widget=CustomPassSwitcher
   slot {class=VerticalBoxSlot, properties {padding=<ltrb 0 16 0 0>}}
  }
  VerticalBox_0 {widgetClassPath=&VerticalBox, parent=CustomPassSwitcher, widget=VerticalBox_0, slot {class=CommonVisibilitySwitcherSlot}}
  Text_Season {
   widgetClassPath=&CommonTextBlock
   bIsVariable=true
   parent=VerticalBox_0
   widget=Text_Season
   properties {
    justification=Center
    mobileFontSizeMultiplier=0.75
    renderOpacity=0
    scrollStyle=/FortUILibrary/Styles/Typography/Foundation/TextScrollStyle-NoFade.TextScrollStyle-NoFade_C
    text=Season X
    textTransformPolicy=ToUpper
    colorAndOpacity {colorUseRule=UseColor_Specified, specifiedColor=<rgba 1 0.6327 0 1>}
    font {
     bForceMonospaced=false
     bMaterialIsStencil=false
     fontMaterial=null
     fontObject=&HeadingNow
     letterSpacing=-40
     monospacedWidth=1
     size=51.75
     skewAmount=0
     typefaceFontName=86 Bold Italic
     outlineSettings {
      bApplyOutlineToDropShadows=false
      bMiteredCorners=false
      bSeparateFillAlpha=false
      outlineMaterial=null
      outlineSize=0
      outlineColor=<rgba 0 0 0 1>
     }
    }
    shadowOffset=<0 0>
   }
   slot {class=VerticalBoxSlot}
  }
  Text_Header {
   widgetClassPath=&CommonTextBlock
   bIsVariable=true
   parent=VerticalBox_0
   widget=Text_Header
   properties {
    autoWrapText=true
    bIsScrollingEnabled=false
    justification=Center
    mobileFontSizeMultiplier=0.75
    renderOpacity=0
    scrollStyle=&TextScrollStyle-NoFade_C
    text=Pass Name
    textTransformPolicy=ToUpper
    font {
     bForceMonospaced=false
     bMaterialIsStencil=false
     fontMaterial=null
     fontObject=&HeadingNow
     letterSpacing=-40
     monospacedWidth=1
     size=51.75
     skewAmount=0
     typefaceFontName=86 Bold Italic
     outlineSettings {
      bApplyOutlineToDropShadows=false
      bMiteredCorners=false
      bSeparateFillAlpha=false
      outlineMaterial=null
      outlineSize=0
      outlineColor=<rgba 0 0 0 1>
     }
    }
    shadowOffset=<0 0>
   }
   slot {class=VerticalBoxSlot, properties {padding=<ltrb 0 -4 0 0>}}
  }
  VerticalBox {widgetClassPath=&VerticalBox, parent=CustomPassSwitcher, widget=VerticalBox, slot {class=CommonVisibilitySwitcherSlot}}
  CustomPassImage {
   widgetClassPath=&FortLazyImage
   bIsVariable=true
   parent=VerticalBox
   widget=CustomPassImage
   properties {brush {
     bIsDynamicallyLoaded=false
     drawAs=Image
     imageType=NoImage
     mirroring=NoMirror
     resourceName=None
     resourceObject=null
     tiling=NoTile
     imageSize=<300 150>
     margin=<ltrb 0 0 0 0>
     outlineSettings {
      bUseBrushTransparency=false
      roundingType=HalfHeightRadius
      width=0
      color {colorUseRule=UseColor_Specified, specifiedColor=<rgba 0 0 0 0>}
      cornerRadii=<0 0 0 0>
     }
     tintColor {colorUseRule=UseColor_Specified, specifiedColor=<rgba 1 1 1 1>}
     uVRegion {bIsValid=false, max=<0 0>, min=<0 0>}
    }}
   slot {class=VerticalBoxSlot, properties {horizontalAlignment=HAlign_Center, verticalAlignment=VAlign_Top, padding=<ltrb 0 -16 0 0>}}
  }
  CustomPassName {
   widgetClassPath=&CommonTextBlock
   bIsVariable=true
   parent=VerticalBox
   widget=CustomPassName
   properties {
    autoWrapText=true
    bIsScrollingEnabled=false
    justification=Center
    mobileFontSizeMultiplier=0.75
    renderOpacity=0
    scrollStyle=&TextScrollStyle-NoFade_C
    text=Pass Name
    textTransformPolicy=ToUpper
    font {
     bForceMonospaced=false
     bMaterialIsStencil=false
     fontMaterial=null
     fontObject=&HeadingNow
     letterSpacing=-40
     monospacedWidth=1
     size=51.75
     skewAmount=0
     typefaceFontName=86 Bold Italic
     outlineSettings {
      bApplyOutlineToDropShadows=false
      bMiteredCorners=false
      bSeparateFillAlpha=false
      outlineMaterial=null
      outlineSize=0
      outlineColor=<rgba 0 0 0 1>
     }
    }
    shadowOffset=<0 0>
   }
   slot {class=VerticalBoxSlot, properties {padding=<ltrb 0 6 0 0>}}
  }
  Text_DevOnly {
   widgetClassPath=&CommonTextBlock
   bIsVariable=true
   parent=VerticalBox_55
   widget=Text_DevOnly
   properties {
    autoWrapText=true
    bIsScrollingEnabled=false
    justification=Center
    mobileFontSizeMultiplier=0.75
    renderOpacity=0
    scrollStyle=&TextScrollStyle-NoFade_C
    text=DEV ONLY
    textTransformPolicy=ToUpper
    visibility=Collapsed
    colorAndOpacity {colorUseRule=UseColor_Specified, specifiedColor=<rgba 1 0.0037 0.0328 1>}
    font {
     bForceMonospaced=false
     bMaterialIsStencil=false
     fontMaterial=null
     fontObject=&HeadingNow
     letterSpacing=10
     monospacedWidth=1
     size=20.25
     skewAmount=0
     typefaceFontName=86 Bold
     outlineSettings {
      bApplyOutlineToDropShadows=false
      bMiteredCorners=false
      bSeparateFillAlpha=false
      outlineMaterial=null
      outlineSize=0
      outlineColor=<rgba 0 0 0 1>
     }
    }
    shadowOffset=<0 0>
   }
   slot {class=VerticalBoxSlot, properties {padding=<ltrb 0 -4 0 0>}}
  }
  Text_Description {
   widgetClassPath=&CommonTextBlock
   bIsVariable=true
   parent=VerticalBox_55
   widget=Text_Description
   properties {
    autoWrapText=true
    justification=Center
    mobileFontSizeMultiplier=0.75
    scrollStyle=&TextScrollStyle-NoFade_C
    text=Current Pass description text
    textTransformPolicy=ToUpper
    visibility=Collapsed
    wrapTextAt=700
    colorAndOpacity {colorUseRule=UseColor_Specified, specifiedColor=<rgba 1 1 1 0.75>}
    font {
     bForceMonospaced=false
     bMaterialIsStencil=false
     fontMaterial=null
     fontObject=&HeadingNow
     letterSpacing=-40
     monospacedWidth=1
     size=51.75
     skewAmount=0
     typefaceFontName=86 Bold Italic
     outlineSettings {
      bApplyOutlineToDropShadows=false
      bMiteredCorners=false
      bSeparateFillAlpha=false
      outlineMaterial=null
      outlineSize=0
      outlineColor=<rgba 0 0 0 1>
     }
    }
    shadowOffset=<0 0>
   }
   slot {class=VerticalBoxSlot, properties {padding=<ltrb 0 20 0 0>}}
  }
  WrapBox {
   widgetClassPath=&WrapBox
   bIsVariable=true
   parent=VerticalBox_55
   widget=WrapBox
   properties {horizontalAlignment=HAlign_Center, innerSlotPadding=<0 16>}
   slot {class=VerticalBoxSlot, properties {padding=<ltrb 0 40 0 0>}}
  }
  PassButtonStates {
   widgetClassPath=&CommonVisibilitySwitcher
   bIsVariable=true
   parent=WrapBox
   widget=PassButtonStates
   slot {class=WrapBoxSlot, properties {horizontalAlignment=HAlign_Center}}
  }
  PurchasePass_Button {
   widgetClassPath=/FortUILibrary/UIKit/Buttons/WBP_UIKit_Button_Loud
   bIsVariable=true
   parent=PassButtonStates
   widget=PurchasePass_Button
   properties {bHideInputAction=true, minHeight=56, minWidth=465, text="Get the [Pass Name]"}
   slot {class=CommonVisibilitySwitcherSlot, properties {horizontalAlignment=HAlign_Center, padding=<ltrb 8 0 8 0>}}
  }
  BuyRewards_Button {
   widgetClassPath=&WBP_UIKit_Button_Loud
   bIsVariable=true
   parent=PassButtonStates
   widget=BuyRewards_Button
   properties {bHideInputAction=true, text=BUY REWARDS}
   slot {class=CommonVisibilitySwitcherSlot, properties {horizontalAlignment=HAlign_Center, padding=<ltrb 8 0 8 0>}}
  }
  GIftPass_Button {
   widgetClassPath=&WBP_UIKit_Button_Regular
   bIsVariable=true
   parent=WrapBox
   widget=GIftPass_Button
   properties {bHideInputAction=true, minHeight=56, text=Gift the pass}
   slot {class=WrapBoxSlot, properties {horizontalAlignment=HAlign_Center, padding=<ltrb 8 0 8 0>}}
  }
  SeasonPass_CrewPassAccess {
   widgetClassPath=/SeasonPass/Widgets/SeasonPass_CrewPassAccess
   bIsVariable=true
   parent=VerticalBox_55
   widget=SeasonPass_CrewPassAccess
   slot {class=VerticalBoxSlot, properties {padding=<ltrb 0 20 0 0>}}
  }
 ]
}
