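# Season Pass hub page, a CommonActivatableWidget with a header above a horizontally scrolling row of category tiles and an availability line.
# Category list changes rebuild the DynamicEntryBox, passing viewmodels to each button and making the first double width.
# Custom up-navigation picks an available header button, else broadcasts NavUpButNoButtonsLeft and returns default focus (first tile).
# PreConstruct applies compact-UI padding by aspect ratio; OpeningTile plays the fade-and-slide intro.
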WidgetBlueprint /SeasonPassHub/Widgets/SeasonPassPage
animations [TilesIntro {
  binding_count=2
  length_seconds=0.4
  object=&self:TilesIntro
  sequence {bindings [
    Categories {
     id=#1
     possessed_class=DynamicEntryBox
     tracks [
      {
       type=MovieSceneFloatTrack
       display_name=Render Opacity
       sections [{type=MovieSceneFloatSection, end_frame=6, start_frame=-1, channels [None {type=MovieSceneScriptingFloatChannel, keys [[2,0],[5.0003,1]]}]}]
      }
      {
       type=MovieScene2DTransformTrack
       sections [{
         type=MovieScene2DTransformSection
         end_frame=6
         start_frame=-1
         channels [
          Translation.X {type=MovieSceneScriptingFloatChannel, const [0,2]}
          Translation.Y {type=MovieSceneScriptingFloatChannel, keys [[2,140],[5.0003,0]]}
          Angle {type=MovieSceneScriptingFloatChannel, const [0,1]}
          Scale.X {type=MovieSceneScriptingFloatChannel, const [1,1]}
          Scale.Y {type=MovieSceneScriptingFloatChannel, const [1,1]}
          Shear.X {type=MovieSceneScriptingFloatChannel, const [0,1]}
          Shear.Y {type=MovieSceneScriptingFloatChannel, const [0,1]}
         ]
        }]
      }
     ]
    }
    Availability {
     id=#2
     possessed_class=UEFN_TextBlock_C
     tracks [{
       type=MovieSceneFloatTrack
       display_name=Render Opacity
       sections [{type=MovieSceneFloatSection, end_frame=7, start_frame=-1, channels [None {type=MovieSceneScriptingFloatChannel, keys [[4,0],[6,1]]}]}]
      }]
    }
   ]}
 }]
graphs [
 BP_GetDesiredFocusTarget {nodes [
   Get Desired Focus Target()
   n1 = Get Categories() -> (Categories:Dynamic Entry Box Object Reference)
   > Return Node!(
    execute
    ReturnValue:Widget Object Reference="Get (a copy)"(Array:Array of User Widget Object References=GetAllEntries!(self:Dynamic Entry Box Base Object Reference=@n1) -> (ReturnValue:Array of User Widget Object References)) -> (Output:User Widget Object Reference)
   )
  ]}
 Categories Changed {nodes [
   n1 = Categories Changed!() -> (then, Categories:Array of Season Pass Category VM Object References) {is=title K2Node_FunctionEntry}
   > Branch(Condition=Is Not Empty!(TargetArray:Array of Season Pass Category VM Object References=@n1) -> (ReturnValue:bool))
   "Get (a copy)"(Array:Array of Season Pass Category VM Object References=@n1) -> (Output:Season Pass Category VM Object Reference)
  ]}
 EnteredPassSubscreen {nodes [EnteredPassSubscreen() {is=title K2Node_FunctionEntry}]}
 EventGraph {nodes [
   n1 = CategoriesListChanged!() -> (OutputDelegate, then, SeasonPassCategoryVMs:Array of Season Pass Category VM Object References) {is=title K2Node_CustomEvent}
   > Reset!(execute, self:Dynamic Entry Box Object Reference=Get Categories(), bDeleteWidgets:bool=true) -> (then)
   > n2 = For Each Loop(Array:Array of Season Pass Category VM Object References=@n1.SeasonPassCategoryVMs) -> (Array Element:Season Pass Category VM Object Reference)
   > n3 = Create Entry!(execute, self:Dynamic Entry Box Object Reference=Get Categories()) -> (then, ReturnValue:User Widget Object Reference)
   > n4 = Cast To SeasonPassCategoryButton!(execute, Object=@n3) -> (then, CastFailed, AsSeason Pass Category Button:Season Pass Category Button Object Reference) {is=title K2Node_DynamicCast}
   > SetSeasonPassCategoryVM!(execute, self:Season Pass Category Button Object Reference=@n4, Viewmodel:Season Pass Category VM Object Reference=@n2) -> (then)
   > SetSeasonPassVM!(
    execute
    self:Season Pass Category Button Object Reference=@n4
    Viewmodel:Season Pass VM Object Reference=Get SeasonPassVM() -> (SeasonPassVM:Season Pass VM Object Reference)
   ) -> (then)
   > Set Double Width!(execute, self:Season Pass Category Button Object Reference=@n4, Is Double Width:bool="Equal (Integer)"(A=@n2.Array Index)) -> (then) {is=title K2Node_CallFunction}
   > SetLockedState!(execute, self:Season Pass Category Button Object Reference=@n4, Is Locked:bool=false) -> (then)
   > Assign Season Pass Category Entered!(execute:Exec, self:Season Pass Category Button Object Reference=@n4, Delegate:"Delegate (by ref)"=Create Event()) -> (then:Exec) {is=title K2Node_AssignDelegate}
   OpeningTile() {is=title K2Node_CustomEvent}
   > PlayAnimation(InAnimation=Get TilesIntro() -> (TilesIntro:Widget Animation Object Reference))
   OnEnteredSubscreen() {is=title K2Node_CustomEvent}
   > Call Entered Pass Subscreen!(execute:Exec, self:Season Pass Page Object Reference) -> (then:Exec) {is=title K2Node_CallDelegate}
   Event PreConstruct()
   > SetMobileStyle() {is=title K2Node_CallFunction}
   > Bind Event to Header Entered Subscreen(
    self:Season Pass Header Object Reference=Get SeasonPass_Header() -> (SeasonPass_Header:Season Pass Header Object Reference)
    Delegate=Create Event()
   ) {is=title K2Node_AddDelegate}
  ]}
 NavUpButNoButtonsLeft {nodes [NavUpButNoButtonsLeft() {is=title K2Node_FunctionEntry}]}
 NavigateToHeader {nodes [
   NavigateToHeader!() -> (then, Navigation:EUINavigation Enum) {is=title K2Node_FunctionEntry}
   n1 = Get SeasonPass_Header()
   n2 = Get PassButtonStates!(self:Season Pass Header Object Reference=@n1) -> (PassButtonStates:Common Visibility Switcher Object Reference)
   n3 = "Equal (Enum)"(
    A:ESlateVisibility Enum=GetVisibility!(self:Widget Object Reference=@n2) -> (ReturnValue:ESlateVisibility Enum=Visible)
    B:ESlateVisibility Enum=Collapsed
   )
   n4 = Select(
    Option 0:Widget Object Reference=Get PurchasePass_Button!(self:Season Pass Header Object Reference=@n1) -> (PurchasePass_Button:WBP UIKit Button Loud Object Reference)
    Option 1:Widget Object Reference=Get BuyRewards_Button!(self:Season Pass Header Object Reference=@n1) -> (BuyRewards_Button:WBP UIKit Button Loud Object Reference)
    Index:int=GetActiveWidgetIndex!(self:Common Visibility Switcher Object Reference=@n2) -> (ReturnValue:int=0)
   ) -> (ReturnValue:Widget Object Reference)
   n5 = "Equal (Boolean)"(A:bool=GetIsEnabled!(self:Widget Object Reference=@n4) -> (ReturnValue:bool), B:bool) -> (ReturnValue:bool=~) {is=title K2Node_PromotableOperator}
   n6 = Get GIftPass_Button!(self:Season Pass Header Object Reference=Get SeasonPass_Header()) -> (GIftPass_Button:WBP UIKit Button Regular Object Reference)
   n7 = GetIsEnabled(self=@n6)
   n8 = "Equal (Enum)"(A=GetVisibility(self=@n6) -> (ReturnValue=Visible), B=SelfHitTestInvisible)
   n9 = Get SeasonPass_Header()
   n10 = "float > float"(
    A=GetDesiredSize!(self:Widget Object Reference=Get WrapBox!(self:Season Pass Header Object Reference=@n9) -> (WrapBox:Wrap Box Object Reference)) -> (ReturnValue_X:double, ReturnValue_Y:double).ReturnValue_Y
    B=80
   )
   n11 = Select(
    Option 0=@n4
    Option 1=Get GIftPass_Button(self=Get SeasonPass_Header())
    Index:bool=OR Boolean(A=OR Boolean(A=@n3, B=@n5), B=AND Boolean!(A=@n8, B=@n7, C:bool=@n10) -> (ReturnValue))
   )
   n12 = GetVisibility(self=@n11) -> (ReturnValue=Visible)
   n13 = GetIsEnabled(self=@n11)
   > n14 = Branch(Condition=OR Boolean(A="Equal (Boolean)"(A=@n13) -> (ReturnValue=~), B="Equal (Enum)"(A=@n12, B=Collapsed)))
   > Call Nav Up But No Buttons Left!(execute:Exec, self:Season Pass Page Object Reference) -> (then:Exec) {is=title K2Node_CallDelegate}
   > Return Node(ReturnValue=GetDesiredFocusTarget!(self:Common Activatable Widget Object Reference) -> (ReturnValue:Widget Object Reference))
   Return Node(execute=@n14.else, ReturnValue=@n11)
  ]}
 SeasonPassVmChanged {nodes [
   n1 = SeasonPassVmChanged!() -> (then, PassVM:Season Pass VM Object Reference) {is=title K2Node_FunctionEntry}
   > SetSeasonPassVM(self:Season Pass Header Object Reference=Get SeasonPass_Header(), Viewmodel=@n1)
  ]}
 SetMobileStyle {nodes [
   SetMobileStyle() {is=title K2Node_FunctionEntry}
   > Branch(Condition=IsUsingCompactUI())
   n1 = CompareToScreenAspectRatio!(Comparison:EAspectRatioComparison Enum=GreaterThan, AspectRatio:EAspectRatio Enum=Ratio_4_3) -> (ReturnValue:bool)
   > SetPadding!(
    execute
    self:Overlay Slot Object Reference=SlotAsOverlaySlot(Widget=Get SeasonPass_Header())
    InPadding_Left:float=0
    InPadding_Top:float=0
    InPadding_Right:float=0
    InPadding_Bottom:float=SelectFloat(A=112, B=0, bPickA=@n1)
   ) -> (then)
   n2 = SlotAsVerticalBoxSlot(Widget=Get Overlay_1() -> (Overlay_1:Overlay Object Reference))
   n3 = Get!(self:Vertical Box Slot Object Reference=@n2) -> (Padding_Left:float=~, Padding_Top:float=~, Padding_Right:float=~, Padding_Bottom:float=~)
   > SetPadding(
    self:Vertical Box Slot Object Reference=@n2
    InPadding_Left=@n3
    InPadding_Top=SelectFloat(A=24, B=60, bPickA=@n1)
    InPadding_Right=@n3.Padding_Right
    InPadding_Bottom=@n3.Padding_Bottom
   )
   n4 = SlotAsVerticalBoxSlot(Widget=Get Availability() -> (Availability:Common Rich Text Block Object Reference))
   n5 = Get(self=@n4) -> (Padding_Left=~, Padding_Top=~, Padding_Right=~, Padding_Bottom=~)
   > SetPadding(self=@n4, InPadding_Left=@n5, InPadding_Top=8, InPadding_Right=@n5.Padding_Right, InPadding_Bottom=@n5.Padding_Bottom)
  ]}
]
variables [
 EnteredPassSubscreen {}
 NavUpButNoButtonsLeft {}
]
widgets {
 info {inheritedWidgetCount=0, parentClass=&CommonActivatableWidget, rootWidgetClass=&VerticalBox, widgetCount=10}
 widgets [
  VerticalBox_0 {widgetClassPath=&VerticalBox, bIsVariable=true, widget=VerticalBox_0}
  HorizontalBox_1 {
   widgetClassPath=&HorizontalBox
   parent=VerticalBox_0
   widget=HorizontalBox_1
   slot {class=VerticalBoxSlot, properties {size {sizeRule=Fill, value=1}}}
  }
  SeasonPass_Header {
   widgetClassPath=&/SeasonPass_Header
   bIsVariable=true
   parent=HorizontalBox_1
   widget=SeasonPass_Header
   slot {class=HorizontalBoxSlot}
  }
  Overlay_3 {
   widgetClassPath=&Overlay
   parent=HorizontalBox_1
   widget=Overlay_3
   slot {class=HorizontalBoxSlot, properties {size {sizeRule=Fill, value=1}}}
  }
  SeasonPassBubble {
   widgetClassPath=&/SeasonPassBubble
   bIsVariable=true
   parent=Overlay_3
   widget=SeasonPassBubble
   properties {visibility=Hidden}
   slot {class=OverlaySlot, properties {horizontalAlignment=HAlign_Right, verticalAlignment=VAlign_Bottom}}
  }
  Overlay_1 {
   widgetClassPath=&Overlay
   bIsVariable=true
   parent=VerticalBox_0
   widget=Overlay_1
   slot {class=VerticalBoxSlot, properties {padding=<ltrb 0 40 0 0>}}
  }
  VerticalBox_1 {
   widgetClassPath=&VerticalBox
   parent=Overlay_1
   widget=VerticalBox_1
   slot {class=OverlaySlot, properties {verticalAlignment=VAlign_Bottom}}
  }
  ScrollBox_0 {
   widgetClassPath=&ScrollBox
   parent=VerticalBox_1
   widget=ScrollBox_0
   properties {
    clipping=Inherit
    navigation=&self:WidgetTree.ScrollBox_0.WidgetNavigation_0
    orientation=Orient_Horizontal
    scrollBarVisibility=Collapsed
    widgetBarStyle {
     thickness=0
     draggedThumbImage {
      bIsDynamicallyLoaded=false
      drawAs=Box
      imageType=FullColor
      mirroring=NoMirror
      resourceName=../../../Engine/Content/Slate/Common/Scrollbar_Thumb.png
      resourceObject=null
      tiling=NoTile
      imageSize=<0 0>
      margin=<ltrb 0.25 0.25 0.25 0.25>
      outlineSettings {
       bUseBrushTransparency=false
       roundingType=HalfHeightRadius
       width=0
       color {colorUseRule=UseColor_Specified, specifiedColor=<rgba 0 0 0 0>}
       cornerRadii=<0 0 0 0>
      }
      tintColor {colorUseRule=UseColor_Specified, specifiedColor=<rgba 1 1 1 1>}
      uVRegion {bIsValid=false, max=<0 0>, min=<0 0>}
     }
     horizontalBackgroundImage {
      bIsDynamicallyLoaded=false
      drawAs=NoDrawType
      imageType=NoImage
      mirroring=NoMirror
      resourceName=None
      resourceObject=null
      tiling=NoTile
      imageSize=<0 0>
      margin=<ltrb 0 0 0 0>
      outlineSettings {
       bUseBrushTransparency=false
       roundingType=HalfHeightRadius
       width=0
       color {colorUseRule=UseColor_Specified, specifiedColor=<rgba 0 0 0 0>}
       cornerRadii=<0 0 0 0>
      }
      tintColor {colorUseRule=UseColor_Specified, specifiedColor=<rgba 1 1 1 1>}
      uVRegion {bIsValid=false, max=<0 0>, min=<0 0>}
     }
     horizontalBottomSlotImage {
      bIsDynamicallyLoaded=false
      drawAs=Image
      imageType=FullColor
      mirroring=NoMirror
      resourceName=../../../Engine/Content/Slate/Common/Scrollbar_Background_Horizontal.png
      resourceObject=null
      tiling=NoTile
      imageSize=<8 8>
      margin=<ltrb 0 0 0 0>
      outlineSettings {
       bUseBrushTransparency=false
       roundingType=HalfHeightRadius
       width=0
       color {colorUseRule=UseColor_Specified, specifiedColor=<rgba 0 0 0 0>}
       cornerRadii=<0 0 0 0>
      }
      tintColor {colorUseRule=UseColor_Specified, specifiedColor=<rgba 1 1 1 1>}
      uVRegion {bIsValid=false, max=<0 0>, min=<0 0>}
     }
     horizontalTopSlotImage {
      bIsDynamicallyLoaded=false
      drawAs=Image
      imageType=FullColor
      mirroring=NoMirror
      resourceName=../../../Engine/Content/Slate/Common/Scrollbar_Background_Horizontal.png
      resourceObject=null
      tiling=NoTile
      imageSize=<8 8>
      margin=<ltrb 0 0 0 0>
      outlineSettings {
       bUseBrushTransparency=false
       roundingType=HalfHeightRadius
       width=0
       color {colorUseRule=UseColor_Specified, specifiedColor=<rgba 0 0 0 0>}
       cornerRadii=<0 0 0 0>
      }
      tintColor {colorUseRule=UseColor_Specified, specifiedColor=<rgba 1 1 1 1>}
      uVRegion {bIsValid=false, max=<0 0>, min=<0 0>}
     }
     hoveredThumbImage {
      bIsDynamicallyLoaded=false
      drawAs=Box
      imageType=FullColor
      mirroring=NoMirror
      resourceName=../../../Engine/Content/Slate/Common/Scrollbar_Thumb.png
      resourceObject=null
      tiling=NoTile
      imageSize=<0 0>
      margin=<ltrb 0.25 0.25 0.25 0.25>
      outlineSettings {
       bUseBrushTransparency=false
       roundingType=HalfHeightRadius
       width=0
       color {colorUseRule=UseColor_Specified, specifiedColor=<rgba 0 0 0 0>}
       cornerRadii=<0 0 0 0>
      }
      tintColor {colorUseRule=UseColor_Specified, specifiedColor=<rgba 1 1 1 1>}
      uVRegion {bIsValid=false, max=<0 0>, min=<0 0>}
     }
     normalThumbImage {
      bIsDynamicallyLoaded=false
      drawAs=Box
      imageType=FullColor
      mirroring=NoMirror
      resourceName=../../../Engine/Content/Slate/Common/Scrollbar_Thumb.png
      resourceObject=null
      tiling=NoTile
      imageSize=<0 0>
      margin=<ltrb 0.25 0.25 0.25 0.25>
      outlineSettings {
       bUseBrushTransparency=false
       roundingType=HalfHeightRadius
       width=0
       color {colorUseRule=UseColor_Specified, specifiedColor=<rgba 0 0 0 0>}
       cornerRadii=<0 0 0 0>
      }
      tintColor {colorUseRule=UseColor_Specified, specifiedColor=<rgba 1 1 1 1>}
      uVRegion {bIsValid=false, max=<0 0>, min=<0 0>}
     }
     verticalBackgroundImage {
      bIsDynamicallyLoaded=false
      drawAs=NoDrawType
      imageType=NoImage
      mirroring=NoMirror
      resourceName=None
      resourceObject=null
      tiling=NoTile
      imageSize=<0 0>
      margin=<ltrb 0 0 0 0>
      outlineSettings {
       bUseBrushTransparency=false
       roundingType=HalfHeightRadius
       width=0
       color {colorUseRule=UseColor_Specified, specifiedColor=<rgba 0 0 0 0>}
       cornerRadii=<0 0 0 0>
      }
      tintColor {colorUseRule=UseColor_Specified, specifiedColor=<rgba 1 1 1 1>}
      uVRegion {bIsValid=false, max=<0 0>, min=<0 0>}
     }
     verticalBottomSlotImage {
      bIsDynamicallyLoaded=false
      drawAs=Image
      imageType=FullColor
      mirroring=NoMirror
      resourceName=../../../Engine/Content/Slate/Common/Scrollbar_Background_Vertical.png
      resourceObject=null
      tiling=NoTile
      imageSize=<8 8>
      margin=<ltrb 0 0 0 0>
      outlineSettings {
       bUseBrushTransparency=false
       roundingType=HalfHeightRadius
       width=0
       color {colorUseRule=UseColor_Specified, specifiedColor=<rgba 0 0 0 0>}
       cornerRadii=<0 0 0 0>
      }
      tintColor {colorUseRule=UseColor_Specified, specifiedColor=<rgba 1 1 1 1>}
      uVRegion {bIsValid=false, max=<0 0>, min=<0 0>}
     }
     verticalTopSlotImage {
      bIsDynamicallyLoaded=false
      drawAs=Image
      imageType=FullColor
      mirroring=NoMirror
      resourceName=../../../Engine/Content/Slate/Common/Scrollbar_Background_Vertical.png
      resourceObject=null
      tiling=NoTile
      imageSize=<8 8>
      margin=<ltrb 0 0 0 0>
      outlineSettings {
       bUseBrushTransparency=false
       roundingType=HalfHeightRadius
       width=0
       color {colorUseRule=UseColor_Specified, specifiedColor=<rgba 0 0 0 0>}
       cornerRadii=<0 0 0 0>
      }
      tintColor {colorUseRule=UseColor_Specified, specifiedColor=<rgba 1 1 1 1>}
      uVRegion {bIsValid=false, max=<0 0>, min=<0 0>}
     }
    }
    widgetStyle {
     barThickness=0
     bottomShadowBrush {
      bIsDynamicallyLoaded=false
      drawAs=NoDrawType
      imageType=FullColor
      mirroring=NoMirror
      resourceName=../../../Engine/Content/Slate/Common/ScrollBoxShadowBottom.png
      resourceObject=null
      tiling=NoTile
      imageSize=<16 8>
      margin=<ltrb 0.5 0 0.5 1>
      outlineSettings {
       bUseBrushTransparency=false
       roundingType=HalfHeightRadius
       width=0
       color {colorUseRule=UseColor_Specified, specifiedColor=<rgba 0 0 0 0>}
       cornerRadii=<0 0 0 0>
      }
      tintColor {colorUseRule=UseColor_Specified, specifiedColor=<rgba 1 1 1 1>}
      uVRegion {bIsValid=false, max=<0 0>, min=<0 0>}
     }
     horizontalScrolledContentPadding=<ltrb 0 0 1 0>
     leftShadowBrush {
      bIsDynamicallyLoaded=false
      drawAs=NoDrawType
      imageType=FullColor
      mirroring=NoMirror
      resourceName=../../../Engine/Content/Slate/Common/ScrollBoxShadowLeft.png
      resourceObject=null
      tiling=NoTile
      imageSize=<8 16>
      margin=<ltrb 1 0.5 0 0.5>
      outlineSettings {
       bUseBrushTransparency=false
       roundingType=HalfHeightRadius
       width=0
       color {colorUseRule=UseColor_Specified, specifiedColor=<rgba 0 0 0 0>}
       cornerRadii=<0 0 0 0>
      }
      tintColor {colorUseRule=UseColor_Specified, specifiedColor=<rgba 1 1 1 1>}
      uVRegion {bIsValid=false, max=<0 0>, min=<0 0>}
     }
     rightShadowBrush {
      bIsDynamicallyLoaded=false
      drawAs=NoDrawType
      imageType=FullColor
      mirroring=NoMirror
      resourceName=../../../Engine/Content/Slate/Common/ScrollBoxShadowRight.png
      resourceObject=null
      tiling=NoTile
      imageSize=<8 16>
      margin=<ltrb 0 0.5 1 0.5>
      outlineSettings {
       bUseBrushTransparency=false
       roundingType=HalfHeightRadius
       width=0
       color {colorUseRule=UseColor_Specified, specifiedColor=<rgba 0 0 0 0>}
       cornerRadii=<0 0 0 0>
      }
      tintColor {colorUseRule=UseColor_Specified, specifiedColor=<rgba 1 1 1 1>}
      uVRegion {bIsValid=false, max=<0 0>, min=<0 0>}
     }
     topShadowBrush {
      bIsDynamicallyLoaded=false
      drawAs=NoDrawType
      imageType=FullColor
      mirroring=NoMirror
      resourceName=../../../Engine/Content/Slate/Common/ScrollBoxShadowTop.png
      resourceObject=null
      tiling=NoTile
      imageSize=<16 8>
      margin=<ltrb 0.5 1 0.5 0>
      outlineSettings {
       bUseBrushTransparency=false
       roundingType=HalfHeightRadius
       width=0
       color {colorUseRule=UseColor_Specified, specifiedColor=<rgba 0 0 0 0>}
       cornerRadii=<0 0 0 0>
      }
      tintColor {colorUseRule=UseColor_Specified, specifiedColor=<rgba 1 1 1 1>}
      uVRegion {bIsValid=false, max=<0 0>, min=<0 0>}
     }
     verticalScrolledContentPadding=<ltrb 0 0 0 1>
    }
   }
   slot {class=VerticalBoxSlot}
  }
  Categories {
   widgetClassPath=&DynamicEntryBox
   bIsVariable=true
   parent=ScrollBox_0
   widget=Categories
   properties {entryWidgetClass=/SeasonPass/Widgets/SeasonPassCategoryButton.SeasonPassCategoryButton_C, entrySpacing=<24 0>}
   slot {class=ScrollBoxSlot}
  }
  Availability {
   widgetClassPath=&CommonRichTextBlock
   bIsVariable=true
   parent=VerticalBox_1
   widget=Availability
   properties {
    bOverrideDefaultStyle=true
    text=Available through date
    defaultTextStyleOverride {
     overflowPolicy=Clip
     transformPolicy=None
     colorAndOpacity {colorUseRule=UseColor_Specified, specifiedColor=<rgba 1 1 1 0.6>}
     font {
      bForceMonospaced=false
      bMaterialIsStencil=false
      fontMaterial=null
      fontObject=&HeadingNow
      letterSpacing=0
      monospacedWidth=1
      size=15.75
      skewAmount=0
      typefaceFontName=74 Regular
      outlineSettings {
       bApplyOutlineToDropShadows=false
       bMiteredCorners=false
       bSeparateFillAlpha=false
       outlineMaterial=null
       outlineSize=0
       outlineColor=<rgba 0 0 0 1>
      }
     }
     highlightColor {colorUseRule=UseColor_Specified, specifiedColor=<rgba 0 0 0 1>}
     highlightShape {
      bIsDynamicallyLoaded=false
      drawAs=Image
      imageType=NoImage
      mirroring=NoMirror
      resourceName=None
      resourceObject=null
      tiling=NoTile
      imageSize=<32 32>
      margin=<ltrb 0 0 0 0>
      outlineSettings {
       bUseBrushTransparency=false
       roundingType=HalfHeightRadius
       width=0
       color {colorUseRule=UseColor_Specified, specifiedColor=<rgba 0 0 0 0>}
       cornerRadii=<0 0 0 0>
      }
      tintColor {colorUseRule=UseColor_Specified, specifiedColor=<rgba 1 1 1 1>}
      uVRegion {bIsValid=false, max=<0 0>, min=<0 0>}
     }
     selectedBackgroundColor {colorUseRule=UseColor_Foreground, specifiedColor=<rgba 1 0 1 1>}
     shadowColorAndOpacity=<rgba 0 0 0 1>
     shadowOffset=<0 0>
     strikeBrush {
      bIsDynamicallyLoaded=false
      drawAs=Image
      imageType=NoImage
      mirroring=NoMirror
      resourceName=None
      resourceObject=null
      tiling=NoTile
      imageSize=<32 32>
      margin=<ltrb 0 0 0 0>
      outlineSettings {
       bUseBrushTransparency=false
       roundingType=HalfHeightRadius
       width=0
       color {colorUseRule=UseColor_Specified, specifiedColor=<rgba 0 0 0 0>}
       cornerRadii=<0 0 0 0>
      }
      tintColor {colorUseRule=UseColor_Specified, specifiedColor=<rgba 1 1 1 1>}
      uVRegion {bIsValid=false, max=<0 0>, min=<0 0>}
     }
     underlineBrush {
      bIsDynamicallyLoaded=false
      drawAs=Image
      imageType=NoImage
      mirroring=NoMirror
      resourceName=None
      resourceObject=null
      tiling=NoTile
      imageSize=<32 32>
      margin=<ltrb 0 0 0 0>
      outlineSettings {
       bUseBrushTransparency=false
       roundingType=HalfHeightRadius
       width=0
       color {colorUseRule=UseColor_Specified, specifiedColor=<rgba 0 0 0 0>}
       cornerRadii=<0 0 0 0>
      }
      tintColor {colorUseRule=UseColor_Specified, specifiedColor=<rgba 1 1 1 1>}
      uVRegion {bIsValid=false, max=<0 0>, min=<0 0>}
     }
    }
   }
   slot {class=VerticalBoxSlot, properties {padding=<ltrb 0 24 0 0>}}
  }
 ]
}
