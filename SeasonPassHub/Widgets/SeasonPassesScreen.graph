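# SeasonPassesScreen is the season pass hub screen, with side-navigation tabs for Crew and each season pass, plus upsell and purchase flows.
# HandleChangePage switches between the Crew landing and pass pages, updating the selection view model, legal text and buttons.
# Swipes and mouse wheel step through tabs, with short animations gating repeat scrolls.
# Vault backgrounds load asynchronously, discard stale requests, and transition once the material and CMS texture are both ready.
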WidgetBlueprint /SeasonPassHub/Widgets/SeasonPassesScreen
animations [
 AnimMoveUp {
  binding_count=1
  length_seconds=0.25
  object=&self:AnimMoveUp
  sequence {bindings [SeasonPassesScreen_C_57 {
     id=#1
     possessed_class=SeasonPassesScreen_C
     tracks [{
       type=MovieSceneBoolTrack
       display_name=Is Vertical Transitioning
       sections [{type=MovieSceneBoolSection, end_frame=5, start_frame=0, channels [None {type=MovieSceneScriptingBoolChannel, keys [[0,true],[5,false]]}]}]
      }]
    }]}
 }
 AnimMoveDown {
  binding_count=1
  length_seconds=0.25
  object=&self:AnimMoveDown
  sequence {bindings [SeasonPassesScreen_C_58 {
     id=#2
     possessed_class=SeasonPassesScreen_C
     tracks [{
       type=MovieSceneBoolTrack
       display_name=Is Vertical Transitioning
       sections [{
         type=MovieSceneBoolSection
         end_frame=5
         start_frame=0
         channels [None {type=MovieSceneScriptingBoolChannel, keys [[0,true],[5,false],[5.0003,false]]}]
        }]
      }]
    }]}
 }
 Intro {
  binding_count=4
  length_seconds=0.6
  object=&self:Intro
  sequence {bindings [
    CrewNavWrapper {
     id=#3
     possessed_class=Overlay
     tracks [
      {
       type=MovieSceneFloatTrack
       display_name=Render Opacity
       sections [{type=MovieSceneFloatSection, end_frame=12, start_frame=0, channels [None {type=MovieSceneScriptingFloatChannel, keys [[0,0],[7,1]]}]}]
      }
      {
       type=MovieScene2DTransformTrack
       sections [{
         type=MovieScene2DTransformSection
         end_frame=12
         start_frame=0
         channels [
          Translation.X {type=MovieSceneScriptingFloatChannel, const [0,1]}
          Translation.Y {type=MovieSceneScriptingFloatChannel, keys [[0,100],[10,0]]}
          Angle {type=MovieSceneScriptingFloatChannel, const [0,1]}
          Scale.X {type=MovieSceneScriptingFloatChannel, const [1,1]}
          Scale.Y {type=MovieSceneScriptingFloatChannel, const [1,1]}
          Shear.X {type=MovieSceneScriptingFloatChannel, const [0,1]}
          Shear.Y {type=MovieSceneScriptingFloatChannel, const [0,1]}
         ]
        }]
      }
     ]
    }
    CategoriesNavWrapper {
     id=#4
     possessed_class=Overlay
     tracks [
      {
       type=MovieSceneFloatTrack
       display_name=Render Opacity
       sections [{type=MovieSceneFloatSection, end_frame=12, start_frame=0, channels [None {type=MovieSceneScriptingFloatChannel, keys [[0,0],[7,1]]}]}]
      }
      {
       type=MovieScene2DTransformTrack
       sections [{
         type=MovieScene2DTransformSection
         end_frame=12
         start_frame=0
         channels [
          Translation.X {type=MovieSceneScriptingFloatChannel, const [0,2]}
          Translation.Y {type=MovieSceneScriptingFloatChannel, keys [[0,100],[10,0]]}
          Angle {type=MovieSceneScriptingFloatChannel, const [0,1]}
          Scale.X {type=MovieSceneScriptingFloatChannel, const [1,1]}
          Scale.Y {type=MovieSceneScriptingFloatChannel, const [1,1]}
          Shear.X {type=MovieSceneScriptingFloatChannel, const [0,1]}
          Shear.Y {type=MovieSceneScriptingFloatChannel, const [0,1]}
         ]
        }]
      }
     ]
    }
    SeasonPass_Level {
     id=#5
     possessed_class=SeasonPass_Level_C
     tracks [
      {
       type=MovieSceneFloatTrack
       display_name=Render Opacity
       sections [{type=MovieSceneFloatSection, end_frame=10, start_frame=0, channels [None {type=MovieSceneScriptingFloatChannel, keys [[0,0],[7,1]]}]}]
      }
      {
       type=MovieScene2DTransformTrack
       sections [{
         type=MovieScene2DTransformSection
         end_frame=10
         start_frame=0
         channels [
          Translation.X {type=MovieSceneScriptingFloatChannel, keys [[0,100],[7,0]]}
          Translation.Y {type=MovieSceneScriptingFloatChannel, const [0,2]}
          Angle {type=MovieSceneScriptingFloatChannel, const [0,1]}
          Scale.X {type=MovieSceneScriptingFloatChannel, const [1,1]}
          Scale.Y {type=MovieSceneScriptingFloatChannel, const [1,1]}
          Shear.X {type=MovieSceneScriptingFloatChannel, const [0,1]}
          Shear.Y {type=MovieSceneScriptingFloatChannel, const [0,1]}
         ]
        }]
      }
     ]
    }
    CategoryTabs_NavWrapper {
     id=#6
     possessed_class=GridPanel
     tracks [
      {
       type=MovieScene2DTransformTrack
       sections [{
         type=MovieScene2DTransformSection
         end_frame=10
         start_frame=0
         channels [
          Translation.X {type=MovieSceneScriptingFloatChannel, keys [[0,-100],[7,0]]}
          Translation.Y {type=MovieSceneScriptingFloatChannel, const [0,2]}
          Angle {type=MovieSceneScriptingFloatChannel, const [0,1]}
          Scale.X {type=MovieSceneScriptingFloatChannel, const [1,1]}
          Scale.Y {type=MovieSceneScriptingFloatChannel, const [1,1]}
          Shear.X {type=MovieSceneScriptingFloatChannel, const [0,1]}
          Shear.Y {type=MovieSceneScriptingFloatChannel, const [0,1]}
         ]
        }]
      }
      {
       type=MovieSceneFloatTrack
       display_name=Render Opacity
       sections [{type=MovieSceneFloatSection, end_frame=10, start_frame=0, channels [None {type=MovieSceneScriptingFloatChannel, keys [[0,0],[7,1]]}]}]
      }
     ]
    }
   ]}
 }
]
graphs [
 AspectRatioCheck {nodes [
   Inputs!()
   n1 = GetViewportSize!() -> (ReturnValue_X:double, ReturnValue_Y:double)
   Outputs!(ReturnValue:bool="float >= float"(A=float / float(A=@n1, B=@n1.ReturnValue_Y), B=1.34))
  ]}
 BP_GetDesiredFocusTarget {nodes [
   Get Desired Focus Target()
   n1 = IsVisible(self=Get CrewNavWrapper() -> (CrewNavWrapper:Overlay Object Reference))
   n2 = Get SeasonPassPage() -> (SeasonPassPage:Season Pass Page Object Reference)
   n3 = Get CrewLandingPage() -> (CrewLandingPage:Crew Landing Page Object Reference)
   > Return Node!(
    execute
    ReturnValue:Widget Object Reference=GetDesiredFocusTarget!(self:Common Activatable Widget Object Reference=Select(Option 0:Common Activatable Widget Object Reference=@n2, Option 1:Common Activatable Widget Object Reference=@n3, Index=@n1) -> (ReturnValue:Common Activatable Widget Object Reference)) -> (ReturnValue:Widget Object Reference)
   )
  ]}
 BP_GetDesiredInputConfig {nodes [
   Get Desired Input Config() {is=title K2Node_FunctionEntry}
   > Return Node(ReturnValue:UIInput Config Structure=Make UIInput Config!(
     bIgnoreMoveInput:bool=False
     bIgnoreLookInput:bool=False
     InputMode:ECommonInputMode Enum=All
     MouseCaptureMode:EMouseCaptureMode Enum=CaptureDuringMouseDown
     MouseLockMode:EMouseLockMode Enum=DoNotLock
     bHideCursorDuringViewportCapture:bool=True
    ) -> (UIInputConfig:UIInput Config Structure) {is=title K2Node_MakeStruct})
  ]}
 ControllerPageScroll {nodes [
   n1 = ControllerPageScroll!() -> (then, Navigation:EUINavigation Enum) {is=title K2Node_FunctionEntry}
   > ScrollToNextCategory!(
    execute
    self
    Up:bool=Select!(
     Left:bool=false
     Right:bool=true
     Up:bool=true
     Down:bool=false
     Next:bool=true
     Previous:bool=false
     Invalid:bool=false
     Index:EUINavigation Enum=@n1
    ) -> (ReturnValue:bool)
   ) -> (then)
   > Return Node(ReturnValue:Widget Object Reference=GetDesiredFocusTarget())
  ]}
 Decide If Showing Upsell A/B Test {nodes [
   Inputs!() -> (Exec:Exec)
   n1 = Get bDoesOwnsPass!(self:Season Pass VM Object Reference=FindSeasonPass!(
     self:Season Pass System VM Object Reference=Get SeasonPassSystemVM() -> (SeasonPassSystemVM:Season Pass System VM Object Reference)
     TemplateId:String=battlepassdata
    ) -> (ReturnValue:Season Pass VM Object Reference)) -> (bDoesOwnsPass:bool)
   n2 = GetConfigDataBoolWithDefault!(EntryName:String=HasSeenPassUpsellTest, bDefaultValue:bool=false) -> (ReturnValue:bool)
   > n3 = Branch(Condition=AND Boolean!(
     A=NOT Boolean(A=@n2)
     B=GetConfigDataBoolWithDefault(EntryName=present_pass_purchase_screen.enabled, bDefaultValue=false)
     C:bool=NOT Boolean(A=@n1)
    ) -> (ReturnValue))
   > SetUserSettingBool!(execute, Name:String=HasSeenPassUpsellTest, bValue:bool=true) -> (then)
   > n4 = PushPrimaryContentClassAsync!(execute:Exec, WidgetClass:Common Activatable Widget Class Reference=&/CrewUpsellScreen.CrewUpsellScreen_C) -> (then:Exec, OnPrimaryContentPushed:Exec, ActivatableWidget:Common Activatable Widget Object Reference) {is=title K2Node_AsyncAction}
   n5 = Cast To CrewUpsellScreen!(execute=@n4.OnPrimaryContentPushed, Object=@n4) -> (then, CastFailed, AsCrew Upsell Screen:Crew Upsell Screen Object Reference) {is=title K2Node_DynamicCast}
   > SetSeasonPassVM!(
    execute
    self:Crew Upsell Screen Object Reference=@n5
    Viewmodel:Season Pass VM Object Reference=FindSeasonPass(self=Get SeasonPassSystemVM(), TemplateId=battlepassdata)
   ) -> (then)
   Outputs!(Else:Exec=@n3.else)
  ]}
 EventGraph {nodes [
   n1 = "Crew Pass Button Click (CrewLandingPage)"!() -> (OutputDelegate, then, PassVM:Season Pass VM Object Reference) {is=title K2Node_ComponentBoundEvent}
   n2 = String To Name(InString=GetObjectName(Object=@n1.PassVM))
   > HandleChangePage!(execute, self, TabID:Name=@n2) -> (then)
   > SelectSideNavigationTab!(
    execute
    self:WBP UIKit Side Navigation Base Object Reference=Get UIKit_SideNavigation() -> (UIKit_SideNavigation:WBP UIKit Side Navigation Object Reference)
    TabID:Name=@n2
    InnerTabID:Name=None
    bSuppressClickFeedback:bool=true
   ) -> (then, Success:bool)
   > RequestRefreshFocus(self:Common Activatable Widget Object Reference)
   n3 = "On Side Navigation Selection Changed (UIKit_SideNavigation)"!() -> (OutputDelegate, then, TabID:Name=None, InnerTabID:Name=None, TabViewModel:MVVM Base Viewmodel Object Reference) {is=title K2Node_ComponentBoundEvent}
   > HandleChangePage(TabID=@n3.TabID)
   "On Swipe Down (SwipePanel)"() {is=title K2Node_ComponentBoundEvent}
   > ScrollToNextCategory(Up=false)
   "Nav Up But No Buttons Left (SeasonPassPage)"() {is=title K2Node_ComponentBoundEvent}
   > ScrollToNextCategory(Up=true)
   n4 = "Entered Pass Subscreen (SeasonPassPage)"() {is=title K2Node_ComponentBoundEvent}
   Set IsActivatingFromSubscreen(execute=[@n4.then, @n5.then], IsActivatingFromSubscreen:bool=true) -> (Output_Get:bool)
   n5 = "Entered Crew Subscreen (CrewLandingPage)"() {is=title K2Node_ComponentBoundEvent}
   n6 = "On External Media Changed (EpicCMSImage_BackgroundLoaderHelper)"!() -> (OutputDelegate, then, Texture:Texture 2D Object Reference) {is=title K2Node_ComponentBoundEvent}
   > Set PendingCMSTexture(PendingCMSTexture:Texture Object Reference=@n6.Texture) -> (Output_Get:Texture Object Reference)
   > Set IsCMSTextureLoaded(IsCMSTextureLoaded:bool=true) -> (Output_Get:bool)
   > Branch(Condition=Get IsBackgroundLoaded() -> (IsBackgroundLoaded:bool))
   > OnBackgroundReady()
   "On Side Navigation Tab Clicked (UIKit_SideNavigation)"!() -> (OutputDelegate, then, TabID:Name=None, IsTabContainerExpanded:bool, TabViewModel:MVVM Base Viewmodel Object Reference) {is=title K2Node_ComponentBoundEvent}
   > RequestRefreshFocus()
   "On Clicked (Button_OpenNav_PromptBelowNav)"!() -> (OutputDelegate, then, Button:Common Button Base Object Reference) {is=title K2Node_ComponentBoundEvent}
   > n7 = NavigateToSideNav!(execute, self, Navigation:EUINavigation Enum=Left) -> (then, ReturnValue:Widget Object Reference)
   > SetFocus(self=@n7)
   n8 = "On Side Navigation Collapsed State Changed (UIKit_SideNavigation)"!() -> (OutputDelegate, then, IsCollapsed:bool) {is=title K2Node_ComponentBoundEvent}
   > SetVisibility(
    self=Get Button_OpenNav_PromptBelowNav() -> (Button_OpenNav_PromptBelowNav:WBP UIKit Button Generic Object Reference)
    InVisibility=Select!(Option 0:ESlateVisibility Enum=Collapsed, Option 1:ESlateVisibility Enum=SelfHitTestInvisible, Index:bool=@n8.IsCollapsed) -> (ReturnValue:ESlateVisibility Enum)
   )
   "On Clicked (Button_JoinCrew)"!() -> (OutputDelegate, then, Button:Common Button Base Object Reference) {is=title K2Node_ComponentBoundEvent}
   > n9 = Get!(execute:Exec) -> (then:Exec, else:Exec, CrewVM:Crew VM Object Reference)
   n10 = Get UIKit_SideNavigation()
   > n11 = SelectSideNavigationViewModel!(
    execute
    self:WBP UIKit Side Navigation Base Object Reference=@n10
    TabViewModel:MVVM Base Viewmodel Object Reference=@n9
    bSuppressClickFeedback:bool=true
   ) -> (then, Success:bool)
   RequestRefreshFocus(execute=[@n11.then, @n12.then])
   n12 = SelectTabByIndex!(execute=@n9.else, self:UIKit Tab Group Object Reference=@n10, Index:int=0, bSuppressClickFeedback:bool=true) -> (then, ReturnValue:bool)
   "On Clicked (Button_SubTerms)"!() -> (OutputDelegate, then, Button:Common Button Base Object Reference) {is=title K2Node_ComponentBoundEvent}
   > n13 = Create Widget!(
    execute:Exec
    Class:User Widget Class Reference=/Crew/UI/Subscription/SubscriptionLegalInfoModal.SubscriptionLegalInfoModal_C
    OwningPlayer:Player Controller Object Reference=GetOwningPlayer()
   ) -> (then:Exec, ReturnValue:Subscription Legal Info Modal Object Reference)
   > TryPushWidgetAsModalPanel!(execute, PushWidget:User Widget Object Reference=@n13) -> (then)
   > SetPurchaseFlow!(execute, self:Subscription Legal Info Modal Object Reference=@n13, IsPurchase:bool=false) -> (then)
   "On Clicked (Button_CancelInfo)"!() -> (OutputDelegate, then, Button:Common Button Base Object Reference) {is=title K2Node_ComponentBoundEvent}
   > n14 = Create Widget!(
    execute
    Class=/Crew/UI/Subscription/SubscriptionCancellationInfoModal.SubscriptionCancellationInfoModal_C
    OwningPlayer=GetOwningPlayer()
    CrewSubscriptionModalInfoVM:Crew Subscription Modal Info VM Object Reference
    CrewCancellationInfoVM:Crew Cancellation Info VM Object Reference
   ) -> (then, ReturnValue:Subscription Cancellation Info Modal Object Reference)
   > TryPushWidgetAsModalPanel(PushWidget=@n14)
   "On Swipe Up (SwipePanel)"() {is=title K2Node_ComponentBoundEvent}
   > ScrollToNextCategory(Up=true)
   On Open Pass Screen() {is=title K2Node_CustomEvent}
   n15 = Get SeasonPassPage()
   > Opening(self:Season Pass Header Object Reference=Get SeasonPass_Header!(self:Season Pass Page Object Reference=@n15) -> (SeasonPass_Header:Season Pass Header Object Reference))
   > OpeningTile(self:Season Pass Page Object Reference=@n15)
   n16 = RequestLoadBackground!() -> (OutputDelegate, then, BackgroundAsset:Soft Object Reference, Direction:EVaultWorldTransitionDirection Enum=Forward, FocalPointValue:double=~) {is=title K2Node_CustomEvent}
   > n17 = Sequence()
   > Set RequestedBackgroundSoftPath(RequestedBackgroundSoftPath:Soft Object Path Structure=To Soft Object Path!(SoftObjectReference:Soft Object Reference=@n16.BackgroundAsset) -> (ReturnValue:Soft Object Path Structure)) -> (Output_Get:Soft Object Path Structure)
   > Set PendingDirection(PendingDirection:EVaultWorldTransitionDirection Enum=@n16.Direction) -> (Output_Get:EVaultWorldTransitionDirection Enum=Forward)
   > Set PendingFocalPointValue(PendingFocalPointValue:double=@n16.FocalPointValue) -> (Output_Get:double)
   > Set PendingLoadedBackground(PendingLoadedBackground:Object Reference) -> (Output_Get:Object Reference)
   > Set IsBackgroundLoaded(IsBackgroundLoaded:bool=false) -> (Output_Get:bool)
   n18 = Async Load Asset(execute=@n17.then_1, Asset=@n16.BackgroundAsset)
   n19 = GetSoftObjectPath(Object=@n18)
   n20 = Get RequestedBackgroundSoftPath() -> (RequestedBackgroundSoftPath:Soft Object Path Structure)
   Branch(
    execute=@n18.Completed
    Condition="Equal Exactly (String)"(A=BreakSoftObjectPath(InSoftObjectPath=@n19), B=BreakSoftObjectPath(InSoftObjectPath=@n20))
   )
   > Branch(Condition=Get bIsActive() -> (bIsActive:bool))
   > Set PendingLoadedBackground(PendingLoadedBackground=@n18)
   > Set IsBackgroundLoaded(IsBackgroundLoaded=true)
   > Branch(Condition=Get IsCMSTextureLoaded() -> (IsCMSTextureLoaded:bool))
   > OnBackgroundReady()
   OnBackgroundReady() {is=title K2Node_CustomEvent}
   > n21 = Sequence()
   > n22 = Cast To MaterialInterface!(execute, Object=Get PendingLoadedBackground() -> (PendingLoadedBackground:Object Reference)) -> (then, CastFailed, AsMaterial Interface:Material Interface Object Reference) {is=title K2Node_DynamicCast}
   > n23 = CreateDynamicMaterialInstance!(execute, Parent=@n22, OptionalName=None, CreationFlags=None) -> (then, ReturnValue)
   > SetTextureParameterValue(self=@n23, ParameterName=CMS_Crew_Bento, Value=Get PendingCMSTexture() -> (PendingCMSTexture:Texture Object Reference))
   > UpdateBackground!(
    execute
    self:Fort Item Preview Background Object Reference=Get FortItemPreviewBackground() -> (FortItemPreviewBackground:Fort Item Preview Background Object Reference)
    LoadedBackground:Object Reference=@n23
    TransitionDirection:EVaultWorldTransitionDirection Enum=Get PendingDirection() -> (PendingDirection:EVaultWorldTransitionDirection Enum=Forward)
    PrimaryColor:Linear Color Structure=(R=0,G=0,B=0,A=0)
    SecondaryColor:Linear Color Structure=(R=0,G=0,B=0,A=0)
   ) -> (then)
   UpdateBackground(
    execute=@n22.CastFailed
    self=Get FortItemPreviewBackground()
    LoadedBackground=Get PendingLoadedBackground()
    TransitionDirection=Get PendingDirection() -> (PendingDirection=Forward)
    PrimaryColor=(R=0,G=0,B=0,A=0)
    SecondaryColor=(R=0,G=0,B=0,A=0)
   )
   SetScalarParameterValue!(
    execute=@n21.then_1
    Collection:Material Parameter Collection Object Reference=/Game/Environments/FrontEnd/Blueprints/ItemPreview/VaultWorldTransitions/MaterialCollectionParameters/MPC_UI_Master_VaultWorld_Settings
    ParameterName:Name=Is Horizontal Orientation
    ParameterValue:float=0
   ) -> (then) {is=title K2Node_CallMaterialParameterCollectionFunction}
   > SetVectorParameterValue!(
    execute
    Collection:Material Parameter Collection Object Reference=&MPC_UI_Master_VaultWorld_Settings
    ParameterName:Name=FocalPointPosition
    ParameterValue_R:float=Get PendingFocalPointValue() -> (PendingFocalPointValue:double)
    ParameterValue_G:float=0
    ParameterValue_B:float=0
    ParameterValue_A:float=0
   ) -> (then) {is=title K2Node_CallMaterialParameterCollectionFunction}
   n24 = OnPreviewItemShown!() -> (OutputDelegate, then, ItemDefinition:Fort Account Item Definition Object Reference) {is=title K2Node_CustomEvent}
   > UpdatePreviewWindowPadding!(execute, self, ItemDefinition:Fort Account Item Definition Object Reference=@n24.ItemDefinition) -> (then)
   Event On Activated()
   > ToggleGiftbox!(execute, self, Enable:bool=true) -> (then)
   > n25 = GetAndClearIsNavigateToPurchaseExternal!(
    execute
    self:Season Pass Selection VM Object Reference=Get SeasonPassSelectionVM() -> (SeasonPassSelectionVM:Season Pass Selection VM Object Reference)
   ) -> (then, ReturnValue:bool)
   > n26 = Branch(Condition=@n25)
   > Show Purchase Screen()
   SetHasSeenPassOrHUB(execute=@n26.else, self:Season Pass System VM Object Reference=Get SeasonPassSystemVM())
   > n27 = GetAndClearIsExternalNavigation!(execute, self:Season Pass Selection VM Object Reference=Get SeasonPassSelectionVM()) -> (then, ReturnValue:bool)
   > n28 = Branch(Condition=@n27)
   > PushPrimaryContentClassAsync()
   Decide if Showing Upsell A/B Test!(Exec:Exec=@n28.else) -> (Else:Exec) {is=title K2Node_MacroInstance}
   > ClearSelectedPage(self:Season Pass Selection VM Object Reference=Get SeasonPassSelectionVM())
   > DelayUntilNextTick!(execute) -> (then)
   > Branch(Condition=Get bIsActive())
   > PlayUpsellMusic!(execute, self, bPlay:bool=true) -> (then)
   > UpdatePreferredPass()
   > n29 = Branch(Condition=Get IsActivatingFromSubscreen() -> (IsActivatingFromSubscreen:bool))
   n30 = Get SelectedPass!(self:Season Pass Selection VM Object Reference=Get SeasonPassSelectionVM()) -> (SelectedPass:Season Pass VM Object Reference)
   n31 = Is Valid(exec=[@n41.then, @n29.then, @n40.else], InputObject=@n30)
   n32 = HandleChangePage(execute=@n31.Is Not Valid, TabID=Crew)
   SelectTabByIndex(
    execute=[@n32.then, @n39.then]
    self=Get UIKit_SideNavigation()
    Index=Get CachedTabIndex() -> (CachedTabIndex:int=0)
    bSuppressClickFeedback=true
   )
   > SetFrontEndCamera!(
    execute
    self:Fort Front End Context Object Reference=Get FortFrontEndContext() -> (ReturnValue:Fort Front End Context Object Reference) {is=title K2Node_GetSubsystem}
    NewState:EFrontEndCamera Enum=CosmeticDisplay
    TransitionParams:View Target Transition Params Structure=(BlendTime=0,BlendFunction=VTBlend_Cubic,BlendExp=2,bLockOutgoing=False)
   ) -> (then)
   n33 = Get SeasonPassFortItemPreviewWidget() -> (SeasonPassFortItemPreviewWidget:Season Pass Fort Item Preview Widget Object Reference)
   > SetZoomLevel!(execute, self:Fort Item Preview Screen Object Reference=@n33, NewZoomLevel:float=0, bSnap:bool=true, bZoomEvenIfNotAllowed:bool=false) -> (then)
   > ActivateWidget(self=@n33)
   > Bind Event to On Preview Item Shown(self:Fort Item Preview Widget Object Reference=@n33, Delegate=@n24) {is=title K2Node_AddDelegate}
   > n34 = Get!(execute) -> (then, else, SeasonPassSelectionVM:Season Pass Selection VM Object Reference)
   > n35 = SetPreviewData!(
    execute
    self:Fort Item Preview Widget Object Reference=@n33
    Data:"Fort Cosmetic Preview Data Structure (by ref)"=GetPreviewData!(self:Season Pass Selection VM Object Reference=@n34) -> (ReturnValue:"Fort Cosmetic Preview Data Structure (by ref)")
   ) -> (then)
   PlayAnimation(execute=[@n35.then, @n34.else], InAnimation=Get Intro())
   > UpdateActionbarBackplateVisibility()
   > SetIsCollapsed!(
    execute
    self:WBP UIKit Side Navigation Base Object Reference=Get UIKit_SideNavigation()
    IsCollapsed:bool=NOT Boolean(A=IsMobileGame())
    SkipAnimation:bool=true
   ) -> (then)
   > UpdateNexusFrontendMode!(execute, Target:Widget Object Reference=Self-Reference(), Is Activating:bool=true) -> (then)
   > Set Clipping!(
    execute
    Clipping:EWidgetClipping Enum=Inherit
    self:Widget Object Reference=Get ScrollBox_Tabs!(self:WBP UIKit Tab Group Base Object Reference=Get UIKit_SideNavigation()) -> (ScrollBox_Tabs:Scroll Box Object Reference)
   ) -> (then, Output_Get:EWidgetClipping Enum=Inherit)
   > Set IsActivatingFromSubscreen(IsActivatingFromSubscreen=false)
   > Fire Event Main View Opened!(execute, View:Widget Object Reference=Self-Reference(), AdditionalInfo:Flow Analytics Additional Info Structure, bIsOverlay:bool=false) -> (then)
   > Is Valid(InputObject=Get SelectedPass(self=Get SeasonPassSelectionVM()))
   > n36 = Cast To FortPlayerController(Object=GetOwningPlayer())
   n37 = Get SeasonPassSelectionVM()
   n38 = Get Source String from Text!(Text:Text=GetDisplayName!(self:Season Pass VM Object Reference=Get SelectedPass(self=@n37)) -> (ReturnValue:Text)) -> (ReturnValue:String) {is=title K2Node_CallFunction}
   > Branch(Condition=NOT Boolean(A=IsEmpty(InString=@n38)))
   > "Fire Event Sub View Opened (Non-Localized)"!(
    execute
    FortPC:Fort Player Controller Object Reference=@n36
    ViewName:String=@n38
    AdditionalInfo:Flow Analytics Additional Info Structure
    bIsOverlay:bool=false
   ) -> (then)
   n39 = HandleChangePage(execute=@n31.Is Valid, TabID=String To Name(InString=GetObjectName(Object=@n30)))
   n40 = Get!(execute=@n29.else) -> (then, else, PreferredPass:Season Pass VM Object Reference)
   > n41 = Set SelectedPass!(execute, SelectedPass:Season Pass VM Object Reference=@n40, self:Season Pass Selection VM Object Reference=Get SeasonPassSelectionVM()) -> (then, Output_Get:Season Pass VM Object Reference)
   K2Node_Event_1 = Show Purchase Screen() {is=title K2Node_CustomEvent}
   > n42 = Is Valid(InputObject=Get SelectedPass(self=Get SeasonPassSelectionVM()))
   PushPrimaryContentClassAsync(execute=@n42.Is Not Valid, WidgetClass=/Crew/UI/CrewInfoContainer.CrewInfoContainer_C)
   n43 = PushPrimaryContentSoftClassAsync!(
    execute:Exec=@n42.Is Valid
    SoftWidgetClass:Common Activatable Widget Soft Class Reference=GetActivatableWidgetForAlias!(WidgetAlias:Gameplay Tag Structure=(TagName="UI.Alias.PassPurchase")) -> (ReturnValue:Common Activatable Widget Soft Class Reference)
   ) -> (then:Exec, OnPrimaryContentPushed:Exec, ActivatableWidget:Common Activatable Widget Object Reference) {is=title K2Node_AsyncAction}
   SetViewModelByClass!(
    execute=@n43.OnPrimaryContentPushed
    self:MVVM View Object Reference=GetExtension!(self:User Widget Object Reference=@n43, ExtensionType:User Widget Extension Class Reference=/Script/ModelViewViewModel.MVVMView) -> (ReturnValue:MVVM View Object Reference)
    NewValue:Notify Field Value Changed Interface=Get SelectedPass(self=Get SeasonPassSelectionVM())
   ) -> (then, ReturnValue:bool)
   Event Destruct()
   > StopListeningToEvent(self:Async Action Start Listening To Event Object Reference=Get ChatEventRef() -> (ChatEventRef:Async Action Start Listening To Event Object Reference))
   Event Construct()
   > SetMobileStyle() {is=title K2Node_CallFunction}
   > n44 = Handle Bottom Bar Visibility for Text Chat!(
    execute:Exec
    Target:Widget Object Reference=Self-Reference()
    BottomBarWidget:Widget Object Reference=Get WBP_ActionBarContainer() -> (WBP_ActionBarContainer:WBP Action Bar Container Object Reference)
   ) -> (then:Exec, AsyncTaskRef:Async Action Start Listening To Event Object Reference) {is=title K2Node_MacroInstance}
   > Set ChatEventRef(ChatEventRef:Async Action Start Listening To Event Object Reference=@n44) -> (Output_Get:Async Action Start Listening To Event Object Reference)
   Event On Deactivated()
   > ToggleGiftbox(Enable=false)
   n45 = Get SeasonPassFortItemPreviewWidget()
   > ClearPreviewItem(self:Fort Item Preview Widget Object Reference=@n45)
   > Unbind all Events from On Preview Item Shown!(execute:Exec, self:Fort Item Preview Widget Object Reference=@n45) -> (then:Exec) {is=title K2Node_ClearDelegate}
   > DeactivateWidget(self=@n45)
   > QuickResetIntroAnimation()
   > UpdateNexusFrontendMode(Target=Self-Reference(), Is Activating=false)
   > PlayUpsellMusic(bPlay=false)
   > n46 = Branch(Condition=Get IsActivatingFromSubscreen())
   UpdateBackground(
    execute=@n46.else
    self=Get FortItemPreviewBackground()
    TransitionDirection=None
    PrimaryColor=(R=0,G=0,B=0,A=0)
    SecondaryColor=(R=0,G=0,B=0,A=0)
   )
  ]}
 GetPrimaryContentSetup {nodes [
   GetPrimaryContentSetup() {is=title K2Node_FunctionEntry}
   > Return Node(ReturnValue:Primary Content Setup Structure)
  ]}
 GetSeasonPassTemplateIds {nodes [
   n1 = GetSeasonPassTemplateIds!() -> (then, Benefits:Array of Crew Benefits Group VM Object References) {is=title K2Node_FunctionEntry}
   > n2 = For Each Loop(Array:Array of Crew Benefits Group VM Object References=@n1) -> (Array Element:Crew Benefits Group VM Object Reference)
   > n3 = For Each Loop(Array:Array of Crew Benefit VM Object References=Get Benefits!(self:Crew Benefits Group VM Object Reference=@n2) -> (Benefits:Array of Crew Benefit VM Object References)) -> (Array Element:Crew Benefit VM Object Reference)
   Return Node!(execute=@n3.Completed, TemplateIds:Array of Strings=Get Results() -> (Results:Array of Strings))
   > Branch(Condition="Equal (Enum)"(
     A:ECrewTileType Enum=GetType!(self:Crew Benefit VM Object Reference=@n3) -> (ReturnValue:ECrewTileType Enum=None)
     B:ECrewTileType Enum=BattlePass
    ))
   > Add(TargetArray=Get Results(), NewItem=GetTemplateId!(self:Crew Benefit VM Object Reference=@n3) -> (ReturnValue:String)) {is=title K2Node_CallArrayFunction}
  ]}
 HandleChangePage {nodes [
   HandleChangePage!() -> (then, TabID:Name) {is=title K2Node_FunctionEntry}
   n1 = GetTabIndexByID!(self:UIKit Tab Group Object Reference=Get UIKit_SideNavigation(), TabNameID:Name=Get TabID() -> (TabID:Name=None)) -> (ReturnValue:int=0)
   n2 = Get CachedTabIndex() -> (CachedTabIndex=0)
   n3 = "Equal (Integer)"(A=Get CachedTabIndex() -> (CachedTabIndex=0), B=@n1)
   > Set ScrollDirection(ScrollDirection:EVaultWorldTransitionDirection Enum=Select(
     Option 0:EVaultWorldTransitionDirection Enum=Select(Option 0=Backward, Option 1=None, Index=@n3)
     Option 1:EVaultWorldTransitionDirection Enum=Forward
     Index="integer > integer"(A=@n1, B=@n2)
    ) -> (ReturnValue:EVaultWorldTransitionDirection Enum)) -> (Output_Get:EVaultWorldTransitionDirection Enum=Forward)
   > Set CachedTabIndex(CachedTabIndex:int=GetTabIndexByID(self=Get UIKit_SideNavigation(), TabNameID=Get TabID() -> (TabID=None)) -> (ReturnValue=0)) -> (Output_Get:int=0)
   > n4 = Sequence!(execute) -> (then_0, then_1, then_2:Exec)
   UpdateActionbarBackplateVisibility(execute=@n4.then_1)
   UpdateUpsellMusic!(execute=@n4.then_2, self, ParamName:Name=Get TabID() -> (TabID=None)) -> (then)
   n5 = Cast To SeasonPassVM!(
    execute=@n4.then_0
    Object=GetTabViewModel!(self:UIKit Tab Group Object Reference=Get UIKit_SideNavigation(), TabID:Name=Get TabID() -> (TabID=None)) -> (ReturnValue:MVVM Base Viewmodel Object Reference)
   ) -> (then, CastFailed, AsSeason Pass VM:Season Pass VM Object Reference) {is=title K2Node_DynamicCast}
   n6 = Get Button_SubTerms() -> (Button_SubTerms:WBP UIKit Button Quiet Object Reference)
   n7 = Get CrewNavWrapper()
   > SetVisibility(self=[@n6, @n7, Get Button_CancelInfo() -> (Button_CancelInfo:WBP UIKit Button Quiet Object Reference)], InVisibility=Collapsed)
   n8 = Get SeasonPass_Level() -> (SeasonPass_Level:Season Pass Level Object Reference)
   n9 = Get CategoriesNavWrapper() -> (CategoriesNavWrapper:Overlay Object Reference)
   > SetVisibility(self=[@n8, @n9], InVisibility=SelfHitTestInvisible)
   > n10 = Set SeasonPassVM!(execute, SeasonPassVM:Season Pass VM Object Reference=@n5, self:Season Pass Page Object Reference=Get SeasonPassPage()) -> (then, Output_Get:Season Pass VM Object Reference)
   > On Open Pass Screen() {is=title K2Node_CallFunction}
   > n11 = Set SelectedPass(SelectedPass=@n10, self=Get SeasonPassSelectionVM())
   > HandleSeasonPassBackgroundTransition!(
    execute
    self
    TransitionDirection:EVaultWorldTransitionDirection Enum=Get ScrollDirection() -> (ScrollDirection:EVaultWorldTransitionDirection Enum=Forward)
   ) -> (then)
   n12 = Get LegalText() -> (LegalText:Common Text Object Reference)
   > "SetText (Text)"(self=@n12, InText=GetLegalText!(self:Season Pass VM Object Reference=@n11) -> (ReturnValue:Text))
   n13 = Get CrewVM() -> (CrewVM:Crew VM Object Reference)
   n14 = Get Button_JoinCrew() -> (Button_JoinCrew:WBP UIKit Button Quiet Object Reference)
   > SetVisibility(
    self=@n14
    InVisibility=Select(
     Option 0:ESlateVisibility Enum=SelfHitTestInvisible
     Option 1:ESlateVisibility Enum=Collapsed
     Index=IsSubscribed!(self:Crew VM Object Reference=@n13) -> (ReturnValue:bool)
    ) -> (ReturnValue:ESlateVisibility Enum)
   )
   SetVisibility(execute=@n5.CastFailed, self=[@n6, @n7], InVisibility=SelfHitTestInvisible)
   > OpeningTile(self:Crew Landing Page Object Reference=Get CrewLandingPage())
   n15 = Get SubscriptionState!(self:Crew VM Object Reference=Get CrewVM()) -> (SubscriptionState:EMcpSubscriptionState Enum=Inactive)
   > SetVisibility(
    self=Get Button_CancelInfo()
    InVisibility=Select(
     Option 0=Collapsed
     Option 1=SelfHitTestInvisible
     Index=Select!(
      Inactive:bool=false
      Active:bool=true
      ActiveWithExternalBilling:bool=true
      NotAutoRenewable:bool=false
      Canceled:bool=false
      PaymentProcessError:bool=true
      PlatformAuthError:bool=true
      RewardsDelayed:bool=true
      Unavailable:bool=false
      Unknown:bool=false
      Index:EMcpSubscriptionState Enum=@n15
     ) -> (ReturnValue:bool)
    )
   )
   > SetVisibility(self=[@n8, @n9], InVisibility=Collapsed)
   > ClearSelectedPass(self:Season Pass Selection VM Object Reference=Get SeasonPassSelectionVM())
   > n16 = Branch(Condition=IsSubscribed(self=Get CrewVM()))
   > n17 = SetPreviewData!(
    execute
    self:Season Pass Selection VM Object Reference=Get SeasonPassSelectionVM()
    InPreviewData:Array of Fort Cosmetic Preview Data Structures=Get PreviewData!(self:Crew Progressive Items VM Object Reference=Get CrewProgressiveItemsVM() -> (CrewProgressiveItemsVM:Crew Progressive Items VM Object Reference)) -> (PreviewData:Array of Fort Cosmetic Preview Data Structures)
   ) -> (then)
   HandleCrewBackgroundTransition!(
    execute=[@n17.then, @n18.then]
    self
    TransitionDirection:EVaultWorldTransitionDirection Enum=Get ScrollDirection() -> (ScrollDirection=Forward)
   ) -> (then)
   > "SetText (Text)"(self=@n12, InText=Get CrewLegalText() -> (CrewLegalText:Text))
   > SetVisibility(self=@n14, InVisibility=Collapsed)
   n18 = ClearPreviewItem(execute=@n16.else, self=Get SeasonPassFortItemPreviewWidget())
  ]}
 HandleCrewBackgroundTransition {nodes [
   n1 = HandleCrewBackgroundTransition!() -> (then, TransitionDirection:EVaultWorldTransitionDirection Enum=Forward) {is=title K2Node_FunctionEntry}
   n2 = Get CrewVM()
   > RequestLoadBackground!(
    execute
    self
    BackgroundAsset:Soft Object Reference=Select!(
     Option 0:Soft Object Reference=/Crew/UI/Materials/M_UI_CrewUpsel_VW__US
     Option 1:Soft Object Reference=/Crew/UI/Materials/M_UI_CrewUpsel_VW_S
     Index:bool=IsSubscribed(self=@n2)
    ) -> (ReturnValue:Soft Object Reference)
    Direction:EVaultWorldTransitionDirection Enum=@n1
    FocalPointValue:double=0
   ) -> (then) {is=title K2Node_CallFunction}
   > PlaySound2D(Sound=Get CrewScreenAppear() -> (CrewScreenAppear:Sound Base Object Reference))
  ]}
 HandleSeasonPassBackgroundTransition {nodes [
   n1 = HandleSeasonPassBackgroundTransition!() -> (then, TransitionDirection:EVaultWorldTransitionDirection Enum=Forward) {is=title K2Node_FunctionEntry}
   n2 = Get SelectedPass(self=Get SeasonPassSelectionVM())
   > RequestLoadBackground(
    BackgroundAsset=Break Season Pass Theme Data!(SeasonPassThemeData:"Season Pass Theme Data Structure (by ref)"=GetThemeData!(self:Season Pass VM Object Reference=@n2) -> (ReturnValue:"Season Pass Theme Data Structure (by ref)")) -> (
     VaultBackgroundMaterial:Soft Object Reference
     GiftboxVaultBackgroundMaterial:Soft Object Reference
     CustomLogo:Soft Object Reference
     CustomSeasonTitle:Text
     CustomSeasonTitleColor:Linear Color Structure
     Upsell2BackgrounMaterial:Soft Object Reference
     ImageReferenceTable:Fort Image Reference Table Structure
     ColorReferenceTable:Fort Color Reference Table Structure
     TextReferenceTable:Fort Text Reference Table Structure
    ) {is=title K2Node_BreakStruct}
    Direction=@n1
    FocalPointValue=Make Literal Float(Value=0.4)
   )
   > PlaySound2D(Sound=Get SeasonPassScreenAppear() -> (SeasonPassScreenAppear:Sound Base Object Reference))
  ]}
 NavigateToContent {nodes [
   NavigateToContent!() -> (then, Navigation:EUINavigation Enum) {is=title K2Node_FunctionEntry}
   n1 = IsVisible(self=Get CrewNavWrapper())
   n2 = Get SeasonPassPage()
   n3 = Get CrewLandingPage()
   > Return Node!(
    execute
    ReturnValue:Widget Object Reference=GetDesiredFocusTarget(self=Select(Option 0:Common Activatable Widget Object Reference=@n2, Option 1:Common Activatable Widget Object Reference=@n3, Index=@n1) -> (ReturnValue:Common Activatable Widget Object Reference))
   )
  ]}
 NavigateToSideNav {nodes [
   NavigateToSideNav!() -> (then, Navigation:EUINavigation Enum) {is=title K2Node_FunctionEntry}
   > Return Node(ReturnValue=GetSideNavigationSelectedButtonBase!(self:WBP UIKit Side Navigation Base Object Reference=Get UIKit_SideNavigation()) -> (Button:Common Button Base Object Reference))
  ]}
 OnBackgroundURLChanged {nodes [
   n1 = OnBackgroundURLChanged!() -> (then, URL:String) {is=title K2Node_FunctionEntry}
   > Set PendingCMSTexture()
   > Set IsCMSTextureLoaded(IsCMSTextureLoaded=false)
   > SetMediaURL!(
    execute
    self:Epic CMSImage Object Reference=Get EpicCMSImage_BackgroundLoaderHelper() -> (EpicCMSImage_BackgroundLoaderHelper:Epic CMSImage Object Reference)
    MediaUrl:String=@n1
   ) -> (then)
  ]}
 OnMouseWheel {nodes [
   n1 = OnMouseWheel!() -> (then, MyGeometry:Geometry Structure, MouseEvent:"Pointer Event Structure (by ref)") {is=title K2Node_FunctionEntry}
   > ScrollToNextCategory(Up="float > float"(A=Get Wheel Delta!(Input:"Pointer Event Structure (by ref)"=@n1.MouseEvent) -> (ReturnValue:float) {is=title K2Node_CallFunction}))
   > Return Node(ReturnValue:Event Reply Structure=Handled())
  ]}
 PlayUpsellMusic {nodes [PlayUpsellMusic!() -> (then, bPlay:bool=~) {is=title K2Node_FunctionEntry}]}
 QuickResetIntroAnimation {nodes [
   QuickResetIntroAnimation() {is=title K2Node_FunctionEntry}
   > SetRenderOpacity(
    self=[
     Get CrewNavWrapper()
     Get CategoriesNavWrapper()
     Get CategoryTabs_NavWrapper() -> (CategoryTabs_NavWrapper:Grid Panel Object Reference)
     Get SeasonPass_Level()
    ]
    InOpacity=0
   )
  ]}
 ScrollToNextCategory {nodes [
   ScrollToNextCategory!() -> (then, Up:bool=~) {is=title K2Node_FunctionEntry}
   n1 = Get UIKit_SideNavigation()
   n2 = int + int(
    A=SelectInt(A=-1, B=1, bPickA=Get Up() -> (Up:bool))
    B=GetSelectedTabIndex!(self:UIKit Tab Group Object Reference=@n1) -> (ReturnValue:int=0)
   )
   > Set NewTabIndex(NewTabIndex:int=@n2) -> (Output_Get:int=0)
   n3 = GetTabCount!(self:Common Tab List Widget Base Object Reference=@n1) -> (ReturnValue:int=0)
   > Branch(Condition=AND Boolean!(A="integer >= integer"(A=@n2), B="integer < integer"(A=@n2, B=@n3)) -> (ReturnValue))
   n4 = Get AnimMoveUp() -> (AnimMoveUp:Widget Animation Object Reference)
   n5 = Get AnimMoveDown() -> (AnimMoveDown:Widget Animation Object Reference)
   > n6 = Branch(Condition=OR Boolean(A=IsAnimationPlaying(InAnimation=@n5), B=IsAnimationPlaying(InAnimation=@n4)))
   PlayAnimation(
    execute=@n6.else
    InAnimation=Select(Option 0:Widget Animation Object Reference=Get AnimMoveDown(), Option 1:Widget Animation Object Reference=Get AnimMoveUp(), Index=Get Up()) -> (ReturnValue:Widget Animation Object Reference)
    PlaybackSpeed=2
   )
   > SelectTabByIndex(self=Get UIKit_SideNavigation(), Index=Get NewTabIndex() -> (NewTabIndex:int=0), bSuppressClickFeedback=true)
  ]}
 SeasonPassesChanged {nodes [
   n1 = SeasonPassesChanged!() -> (then, SeasonPasses:Array of Season Pass VM Object References) {is=title K2Node_FunctionEntry}
   n2 = Get UIKit_SideNavigation()
   > ClearSideNavigationTabs(self:WBP UIKit Side Navigation Base Object Reference=@n2)
   > AddSideNavigationTab!(
    execute
    self:WBP UIKit Side Navigation Base Object Reference=@n2
    TabViewModel:MVVM Base Viewmodel Object Reference=Get CrewVM()
    TabID:Name=Crew
    Index:int=-1
   ) -> (then)
   > AddDivider!(execute, self:WBP UIKit Tab Group Base Object Reference=@n2, TabID:Name=Crew, AddAfter:bool=true) -> (then)
   > n3 = For Each Loop(Array:Array of Season Pass VM Object References=@n1) -> (Array Element:Season Pass VM Object Reference)
   > AddSideNavigationTab(self=Get UIKit_SideNavigation(), TabViewModel=@n3, TabID=String To Name(InString=GetObjectName(Object=@n3)), Index=-1)
  ]}
 SetMobileStyle {nodes [
   SetMobileStyle() {is=title K2Node_FunctionEntry}
   > Branch(Condition=IsMobileGame())
   > Set EnableCollapsedState!(execute, EnableCollapsedState:bool=false, self:WBP UIKit Side Navigation Base Object Reference=Get UIKit_SideNavigation()) -> (then, Output_Get:bool)
   n1 = SlotAsGridSlot(Widget=Get CategoriesNavWrapper())
   > SetVerticalAlignment(self:Grid Slot Object Reference=@n1, InVerticalAlignment=VAlign_Bottom)
   > SetPadding!(
    execute
    self:Grid Slot Object Reference=@n1
    InPadding_Left:float=0
    InPadding_Top:float=0
    InPadding_Right:float=0
    InPadding_Bottom:float=SelectFloat(A=8, B=72, bPickA=Aspect Ratio Check!() -> (ReturnValue:bool=~) {is=title K2Node_MacroInstance})
   ) -> (then)
   > SetPadding(
    self:Size Box Slot Object Reference=SlotAsSizeBoxSlot!(Widget:Widget Object Reference=Get Overlay() -> (Overlay:Overlay Object Reference)) -> (ReturnValue:Size Box Slot Object Reference)
    InPadding_Left=48
    InPadding_Top=24
    InPadding_Right=48
    InPadding_Bottom=24
   )
   > SetPadding(
    self:Grid Slot Object Reference=SlotAsGridSlot(Widget=Get ProgressStatus() -> (ProgressStatus:Vertical Box Object Reference))
    InPadding_Left=0
    InPadding_Top=128
    InPadding_Right=0
    InPadding_Bottom=0
   )
   n2 = Get SizeBox_0() -> (SizeBox_0:Size Box Object Reference)
   > ClearWidthOverride(self:Size Box Object Reference=@n2)
   n3 = SlotAsOverlaySlot(Widget=@n2)
   > SetHorizontalAlignment(self=@n3, InHorizontalAlignment=HAlign_Fill)
   > SetVerticalAlignment(self:Overlay Slot Object Reference=@n3, InVerticalAlignment=VAlign_Fill)
   > SetWrapTextWidth!(execute, self:Common Text Object Reference=Get LegalText(), InWrapTextAt:int=420) -> (then)
   n4 = SlotAsHorizontalBoxSlot!(Widget:Widget Object Reference=Get SizeBox_PreviewWindow() -> (SizeBox_PreviewWindow:Size Box Object Reference)) -> (ReturnValue:Horizontal Box Slot Object Reference)
   > SetPadding(
    self:Horizontal Box Slot Object Reference=@n4
    InPadding_Left=SelectFloat(A=-256, B=-128, bPickA=Aspect Ratio Check() -> (ReturnValue=~))
    InPadding_Top=128
    InPadding_Right=0
    InPadding_Bottom=0
   )
   > Set CachedCompactUIPadding(CachedCompactUIPadding:Margin Structure=Get Padding!(self:Horizontal Box Slot Object Reference=@n4) -> (Padding:Margin Structure)) -> (Output_Get:Margin Structure)
   > SetPadding(
    self:Grid Slot Object Reference=SlotAsGridSlot(Widget=Get CategoryTabs_NavWrapper())
    InPadding_Left=0
    InPadding_Top=171
    InPadding_Right=56
    InPadding_Bottom=0
   )
   n5 = Aspect Ratio Check() -> (ReturnValue=~)
   > SetPadding(
    self:Overlay Slot Object Reference=SlotAsOverlaySlot(Widget=Get CrewLandingPage())
    InPadding_Left=0
    InPadding_Top=SelectFloat(A=0, B=200, bPickA=@n5)
    InPadding_Right=0
    InPadding_Bottom=SelectFloat(A=16, B=72, bPickA=@n5)
   )
  ]}
 ToggleGiftbox {nodes [
   n1 = ToggleGiftbox!() -> (then, Enable:bool=~) {is=title K2Node_FunctionEntry}
   > n2 = Branch(Condition=@n1)
   > EnableGiftboxesOfType!(execute, DisplayTag:Gameplay Tag Structure=(TagName="Giftbox.Display.PassHub")) -> (then)
   DisableGiftboxesOfType!(execute=@n2.else, DisplayTag:Gameplay Tag Structure=(TagName="Giftbox.Display.PassHub")) -> (then)
  ]}
 UpdateActionbarBackplateVisibility {nodes [
   UpdateActionbarBackplateVisibility() {is=title K2Node_FunctionEntry}
   n1 = Get Button_JoinCrew()
   n2 = Get Button_SubTerms()
   > SetHasPrompts!(
    execute
    self:WBP Action Bar Container Object Reference=Get WBP_ActionBarContainer()
    HasPrompts:bool=OR Boolean(A=IsVisible(self=@n1), B=IsVisible(self=@n2))
   ) -> (then)
  ]}
 UpdatePreferredPass {nodes [
   UpdatePreferredPass() {is=title K2Node_FunctionEntry}
   > n1 = Get!(execute) -> (then, else, SeasonPassSystemVM:Season Pass System VM Object Reference)
   > n2 = For Each Loop(Array=Get SeasonPasses!(self:Season Pass System VM Object Reference=@n1) -> (SeasonPasses:Array of Season Pass VM Object References))
   Branch(execute=@n2.Completed, Condition="integer > integer"(A=Get PassesOwned() -> (PassesOwned:int=0), B=1))
   > Set PreferredPass(PreferredPass:Season Pass VM Object Reference) -> (Output_Get:Season Pass VM Object Reference)
   Branch(execute=@n2.LoopBody, Condition=Get bDoesOwnsPass(self=@n2))
   > Set PreferredPass(PreferredPass=@n2)
   > Set PassesOwned(PassesOwned:int=int + int(A=Get PassesOwned() -> (PassesOwned=0))) -> (Output_Get:int=0)
  ]}
 UpdatePreviewWindowPadding {nodes [
   n1 = UpdatePreviewWindowPadding!() -> (then, ItemDefinition:Fort Account Item Definition Object Reference) {is=title K2Node_FunctionEntry}
   > n2 = Cast To AthenaCosmeticItemDefinition!(execute, Object=@n1) -> (then, CastFailed, AsAthena Cosmetic Item Definition:Athena Cosmetic Item Definition Object Reference) {is=title K2Node_DynamicCast}
   > n3 = Branch(Condition=IsUsingCompactUI())
   n4 = "Equal (Enum)"(
    A:EFortItemType Enum=GetItemType!(ItemDefinition:Item Definition Base Object Reference=@n2) -> (ReturnValue:EFortItemType Enum=WorldItem)
    B:EFortItemType Enum=CosmeticMimosa
   )
   n5 = Get!() -> (
    CachedCompactUIPadding_Left:float=~
    CachedCompactUIPadding_Top:float=~
    CachedCompactUIPadding_Right:float=~
    CachedCompactUIPadding_Bottom:float=~
   )
   > SetPadding(
    self:Horizontal Box Slot Object Reference=SlotAsHorizontalBoxSlot(Widget=Get SizeBox_PreviewWindow())
    InPadding_Left=@n5
    InPadding_Top=@n5.CachedCompactUIPadding_Top
    InPadding_Right=@n5.CachedCompactUIPadding_Right
    InPadding_Bottom=SelectFloat(A=145, B=@n5.CachedCompactUIPadding_Bottom, bPickA=@n4)
   )
   n6 = SlotAsHorizontalBoxSlot(Widget=Get SizeBox_PreviewWindow())
   n7 = Get!(self:Horizontal Box Slot Object Reference=@n6) -> (Padding_Left:float=~, Padding_Top:float=~, Padding_Right:float=~, Padding_Bottom:float=~)
   SetPadding(
    execute=@n3.else
    self=@n6
    InPadding_Left=@n7
    InPadding_Top=@n7.Padding_Top
    InPadding_Right=@n7.Padding_Right
    InPadding_Bottom=SelectFloat(A=-110, B=-510, bPickA=@n4)
   )
  ]}
 UpdateUpsellMusic {nodes [UpdateUpsellMusic!() -> (then, ParamName:Name) {is=title K2Node_FunctionEntry}]}
]
variables [
 CachedCompactUIPadding {type=Margin, default=<ltrb 0 0 0 0>}
 CachedTabIndex {type=integer}
 ChatEventRef {type=AsyncAction_StartListeningToEvent, default=null}
 CrewLegalText {type=string, default=""}
 CrewScreenAppear {type=SoundBase, default=null}
 IsActivatingFromSubscreen {type=boolean}
 IsBackgroundLoaded {type=boolean}
 IsCMSTextureLoaded {type=boolean}
 IsVerticalTransitioning {type=boolean}
 PendingCMSTexture {type=Texture, default=null}
 PendingDirection {type=EVaultWorldTransitionDirection, default=Forward}
 PendingFocalPointValue {type=number}
 PendingLoadedBackground {type=Object, default=null}
 PreferredPass {type=SeasonPassVM, default=null}
 RequestedBackgroundSoftPath {type=SoftObjectPath, default=""}
 Season Pass Header {type=SeasonPass_Header_C, default=null}
 SeasonPassScreenAppear {type=SoundBase, default=null}
]
widgets {
 info {
  inheritedWidgetCount=0
  namedSlotCount=1
  parentClass=&CommonActivatableWidget
  rootWidgetClass=/Script/FortniteUI.FortSwipePanel
  widgetCount=35
 }
 widgets [
  SwipePanel {
   widgetClassPath=&FortSwipePanel
   bIsVariable=true
   widget=SwipePanel
   properties {bConsumePointerInput=false, thresholdUnit=Inches, swipeThreshold=<0 0.25>}
  }
  SafeZone {widgetClassPath=&SafeZone, bIsVariable=true, parent=SwipePanel, widget=SafeZone, slot {class=FortSwipePanelSlot}}
  Overlay_1 {widgetClassPath=&Overlay, parent=SafeZone, widget=Overlay_1, slot {class=SafeZoneSlot}}
  EpicCMSImage_BackgroundLoaderHelper {
   widgetClassPath=/Script/EpicCMSUIFramework.EpicCMSImage
   bIsVariable=true
   parent=Overlay_1
   widget=EpicCMSImage_BackgroundLoaderHelper
   properties {visibility=Collapsed}
   slot {class=OverlaySlot}
  }
  WBP_CaptureForPostBufferUpdate {
   widgetClassPath=&WBP_CaptureForPostBufferUpdate
   bIsVariable=true
   parent=Overlay_1
   slot {class=OverlaySlot, properties {horizontalAlignment=HAlign_Fill, verticalAlignment=VAlign_Fill}}
  }
  SizeBox_0 {
   widgetClassPath=&SizeBox
   bIsVariable=true
   parent=Overlay_1
   widget=SizeBox_0
   properties {bOverride_WidthOverride=true, widthOverride=1920}
   slot {class=OverlaySlot, properties {horizontalAlignment=HAlign_Center, verticalAlignment=VAlign_Center}}
  }
  Overlay {
   widgetClassPath=&Overlay
   bIsVariable=true
   parent=SizeBox_0
   widget=Overlay
   slot {class=SizeBoxSlot, properties {padding=<ltrb 80 80 80 40>}}
  }
  Grid_Main {
   widgetClassPath=&GridPanel
   bIsVariable=true
   parent=Overlay
   widget=Grid_Main
   properties {columnFill [0,1,1], rowFill [1]}
   slot {class=OverlaySlot, properties {horizontalAlignment=HAlign_Fill, verticalAlignment=VAlign_Fill, padding=<ltrb 0 0 0 80>}}
  }
  Widget_CameraFrame {
   widgetClassPath=&SizeBox
   bIsVariable=true
   parent=Grid_Main
   widget=Widget_CameraFrame
   slot {class=GridSlot, properties {column=1, columnSpan=2}}
  }
  HorizontalBox_0 {widgetClassPath=&HorizontalBox, parent=Widget_CameraFrame, widget=HorizontalBox_0, slot {class=SizeBoxSlot}}
  Camera_Spacer_Left {
   widgetClassPath=&Spacer
   bIsVariable=true
   parent=HorizontalBox_0
   widget=Camera_Spacer_Left
   slot {class=HorizontalBoxSlot, properties {size {sizeRule=Fill, value=1}}}
  }
  SizeBox_PreviewWindow {
   widgetClassPath=&SizeBox
   bIsVariable=true
   parent=HorizontalBox_0
   widget=SizeBox_PreviewWindow
   properties {bOverride_HeightOverride=true, bOverride_WidthOverride=true, heightOverride=2048, widthOverride=1024}
   slot {class=HorizontalBoxSlot, properties {horizontalAlignment=HAlign_Center, verticalAlignment=VAlign_Center, padding=<ltrb -100 -200 0 -510>}}
  }
  SeasonPassFortItemPreviewWidget {
   widgetClassPath=&/SeasonPassFortItemPreviewWidget
   bIsVariable=true
   parent=SizeBox_PreviewWindow
   widget=SeasonPassFortItemPreviewWidget
   properties {
    bCanDisplayUserAccessories=false
    bPreviewActionsEnabled=false
    bSupportsActivationFocus=false
    itemViewObject=&self:WidgetTree.SeasonPassFortItemPreviewWidget.ItemViewObject
    visibility=Visible
    previewSceneChangerTag {tagName=UI.ItemPreview.SceneChanger.MPItemShop}
   }
   slot {class=SizeBoxSlot, properties {padding=<ltrb 0 0 0 96>}}
  }
  Camera_Spacer_Right {
   widgetClassPath=&Spacer
   bIsVariable=true
   parent=HorizontalBox_0
   widget=Camera_Spacer_Right
   slot {class=HorizontalBoxSlot, properties {size {sizeRule=Fill, value=0}}}
  }
  FortItemPreviewBackground {
   widgetClassPath=/Script/FortniteUI.FortItemPreviewBackground
   bIsVariable=true
   parent=HorizontalBox_0
   widget=FortItemPreviewBackground
   properties {previewSceneChangerTag {tagName=UI.ItemPreview.SceneChanger.MPItemShop}}
   slot {class=HorizontalBoxSlot}
  }
  CategoryTabs_NavWrapper {
   widgetClassPath=&GridPanel
   bIsVariable=true
   parent=Grid_Main
   widget=CategoryTabs_NavWrapper
   properties {navigation=&self:WidgetTree.CategoryTabs_NavWrapper.WidgetNavigation_0}
   slot {class=GridSlot, properties {horizontalAlignment=HAlign_Left, layer=3, verticalAlignment=VAlign_Top, padding=<ltrb 0 168 56 0>}}
  }
  SizeBox_Mobile {
   widgetClassPath=&SizeBox
   bIsVariable=true
   parent=CategoryTabs_NavWrapper
   widget=SizeBox_Mobile
   slot {class=GridSlot, properties {verticalAlignment=VAlign_Center}}
  }
  UIKit_SideNavigation {
   widgetClassPath=/FortUILibrary/UIKit/Tabs/WBP_UIKit_SideNavigation
   bIsVariable=true
   parent=SizeBox_Mobile
   widget=UIKit_SideNavigation
   properties {
    bIsFocusable=true
    buttonClass=/SeasonPass/Widgets/SeasonPassSideNavigationButton.SeasonPassSideNavigationButton_C
    dividerClass=/Locker/MPL/WBP_MPLocker_Sidebar_Divider.WBP_MPLocker_Sidebar_Divider_C
    dividerSpacing=10
    enableCollapsedState=true
    hidePrompts=true
    innerTabButtonClassOverride=/SeasonPass/Widgets/SeasonPassButton.SeasonPassButton_C
    inputActions=Custom
    nextPromptHorizontalAlignment=HAlign_Right
    prevPromptHorizontalAlignment=HAlign_Left
    promptBackplateBrightness=Dark
    selectInnerTabOnContainerExpand=true
    tabSpacing=0
    usePromptBackplate=false
    buttonClassOverride {crew=&/CrewSideNavigationButton.CrewSideNavigationButton_C}
    buttonPadding=<ltrb 8 4 8 4>
    dividerPadding=<ltrb 17 0 17 0>
    innerTabGroupPadding=<ltrb 4 4 4 4>
    promptBackplateMinSize=<62 48>
    tabGroupPadding=<ltrb 0 4 0 4>
   }
   slot {class=SizeBoxSlot}
  }
  CommonHardwareVisibilityBorder_0 {
   widgetClassPath=&CommonHardwareVisibilityBorder
   parent=CategoryTabs_NavWrapper
   widget=CommonHardwareVisibilityBorder_0
   properties {visibilityQuery {
     autoDescription=" ANY( Input.Gamepad )"
     tokenStreamVersion=0
     userDescription=""
     queryTokenStream [0,1,1,1,0]
     tagDictionary [{tagName=Input.Gamepad}]
    }}
   slot {class=GridSlot, properties {horizontalAlignment=HAlign_Center, row=1, verticalAlignment=VAlign_Center, padding=<ltrb 0 8 0 0>}}
  }
  Button_OpenNav_PromptBelowNav {
   widgetClassPath=&WBP_UIKit_Button_Generic
   bIsVariable=true
   parent=CommonHardwareVisibilityBorder_0
   widget=Button_OpenNav_PromptBelowNav
   properties {
    bHideInputActionWithKeyboard=true
    bIsFocusable=false
    minHeight=40
    promptHorizontalAlignment=HAlign_Center
    useBackgroundBlock=false
    useIconBlock=false
    useTextBlock=false
    visibility=Collapsed
    buttonPadding=<ltrb 0 0 0 0>
    promptNudge=<0 0>
    triggeringInputAction {dataTable=&InputActionDataTable, rowName=GenericBack}
   }
   slot {class=BorderSlot, properties {padding=<ltrb 0 0 0 0>}}
  }
  CategoriesNavWrapper {
   widgetClassPath=&Overlay
   bIsVariable=true
   parent=Grid_Main
   widget=CategoriesNavWrapper
   properties {navigation=&self:WidgetTree.CategoriesNavWrapper.WidgetNavigation_0, visibility=Collapsed}
   slot {class=GridSlot, properties {column=1, columnSpan=2, layer=3}}
  }
  SeasonPassPage {
   widgetClassPath=&/SeasonPassPage
   bIsVariable=true
   parent=CategoriesNavWrapper
   widget=SeasonPassPage
   properties {bSupportsActivationFocus=true}
   slot {class=OverlaySlot, properties {horizontalAlignment=HAlign_Fill, verticalAlignment=VAlign_Bottom}}
  }
  CrewNavWrapper {
   widgetClassPath=&Overlay
   bIsVariable=true
   parent=Grid_Main
   widget=CrewNavWrapper
   properties {navigation=&self:WidgetTree.CrewNavWrapper.WidgetNavigation_0}
   slot {class=GridSlot, properties {column=1, columnSpan=2, layer=3}}
  }
  CrewLandingPage {
   widgetClassPath=&/CrewLandingPage
   bIsVariable=true
   parent=CrewNavWrapper
   widget=CrewLandingPage
   properties {bSetVisibilityOnActivated=true, bSetVisibilityOnDeactivated=true, bSupportsActivationFocus=true}
   slot {class=OverlaySlot, properties {horizontalAlignment=HAlign_Fill, verticalAlignment=VAlign_Fill}}
  }
  ProgressStatus {
   widgetClassPath=&VerticalBox
   bIsVariable=true
   parent=Grid_Main
   widget=ProgressStatus
   slot {class=GridSlot, properties {column=2, horizontalAlignment=HAlign_Right, verticalAlignment=VAlign_Top, padding=<ltrb 0 60 0 0>}}
  }
  SeasonPass_Level {
   widgetClassPath=/FortniteCoreUI/MultiProduct/SeasonPass_Level
   bIsVariable=true
   parent=ProgressStatus
   widget=SeasonPass_Level
   slot {class=VerticalBoxSlot}
  }
  LegalandActionBar {
   widgetClassPath=&HorizontalBox
   parent=Overlay
   widget=LegalandActionBar
   slot {class=OverlaySlot, properties {horizontalAlignment=HAlign_Fill, verticalAlignment=VAlign_Bottom}}
  }
  LegalText {
   widgetClassPath=&CommonTextBlock
   bIsVariable=true
   parent=LegalandActionBar
   widget=LegalText
   properties {
    autoWrapText=true
    mobileFontSizeMultiplier=1.25
    text=Legal text goes here
    wrapTextAt=675
    colorAndOpacity {colorUseRule=UseColor_Specified, specifiedColor=<rgba 1 1 1 0.6>}
    font {
     bForceMonospaced=false
     bMaterialIsStencil=false
     fontMaterial=null
     fontObject=&HeadingNow
     letterSpacing=0
     monospacedWidth=1
     size=24
     skewAmount=0
     typefaceFontName=86 Bold
     outlineSettings {
      bApplyOutlineToDropShadows=false
      bMiteredCorners=false
      bSeparateFillAlpha=false
      outlineMaterial=null
      outlineSize=0
      outlineColor=<rgba 0 0 0 1>
     }
    }
    shadowOffset=<0 0>
   }
   slot {class=HorizontalBoxSlot, properties {verticalAlignment=VAlign_Bottom, size {sizeRule=Fill, value=1}}}
  }
  Spacer_151 {widgetClassPath=&Spacer, parent=LegalandActionBar, widget=Spacer_151, properties {size=<1 48>}, slot {class=HorizontalBoxSlot}}
  WBP_ActionBarContainer {
   widgetClassPath=/Game/UI/Frontend/BottomBar/WBP_ActionBarContainer
   bIsVariable=true
   parent=LegalandActionBar
   widget=WBP_ActionBarContainer
   slot {class=HorizontalBoxSlot, properties {verticalAlignment=VAlign_Bottom}}
  }
  HorizontalBox_1 {widgetClassPath=&HorizontalBox, namedSlotHost=WBP_ActionBarContainer, widget=HorizontalBox_1}
  Button_CancelInfo {
   widgetClassPath=/FortUILibrary/UIKit/Buttons/WBP_UIKit_Button_Quiet
   bIsVariable=true
   parent=HorizontalBox_1
   widget=Button_CancelInfo
   properties {
    bIsFocusable=false
    text=Cancellation Info
    visibility=Collapsed
    triggeringInputAction {dataTable=&InputActionDataTable, rowName=GenericFaceButtonTop}
   }
   slot {class=HorizontalBoxSlot, properties {padding=<ltrb 0 0 24 0>}}
  }
  Button_SubTerms {
   widgetClassPath=&WBP_UIKit_Button_Quiet
   bIsVariable=true
   parent=HorizontalBox_1
   widget=Button_SubTerms
   properties {
    bIsFocusable=false
    text=Subscription Terms
    visibility=Collapsed
    triggeringInputAction {dataTable=&InputActionDataTable, rowName=GenericFaceButtonLeft}
   }
   slot {class=HorizontalBoxSlot, properties {padding=<ltrb 0 0 24 0>}}
  }
  Button_JoinCrew {
   widgetClassPath=&WBP_UIKit_Button_Quiet
   bIsVariable=true
   parent=HorizontalBox_1
   widget=Button_JoinCrew
   properties {
    bIsFocusable=false
    text=Join Fortnite Crew
    visibility=Collapsed
    triggeringInputAction {dataTable=&InputActionDataTable, rowName=GenericFaceButtonLeft}
   }
   slot {class=HorizontalBoxSlot, properties {padding=<ltrb 0 0 24 0>}}
  }
  WBP_AthenaBottomBar_WithNotifications {
   widgetClassPath=/Game/UI/Frontend/BottomBar/WBP_AthenaBottomBar_WithNotifications
   bIsVariable=true
   parent=HorizontalBox_1
   widget=WBP_AthenaBottomBar_WithNotifications
   slot {class=HorizontalBoxSlot}
  }
 ]
}
