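# Activatable wrapper widget for the Season Pass screen that chooses which version of that screen to display.
# Its root switcher holds three screens: SeasonPassesScreen, WBP_SeasonPassesScreen and WBP_SeasonPassScreen_B.
# On activation it reads the bEnableS39BattlePassRedesign config flag, showing screen B when true (the default) and SeasonPassesScreen when false.
# WBP_SeasonPassesScreen is never selected by the graph; its getter node is left unconnected.
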WidgetBlueprint /SeasonPassHub/Widgets/SeasonPassScreenWrapper
graphs [EventGraph {nodes [
   Event On Activated()
   > SetActiveWidget(
    self:Widget Switcher Object Reference=Get ScreenPassesSwitcher() -> (ScreenPassesSwitcher:Common Activatable Widget Switcher Object Reference)
    Widget=Select(
     Option 0:Widget Object Reference=Get SeasonPassesScreen() -> (SeasonPassesScreen:Season Passes Screen Object Reference)
     Option 1:Widget Object Reference=Get WBP_SeasonPassScreen_B() -> (WBP_SeasonPassScreen_B:WBP Season Pass Screen B Object Reference)
     Index=GetConfigDataBoolWithDefault!(EntryName:String=bEnableS39BattlePassRedesign, bDefaultValue:bool=true) -> (ReturnValue:bool)
    ) -> (ReturnValue:Widget Object Reference)
   )
   Get WBP_SeasonPassesScreen() -> (WBP_SeasonPassesScreen:WBP Season Pass Screen Object Reference)
  ]}]
variables []
widgets {
 info {
  inheritedWidgetCount=0
  parentClass=&CommonActivatableWidget
  rootWidgetClass=&CommonActivatableWidgetSwitcher
  widgetCount=4
 }
 widgets [
  ScreenPassesSwitcher {widgetClassPath=&CommonActivatableWidgetSwitcher, bIsVariable=true, widget=ScreenPassesSwitcher}
  SeasonPassesScreen {
   widgetClassPath=&/SeasonPassesScreen
   bIsVariable=true
   parent=ScreenPassesSwitcher
   widget=SeasonPassesScreen
   slot {class=WidgetSwitcherSlot}
  }
  WBP_SeasonPassesScreen {
   widgetClassPath=/SeasonPassHub/Widgets/Reboot/WBP_SeasonPassScreen
   bIsVariable=true
   parent=ScreenPassesSwitcher
   widget=WBP_SeasonPassesScreen
   slot {class=WidgetSwitcherSlot}
  }
  WBP_SeasonPassScreen_B {
   widgetClassPath=/SeasonPassHub/Widgets/Reboot/WBP_SeasonPassScreen_B
   bIsVariable=true
   parent=ScreenPassesSwitcher
   widget=WBP_SeasonPassScreen_B
   slot {class=WidgetSwitcherSlot}
  }
 ]
}
