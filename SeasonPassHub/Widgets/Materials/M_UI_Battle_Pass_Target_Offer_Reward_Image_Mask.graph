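# Translucent UI material for a Battle Pass target offer reward image, masking it with a soft horizontal fade.
# The parameterized reward texture is rotated, translated and scaled about its center.
# A separate UV ramp, mirrored by a localization Flip parameter, is offset, smoothstepped and sine-eased into the fade.
# Named reroutes carry the color through, while final alpha is the minimum of texture alpha and an inverted mask.
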Material /SeasonPassHub/Widgets/Materials/M_UI_Battle_Pass_Target_Offer_Reward_Image_Mask
graph {
 properties {bAutomaticallySetUsageInEditor=false, blendMode=BLEND_Translucent, materialDomain=MD_UI, usedShadingModels=Default Lit}
 nodes [
  n1 = MaterialFunctionCall(UVs=TextureCoordinate(), "Rotation Angle (0-1)"=ScalarParameter() {parameterName=Rotation, sliderMax=1, sliderMin=-1}) {materialFunction=&CustomRotator}
  n2 = AppendVector(
   A=ScalarParameter() {defaultValue=0.005, parameterName=OffsetX, sliderMax=1, sliderMin=-1}
   B=ScalarParameter() {parameterName=OffsetY, sliderMax=1, sliderMin=-1}
  )
  n3 = TextureSampleParameter2D(UVs=MaterialFunctionCall(
    UVs=MaterialFunctionCall!(Input UVs=@n1.Rotated Values, Translate=@n2) {materialFunction=/FortUILibrary/MaterialFunctions/Basics/MF_UI_Translate}.UVs raw
    Texture Scale=ScalarParameter() {defaultValue=2.1, parameterName=TextureScale}
   ) {materialFunction=&ScaleUVsByCenter}.UVs) {parameterName=RewardTexture, texture=/FortniteCoreUI/Frontend/StoreMain/Materials/Textures/T_UI_VBucks_Details_Exact_Amount}
  n4 = NamedRerouteUsage()
  NamedRerouteDeclaration(None=Min(A=NamedRerouteDeclaration(None=@n3.A) {name=TextureAlpha, nodeColor=<rgba 1 0.4688 0 1>}, B=OneMinus(None=@n4))) {name=FinalAlpha, nodeColor=<rgba 0.1484 0 1 1>}
  NamedRerouteDeclaration(None=NamedRerouteDeclaration(None=@n3.RGB) {name=Texture, nodeColor=<rgba 0 1 0.7891 1>}) {name=FinalTexture, nodeColor=<rgba 0.1719 1 0 1>}
  n5 = ScalarParameter() {defaultValue=0.6, parameterName=FadeSharpenss, sliderMax=1}
  n6 = TextureCoordinate()
  n7 = LinearInterpolate(
   A=Constant2Vector() {g=1, r=1}
   B=Constant2Vector() {g=1, r=-1}
   Alpha=ScalarParameter() {group=Localization, parameterName=Flip, sliderMax=1}
  ) {constAlpha=0}
  n8 = ScalarParameter() {defaultValue=0.95, parameterName=FadeOffset, sliderMax=1}
  n9 = ComponentMask(None=MaterialFunctionCall(
    Input UVs=MaterialFunctionCall(UVs=@n6, Texture Scale=@n7) {desc=Flip Mask, materialFunction=&ScaleUVsByCenter}.UVs
    Translate=Subtract(A=@n8, B=@n5)
   ) {materialFunction=&MF_UI_Translate}.UVs raw) {r=true}
  NamedRerouteDeclaration(None=MaterialFunctionCall!("Value (0-1)"=SmoothStep(Max=@n5, Value=@n9) {constMax=0.1}) {materialFunction=/FortUILibrary/UGC/MaterialFunctions/MF_UGCUI_EaseCurves}.EaseInSine) {name=HorizontalFade, nodeColor=<rgba 0 0.8125 1 1>}
  NamedRerouteUsage()
  NamedRerouteUsage()
  TextureObject()
 ]
}
