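# Season pass hub widget showing a limited-time offer banner with a bonus reward (extra V-Bucks or levels) and an end-time line.
# Functions switch reward layout by type string, set amount and timer text, and toggle visibility; RefreshVM stores the SeasonPassVM, hiding if invalid.
# SetColors tints side, glow and timer font materials from the upsell theme.
# PreConstruct sets a material Flip parameter to mirror art for Arabic culture.
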WidgetBlueprint /SeasonPassHub/Widgets/WBP_UI_Offer_Details
graphs [
 EventGraph {nodes [
   Event PreConstruct()
   n1 = GetCurrentCulture!() -> (ReturnValue:String)
   > FlowDirection!(
    execute
    self
    Index:EFlowDirectionPreference Enum=Select(
     Option 0:EFlowDirectionPreference Enum=LeftToRight
     Option 1:EFlowDirectionPreference Enum=RightToLeft
     Index="Equal Exactly (String)"(A=@n1, B=ar)
    ) -> (ReturnValue:EFlowDirectionPreference Enum)
   ) -> (then)
   Event Construct()
   Event Tick()
  ]}
 ExtraRevardValue {nodes [
   n1 = ExtraRevardValue!() -> (then, Value:int) {is=title K2Node_FunctionEntry}
   > "SetText (Text)"(
    self=[
     Get BattlePassVbucksText01() -> (BattlePassVbucksText01:Common Text Object Reference)
     Get BonuLvlSizeText() -> (BonuLvlSizeText:Text Block Object Reference)
    ]
    InText="To Text (Integer)"(Value=@n1)
   )
  ]}
 FlowDirection {nodes [
   n1 = FlowDirection!() -> (then, Index:EFlowDirectionPreference Enum=Inherit) {is=title K2Node_FunctionEntry}
   > n2 = GetDynamicMaterial(self=Get LazyRevardImage() -> (LazyRevardImage:Fort Lazy Image Object Reference))
   n3 = Select!(Inherit:float=0, Culture:float=0, LeftToRight:float=0, RightToLeft:float=1, Index:EFlowDirectionPreference Enum=@n1) -> (ReturnValue:float)
   > SetScalarParameterValue(self=@n2, ParameterName=Flip, Value=@n3)
   > n4 = GetDynamicMaterial(self=Get Left_Side_Texture() -> (Left_Side_Texture:Image Object Reference))
   > SetScalarParameterValue(self=@n4, ParameterName=Flip, Value=@n3)
   > n5 = GetDynamicMaterial(self=Get Right_Side_Texture() -> (Right_Side_Texture:Image Object Reference))
   > SetScalarParameterValue(self=@n5, ParameterName=Flip, Value=@n3)
  ]}
 IsContainsOffer {nodes [
   n1 = IsContainsOffer!() -> (then, IsOffer:bool=~) {is=title K2Node_FunctionEntry}
   > SetVisibility(InVisibility=Select!(Option 0:ESlateVisibility Enum=Hidden, Option 1:ESlateVisibility Enum=SelfHitTestInvisible, Index:bool=@n1) -> (ReturnValue:ESlateVisibility Enum))
  ]}
 RefreshVM {nodes [
   n1 = RefreshVM!() -> (then, SeasonPass:Season Pass VM Object Reference) {is=title K2Node_FunctionEntry}
   > n2 = Set SeasonPassVM(SeasonPassVM:Season Pass VM Object Reference=@n1) -> (Output_Get:Season Pass VM Object Reference)
   > n3 = Is Valid(InputObject=@n2)
   SetVisibility(execute=@n3.Is Not Valid, InVisibility=Hidden)
   SetColors!(execute, self, UpsellType:ENexusUpsellEntryType Enum=Bundle) -> (then)
  ]}
 RevardTypeFromString {nodes [
   n1 = RevardTypeFromString!() -> (then, CustomOfferType:String) {is=title K2Node_FunctionEntry}
   > n2 = Switch on String!(execute:Exec, Selection:String=@n1) -> (Default:Exec, V-bucks:Exec, BonusLvls:Exec) {pinNames [V-bucks,BonusLvls]}
   Return Node!(execute=@n2.BonusLvls, RevardType:E_CustomOfferType Enum=NewEnumerator1)
   > Return Node(RevardType=NewEnumerator2)
   Return Node(execute=@n2.V-bucks, RevardType=NewEnumerator0)
  ]}
 SetColors {nodes [
   n1 = SetColors!() -> (then, UpsellType:ENexusUpsellEntryType Enum=Free) {is=title K2Node_FunctionEntry}
   > n2 = GetDynamicMaterial(self=Get Left_Side_Texture())
   n3 = GetUpsellThemeColor!(
    Benefits:"Nexus Upsell Benefits Structure (by ref)"=GetBenefitsForUpsellType!(
     self:Season Pass VM Object Reference=Get SeasonPassVM() -> (SeasonPassVM:Season Pass VM Object Reference)
     UpsellType:ENexusUpsellEntryType Enum=@n1
    ) -> (BenefitData:Nexus Upsell Benefits Structure, ReturnValue:Fort Simple Offer VM Object Reference)
    ThemeName:String=Select(Option 0:String=Tile Color 1, Option 1:String=BG Color 1, Index=true) -> (ReturnValue:String)
    DefaultColor:Linear Color Structure=(R=0.5755,G=0.5755,B=0.5755,A=1)
   ) -> (ReturnValue_SpecifiedColor:Linear Color Structure, ReturnValue_ColorUseRule:ESlateColorStylingMode Enum)
   > SetVectorParameterValue(self=@n2, ParameterName=Color, Value=@n3)
   > n4 = GetDynamicMaterial(self=Get Right_Side_Texture())
   > SetVectorParameterValue(self=@n4, ParameterName=Color, Value=@n3)
   > n5 = GetDynamicMaterial(self=Get BG_Glow() -> (BG_Glow:Image Object Reference))
   > SetVectorParameterValue(self=@n5, ParameterName=Color, Value=@n3)
   > n6 = GetDynamicFontMaterial!(execute, self:Text Object Reference=Get Ends_In_Text() -> (Ends_In_Text:Common Text Object Reference)) -> (then, ReturnValue:Material Instance Dynamic Object Reference)
   > SetVectorParameterValue(self=@n6, ParameterName=Color, Value=@n3)
   Switch on ENexusUpsellEntryType!(execute, Selection:ENexusUpsellEntryType Enum=@n1) -> (Free:Exec, Paid:Exec, Bundle:Exec, Crew:Exec) {is=title K2Node_SwitchEnum}
  ]}
 SetRevardType {nodes [
   n1 = SetRevardType!() -> (then, RevardType:String) {is=title K2Node_FunctionEntry}
   > n2 = RevardTypeFromString!(execute, self, CustomOfferType:String=@n1) -> (then, RevardType:E_CustomOfferType Enum=NewEnumerator0) {is=title K2Node_CallFunction}
   > SetActiveWidget(
    self:Widget Switcher Object Reference=Get RewardWidgetSwitcher() -> (RewardWidgetSwitcher:Common Activatable Widget Switcher Object Reference)
    Widget=Select!(
     NewEnumerator0:Widget Object Reference=Get LazyRevardImage()
     NewEnumerator1:Widget Object Reference=Get BonusLvls() -> (BonusLvls:Horizontal Box Object Reference)
     NewEnumerator2:Widget Object Reference=Get Placeholder() -> (Placeholder:Image Object Reference)
     Index:E_CustomOfferType Enum=@n2
    ) -> (ReturnValue:Widget Object Reference)
   )
   > SetVisibility(
    self=Get BattlePassVbucksText01()
    InVisibility=Select(
     NewEnumerator0:ESlateVisibility Enum=SelfHitTestInvisible
     NewEnumerator1:ESlateVisibility Enum=Collapsed
     NewEnumerator2:ESlateVisibility Enum=Collapsed
     Index=@n2
    ) -> (ReturnValue:ESlateVisibility Enum)
   )
   > "SetText (Text)"(
    self=Get RevardDescription() -> (RevardDescription:Common Text Object Reference)
    InText=Select(NewEnumerator0:Text="LOCTEXT(\"Extra V-Bucks!\")", NewEnumerator1:Text="LOCTEXT(\"Extra\\r\\nLevels!\")", NewEnumerator2:Text, Index=@n2) -> (ReturnValue:Text)
   )
  ]}
 UpdateTimer {nodes [
   n1 = UpdateTimer!() -> (then, LocalizedTimer:Text) {is=title K2Node_FunctionEntry}
   > "SetText (Text)"(self=Get Ends_In_Text(), InText=@n1)
   Format Text!(Format="LOCTEXT(\"Ends {Timer}\")", Timer:Text=@n1) -> (Result)
  ]}
]
variables []
widgets {
 info {inheritedWidgetCount=0, parentClass=&UserWidget, rootWidgetClass=&Overlay, widgetCount=20}
 widgets [
  Overlay_37 {widgetClassPath=&Overlay, widget=Overlay_37}
  CommonVisualAttachment_204 {
   widgetClassPath=&CommonVisualAttachment
   parent=Overlay_37
   widget=CommonVisualAttachment_204
   properties {contentAnchor=<0.5 0.5>}
   slot {class=OverlaySlot, properties {horizontalAlignment=HAlign_Center, verticalAlignment=VAlign_Center}}
  }
  BG_Glow {
   widgetClassPath=&Image
   bIsVariable=true
   parent=CommonVisualAttachment_204
   widget=BG_Glow
   properties {
    brush {
     bIsDynamicallyLoaded=false
     drawAs=RoundedBox
     imageType=FullColor
     mirroring=NoMirror
     resourceName=None
     resourceObject=/SeasonPassHub/Widgets/Materials/MaterialInstances/M_UI_Battle_Pass_Target_Offer_Blur_Effect_Inst
     tiling=NoTile
     imageSize=<560 250>
     margin=<ltrb 0 0 0 0>
     outlineSettings {
      bUseBrushTransparency=false
      roundingType=HalfHeightRadius
      width=0
      color {colorUseRule=UseColor_Specified, specifiedColor=<rgba 0 0 0 0>}
      cornerRadii=<0 0 0 0>
     }
     tintColor {colorUseRule=UseColor_Specified, specifiedColor=<rgba 1 1 1 1>}
     uVRegion {bIsValid=false, max=<0 0>, min=<0 0>}
    }
    renderTransformPivot=<0 0>
   }
   slot {class=SizeBoxSlot, properties {horizontalAlignment=HAlign_Center, verticalAlignment=VAlign_Center}}
  }
  HorizontalBox_40 {
   widgetClassPath=&HorizontalBox
   parent=Overlay_37
   widget=HorizontalBox_40
   slot {class=OverlaySlot, properties {horizontalAlignment=HAlign_Center, verticalAlignment=VAlign_Center}}
  }
  Left_Side_Texture {
   widgetClassPath=&Image
   bIsVariable=true
   parent=HorizontalBox_40
   widget=Left_Side_Texture
   properties {
    brush {
     bIsDynamicallyLoaded=false
     drawAs=Image
     imageType=FullColor
     mirroring=NoMirror
     resourceName=None
     resourceObject=/SeasonPassHub/Widgets/Materials/MaterialInstances/M_UI_Battle_Pass_Target_Offer_Triangle_Effect_Inst
     tiling=NoTile
     imageSize=<160 140>
     margin=<ltrb 0 0 0 0>
     outlineSettings {
      bUseBrushTransparency=false
      roundingType=HalfHeightRadius
      width=0
      color {colorUseRule=UseColor_Specified, specifiedColor=<rgba 0 0 0 0>}
      cornerRadii=<0 0 0 0>
     }
     tintColor {colorUseRule=UseColor_Specified, specifiedColor=<rgba 1 1 1 1>}
     uVRegion {bIsValid=false, max=<0 0>, min=<0 0>}
    }
    renderTransform {angle=0, scale=<1 1>, shear=<0 0>, translation=<85 0>}
   }
   slot {class=HorizontalBoxSlot, properties {horizontalAlignment=HAlign_Left, verticalAlignment=VAlign_Top}}
  }
  VerticalBox_1 {
   widgetClassPath=&VerticalBox
   parent=HorizontalBox_40
   widget=VerticalBox_1
   slot {class=HorizontalBoxSlot, properties {horizontalAlignment=HAlign_Center, verticalAlignment=VAlign_Center, size {sizeRule=Fill, value=1}}}
  }
  SizeBox_3 {
   widgetClassPath=&SizeBox
   parent=VerticalBox_1
   widget=SizeBox_3
   properties {bOverride_HeightOverride=true, heightOverride=90}
   slot {class=VerticalBoxSlot, properties {verticalAlignment=VAlign_Center, size {sizeRule=Fill, value=1}}}
  }
  BattlePass_VBucks_Discount {widgetClassPath=&Overlay, parent=SizeBox_3, widget=BattlePass_VBucks_Discount, slot {class=SizeBoxSlot}}
  HorizontalBox_0 {
   widgetClassPath=&HorizontalBox
   parent=BattlePass_VBucks_Discount
   widget=HorizontalBox_0
   properties {flowDirectionPreference=Culture}
   slot {class=OverlaySlot, properties {horizontalAlignment=HAlign_Center, verticalAlignment=VAlign_Center, padding=<ltrb 0 0 10 0>}}
  }
  RewardWidgetSwitcher {
   widgetClassPath=&CommonActivatableWidgetSwitcher
   bIsVariable=true
   parent=HorizontalBox_0
   widget=RewardWidgetSwitcher
   slot {class=HorizontalBoxSlot}
  }
  Placeholder {
   widgetClassPath=&Image
   bIsVariable=true
   parent=RewardWidgetSwitcher
   widget=Placeholder
   properties {colorAndOpacity=<rgba 1 1 1 0>}
   slot {class=WidgetSwitcherSlot, properties {horizontalAlignment=HAlign_Center, verticalAlignment=VAlign_Center}}
  }
  LazyRevardImage {
   widgetClassPath=&FortLazyImage
   bIsVariable=true
   parent=RewardWidgetSwitcher
   widget=LazyRevardImage
   properties {
    materialTextureParamName=RewardTexture
    brush {
     bIsDynamicallyLoaded=false
     drawAs=Image
     imageType=FullColor
     mirroring=NoMirror
     resourceName=None
     resourceObject=/SeasonPassHub/Widgets/Materials/M_UI_Battle_Pass_Target_Offer_Reward_Image_Mask
     tiling=NoTile
     imageSize=<95 95>
     margin=<ltrb 0 0 0 0>
     outlineSettings {
      bUseBrushTransparency=false
      roundingType=HalfHeightRadius
      width=0
      color {colorUseRule=UseColor_Specified, specifiedColor=<rgba 0 0 0 0>}
      cornerRadii=<0 0 0 0>
     }
     tintColor {colorUseRule=UseColor_Specified, specifiedColor=<rgba 1 1 1 1>}
     uVRegion {bIsValid=false, max=<0 0>, min=<0 0>}
    }
   }
   slot {class=WidgetSwitcherSlot}
  }
  BonusLvls {
   widgetClassPath=&HorizontalBox
   bIsVariable=true
   parent=RewardWidgetSwitcher
   widget=BonusLvls
   properties {flowDirectionPreference=LeftToRight}
   slot {class=WidgetSwitcherSlot, properties {verticalAlignment=VAlign_Center, padding=<ltrb 14 0 14 0>}}
  }
  BonusPlusText {
   widgetClassPath=&UEFN_TextBlock
   parent=BonusLvls
   widget=BonusPlusText
   properties {
    text=+
    font {
     bForceMonospaced=false
     bMaterialIsStencil=false
     fontMaterial=null
     fontObject=&HeadingNow
     letterSpacing=0
     monospacedWidth=1
     size=30
     skewAmount=0
     typefaceFontName=86 Bold
     outlineSettings {
      bApplyOutlineToDropShadows=false
      bMiteredCorners=false
      bSeparateFillAlpha=false
      outlineMaterial=null
      outlineSize=0
      outlineColor=<rgba 0 0 0 1>
     }
    }
   }
   slot {class=HorizontalBoxSlot, properties {verticalAlignment=VAlign_Center}}
  }
  BonuLvlSizeText {
   widgetClassPath=&UEFN_TextBlock
   bIsVariable=true
   parent=BonusLvls
   widget=BonuLvlSizeText
   properties {
    text=5
    font {
     bForceMonospaced=false
     bMaterialIsStencil=false
     fontMaterial=null
     fontObject=&HeadingNow
     letterSpacing=0
     monospacedWidth=1
     size=37.5
     skewAmount=0
     typefaceFontName=86 Bold
     outlineSettings {
      bApplyOutlineToDropShadows=false
      bMiteredCorners=false
      bSeparateFillAlpha=false
      outlineMaterial=null
      outlineSize=0
      outlineColor=<rgba 0 0 0 1>
     }
    }
   }
   slot {class=HorizontalBoxSlot}
  }
  VerticalBox_0 {
   widgetClassPath=&VerticalBox
   parent=HorizontalBox_0
   widget=VerticalBox_0
   slot {class=HorizontalBoxSlot, properties {verticalAlignment=VAlign_Center}}
  }
  BattlePassVbucksText01 {
   widgetClassPath=&CommonTextBlock
   bIsVariable=true
   parent=VerticalBox_0
   widget=BattlePassVbucksText01
   properties {
    bAutoCollapseWithEmptyText=true
    justification=Center
    text=200
    wrapTextAt=700
    font {
     bForceMonospaced=false
     bMaterialIsStencil=false
     fontMaterial=null
     fontObject=&HeadingNow
     letterSpacing=0
     monospacedWidth=1
     size=30
     skewAmount=0
     typefaceFontName=86 Bold
     outlineSettings {
      bApplyOutlineToDropShadows=false
      bMiteredCorners=false
      bSeparateFillAlpha=false
      outlineMaterial=null
      outlineSize=0
      outlineColor=<rgba 0 0 0 1>
     }
    }
    shadowOffset=<0 0>
   }
   slot {class=VerticalBoxSlot, properties {horizontalAlignment=HAlign_Left, verticalAlignment=VAlign_Center}}
  }
  RevardDescription {
   widgetClassPath=&CommonTextBlock
   bIsVariable=true
   parent=VerticalBox_0
   widget=RevardDescription
   properties {
    bAutoCollapseWithEmptyText=true
    justification=Center
    text="Extra V-Bucks!"
    wrapTextAt=700
    font {
     bForceMonospaced=false
     bMaterialIsStencil=false
     fontMaterial=null
     fontObject=&HeadingNow
     letterSpacing=0
     monospacedWidth=1
     size=15.75
     skewAmount=0
     typefaceFontName=75 Medium
     outlineSettings {
      bApplyOutlineToDropShadows=false
      bMiteredCorners=false
      bSeparateFillAlpha=false
      outlineMaterial=null
      outlineSize=0
      outlineColor=<rgba 0 0 0 1>
     }
    }
    shadowOffset=<0 0>
   }
   slot {class=VerticalBoxSlot, properties {horizontalAlignment=HAlign_Center}}
  }
  Ends_In_Text {
   widgetClassPath=&CommonTextBlock
   bIsVariable=true
   parent=VerticalBox_1
   widget=Ends_In_Text
   properties {
    bAutoCollapseWithEmptyText=true
    justification=Center
    text=Ends March 26 11:59PM ET
    wrapTextAt=700
    colorAndOpacity {colorUseRule=UseColor_Specified, specifiedColor=<rgba 0.9387 0.9387 0.9387 1>}
    font {
     bForceMonospaced=false
     bMaterialIsStencil=false
     fontMaterial=/SeasonPassHub/Widgets/Materials/MaterialInstances/M_UI_Battle_Pass_Target_Offer_EndsInText_Inst
     fontObject=&HeadingNow
     letterSpacing=0
     monospacedWidth=1
     size=15.75
     skewAmount=0
     typefaceFontName=74 Regular
     outlineSettings {
      bApplyOutlineToDropShadows=false
      bMiteredCorners=false
      bSeparateFillAlpha=false
      outlineMaterial=null
      outlineSize=0
      outlineColor=<rgba 0 0 0 1>
     }
    }
    shadowOffset=<0 0>
   }
   slot {class=VerticalBoxSlot, properties {horizontalAlignment=HAlign_Center, verticalAlignment=VAlign_Center}}
  }
  Right_Side_Texture {
   widgetClassPath=&Image
   bIsVariable=true
   parent=HorizontalBox_40
   widget=Right_Side_Texture
   properties {
    brush {
     bIsDynamicallyLoaded=false
     drawAs=Image
     imageType=FullColor
     mirroring=NoMirror
     resourceName=None
     resourceObject=&M_UI_Battle_Pass_Target_Offer_Triangle_Effect_Inst
     tiling=NoTile
     imageSize=<160 140>
     margin=<ltrb 0 0 0 0>
     outlineSettings {
      bUseBrushTransparency=false
      roundingType=HalfHeightRadius
      width=0
      color {colorUseRule=UseColor_Specified, specifiedColor=<rgba 0 0 0 0>}
      cornerRadii=<0 0 0 0>
     }
     tintColor {colorUseRule=UseColor_Specified, specifiedColor=<rgba 1 1 1 1>}
     uVRegion {bIsValid=false, max=<0 0>, min=<0 0>}
    }
    renderTransform {angle=0, scale=<-1 -1>, shear=<0 0>, translation=<-85 0>}
   }
   slot {class=HorizontalBoxSlot, properties {horizontalAlignment=HAlign_Center, verticalAlignment=VAlign_Center}}
  }
 ]
}
